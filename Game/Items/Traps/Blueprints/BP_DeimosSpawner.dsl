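; Creative creature spawner device that spawns Deimos AI through rifts and responds to on, off, destroy and eliminate-creatures events.
; On minigame start, preset indices become spawn, despawn and rift settings, including an EQS query chosen by visibility and area options.
; Spawned creatures get class-mapped tags via a gameplay effect, and kills broadcast an event.
; Option changes rescale the radius preview and async-load the selected meshes and materials.
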(event EventOnWorldReady
  (Visibility|SetVisibilityComponents (Variables|Default|GetVisibleInGameComponent) (Utilities|Array|MakeArray (Variables|Debug|GetStaticMeshComponent) (Variables|Default|GetCosmeticBase)) true)
  (Visibility|SetVisibilityComponents (Variables|Default|GetRadiusMeshVisibleInGameComponent) (Utilities|Array|MakeArray (Variables|Default|GetSpawnAreaRadiusMesh)) true))

(event EventBP_NotifyAIPawnKilled
  (params (KilledAI :type Fort%20AIPawn%20Object%20Reference) (KillingInstigator :type Controller%20Object%20Reference))
  (GameplayEvent|BroadcastEvent (Variables|UserOptions|Events|GetOn%20a%20Creature%20is%20Eliminated) 0 (Utilities|Struct|MakeGameplayEventCreatureEliminated KillingInstigator)))

(event EventBeginPlay
  (CallFunction|SetSpawnLocation (Variables|Default|GetSpawnPoint)))

(event Custom|TurnOnEvent
  (params (Instigator :type Controller%20Object%20Reference))
  (CallFunction|OnTurnonReceived Instigator))

(event Custom|TurnOffEvent
  (params (Instigator :type Controller%20Object%20Reference))
  (CallFunction|OnTurnOffReceived Instigator))

(event OnAnyPropertyChanged(ToyOptionsComponent)
  (CallFunction|UpdateSpawnLocationVisuals))

(event Custom|DestroySpawnerEvent
  (params (Instigator :type Controller%20Object%20Reference))
  (CallFunction|OnKillRiftReceived Instigator))

(event OnPropertiesLoaded(ToyOptionsComponent)
  (CallFunction|UpdateSpawnLocationVisuals))

(event Custom|EliminateCreaturesEvent
  (params (Instigator :type Controller%20Object%20Reference))
  (CallFunction|OnKillAIReceived Instigator))

(event EventMinigameStarted
  (CallFunction|SetupConfigurables)
  (Networking|SwitchHasAuthority
    (:Authority
      (|CallParentFunction))))

(event EventOnSpawnedRift
  (params (SpawnedRift :type Building%20Rift%20Object%20Reference))
  (Networking|SwitchHasAuthority
    (:Authority
      (bind _asbp_creative_deimos_rift (Utilities|Casting|CastToBP_CreativeDeimosRift SpawnedRift))
      (if (not (Variables|Configurables|GetbIsVisible))
        (Variables|/Game/Athena/Deimos/Spawners/RiftSpawners/BP_CreativeDeimosRift.BP_CreativeDeimosRift_C|SetbDamageSelfOnSpawn false _asbp_creative_deimos_rift)
        (else
          (Variables|/Game/Athena/Deimos/Spawners/RiftSpawners/BP_CreativeDeimosRift.BP_CreativeDeimosRift_C|SetbDamageSelfOnSpawn (Variables|Configurables|GetbDamageOnSpawn) _asbp_creative_deimos_rift)))
      (Variables|/Game/Athena/Deimos/Spawners/RiftSpawners/BP_CreativeDeimosRift.BP_CreativeDeimosRift_C|SetbHidesVFX (not (Variables|Configurables|GetbVFXVisible)) _asbp_creative_deimos_rift))))

(event MinigameStateChangedDelegate(MinigameProgress)
  (params (Minigame :type Fort%20Minigame%20Object%20Reference) (NewMinigameState :type EFortMinigameState%20Enum :default "PreGame"))
  (CallFunction|UpdateCollision (Utilities|Enum|Equal(Enum) NewMinigameState "PreGame")))

(event Custom|LoadSelectedStaticMeshAndMaterials
  ;[1 1 1] Logic to async load the desired static mesh and base mesh, along with their associated materials
  (bind (_cube _base) (CallFunction|GetDesiredMeshMaterial))
  (Utilities|FlowControl|Sequence
    (:then_0
      (bind _loaded_object (Creative|Loading|AsyncLoadEditorSafe (Utilities|Struct|BreakMeshMaterial _cube))
        (:Loaded
          (bind _asstatic_mesh (Utilities|Casting|CastToStaticMesh _loaded_object))
          (Components|StaticMesh|SetStaticMesh (Variables|Debug|GetStaticMeshComponent) _asstatic_mesh)
          (bind _loaded_object_2 (Creative|Loading|AsyncLoadEditorSafe (.Material_9_0C040D9B4C3EF68C191FE79CEE91901E (Utilities|Struct|BreakMeshMaterial _cube)))
            (:Loaded
              (bind _asmaterial_interface (Utilities|Casting|CastToMaterialInterface _loaded_object_2))
              (Rendering|Material|SetMaterial (Variables|Debug|GetStaticMeshComponent) 0 _asmaterial_interface))))))
    (:then_1
      (bind _loaded_object_3 (Creative|Loading|AsyncLoadEditorSafe (Utilities|Struct|BreakMeshMaterial _base))
        (:Loaded
          (bind _asstatic_mesh_2 (Utilities|Casting|CastToStaticMesh _loaded_object_3))
          (Components|StaticMesh|SetStaticMesh (Variables|Default|GetCosmeticBase) _asstatic_mesh_2)
          (bind _loaded_object_4 (Creative|Loading|AsyncLoadEditorSafe (.Material_9_0C040D9B4C3EF68C191FE79CEE91901E (Utilities|Struct|BreakMeshMaterial _base)))
            (:Loaded
              (bind _asmaterial_interface_2 (Utilities|Casting|CastToMaterialInterface _loaded_object_4))
              (Rendering|Material|SetMaterial (Variables|Default|GetCosmeticBase) 0 _asmaterial_interface_2))))))))

(event EventBP_NotifyAIPawnSpawned
  (params (SpawnedAI :type Fort%20AIPawn%20Object%20Reference) (SpawnSlotLocation :type Vector%20%28by%20ref%29 :default "0, 0, 0"))
  (bind _returnvalue (Utilities|GetClass SpawnedAI))
  (bind _output (Utilities|ResolveSoftReference _returnvalue))
  (bind _classtotagmap (Variables|Default|GetClassToTagMap))
  (bind (_value _returnvalue_2) (Utilities|Map|Find _classtotagmap _output))
  (bind _aipawnabilitysystemcomponent (Variables|/Script/FortniteGame.FortAIPawn|GetAIPawnAbilitySystemComponent SpawnedAI))
  (Utilities|IsValid SpawnedAI
    (:Is%20Valid
      (if _returnvalue_2
        (bind _returnvalue_3 (Ability|GameplayEffect|AddGrantedTag (GameplayEffects|MakeOutgoingSpec :self _aipawnabilitysystemcomponent :GameplayEffectClass "/Game/Creative/Devices/CreaturePlacer/GameplayEffects/GE_EliminationManagerAddTag.GE_EliminationManagerAddTag_C" :Level 0.0) _value))
        (GameplayEffects|ApplyGameplayEffectSpecToSelf _aipawnabilitysystemcomponent _returnvalue_3)))))

(fn BP_ShouldTrigger
  (params (TouchingActors :type Array%20of%20Actor%20Object%20References))
  (return true))

(fn ChooseEQSRift
  (LocalVariables|SetbUseVisibility (!= (Variables|Configurables|GetSpawnAtVisiblePreset) 1))
  (LocalVariables|SetbUseArea (!= (Variables|Configurables|GetSpawnWithinRadiusPreset) 0))
  (Variables|BuildingProp_DeimosSpawner|SetbSortSlotsByBestScore false)
  (if (LocalVariables|GetbUseVisibility)
    (if (LocalVariables|GetbUseArea)
      (LocalVariables|SetCalculatedIndex 3)
      (else
        (LocalVariables|SetCalculatedIndex 1)))
    (elif (LocalVariables|GetbUseArea)
      (LocalVariables|SetCalculatedIndex 2)
      (else
        (LocalVariables|SetCalculatedIndex 0))))
  (Variables|BuildingProp_DeimosSpawner|SetSelectedEQSQuery (Utilities|Array|Get(acopy) (Variables|CustomizeOptions|GetEQSQueryOptions) (LocalVariables|GetCalculatedIndex))))

(fn GetDesiredMeshMaterial
  (bind _spawnertypepresetindex (Variables|Configurables|GetSpawnerTypePresetIndex))
  (bind _returnvalue (* _spawnertypepresetindex 2))
  (bind _returnvalue_2 (+ _returnvalue 1))
  (bind _meshmaterialarray (Variables|Configurables|GetMeshMaterialArray))
  (if (and (Utilities|Array|IsValidIndex _meshmaterialarray _returnvalue_2) (Utilities|Array|IsValidIndex _meshmaterialarray _returnvalue))
    (return (Utilities|Array|Get(aref) _meshmaterialarray _returnvalue) (Utilities|Array|Get(aref) _meshmaterialarray _returnvalue_2))
    (else
      (return 0 0))))

(fn GetOverrideMeshMaterial
  (params (MeshComp :type Static%20Mesh%20Component%20Object%20Reference) (MatIdx :type Integer))
  (if (Utilities|Equal(Object) MeshComp (Variables|Default|GetSpawnAreaRadiusMesh))
    (bind _spawnareamid (Variables|Default|GetSpawnAreaMID)
      (:then
        (return _spawnareamid))
      (:else))))

(fn SetupConfigurables
  (Variables|BuildingProp_DeimosSpawner|SetSelectedPawnNumberCap (Utilities|Array|Get(acopy) (Variables|CustomizeOptions|GetPawnNumberCapOptions) (Variables|Configurables|GetAINumberPresetIndex)))
  (Variables|BuildingProp_DeimosSpawner|SetSelectedSpawnGroup (Utilities|Array|Get(acopy) (Variables|CustomizeOptions|GetSpawnGroupOptions) (Variables|Configurables|GetAITypePresetIndex)))
  (Variables|BuildingProp_DeimosSpawner|SetSelectedSpawnTiming (Utilities|Array|Get(acopy) (Variables|CustomizeOptions|GetSpawnTimingOptions) (Variables|Configurables|GetSpawnFreqPresetIndex)))
  (Variables|BuildingProp_DeimosSpawner|SetSelectedSpawnLimit (Utilities|Array|Get(acopy) (Variables|CustomizeOptions|GetSpawnLimitOptions) (Variables|Configurables|GetSpawnLimitPresetIndex)))
  (Variables|BuildingProp_DeimosSpawner|SetbNoSpawnLimitEnabled (== (Variables|Configurables|GetSpawnLimitPresetIndex) 0))
  (if (>= (Variables|Configurables|GetActivationRangePresetIndex) 0)
    (Variables|BuildingProp_AISpawner|SetActivationRangeLevel (Math|Conversions|ToFloat(Integer) (+ (Variables|Configurables|GetActivationRangePresetIndex) 1)))
    (else
      (Variables|BuildingProp_AISpawner|SetActivationRangeLevel 5.0)))
  (Variables|BuildingProp_AISpawner|SetbUseDistanceToAI (== (Variables|Configurables|GetDespawnTypePresetIndex) 0))
  (Variables|BuildingProp_AISpawner|SetbCanAIDespawnWhenTooFar (!= (Variables|Configurables|GetDespawnTypePresetIndex) 2))
  (Variables|BuildingProp_AISpawner|SetDespawnRangeOverride (* (Variables|Configurables|GetDespawnRangePresetIndex) 512.0))
  (Variables|BuildingProp_DeimosSpawner|SetbDamageBuildingsOnSpawn (Variables|Configurables|GetbDamageBuildings))
  (Variables|BuildingProp_DeimosSpawner|SetbInvulnerable (Variables|Configurables|GetbInvincible))
  (Variables|BuildingProp_DeimosSpawner|SetbVisible (Variables|Configurables|GetbIsVisible))
  (Variables|BuildingProp_DeimosSpawner|SetOverrideQueryRadius (Variables|UserOptions|Basic|GetSpawnRadiusPresetIndex))
  (CallFunction|ChooseEQSRift)
  (Variables|BuildingProp_DeimosSpawner|SetbEnabledOnMinigameStart (Variables|Default|GetbEnabledOnStart))
  (Variables|BuildingProp_AISpawner|SetCreativeRiftClass (Utilities|Array|Get(acopy) (Variables|CustomizeOptions|GetDeimosSpawnerOptions) (Variables|Configurables|GetSpawnerTypePresetIndex))))

(fn UpdateCollision
  (params (Collision%20On :type Boolean))
  (bind _returnvalue (not Collision%20On))
  (CallFunction|AdjustCollision _returnvalue (Variables|Debug|GetStaticMeshComponent))
  (CallFunction|AdjustCollision _returnvalue (Variables|Default|GetCosmeticBase)))

(fn UpdateSpawnLocationVisuals
  (Creative|Networking|Authority/Cosmetic
    (:Authority
      _)
    (:Cosmetic
      (Transformation|SetRelativeScale3D (Variables|Default|GetSpawnAreaRadiusMesh) (Math|Vector|MakeVector (* (Variables|UserOptions|Basic|GetSpawnRadiusPresetIndex) 2.0) (* (Variables|UserOptions|Basic|GetSpawnRadiusPresetIndex) 2.0) (.z (Variables|/Script/Engine.SceneComponent|GetRelativeScale3D (Variables|Default|GetSpawnAreaRadiusMesh)))))))
  (CallFunction|LoadSelectedStaticMeshandMaterials))

(fn ConstructionScript
  (Creative|Networking|Authority/Cosmetic
    (:Cosmetic
      (Variables|Default|SetSpawnAreaMID (Material|GetMIDfromMaterial (Variables|Default|GetSpawnAreaRadiusMesh))))))
