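; Ceiling-mounted no-weapon zone device Blueprint; its name suggests a trap-style item that marks an area where weapons are restricted.
; GetTriggerLocation keeps the Trigger box's relative X and Y, and sets Z to the negative of its Z extent times ZoneZScale, so the zone hangs downward.
; The construction script only calls the parent; BeginPlay, overlap and Tick add no logic, and no weapon restriction logic appears here.
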(event EventBeginPlay
  (|CallParentFunction) ; This node is disabled and will not be called.\nDrag off pins to build functionality.
)

(event Collision|EventActorBeginOverlap
  (params (OtherActor :type Actor%20Object%20Reference)))

(event EventTick
  (params (DeltaSeconds :type Float%20%28single-precision%29)))

(fn GetTriggerLocation
  (return (Math|Vector|MakeVector (.x (Variables|/Script/Engine.SceneComponent|GetRelativeLocation (Variables|Default|GetTrigger))) (.y (Variables|/Script/Engine.SceneComponent|GetRelativeLocation (Variables|Default|GetTrigger))) (Math|Float|float*float (.z (Variables|/Script/Engine.BoxComponent|GetBoxExtent (Variables|Default|GetTrigger))) (Variables|UserOptions|GetZoneZScale) -1.0))))

(fn ConstructionScript
  (|CallParentFunction))
