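; Floor trap device that spawns a Baller (OctopusVehicle) and can tune it from configurable presets.
; After the parent spawn attempt, it caches the vehicle's boost recharge rate so the boost toggle can restore it or zero it.
; ConfigureOctopusVehicle rebuilds the boost, spherical drive and towhook structs with preset overrides, and sets damage dealing.
; OnRep of the cached vehicle and Customize run it; property-change customization is unhooked pending fixes.
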(event Custom|Customize
  (CallFunction|ConfigureOctopusVehicle))

(event EventAttemptSpawnVehicle
  (bind _cachedvehicleinterface (Variables|Default|GetCachedVehicleInterface))
  (bind _returnvalue (Utilities|ToObject(Interface) _cachedvehicleinterface))
  (|CallParentFunction)
  (Utilities|IsValid _returnvalue
    (:Is%20Valid
      (bind _asoctopus_vehicle (Utilities|Casting|CastToOctopusVehicle _returnvalue))
      (Variables|Configurables|SetCachedBoostRecharge (Utilities|Struct|BreakFortRechargingActionTimer (Variables|/Script/FortniteGame.FortOctopusVehicle|GetBoostBounceAction _asoctopus_vehicle))))))

(event EventBeginPlay
  (|CallParentFunction))

(event OnAnyPropertyChanged(ToyOptionsComponent)
  ;[1 0 0.01891] Unhook Customize until customization issues have been addressed
  (Utilities|IsValid (Variables|Default|GetCachedOctopusVehicle)))

(fn ShouldTrigger
  (params (TouchingActors :type Array%20of%20Actor%20Object%20References))
  (return true))

(fn ConfigureOctopusVehicle
  (bind (_chargerate _activeexpenserate _passiveexpenserate _minactiveduration _minactivationcharge _activecooldowntime _chargethreshold _charge _bisactive _bischarging _bispassive) (Utilities|Struct|BreakFortRechargingActionTimer (Variables|/Script/FortniteGame.FortOctopusVehicle|GetBoostBounceAction (Variables|Default|GetCachedOctopusVehicle))))
  (bind (_radius _maxspeedkmh _lowspeedaccelerationforce _highspeedaccelerationforce _maxinclinedeg _maxaircontrolforce _maxaircontrolspeedkmh _shellangulardrag _dragcoefficient _dragcoefficientautobrake _maxautobrakespeedkmh _contactrepulsionforce _contactthreshold _massdirectionmaxangledeg _massdirectionstiffness _bmassdirectioninvert) (Utilities|Struct|BreakSphericalDriveParams (Variables|/Script/FortniteGame.FortOctopusVehicle|GetBaseSphericalDriveParams (Variables|Default|GetCachedOctopusVehicle))))
  (bind (_maxcablelength _mincablelength _busespring _buserope _buseastractorbeam _springstiffness _springdamping _springmaxstiffnessforce _springmaxstiffnessvelocity _springdeformationrate _ropegive _ropebreakforce _ropebreaklength _ropeyankforce _extendspeedkmh _contractspeedkmh _bapplyspringtoself _bapplyspringtoother _btakeupslack) (Utilities|Struct|BreakTowhookParams (Variables|/Script/FortniteGame.FortOctopusVehicle|GetBaseTowhookParams (Variables|Default|GetCachedOctopusVehicle))))
  (bind _cachedoctopusvehicle (Variables|Default|GetCachedOctopusVehicle))
  (Utilities|IsValid _cachedoctopusvehicle
    (:Is%20Valid
      (Variables|/Script/FortniteGame.FortOctopusVehicle|SetBoostBounceAction (Utilities|Struct|MakeFortRechargingActionTimer (select (Variables|Configurables|GetbEnableBoostPreset) (Variables|Configurables|GetCachedBoostRecharge) 0.0) _activeexpenserate _passiveexpenserate _minactiveduration _minactivationcharge _activecooldowntime _chargethreshold _charge _bisactive _bischarging _bispassive) _cachedoctopusvehicle)
      (Variables|/Script/FortniteGame.FortOctopusVehicle|SetBaseSphericalDriveParams (Utilities|Struct|MakeSphericalDriveParams _radius (Variables|Configurables|GetMaxSpeedKmhPreset) _lowspeedaccelerationforce _highspeedaccelerationforce (Variables|Configurables|GetAngleThresholdPreset) _maxaircontrolforce _maxaircontrolspeedkmh _shellangulardrag (Variables|Configurables|GetDragCoefficientPreset) _dragcoefficientautobrake _maxautobrakespeedkmh _contactrepulsionforce _contactthreshold _massdirectionmaxangledeg _massdirectionstiffness _bmassdirectioninvert) _cachedoctopusvehicle)
      (Variables|/Script/FortniteGame.FortOctopusVehicle|SetBaseTowhookParams (Utilities|Struct|MakeTowhookParams _maxcablelength _mincablelength _busespring _buserope false (Variables|Configurables|GetSpringStiffnessPreset) _springdamping _springmaxstiffnessforce _springmaxstiffnessvelocity _springdeformationrate _ropegive _ropebreakforce 0.0 1e+03 _extendspeedkmh _contractspeedkmh _bapplyspringtoself _bapplyspringtoother _btakeupslack) _cachedoctopusvehicle)
      (Class|FortAthenaVehicle|SetShouldDealDamage _cachedoctopusvehicle (Variables|Configurables|GetbEnableDamagePreset)))
    (:Is%20Not%20Valid
      (Development|PrintString "Not valid Config"))))

(fn OnRep_CachedOctopusVehicle
  (CallFunction|ConfigureOctopusVehicle))

(fn ConstructionScript)
