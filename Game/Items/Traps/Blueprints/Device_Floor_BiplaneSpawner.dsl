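; Floor-placed biplane spawner device that applies configurable prop-destruction health thresholds to its Ferret plane.
; Option changes, property loads and vehicle spawns all call UpdatePlaneWithNewThresholds, which runs on the authority and casts the cached vehicle to a Ferret vehicle.
; It writes the direct-hit, boosting and normal damage thresholds only where they differ from the plane's values, avoiding needless replication.
; ShouldTrigger always returns true.
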(event EventBeginPlay
  (|CallParentFunction))

(event OnAnyPropertyChanged(ToyOptionsComponent)
  (CallFunction|UpdatePlanewithNewThresholds))

(event OnPropertiesLoaded(ToyOptionsComponent)
  (CallFunction|UpdatePlanewithNewThresholds))

(event EventOnVehicleSpawned
  (params (SpawnedVehicle :type Actor%20Object%20Reference))
  (|CallParentFunction SpawnedVehicle)
  (CallFunction|UpdatePlanewithNewThresholds))

(fn ShouldTrigger
  (params (TouchingActors :type Array%20of%20Actor%20Object%20References))
  (return true))

(fn UpdatePlaneWithNewThresholds
  (bind _boostdamagehealththreshold (Variables|Default|GetBoostDamageHealthThreshold))
  (bind _damagehealththreshold (Variables|Default|GetDamageHealthThreshold))
  (bind _directdamagehealththreshold (Variables|Default|GetDirectDamageHealthThreshold))
  (Networking|SwitchHasAuthority
    (:Authority
      (if (Utilities|IsValid (Utilities|ToObject(Interface) (Variables|Default|GetCachedVehicleInterface)) :!class /Script/Engine.KismetSystemLibrary)
        (bind _asferret_vehicle (Utilities|Casting|CastToFerretVehicle (Utilities|ToObject(Interface) (Variables|Default|GetCachedVehicleInterface))))
        (if (Utilities|IsValid _asferret_vehicle :!class /Script/Engine.KismetSystemLibrary)
          (LocalVariables|SetFerretPlane _asferret_vehicle)
          (Utilities|FlowControl|Sequence
            (:then_0
              ;[1 1 1] Only set new values if they are different from what is already set.  Prevents data from replicating that doesn't need to
              (if (!= (Variables|/Script/FortniteGame.FortAthenaFerretVehicle|GetMaxHealthToDestroyPropWithDirectHit (LocalVariables|GetFerretPlane)) _directdamagehealththreshold)
                (Variables|/Script/FortniteGame.FortAthenaFerretVehicle|SetMaxHealthToDestroyPropWithDirectHit _directdamagehealththreshold (LocalVariables|GetFerretPlane))))
            (:then_1
              (if (!= (Variables|/Script/FortniteGame.FortAthenaFerretVehicle|GetMaxHealthToDestroyPropWhileBoosting (LocalVariables|GetFerretPlane)) _boostdamagehealththreshold)
                (Variables|/Script/FortniteGame.FortAthenaFerretVehicle|SetMaxHealthToDestroyPropWhileBoosting _boostdamagehealththreshold (LocalVariables|GetFerretPlane))))
            (:then_2
              (if (!= (Variables|/Script/FortniteGame.FortAthenaFerretVehicle|GetMaxHealthToDestroyProp (LocalVariables|GetFerretPlane)) _damagehealththreshold)
                (Variables|/Script/FortniteGame.FortAthenaFerretVehicle|SetMaxHealthToDestroyProp _damagehealththreshold (LocalVariables|GetFerretPlane))))))))))

(fn ConstructionScript)
