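; Creative floor device Elimination Zone: a Device_Volume_Floor child that damages players, bots, creatures and optionally vehicles inside a scalable trigger volume.
; A use-restriction component filters targets by team and class, and pawns are revalidated on team change, class change, spawn or selected team/class updates.
; Vehicle seat events re-activate the damage ability as occupancy changes.
; Enter/exit events broadcast to transmitters, and edge particles follow zone scale and damage type.
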(event EventBeginPlay
  (|CallParentFunction)
  ;[0 0 0 1 10] Trigger acts as a overlap volume for the function OnAffectedPlayerExitedVehicle in Device_Volume_Floor (parent of this)
  (Trap|AddDamageComponent (Variables|Default|GetTrigger)))

(event EventUpdateVolumeSize
  (|CallParentFunction)
  (Components|Activation|Deactivate (Variables|Default|GetP_EliminationZone_Edges))
  (Transformation|SetWorldScale3D (Variables|Default|GetP_EliminationZone_Edges) (Math|Vector|MakeVector (Variables|UserOptions|GetZoneXScale) (Variables|UserOptions|GetZoneYScale) (Variables|UserOptions|GetZoneZScale)))
  (Variables|/Script/CascadeRuntime.ParticleSystemComponent|SetInstanceParameters (CallFunction|VFXParameters) (Variables|Default|GetP_EliminationZone_Edges))
  (Effects|Components|ParticleSystem|SetTemplate (Variables|Default|GetP_EliminationZone_Edges) (select (== (Variables|UserOptions|GetZoneDamageType) 0) (Variables|Default|GetP_EliminationZone_DirectDamage) (Variables|Default|GetP_EliminationZone_DOTDamage)))
  (Rendering|SetBoundsScale (Variables|Default|GetP_EliminationZone_Edges) (Math|Float|Max(Float) (* (Variables|UserOptions|GetZoneXScale) 512.0) (* (Variables|UserOptions|GetZoneYScale) 512.0) (* (Variables|UserOptions|GetZoneZScale) 384.0) (Variables|Default|GetParticlesMaxDrawDistance))))

(event Custom|UpdateSelectedTeam
  (params (Instigator :type Controller%20Object%20Reference))
  (Networking|SwitchHasAuthority
    (:Authority
      (bind _currentminigame (Utilities|GetCurrentMinigame)
        (:Is%20Valid
          (if (and (Minigame|IsinProgress _currentminigame) (Variables|Default|GetbEnabled))
            (bind _asfort_player_controller_athena (Utilities|Casting|CastToFortPlayerControllerAthena Instigator))
            (Utilities|IsValid _asfort_player_controller_athena
              (:Is%20Valid
                (CallFunction|CustomInitializeTeams (Knobs|Team|ReshiftTeamIndexValue (Math|Conversions|ToInteger(Byte) (Team|GetActorPvPTeamIndex _asfort_player_controller_athena))))
                (CallFunction|RevalidateOverlappingPawns)))))))))

(event Custom|UpdatedSelectedClass
  (params (Instigator :type Controller%20Object%20Reference))
  (Networking|SwitchHasAuthority
    (:Authority
      (bind _currentminigame (Utilities|GetCurrentMinigame)
        (:Is%20Valid
          (Utilities|IsValid (Minigame|GetFNEClassComponent _currentminigame)
            (:Is%20Valid
              (if (and (Minigame|IsinProgress _currentminigame) (Variables|Default|GetbEnabled))
                (bind _asfort_player_controller_athena (Utilities|Casting|CastToFortPlayerControllerAthena Instigator))
                (bind _asfort_player_state (Utilities|Casting|CastToFortPlayerState (Variables|/Script/Engine.Controller|GetPlayerState _asfort_player_controller_athena)))
                (CallFunction|CustomInitializeClass (Math|Conversions|ToInteger(Byte) (Minigame|GetClassSlotforPlayer (Minigame|GetFNEClassComponent _currentminigame) _asfort_player_state false)))
                (CallFunction|RevalidateOverlappingPawns)))))))))

(event OnAddedtoMinigame(FortMinigameProgress)
  (params (Minigame :type Fort%20Minigame%20Object%20Reference))
  (Networking|SwitchHasAuthority
    (:Authority
      ;[1 1 1] Setup FortMinigameProgress to enable ClassRestrictions
      (Initialization|InitializeComponent (Variables|Default|GetUseRestriction) (Variables|Default|GetFortMinigameProgress) 0 0 (Knobs|Team|GetShiftedAnyTeamValue) -1 0 1))))

(event MinigameStartedDelegate(FortMinigameProgress)
  (Networking|SwitchHasAuthority
    (:Authority
      (CallFunction|CustomInitializeClass (Variables|UserOptions|GetKnob_SelectedClass))
      (CallFunction|CustomInitializeTeams (Creative|TeamOption|GetLegacyTeamIndex (Variables|UserOptions|GetZoneProtectedTeam)))
      (bind _currentminigame (Utilities|GetCurrentMinigame)
        (:Is%20Valid
          (Minigame|BindEventtoOnPlayerClassApplied _currentminigame (EventDispatchers|CreateEvent :!event Custom|OnPlayerClassApplied))
          (Minigame|BindEventtoPlayerChangedTeamDelegate _currentminigame (EventDispatchers|CreateEvent :!event Custom|OnPlayerChangedTeam))
          (Minigame|AssignOnMinigamePlayerSpawned _currentminigame (EventDispatchers|CreateEvent :!event Custom|OnMinigamePlayerSpawned)))))))

(event EventInitializeDevice
  ;[1 1 1] Initialize Device
  (|CallParentFunction)
  (CallFunction|CustomInitializeClass (Variables|UserOptions|GetKnob_SelectedClass)))

(event MinigameEndedDelegate(FortMinigameProgress)
  (Networking|SwitchHasAuthority
    (:Authority
      (bind _currentminigame (Utilities|GetCurrentMinigame)
        (:Is%20Valid
          (Minigame|UnbindEventfromOnPlayerClassApplied _currentminigame (EventDispatchers|CreateEvent :!event Custom|OnPlayerClassApplied))
          (Minigame|UnbindEventfromPlayerChangedTeamDelegate _currentminigame (EventDispatchers|CreateEvent :!event Custom|OnPlayerChangedTeam))
          (Minigame|UnbindEventfromOnMinigamePlayerSpawned _currentminigame (EventDispatchers|CreateEvent :!event Custom|OnMinigamePlayerSpawned)))))))

(event Custom|OnPlayerClassApplied
  (params (PlayerState :type Fort%20Player%20State%20Object%20Reference) (ClassSlotIndex :type Byte))
  (bind _returnvalue (PlayerState|GetCurrentPawn PlayerState))
  (Networking|SwitchHasAuthority
    (:Authority
      ;[1 1 1] If the device is disabled there is no overlap
      (if (Collision|IsOverlappingActor (Variables|Default|GetTrigger) _returnvalue :!class /Script/Engine.PrimitiveComponent)
        (CallFunction|RevalidatePawn _returnvalue)))))

(event Custom|OnPlayerChangedTeam
  (params (PlayerState :type Fort%20Player%20State%20Object%20Reference) (TeamId :type Byte) (OldTeamId :type Byte) (bTeamChangedWithoutRespawn :type Boolean))
  (bind _returnvalue (PlayerState|GetCurrentPawn PlayerState))
  (Networking|SwitchHasAuthority
    (:Authority
      ;[1 1 1] If the device is disabled there is no overlap
      (if (Collision|IsOverlappingActor (Variables|Default|GetTrigger) _returnvalue :!class /Script/Engine.PrimitiveComponent)
        (CallFunction|RevalidatePawn _returnvalue)))))

(event Custom|OnMinigamePlayerSpawned
  (params (SpawningPlayer :type Controller%20Object%20Reference))
  (bind _returnvalue (Pawn|GetControlledPawn SpawningPlayer))
  (if (and (Utilities|IsValid _returnvalue :!class /Script/Engine.KismetSystemLibrary) (Collision|IsOverlappingActor (Variables|Default|GetTrigger) _returnvalue :!class /Script/Engine.PrimitiveComponent))
    (CallFunction|RevalidatePawn _returnvalue)))

(event EventPawnEnteredTriggerVolume
  (params (Pawn :type Pawn%20Object%20Reference) (PlayerController :type Player%20Controller%20Object%20Reference))
  ;[0 0.471 1 0.5] Overlapping Volume for Transmitters
  (|CallParentFunction Pawn PlayerController)
  (CallFunction|TransmitEnter/ExitEvent true Pawn PlayerController)
  ;[0 0.4 0.8 0.5] If a vehicle is occupied when it enters the volume, our base class will fire the Pawn trigger volume event but not the non-pawn event, so we add the Non- Pawn event ourselves if needed.
  (bind _vehicle_occupied (Utilities|IsinVehicle Pawn)
    (:true
      (bind _returnvalue (Vehicle|GetVehicleActor _vehicle_occupied :!class /Script/FortniteGame.FortVehicleInterface))
      (CallFunction|Non-PawnEnteredTriggerVolume _returnvalue))))

(event EventPawnExitedTriggerVolume
  (params (Pawn :type Pawn%20Object%20Reference) (PlayerController :type Player%20Controller%20Object%20Reference))
  ;[0 0.471 1 0.5] Overlapping Volume for Transmitters
  (|CallParentFunction Pawn PlayerController)
  (CallFunction|TransmitEnter/ExitEvent false Pawn PlayerController)
  ;[0 0.4 0.8 0.5] If a vehicle is occupied when it enters the volume, our base class will fire the Pawn trigger volume event but not the non-pawn event, so we add the Non- Pawn event ourselves if needed.
  (bind _vehicle_occupied (Utilities|IsinVehicle Pawn)
    (:true
      (bind _returnvalue (Vehicle|GetVehicleActor _vehicle_occupied :!class /Script/FortniteGame.FortVehicleInterface))
      (CallFunction|Non-PawnExitedTriggerVolume _returnvalue))))

(event EventNon-PawnEnteredTriggerVolume
  (params (Actor :type Actor%20Object%20Reference))
  ;[0 0.6 0.6 0.5] Events triggered when a vehicle enters or leaves the damage volume
  (|CallParentFunction Actor)
  (if (Variables|UserOptions|GetbDamageVehicles)
    (bind _asfort_athena_vehicle (Utilities|Casting|CastToFortAthenaVehicle Actor))
    (CallFunction|BindVehicleEvents _asfort_athena_vehicle)))

(event EventNon-PawnExitedTriggerVolume
  (params (Actor :type Actor%20Object%20Reference))
  ;[0 0.6 0.6 0.5] Events triggered when a vehicle enters or leaves the damage volume
  (|CallParentFunction Actor)
  (if (Variables|UserOptions|GetbDamageVehicles)
    (bind _asfort_athena_vehicle (Utilities|Casting|CastToFortAthenaVehicle Actor))
    (CallFunction|UnbindVehicleEvents _asfort_athena_vehicle)))

(event Custom|Player%20Entered%20Vehicle
  (params (Vehicle :type Fort%20Vehicle%20Interface%20Interface%20%28by%20ref%29) (Player%20Pawn :type Fort%20Pawn%20Object%20Reference) (Seat%20Index :type Integer))
  ;[0 0.8 0.4 0.5] Events generated when a player enters or leaves a vehicle inside the damage volume
  (if (== (Utilities|Array|Length (Vehicle|GetAllPassengers Vehicle)) 1)
    (CallFunction|RetriggerEffects)))

(event Custom|Player%20Exited%20Vehicle
  (params (Vehicle :type Fort%20Vehicle%20Interface%20Interface%20%28by%20ref%29) (Player%20Pawn :type Fort%20Pawn%20Object%20Reference) (Seat%20Index :type Integer))
  ;[0 0.8 0.4 0.5] Events generated when a player enters or leaves a vehicle inside the damage volume
  (if (< (Utilities|Array|Length (Vehicle|GetAllPassengers Vehicle)) 1)
    (CallFunction|RetriggerEffects)))

(event EventPlayerBotEnteredTriggerVolume
  (params (Pawn :type Pawn%20Object%20Reference) (AI%20Bot%20Controller :type Fort%20Athena%20AIBot%20Controller%20Object%20Reference))
  ;[0.643 0.6823 0.1546 0.5] Events triggered when a player bot enters or leaves the damage volume
  (|CallParentFunction Pawn AI%20Bot%20Controller)
  ;[0.643 0.6823 0.1546 0.5] If a vehicle is occupied when it enters the volume, our base class will fire the Pawn trigger volume event but not the non-pawn event, so we add the Non- Pawn event ourselves if needed.
  (bind _vehicle_occupied (Utilities|IsinVehicle Pawn)
    (:true
      (bind _returnvalue (Vehicle|GetVehicleActor _vehicle_occupied :!class /Script/FortniteGame.FortVehicleInterface))
      (CallFunction|Non-PawnEnteredTriggerVolume _returnvalue))))

(event EventPlayerBotExitedTriggerVolume
  (params (Pawn :type Pawn%20Object%20Reference) (AI%20Bot%20Controller :type Fort%20Athena%20AIBot%20Controller%20Object%20Reference))
  ;[0.643 0.6823 0.1546 0.5] Events triggered when a player bot enters or leaves the damage volume
  (|CallParentFunction Pawn AI%20Bot%20Controller)
  ;[0.643 0.6823 0.1546 0.5] If a vehicle is occupied when it enters the volume, our base class will fire the Pawn trigger volume event but not the non-pawn event, so we add the Non- Pawn event ourselves if needed.
  (bind _vehicle_occupied (Utilities|IsinVehicle Pawn)
    (:true
      (bind _returnvalue (Vehicle|GetVehicleActor _vehicle_occupied :!class /Script/FortniteGame.FortVehicleInterface))
      (CallFunction|Non-PawnExitedTriggerVolume _returnvalue))))

(fn BindVehicleEvents
  (params (Vehicle :type Fort%20Athena%20Vehicle%20Object%20Reference))
  (CallFunction|UnbindVehicleEvents Vehicle)
  (Vehicle|BindEventtoOnPawnEnteredVehicleSeat Vehicle (EventDispatchers|CreateEvent :!event Player%20Entered%20Vehicle))
  (Vehicle|BindEventtoOnPawnExitedVehicleSeat Vehicle (EventDispatchers|CreateEvent :!event Player%20Exited%20Vehicle)))

(fn ShouldTrigger
  (params (TouchingActors :type Array%20of%20Actor%20Object%20References))
  (bind _array_element (Utilities|Array|ForEachLoopwithBreak TouchingActors)
    (:LoopBody
      (Utilities|IsValid _array_element
        (:Is%20Valid
          (bind _valid_controller (DEPRECATED|ActorHasValidController (Variables|Default|GetUseRestriction) _array_element true true))
          (if (Math|Boolean|ORBoolean _valid_controller (CallFunction|ActorIsAVehicle _array_element) (and (CallFunction|ActorIsACreature _array_element) (StateChecks|IsValidTeam (Variables|Default|GetUseRestriction) _array_element)) (and (StateChecks|IsValidTeam (Variables|Default|GetUseRestriction) _array_element) (CallFunction|ActorIsAPlayerBot _array_element)))
            (LocalVariables|SetShouldTrigger true)))))
    (:Completed
      _))
  (return (LocalVariables|GetShouldTrigger)))

(fn CustomInitializeClass
  (params (SelectedClass :type Integer))
  (Networking|SwitchHasAuthority
    (:Authority
      (ClassRestrictions|SetClassWithAffects (Variables|Default|GetUseRestriction) (Variables|GetSelectedClass) (== (Variables|UserOptions|GetKnob_AffectsClass) (Variables|/Game/Creative/Devices/Common/Components/Creative_UseRestriction_Component.Creative_UseRestriction_Component_C|GetAffectsAllButSelectedValue (Variables|Default|GetUseRestriction))) true))))

(fn CustomInitializeTeams
  (params (SelectedTeam :type Integer))
  (Networking|SwitchHasAuthority
    (:Authority
      (TeamRestrictions|SetTeamswithAffects (Variables|Default|GetUseRestriction) (Variables|GetSelectedTeam) (== (Variables|UserOptions|GetKnob_AffectsTeam) (Variables|/Game/Creative/Devices/Common/Components/Creative_UseRestriction_Component.Creative_UseRestriction_Component_C|GetAffectsAllButSelectedValue (Variables|Default|GetUseRestriction)))))))

(fn GetZoneVisualComponents
  (return (Utilities|Array|MakeArray (Variables|Default|GetP_EliminationZone_Edges))))

(fn InitializeTeam
  ;[1 1 1 1 12] Set a team that will never be equal to any member in the game to avoid friendly fire avoiding damage
  (CallFunction|SetTeam 255)
  (CallFunction|CustomInitializeTeams (Creative|TeamOption|GetLegacyTeamIndex (Variables|UserOptions|GetZoneProtectedTeam))))

(fn RetriggerEffects
  (Ability|Advanced|GrantandActivateAbilityonSelf (Ability|GetFortAbilitySystemComponent self) (Variables|Default|GetDamage%20Applier%20GA) self))

(fn RevalidateOverlappingPawns
  (Utilities|FlowControl|Sequence
    (:then_0
      (for _array_element (Collision|GetOverlappingActors "/Script/FortniteGame.FortPawn")
        (bind _asfort_pawn (Utilities|Casting|CastToFortPawn _array_element))
        (Utilities|ReProcessPawnValidation _asfort_pawn
          (:Retrigger
            (LocalVariables|SetRetriggerTrap true))
          (:OverlapEnd
            (Trap|TriggerOverlapEnd _asfort_pawn)))))
    (:then_1
      (if (LocalVariables|GetRetriggerTrap)
        (Trap|TriggerifAppropriate)))))

(fn RevalidatePawn
  (params (PawnActor :type Actor%20Object%20Reference))
  (Utilities|FlowControl|Sequence
    (:then_0
      (bind _asfort_pawn (Utilities|Casting|CastToFortPawn PawnActor))
      (LocalVariables|SetLocal_PlayerPawn _asfort_pawn)
      (LocalVariables|SetPlayerControllerFromActor (CallFunction|GetPlayerControllerfromActor _asfort_pawn)))
    (:then_1
      (Utilities|ReProcessPawnValidation (LocalVariables|GetLocal_PlayerPawn)
        (:Retrigger
          (Trap|TriggerifAppropriate))
        (:OverlapEnd
          (Trap|TriggerOverlapEnd (LocalVariables|GetLocal_PlayerPawn)))))))

(fn TransmitEnter/ExitEvent
  (params (Enter :type Boolean) (Actor :type Actor%20Object%20Reference) (Controller :type Controller%20Object%20Reference))
  (bind _local_controller (LocalVariables|Getlocal%20Controller))
  (Networking|SwitchHasAuthority
    (:Authority
      (bind _shouldaffect (CallFunction|AffectsTarget Actor))
      (if _shouldaffect
        (bind _controller (Variables|GetController)
          (:then
            (LocalVariables|Setlocal%20Controller _controller))
          (:else
            (bind _asfort_player_pawn_athena (Utilities|Casting|CastToFortPlayerPawnAthena Actor))
            (LocalVariables|Setlocal%20Controller (Pawn|GetController _asfort_player_pawn_athena))))
        (if (Variables|GetEnter)
          (GameplayEvent|BroadcastEvent (Variables|UserOptions|Events|GetOn%20Player%20Entering%20Zone) 0 (Utilities|Struct|MakeGameplayEventEnter _local_controller))
          (else
            (GameplayEvent|BroadcastEvent (Variables|UserOptions|Events|GetOn%20Player%20Leaving%20Zone) 0 (Utilities|Struct|MakeGameplayEventExit _local_controller))))))))

(fn UnbindVehicleEvents
  (params (Vehicle :type Fort%20Athena%20Vehicle%20Object%20Reference))
  (Vehicle|UnbindEventfromOnPawnEnteredVehicleSeat Vehicle (EventDispatchers|CreateEvent :!event Player%20Entered%20Vehicle))
  (Vehicle|UnbindEventfromOnPawnExitedVehicleSeat Vehicle (EventDispatchers|CreateEvent :!event Player%20Exited%20Vehicle)))

(fn ConstructionScript
  (|CallParentFunction)
  (Networking|CanExecuteCosmeticEvents
    (:True
      (Variables|/Script/Engine.SceneComponent|SetbAbsoluteScale true (Variables|Default|GetP_EliminationZone_Edges)))))
