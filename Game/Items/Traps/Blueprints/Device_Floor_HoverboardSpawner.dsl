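; Floor trap device Blueprint that, going by its name, serves as a hoverboard spawner.
; It overrides ShouldTrigger to return true unconditionally, ignoring the touching actors passed in.
; BeginPlay only calls the parent and the construction script is empty, so no spawning logic appears in these graphs.
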(event EventBeginPlay
  (|CallParentFunction))

(fn ShouldTrigger
  (params (TouchingActors :type Array%20of%20Actor%20Object%20References))
  (return true))

(fn ConstructionScript)
