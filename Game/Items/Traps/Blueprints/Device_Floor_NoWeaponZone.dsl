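; Creative trigger-volume device (No Weapon Zone) that applies gameplay effects restricting or altering firing, building, editing, movement and gravity for pawns inside.
; Team and class restrictions pick targets, revalidated on changes; effects use SetByCaller magnitudes and are removed on exit.
; It broadcasts enter, exit and emote events, and retries pawns with uninitialised ability systems.
; Pickup lifespans are overridden, with destroys throttled through a tracked-data limit and deferred queue.
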(event Custom|OnDataTrackingResetDestroyCount
  ;[1 1 1] CustomEvent: OnDataTrackingResetDestroyCount -- Handle Pickups left for later
  (bind _currentminigame (Utilities|GetCurrentMinigame)
    (:Is%20Valid
      (if (Minigame|IsinProgress _currentminigame)
        (bind _array_element (Utilities|Array|ForEachLoopwithBreak (Variables|Default|GetPickupsToHandleLater))
          (:LoopBody
            (if (CallFunction|CanDestroyNow)
              _
              (else
                _))
            (CallFunction|HandlePickUpBeginOverlap _array_element))
          (:Completed
            _))
        (CallFunction|CleanPickupstoHandleLater)))))

(event OnAddedtoMinigame(FortMinigameProgress)
  (params (Minigame :type Fort%20Minigame%20Object%20Reference))
  ;[0.3688 0.2453 1] OnAddedToMinigame
  (Networking|SwitchHasAuthority
    (:Authority
      ;;[1 1 1] OnAddedToMinigame: Track Device Reference once we have a Minigame because we need it later
      (CallFunction|InitializeDeviceDataTracker)
      ;;[1 1 1] Setup FortMinigameProgress to enable ClassRestrictions
      (Initialization|InitializeComponent (Variables|Default|GetUseRestriction) (Variables|Default|GetFortMinigameProgress) 0 0 (Knobs|Team|GetShiftedAnyTeamValue) -1 0 1))))

(event OnRemovedfromMinigame(FortMinigameProgress)
  (params (Minigame :type Fort%20Minigame%20Object%20Reference))
  (bind _noweaponzonetrackeddata (Variables|Default|GetNoWeaponZoneTrackedData))
  ;[1 1 1] OnRemovedFromMinigame:
  (Networking|SwitchHasAuthority
    (:Authority
      (Utilities|IsValid _noweaponzonetrackeddata
        (:Is%20Valid
          (Class|NoWeaponZoneTrackedData|UntrackDevice _noweaponzonetrackeddata))))))

(event Custom|OnPickupDestroyed
  (params (DestroyedActor :type Actor%20Object%20Reference))
  ;[1 1 1] OnPickupDestroyed Remove reference from Array and Track destroy
  (Networking|SwitchHasAuthority
    (:Authority
      (Game|UnbindEventfromOnDestroyed DestroyedActor (delegate Custom|OnPickupDestroyed))
      (bind _currentminigame (Utilities|GetCurrentMinigame)
        (:Is%20Valid
          (if (Minigame|IsinProgress _currentminigame)
            (Utilities|Array|RemoveItem (Variables|Default|GetActorsPendingDestroy) DestroyedActor)
            (Class|NoWeaponZoneTrackedData|AddTrackedDestroy (Variables|Default|GetNoWeaponZoneTrackedData))))))))

(event EventInitializeDevice
  ;[0 1 0.1634] Initialization
  (|CallParentFunction)
  (CallFunction|UpdateActiveGameplayEffectsClasses)
  (Default|CustomInitializeClass (Variables|UserOptions|GetKnob_SelectedClass)))

(event Custom|HandleAbilitySystemComponentInitialized
  (params (AbilitySystemComponent :type Fort%20Ability%20System%20Component%20Object%20Reference) (AffectedPawn :type Fort%20Player%20Pawn%20Object%20Reference))
  ;[1 1 1] Re-attempt Trigger Volume Entered if the Actor did not have its ability system component ready.
  (Utilities|IsValid AffectedPawn
    (:Is%20Valid
      (CallFunction|BindOrUnbindAbilitySystemRetryDelegate AffectedPawn)
      (CallFunction|PawnEnteredTriggerVolume AffectedPawn (Fortnite|GetFortPlayerControllerfromActor AffectedPawn)))))

(event EventPawnEnteredTriggerVolume
  (params (Pawn :type Pawn%20Object%20Reference) (PlayerController :type Player%20Controller%20Object%20Reference))
  ;[0 1 0.9238 0.5] Player Entered the Volume
  ;;[0 0 0 1 10] When the device is disabled we need to reset the lifetimes of the items within the Trigger bounds. To do this we need to briefly set the collision of Trigger to Collision Enabled so that overlaps happen. If a pawn is inside, EventPawnEnteredTriggerVolume will also get called, but this brief collision update should not be affecting pawns.
  (if (Variables|Default|GetbEnabled)
    (Networking|SwitchHasAuthority
      (:Authority
        (bind _isvalidtarget (CallFunction|ShouldApplyEffects Pawn PlayerController))
        (if _isvalidtarget
          (bind _applied (CallFunction|ApplyEffectsToPawn Pawn PlayerController))
          (bind _instigator (CallFunction|GetInstigatorforBroadcastifValidActor Pawn PlayerController))
          (Utilities|IsValid _instigator
            (:Is%20Valid
              (GameplayEvent|BroadcastEvent (Variables|UserOptions|Events|GetOn%20Player%20Entering%20Zone) 0 (Utilities|Struct|MakeGameplayEventonPlayerEnteringZone _instigator))
              (if (and (Variables|UserOptions|GetbKnob_EnableVFX) _applied)
                (CallFunction|PlayFX Pawn)))))))))

(event Custom|OnPawnStartedEmote
  (params (MontageItemDef :type Fort%20Item%20Definition%20Object%20Reference) (PawnEmoting :type Fort%20Pawn%20Object%20Reference))
  (bind _returnvalue (Pawn|GetController PawnEmoting))
  ;[0.9747 1 0] Emote Events
  (Utilities|IsValid _returnvalue
    (:Is%20Valid
      (bind _instigator (CallFunction|GetInstigatorforBroadcastifValidActor PawnEmoting _returnvalue))
      (Utilities|IsValid _instigator
        (:Is%20Valid
          (GameplayEvent|BroadcastEvent (Variables|UserOptions|Events|GetOn%20Player%20Starts%20Emoting) 0 (Utilities|Struct|MakeGameplayEventonPlayerStartedEmoting _instigator)))))))

(event Custom|OnPawnStoppedEmote
  (params (MontageItemDef :type Fort%20Item%20Definition%20Object%20Reference) (PawnEmoting :type Fort%20Pawn%20Object%20Reference))
  (bind _returnvalue (Pawn|GetController PawnEmoting))
  ;[0.9747 1 0] Emote Events
  (Utilities|FlowControl|Delay 0.0) ; Wait for next tick to ensure\nEmoteAbility is not Active anymore
  (Utilities|IsValid _returnvalue
    (:Is%20Valid
      (bind _asathena_player_controller (Utilities|Casting|CastToAthena_PlayerController _returnvalue))
      (if (not (Emote|IsPlayingEmote _asathena_player_controller))
        (bind _instigator (CallFunction|GetInstigatorforBroadcastifValidActor PawnEmoting _asathena_player_controller))
        (Utilities|IsValid _instigator
          (:Is%20Valid
            (GameplayEvent|BroadcastEvent (Variables|UserOptions|Events|GetOn%20Player%20Stops%20Emoting) 0 (Utilities|Struct|MakeGameplayEventonPlayerStoppedEmoting _instigator))))))))

(event EventPawnExitedTriggerVolume
  (params (Pawn :type Pawn%20Object%20Reference) (PlayerController :type Player%20Controller%20Object%20Reference))
  ;[1 0.2115 0] Player Exited the Volume
  (Networking|SwitchHasAuthority
    (:Authority
      (bind _shouldaffect (CallFunction|AffectsTarget Pawn))
      (if _shouldaffect
        (CallFunction|RemoveEffectsAndUnbindAbilitySystem Pawn)
        (bind _instigator (CallFunction|GetInstigatorforBroadcastifValidActor Pawn PlayerController))
        (Utilities|IsValid _instigator
          (:Is%20Valid
            (GameplayEvent|BroadcastEvent (Variables|UserOptions|Events|GetOn%20Player%20Leaving%20Zone) 0 (Utilities|Struct|MakeGameplayEventonPlayerExitingZone _instigator))))))))

(event Custom|Update%20Selected%20Team%20When%20Receiving%20From%20Function
  (params (Instigator :type Controller%20Object%20Reference))
  (if (Variables|Default|GetbEnabled)
    (Networking|SwitchHasAuthority
      (:Authority
        (bind _asfort_player_controller_athena (Utilities|Casting|CastToFortPlayerControllerAthena Instigator))
        (Utilities|IsValid _asfort_player_controller_athena
          (:Is%20Valid
            (Default|CustomInitializeTeam (Knobs|Team|ReshiftTeamIndexValue (Math|Conversions|ToInteger(Byte) (Team|GetActorPvPTeamIndex _asfort_player_controller_athena))))
            (CallFunction|UpdateOverlappingPawns)))))))

(event EventNon-PawnEnteredTriggerVolume
  (params (Actor :type Actor%20Object%20Reference))
  ;[0 0.75 1] Non Player Entered the Volume (Test for Pickup to override Lifetime if we have that set)
  (Networking|SwitchHasAuthority
    (:Authority
      (CallFunction|HandlePickUpBeginOverlap Actor))))

(event Custom|Update%20Item%20Lifetimes
  ;[0.25 0.25 0.25] Update item lifespans
  (Networking|SwitchHasAuthority
    (:Authority
      (if (not (Variables|Default|GetbEnabled))
        ;;[0.5 0.5 0.5 1 14] This looks strange but is the simplist way to reset the lifetimes of the items in the Trigger bounds
        (CallFunction|UpdateTriggerCollision false)
        (CallFunction|UpdateTriggerCollision true)))))

(event EventUpdateVolumeState
  ;[0.25 0.25 0] Update volume state
  (|CallParentFunction)
  (CallFunction|OnVolumeStateUpdated))

(event Custom|Update%20Selected%20Class%20When%20Receiving%20From%20Function
  (params (Instigator :type Controller%20Object%20Reference))
  (if (Variables|Default|GetbEnabled)
    (Networking|SwitchHasAuthority
      (:Authority
        (bind _fortminigameprogress (Variables|Default|GetFortMinigameProgress)
          (:then
            (bind _currentminigame (Variables|/Script/FortniteGame.FortMinigameLogicComponent|GetCurrentMinigame _fortminigameprogress)
              (:then
                (Utilities|IsValid (Minigame|GetFNEClassComponent _currentminigame)
                  (:Is%20Valid
                    (bind _asfort_player_controller_athena (Utilities|Casting|CastToFortPlayerControllerAthena Instigator))
                    (bind _asfort_player_state (Utilities|Casting|CastToFortPlayerState (Variables|/Script/Engine.Controller|GetPlayerState _asfort_player_controller_athena)))
                    (Default|CustomInitializeClass (Math|Conversions|ToInteger(Byte) (Minigame|GetClassSlotforPlayer (Minigame|GetFNEClassComponent _currentminigame) _asfort_player_state false)))
                    (CallFunction|UpdateOverlappingPawns))))
              (:else)))
          (:else))))))

(event MinigameStartedDelegate(FortMinigameProgress)
  (Networking|SwitchHasAuthority
    (:Authority
      (Default|CustomInitializeTeam (Creative|TeamOption|GetLegacyTeamIndex (Variables|UserOptions|GetZoneProtectedTeam)))
      (Default|CustomInitializeClass (Variables|UserOptions|GetKnob_SelectedClass))
      (bind _currentminigame (Utilities|GetCurrentMinigame)
        (:Is%20Valid
          (Minigame|BindEventtoPlayerChangedTeamDelegate _currentminigame (EventDispatchers|CreateEvent :!event Custom|OnPlayerChangedTeam))
          (Minigame|AssignOnPlayerClassApplied _currentminigame (EventDispatchers|CreateEvent :!event Custom|OnPlayerClassApplied)))))))

(event MinigameEndedDelegate(FortMinigameProgress)
  (Networking|SwitchHasAuthority
    (:Authority
      (CallFunction|ClearPendingPickupsReferences)
      (bind _currentminigame (Utilities|GetCurrentMinigame)
        (:Is%20Valid
          (Minigame|UnbindEventfromPlayerChangedTeamDelegate _currentminigame (EventDispatchers|CreateEvent :!event Custom|OnPlayerChangedTeam))
          (Minigame|UnbindEventfromOnPlayerClassApplied _currentminigame (EventDispatchers|CreateEvent :!event Custom|OnPlayerClassApplied)))))))

(event Custom|OnPlayerChangedTeam
  (params (PlayerState :type Fort%20Player%20State%20Object%20Reference) (TeamId :type Byte) (OldTeamId :type Byte) (bTeamChangedWithoutRespawn :type Boolean))
  (bind _returnvalue (PlayerState|GetCurrentPawn PlayerState))
  (if bTeamChangedWithoutRespawn
    (if (Collision|IsOverlappingActor (Variables|Default|GetTrigger) _returnvalue :!class /Script/Engine.PrimitiveComponent)
      (bind (_valid_controller _fortplayerpawn _playercontroller) (DEPRECATED|ActorHasValidController (Variables|Default|GetUseRestriction) _returnvalue true true))
      (CallFunction|RemoveEffects _fortplayerpawn)
      (if _valid_controller
        (bind _isvalidtarget (CallFunction|ShouldApplyEffects _fortplayerpawn _playercontroller))
        (if _isvalidtarget
          (bind _applied (CallFunction|ApplyEffectsToPawn _fortplayerpawn))
          (if (and _applied (Variables|UserOptions|GetbKnob_EnableVFX))
            (CallFunction|PlayFX _fortplayerpawn))
          (else
            _))
        (else
          _))
      (CallFunction|BindOrUnbindAbilitySystemRetryDelegate _fortplayerpawn))))

(event Custom|OnPlayerClassApplied
  (params (PlayerState :type Fort%20Player%20State%20Object%20Reference) (ClassSlotIndex :type Byte))
  (bind _returnvalue (PlayerState|GetCurrentPawn PlayerState))
  (Networking|SwitchHasAuthority
    (:Authority
      ;[1 1 1] If the device is disabled there is no overlap
      (if (Collision|IsOverlappingActor (Variables|Default|GetTrigger) _returnvalue :!class /Script/Engine.PrimitiveComponent)
        (CallFunction|RevalidatePawnOverlap _returnvalue)))))

(event EventPlayerBotEnteredTriggerVolume
  (params (Pawn :type Pawn%20Object%20Reference) (AI%20Bot%20Controller :type Fort%20Athena%20AIBot%20Controller%20Object%20Reference))
  ;[0 1 0.9238 0.5] Player Bot Entered the Volume
  ;;[0 0 0 1 10] When the device is disabled we need to reset the lifetimes of the items within the Trigger bounds. To do this we need to briefly set the collision of Trigger to Collision Enabled so that overlaps happen. If a pawn is inside, EventPawnEnteredTriggerVolume will also get called, but this brief collision update should not be affecting pawns.
  (if (Variables|Default|GetbEnabled)
    (Networking|SwitchHasAuthority
      (:Authority
        (bind _isvalidtarget (CallFunction|ShouldApplyEffects Pawn AI%20Bot%20Controller))
        (if _isvalidtarget
          (bind _applied (PlayerBot|ApplyEffectsToPlayerBot Pawn AI%20Bot%20Controller))
          (if (and _applied (Variables|UserOptions|GetbKnob_EnableVFX))
            (CallFunction|PlayFX Pawn)))))))

(event EventPlayerBotExitedTriggerVolume
  (params (Pawn :type Pawn%20Object%20Reference) (AI%20Bot%20Controller :type Fort%20Athena%20AIBot%20Controller%20Object%20Reference))
  ;[1 0.2115 0] Player Bot Exited the Volume
  (Networking|SwitchHasAuthority
    (:Authority
      (bind _shouldaffect (CallFunction|AffectsTarget Pawn))
      (if _shouldaffect
        (CallFunction|RemoveEffectsAndUnbindAbilitySystem Pawn)))))

(event Custom|PlayFX
  (params (Overlapped%20Pawn :type Pawn%20Object%20Reference))
  (bind _ascharacter (Utilities|Casting|CastToCharacter Overlapped%20Pawn))
  (Networking|CanExecuteCosmeticEvents
    (:True
      (Effects|SpawnEmitterAttached :EmitterTemplate "/Game/Abilities/Toys/NoWeaponZone/P_NoWeaponZone_Remove.P_NoWeaponZone_Remove" :AttachToComponent (Variables|/Script/Engine.Character|GetMesh _ascharacter) :AttachPointName "AttachSocket" :Scale "1.000000,1.000000,1.000000" :LocationType "KeepRelativeOffset" :bAutoDestroy true :PoolingMethod "AutoRelease" :bAutoActivate true))))

(fn ApplyEffectstoPawn
  (params (Pawn :type Actor%20Object%20Reference) (Controller :type Player%20Controller%20Object%20Reference))
  (bind _ability_system (LocalVariables|GetAbility%20System))
  (Utilities|IsValid Pawn
    (:Is%20Valid
      (Default|RegisterPlayerinVolume Pawn)
      (bind _asfort_ability_system_component (Utilities|Casting|CastToFortAbilitySystemComponent (Ability|GetAbilitySystemComponent Pawn))
        (:then
          (LocalVariables|SetAbility%20System _asfort_ability_system_component)
          (Tags|CancelAbilitiesWithTags _ability_system (GameplayTags|MakeLiteralGameplayTagContainer (Variables|Default|GetCancelAbilitiesWithTheseTags)))
          (CallFunction|BindorUnbindonPlayerEmote (Variables|GetPawn) true)
          (CallFunction|ExecuteOnPlayerStartEmote (Variables|Default|GetController) (Variables|GetPawn))
          (bind (_should_apply _spec_handles) (CallFunction|SetupEffectData _ability_system))
          (if _should_apply
            (for _array_element _spec_handles
              (GameplayEffects|ApplyGameplayEffectSpecToSelf (LocalVariables|GetAbility%20System) _array_element))
            (return true)
            (else
              (return false))))
        (:CastFailed
          (return false))))))

(fn ApplyEffectstoPlayerBot
  (params (Pawn :type Actor%20Object%20Reference) (AI%20Bot%20Controller :type Fort%20Athena%20AIBot%20Controller%20Object%20Reference))
  (bind _ability_system (LocalVariables|GetAbility%20System))
  (Utilities|IsValid Pawn
    (:Is%20Valid
      (bind _asfort_ability_system_component (Utilities|Casting|CastToFortAbilitySystemComponent (Ability|GetAbilitySystemComponent Pawn))
        (:then
          (LocalVariables|SetAbility%20System _asfort_ability_system_component)
          (Tags|CancelAbilitiesWithTags _ability_system (GameplayTags|MakeLiteralGameplayTagContainer (Variables|Default|GetCancelAbilitiesWithTheseTags)))
          (CallFunction|BindorUnbindonPlayerEmote (Variables|GetPawn) true)
          (bind (_should_apply _spec_handles) (CallFunction|SetupEffectData _ability_system))
          (if _should_apply
            (for _array_element _spec_handles
              (GameplayEffects|ApplyGameplayEffectSpecToSelf (LocalVariables|GetAbility%20System) _array_element))
            (return true)
            (else
              (return false))))
        (:CastFailed
          (return false))))))

(fn BindorUnbindAbilitySystemRetryDelegate
  (params (Pawn :type Pawn%20Object%20Reference) (Bind :type Boolean))
  (bind _outputdelegate (EventDispatchers|CreateEvent :!event HandleAbilitySystemComponentInitialized))
  (bind _asfort_player_pawn (Utilities|Casting|CastToFortPlayerPawn Pawn))
  (if Bind
    (EventDispatchers|BindEventtoOnAbilitySystemComponentInitialized _asfort_player_pawn _outputdelegate)
    (else
      (EventDispatchers|UnbindEventfromOnAbilitySystemComponentInitialized _asfort_player_pawn _outputdelegate))))

(fn BindOrUnbindOnPlayerEmote
  (params (Pawn :type Actor%20Object%20Reference) (Bind :type Boolean))
  (bind _outputdelegate (EventDispatchers|CreateEvent :!event OnPawnStartedEmote))
  (bind _outputdelegate_2 (EventDispatchers|CreateEvent :!event OnPawnStoppedEmote))
  (Networking|SwitchHasAuthority
    (:Authority
      (bind _asfort_pawn (Utilities|Casting|CastToFortPawn (Variables|GetPawn)))
      (if (Variables|GetBind)
        (EventDispatchers|BindEventtoOnPawnStartedEmote _asfort_pawn _outputdelegate)
        (EventDispatchers|BindEventtoOnPawnStoppedEmote _asfort_pawn _outputdelegate_2)
        (else
          (EventDispatchers|UnbindEventfromOnPawnStartedEmote _asfort_pawn _outputdelegate)
          (EventDispatchers|UnbindEventfromOnPawnStoppedEmote _asfort_pawn _outputdelegate_2))))))

(fn BindUnbindToOnResetDestroyCount
  (params (Bind :type Boolean))
  (bind _outputdelegate (EventDispatchers|CreateEvent :!event OnDataTrackingResetDestroyCount))
  (bind _noweaponzonetrackeddata (Variables|Default|GetNoWeaponZoneTrackedData))
  (if Bind
    (EventDispatchers|BindEventtoOnResetActorsDestroyedCount _noweaponzonetrackeddata _outputdelegate)
    (else
      (EventDispatchers|UnbindEventfromOnResetActorsDestroyedCount _noweaponzonetrackeddata _outputdelegate))))

(fn CanDestroyNow
  (bind _returnvalue (Class|NoWeaponZoneTrackedData|GetMaxDestroyCountPerTrackedDevice (Variables|Default|GetNoWeaponZoneTrackedData)))
  (bind _noweaponzonetrackeddata (Variables|Default|GetNoWeaponZoneTrackedData))
  (Utilities|IsValid _noweaponzonetrackeddata
    (:Is%20Valid
      (if (== _returnvalue (LocalVariables|GetMaxDestroyCountDisabled))
        (return true)
        (else
          (return (and (not (Class|NoWeaponZoneTrackedData|IsTrackedDestroyatLimit _noweaponzonetrackeddata)) (< (+ (Utilities|Array|Length (Variables|Default|GetActorsPendingDestroy)) (Class|NoWeaponZoneTrackedData|GetDestroyCount _noweaponzonetrackeddata)) _returnvalue))))))
    (:Is%20Not%20Valid
      (return true))))

(fn CleanPickupsToHandleLater
  (bind _pickupstohandlelater (Variables|Default|GetPickupsToHandleLater))
  (bind _returnvalue (Utilities|Array|Length _pickupstohandlelater))
  (bind _iterationindex (LocalVariables|GetIterationIndex))
  (bind _pickupstohandlelater_2 (Variables|Default|GetPickupsToHandleLater))
  (bind _output (Utilities|Array|Get(acopy) _pickupstohandlelater_2 _iterationindex))
  (if (> _returnvalue 0)
    (for _ (range (LocalVariables|GetIterationIndex) (+ _returnvalue 1))
      (Utilities|IsValid _output
        (:Is%20Valid
          (if (not (Utilities|Array|ContainsItem (Variables|Default|GetActorsPendingDestroy) _output))
            (LocalVariables|SetIterationIndex (+ (LocalVariables|GetIterationIndex) 1))))
        (:Is%20Not%20Valid
          _))
      (Utilities|Array|RemoveIndex (Variables|Default|GetPickupsToHandleLater) (LocalVariables|GetIterationIndex)))))

(fn ClearPendingPickupsReferences
  (Utilities|Array|Clear (Variables|Default|GetActorsPendingDestroy))
  (Utilities|Array|Clear (Variables|Default|GetPickupsToHandleLater)))

(fn CustomInitializeClass
  (params (SelectedClass :type Integer))
  (Networking|SwitchHasAuthority
    (:Authority
      (ClassRestrictions|SetClassWithAffects (Variables|Default|GetUseRestriction) SelectedClass (== (Variables|UserOptions|GetKnob_AffectsClass) (Variables|/Game/Creative/Devices/Common/Components/Creative_UseRestriction_Component.Creative_UseRestriction_Component_C|GetAffectsAllButSelectedValue (Variables|Default|GetUseRestriction))) true))))

(fn CustomInitializeTeam
  (params (SelectedTeam :type Integer))
  (Networking|SwitchHasAuthority
    (:Authority
      ;[1 1 1] This is here to respect parent behaviour
      (CallFunction|SetTeam (Teams|IndextoTeam (Knobs|Team|GetTeamValueasIndex SelectedTeam)))
      (TeamRestrictions|SetTeamswithAffects (Variables|Default|GetUseRestriction) (Variables|GetSelectedTeam) (== (Variables|UserOptions|GetKnob_AffectsTeam) (Variables|/Game/Creative/Devices/Common/Components/Creative_UseRestriction_Component.Creative_UseRestriction_Component_C|GetAffectsAllButSelectedValue (Variables|Default|GetUseRestriction)))))))

(fn ExecuteonPlayerStartEmote
  (params (Controller :type Player%20Controller%20Object%20Reference) (FortPawn :type Actor%20Object%20Reference))
  (bind _asathena_player_controller (Utilities|Casting|CastToAthena_PlayerController Controller))
  (if (Emote|IsPlayingEmote _asathena_player_controller)
    (bind _asfort_pawn (Utilities|Casting|CastToFortPawn FortPawn))
    (CallFunction|OnPawnStartedEmote 0 _asfort_pawn)))

(fn GetAbilitySystemComponentAndPawn
  (params (Object :type Object%20Reference))
  (bind _asfort_pawn (Utilities|Casting|CastToFortPawn Object)
    (:then
      (Utilities|IsValid _asfort_pawn
        (:Is%20Valid
          (bind _asfort_ability_system_component (Utilities|Casting|CastToFortAbilitySystemComponent (Ability|GetAbilitySystemComponent _asfort_pawn))
            (:then
              (return _asfort_ability_system_component _asfort_pawn))
            (:CastFailed
              _)))
        (:Is%20Not%20Valid
          _))
      (return 0 _asfort_pawn))
    (:CastFailed
      _)))

(fn GetDesiredLifeSpan
  (return (Math|Float|SelectFloat (Variables|Default|GetAlmostImmediateLifeSpan) (Variables|UserOptions|GetPickUpLifeSpan) (Math|Float|NearlyEqual(Float) (Variables|UserOptions|GetPickUpLifeSpan)))))

(fn GetZoneVisualComponents
  (return (Utilities|Array|MakeArray (Variables|Default|GetBounds) (Variables|Default|GetIdleFX))))

(fn HandlePickUpBeginOverlap
  (params (Actor :type Actor%20Object%20Reference))
  (if (Math|Boolean|ANDBoolean (Utilities|IsValid Actor :!class /Script/Engine.KismetSystemLibrary) (Utilities|ClassIsChildOf (Utilities|GetClass Actor) "/Script/FortniteGame.FortPickup") (>= (Variables|UserOptions|GetPickUpLifeSpan) 0.0))
    ;[1 1 1] When disabled we reset Pickup Life Span
    (if (Variables|Default|GetbEnabled)
      (if (Math|Float|NearlyEqual(Float) (CallFunction|GetDesiredLifeSpan) -1.0)
        (CallFunction|ResetPickupLifespanandStopDespawnTimer Actor)
        (elif (CallFunction|CanDestroyNow)
          (Utilities|Array|Add (Variables|Default|GetActorsPendingDestroy) Actor)
          (Utilities|FlowControl|Sequence
            (:then_0
              (bind _asfort_pickup (Utilities|Casting|CastToFortPickup Actor))
              (Pickup|SetDespawnTime _asfort_pickup (CallFunction|GetDesiredLifeSpan)))
            (:then_1
              (Game|BindEventtoOnDestroyed Actor (EventDispatchers|CreateEvent :!event OnPickupDestroyed))))
          (else
            (Utilities|Array|Add (Variables|Default|GetPickupsToHandleLater) Actor))))
      (else
        (CallFunction|ResetPickupLifespanandStopDespawnTimer Actor)))))

(fn InitializeTeam
  (Networking|SwitchHasAuthority
    (:Authority
      (Default|CustomInitializeTeam (Creative|TeamOption|GetLegacyTeamIndex (Variables|UserOptions|GetZoneProtectedTeam))))))

(fn InitializeDeviceDataTracker
  (bind _fortminigameprogress (Variables|Default|GetFortMinigameProgress))
  (bind _currentminigame (Variables|/Script/FortniteGame.FortMinigameLogicComponent|GetCurrentMinigame _fortminigameprogress))
  (bind _volume (Variables|/Script/FortniteGame.FortMinigame|GetVolume _currentminigame))
  (bind _returnvalue (Actor|GetComponentbyClass _volume "/Script/FortniteGame.DevicesDataTrackingComponent"))
  (Utilities|IsValid _fortminigameprogress
    (:Is%20Valid
      (Utilities|IsValid _currentminigame
        (:Is%20Valid
          (Utilities|IsValid _volume
            (:Is%20Valid
              (Utilities|IsValid _returnvalue
                (:Is%20Valid
                  (bind _returnvalue_2 (Class|DevicesDataTrackingComponent|AddNewTracker _returnvalue "/Script/FortniteGame.NoWeaponZoneTrackedData"))
                  (bind _asno_weapon_zone_tracked_data (Utilities|Casting|CastToNoWeaponZoneTrackedData _returnvalue_2))
                  (bind _output_get (Variables|Default|SetNoWeaponZoneTrackedData _asno_weapon_zone_tracked_data))
                  (Class|NoWeaponZoneTrackedData|TrackNewDevice _output_get)
                  (bind _currentminigame_2 (Utilities|GetCurrentMinigame)
                    (:Is%20Valid
                      (Class|NoWeaponZoneTrackedData|SetupMinigame _output_get _currentminigame_2))))))))))))

(fn OnVolumeStateUpdated
  (CallFunction|UpdateItemLifetimes)
  (Networking|SwitchHasAuthority
    (:Authority
      (CallFunction|BindUnbindtoonResetDestroyCount (Variables|Default|GetbEnabled))
      (if (not (Variables|Default|GetbEnabled))
        (CallFunction|ClearPendingPickupsReferences)))))

(fn RemoveEffectsandUnbindAbilitySystem
  (params (Pawn :type Pawn%20Object%20Reference))
  (Networking|SwitchHasAuthority
    (:Authority
      (CallFunction|RemoveEffects Pawn)
      (CallFunction|BindOrUnbindAbilitySystemRetryDelegate Pawn))))

(fn RemoveEffects
  (params (Actor :type Actor%20Object%20Reference))
  (Utilities|IsValid Actor
    (:Is%20Valid
      (Default|UnregisterPlayerinVolume Actor)
      (Networking|SwitchHasAuthority
        (:Authority
          (Utilities|FlowControl|Sequence
            (:then_0
              (bind _output_get (LocalVariables|SetAbility%20System (Ability|GetAbilitySystemComponent Actor)))
              (Utilities|IsValid _output_get
                (:Is%20Valid
                  (Utilities|Struct|SetmembersinGameplayEffectQuery (LocalVariables|GetGameplayQuery) self)
                  (bind _returnvalue (GameplayEffects|GetActiveGameplayEffectsforQuery (LocalVariables|GetAbility%20System) (LocalVariables|GetGameplayQuery)))
                  (for _array_element _returnvalue
                    (GameplayEffects|RemoveActiveGameplayEffect (LocalVariables|GetAbility%20System) _array_element -1))
                  (Ability|SendGameplayEventtoActor (Variables|GetActor) "(TagName=\"Creative.Toys.MutatorZone.Exit\")" (Utilities|Struct|MakeGameplayEventData :OptionalObject self)))))
            (:then_1
              (CallFunction|BindorUnbindonPlayerEmote Actor))))))))

(fn ResetPickupLifespanAndStopDespawnTimer
  (params (Pickup :type Actor%20Object%20Reference))
  (Actor|SetLifeSpan Pickup 0.0)
  (bind _asfort_pickup (Utilities|Casting|CastToFortPickup Pickup))
  (Pickup|SetDespawnTime _asfort_pickup 0.0))

(fn RevalidatePawnOverlap
  (params (ActorToRevalidate :type Actor%20Object%20Reference))
  (Utilities|IsValid ActorToRevalidate
    (:Is%20Valid
      (bind _asfort_pawn (Utilities|Casting|CastToFortPawn ActorToRevalidate))
      (if (and (StateChecks|IsValidTeam (Variables|Default|GetUseRestriction) ActorToRevalidate) (StateChecks|IsValidClass (Variables|Default|GetUseRestriction) (CallFunction|GetPlayerControllerfromActor _asfort_pawn)))
        (if (not (Default|IsPlayerRegistered ActorToRevalidate))
          (CallFunction|ComponentBeginOverlap ActorToRevalidate))
        (else
          (if (Default|IsPlayerRegistered ActorToRevalidate)
            (bind _asfort_pawn_2 (Utilities|Casting|CastToFortPawn ActorToRevalidate))
            (CallFunction|RemoveEffectsAndUnbindAbilitySystem _asfort_pawn_2)))))))

(fn SetupEffectData
  (params (Ability%20System :type Ability%20System%20Component%20Object%20Reference))
  (bind _ability_system (Variables|GetAbility%20System))
  (bind _effect_context (LocalVariables|GetEffect%20Context))
  (bind _output_get (LocalVariables|SetEffect%20Context (GameplayEffects|MakeEffectContext Ability%20System)))
  (Ability|EffectContext|AddSourceObject _output_get self)
  (for _array_element (Variables|Default|GetActive%20Classes)
    (if (Utilities|Equal(Class) _array_element (Variables|Default|GetSpeedMultiplierEffectClass))
      (bind _returnvalue (Ability|GameplayEffect|AssignTagSetbyCallerMagnitude (GameplayEffects|MakeOutgoingSpec _ability_system _array_element 0.0 _effect_context) "(TagName=\"SetByCaller.Magnitude\")" (Variables|UserOptions|GetSpeedMultiplier)))
      (Utilities|Array|Add (LocalVariables|GetHandles) _returnvalue)
      (elif (Utilities|Equal(Class) _array_element (Variables|Default|GetGravityOverrideEffectClass))
        (bind _returnvalue_2 (Ability|GameplayEffect|AssignTagSetbyCallerMagnitude (GameplayEffects|MakeOutgoingSpec _ability_system _array_element 0.0 _effect_context) "(TagName=\"SetByCaller.Magnitude\")" (Variables|UserOptions|GetGravityOverride)))
        (Utilities|Array|Add (LocalVariables|GetHandles) _returnvalue_2)
        (elif (Utilities|Equal(Class) _array_element (Variables|Default|GetBuilding%20Speed%20Effect%20Class))
          (bind _returnvalue_3 (Ability|GameplayEffect|AssignTagSetbyCallerMagnitude (GameplayEffects|MakeOutgoingSpec _ability_system _array_element 0.0 _effect_context) "(TagName=\"SetByCaller.Magnitude\")" (Variables|UserOptions|GetBuildingSpeed)))
          (Utilities|Array|Add (LocalVariables|GetHandles) _returnvalue_3)
          (elif (Utilities|Equal(Class) _array_element (Variables|Default|GetBuilding%20Cost%20Effect%20Class))
            (bind _returnvalue_4 (Ability|GameplayEffect|AssignTagSetbyCallerMagnitude (GameplayEffects|MakeOutgoingSpec _ability_system _array_element 0.0 _effect_context) "(TagName=\"SetByCaller.Magnitude\")" (Variables|UserOptions|GetBuildingCost)))
            (Utilities|Array|Add (LocalVariables|GetHandles) _returnvalue_4)
            (else
              (Utilities|Array|Add (LocalVariables|GetHandles) (GameplayEffects|MakeOutgoingSpec _ability_system _array_element 0.0 _effect_context)))))))
    (LocalVariables|SetApply true))
  (return (LocalVariables|GetApply) (LocalVariables|GetHandles)))

(fn ShouldApplyEffects
  (params (object :type Object%20Reference) (Controller :type Controller%20Object%20Reference))
  (bind _returnvalue (Utilities|IsValid Controller :!class /Script/Engine.KismetSystemLibrary))
  (bind (_asfort_ability_system_component _asfort_pawn) (CallFunction|GetAbilitySystemComponentandPawn object))
  (LocalVariables|SetPawn _asfort_pawn)
  (if (and (Utilities|IsValid _asfort_ability_system_component :!class /Script/Engine.KismetSystemLibrary) (or (CallFunction|ActorIsACreature _asfort_pawn) _returnvalue))
    (bind _shouldaffect (CallFunction|AffectsTarget (LocalVariables|GetPawn)))
    (if _shouldaffect
      (if (Collision|IsOverlappingActor (Variables|Default|GetTrigger) (LocalVariables|GetPawn) :!class /Script/Engine.PrimitiveComponent)
        (return true)
        (else
          _))
      (else
        (return false)))
    (elif _returnvalue
      (CallFunction|BindOrUnbindAbilitySystemRetryDelegate (LocalVariables|GetPawn) true)
      (else
        (return false))))
  (return false))

(fn UpdateActiveGameplayEffectsClasses
  (bind _speedmultiplier (Variables|UserOptions|GetSpeedMultiplier))
  (Utilities|Array|Clear (Variables|Default|GetActive%20Classes))
  (Utilities|FlowControl|Sequence
    (:then_0
      (if (Utilities|Enum|Equal(Enum) (Variables|UserOptions|GetAllowBuilding) "None")
        (Utilities|Array|Add (Variables|Default|GetActive%20Classes) (Variables|Default|GetDisable%20Build%20Effect%20Class))))
    (:then_1
      (if (not (Variables|UserOptions|GetKnob_bAllowFiring))
        (Utilities|Array|Add (Variables|Default|GetActive%20Classes) (Variables|Default|GetDisable%20Weapon%20Effect%20Class))))
    (:then_2
      (if (not (Variables|UserOptions|GetbAllowEditing))
        (Utilities|Array|Add (Variables|Default|GetActive%20Classes) (Variables|Default|GetDisable%20Editing%20Effect%20Class))))
    (:then_3
      (if (!= (Variables|UserOptions|GetBuildingSpeed) 1.0)
        (Utilities|Array|Add (Variables|Default|GetActive%20Classes) (Variables|Default|GetBuilding%20Speed%20Effect%20Class))))
    (:then_4
      (if (!= (Variables|UserOptions|GetBuildingCost) 1.0)
        (Utilities|Array|Add (Variables|Default|GetActive%20Classes) (Variables|Default|GetBuilding%20Cost%20Effect%20Class))))
    (:then_5
      (if (!= _speedmultiplier 1.0)
        (Utilities|Array|Add (Variables|Default|GetActive%20Classes) (select (and (== _speedmultiplier 0) (Variables|UserOptions|GetbDisableMovement)) (Variables|Default|GetSpeedMultiplierImmobilizeEffectClass) (Variables|Default|GetSpeedMultiplierEffectClass)))))
    (:then_6
      (if (!= (Variables|UserOptions|GetGravityOverride) 1.0)
        (Utilities|Array|Add (Variables|Default|GetActive%20Classes) (Variables|Default|GetGravityOverrideEffectClass))))))

(fn UpdateOverlappingPawns
  (Networking|SwitchHasAuthority
    (:Authority
      (for _array_element (Collision|GetOverlappingActors (Variables|Default|GetTrigger) "/Script/FortniteGame.FortPawn" :!class /Script/Engine.PrimitiveComponent)
        (CallFunction|RevalidatePawnOverlap _array_element)))))

(fn UpdateVolumeSize
  (bind _zonezscale (Variables|UserOptions|GetZoneZScale))
  (bind _zoneyscale (Variables|UserOptions|GetZoneYScale))
  (bind _zonexscale (Variables|UserOptions|GetZoneXScale))
  (bind _returnvalue (Math|Vector|MakeVector _zonexscale _zoneyscale _zonezscale))
  (|CallParentFunction)
  (Transformation|SetWorldScale3D (Variables|Default|GetBounds) (- (* _returnvalue "512.000000,512.000000,384.000000") 5.0))
  (Components|Activation|Deactivate (Variables|Default|GetIdleFX))
  (Effects|Components|ParticleSystem|SetInstanceParams (Variables|Default|GetIdleFX) (CallFunction|VFXParameters))
  (Transformation|SetWorldScale3D (Variables|Default|GetIdleFX) _returnvalue)
  (Components|Activation|Activate (Variables|Default|GetIdleFX) true)
  (Rendering|SetBoundsScale (Variables|Default|GetIdleFX) (Math|Float|Max(Float) (* (Variables|UserOptions|GetZoneXScale) 512.0) (* (Variables|UserOptions|GetZoneYScale) 512.0) (* (Variables|UserOptions|GetZoneZScale) 384.0) (Variables|Default|GetParticlesMaxDrawDistance))))

(fn ConstructionScript
  (|CallParentFunction)
  (Variables|/Script/Engine.SceneComponent|SetbAbsoluteScale true (Variables|Default|GetIdleFX)))
