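; Creative checkpoint floor device that saves itself as a player's minigame respawn point when a pawn on the allowed team overlaps it while enabled.
; Overlap broadcasts first-activation and per-player events, stores the checkpoint on the minigame player start component, and optionally clears the player's droppable items.
; Enable, disable and minigame-start events toggle its five player starts.
; Cosmetic-only timelines drive emissive, light beam and FX-plane materials, with activation FX behind a DoOnce gate.
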(event EventOnPlaced
  (Audio|PlaySoundatLocation (Variables|Audio|GetPlaced%20Sound) (Transformation|GetActorLocation)))

(event EventBeginPlay
  (bind _fortminigameplayerstart (Variables|Default|GetFortMinigamePlayerStart))
  (|CallParentFunction)
  (CallFunction|SetupTrapMIDs)
  (CallFunction|SetTrapEmissiveColor (CallFunction|GetTrapIdleEmissiveColor))
  (bind _output_get (Variables|/Script/FortniteGame.FortMinigamePlayerStartComponent|SetPlayerStarts (Utilities|GetPlayerStartArray) _fortminigameplayerstart))
  (for _array_element _output_get
    (Variables|/Script/FortniteGame.FortPlayerStartCreative|SetPlayerStartTags (Variables|/Script/FortniteGame.FortMinigameLogicComponent|GetMinigameTags _fortminigameplayerstart) _array_element)
    (Variables|/Script/FortniteGame.FortPlayerStartCreative|SetApplicableTeam 0 _array_element)))

(event OnComponentBeginOverlap(CheckpointCollider)
  (params (OverlappedComponent :type Primitive%20Component%20Object%20Reference) (OtherActor :type Actor%20Object%20Reference) (OtherComp :type Primitive%20Component%20Object%20Reference) (OtherBodyIndex :type Integer) (bFromSweep :type Boolean) (SweepResult :type Hit%20Result%20Structure%20%28by%20ref%29))
  (bind _asfort_player_pawn (Utilities|Casting|CastToFortPlayerPawn OtherActor))
  (if (and (or (not (Knobs|Team|IsValidTeamIndex (Variables|Default|GetTargetTeamIndex))) (Math|Byte|Equal(Byte) (Team|GetActorPvPTeamIndex _asfort_player_pawn) (Math|Conversions|ToByte(Integer) (Knobs|Team|GetTeamValueasIndex (Variables|Default|GetTargetTeamIndex))))) (Variables|Default|GetbIsEnabled))
    (bind _playerstate (Variables|/Script/Engine.Pawn|GetPlayerState _asfort_player_pawn)
      (:then
        (CallFunction|SendCheckpointTriggers (Fortnite|GetFortPlayerControllerfromActor _asfort_player_pawn))
        (Minigame|StorePlayerCheckpoint (Variables|Default|GetFortMinigamePlayerStart) _playerstate)
        (Utilities|FlowControl|Sequence
          (:then_0
            (if (Variables|Default|GetMinigameIsActive)
              (CallFunction|PlayActivateFX)))
          (:then_1
            (bind _gameisrunning (Default|MinigameIsRunning))
            (if (and _gameisrunning (Variables|Default|GetResetInventorySetting))
              (bind _asfort_player_controller_athena (Utilities|Casting|CastToFortPlayerControllerAthena (Pawn|GetController _asfort_player_pawn)))
              (Class|FortPlayerControllerAthena|ClearDroppableItems _asfort_player_controller_athena false)))))
      (:else))))

(event Custom|Enable%20Event
  (params (Instigator :type Controller%20Object%20Reference))
  (bind _output_get (Variables|Default|SetbIsEnabled true))
  (CallFunction|SetPlayerStartsEnabled _output_get)
  (CallFunction|SetTrapEmissiveColor (Variables|Default|GetIdleEmissiveColor)))

(event Custom|Disable%20Event
  (params (Instigator :type Controller%20Object%20Reference))
  (bind _output_get (Variables|Default|SetbIsEnabled false))
  (CallFunction|SetPlayerStartsEnabled _output_get)
  (CallFunction|SetTrapEmissiveColor))

(event MinigameStartedDelegate(FortMinigameProgress)
  (bind _output_get (Variables|Default|SetbIsEnabled (Variables|Default|GetEnabledOnStart)))
  (CallFunction|SetPlayerStartsEnabled _output_get)
  (Variables|Default|SetMinigameIsActive true)
  (CallFunction|ResetFXGate))

(event MinigameEndedDelegate(FortMinigameProgress)
  (Variables|Default|SetMinigameIsActive false)
  (Development|SetHiddeninGame (Variables|Default|GetLightBeam) true false)
  (CallFunction|SetTrapEmissiveColor (CallFunction|GetTrapIdleEmissiveColor)))

(event OnAnyPropertyChanged(ToyOptionsComponent)
  (Networking|CanExecuteCosmeticEvents
    (:True
      (CallFunction|SetTrapEmissiveColor (CallFunction|GetTrapIdleEmissiveColor)))))

(event Custom|SetTrapEmissiveColor
  (params (Color :type Linear%20Color%20Structure))
  ;[1 1 1] Set emissive color on trap
  (Networking|CanExecuteCosmeticEvents
    (:True
      (bind _trapmid (Variables|Default|GetTrapMID)
        (:then
          (bind _returnvalue (Rendering|Material|GetVectorParameterValue _trapmid "EmissiveMultiplier" :!class /Script/Engine.MaterialInstanceDynamic))
          (if (not (or (Math|Color|NearEqual(LinearColor) _returnvalue Color 0.05) (Components|Timeline|IsPlaying (Variables|Device_Floor_PlayerCheckpoint|GetAnimatedEmissiveTL))))
            (Variables|Default|SetStartColor _returnvalue)
            (Variables|Default|SetDestinationColor Color)
            (bind (_direction _lerp) (Timeline|TrapVisualStateTransitionTL PlayFromStart)
              (:Update
                (Rendering|Material|SetVectorParameterValue (Variables|Default|GetTrapMID) "EmissiveMultiplier" (Math|Color|Lerp(LinearColor) (Variables|Default|GetStartColor) (Variables|Default|GetDestinationColor) _lerp) :!class /Script/Engine.MaterialInstanceDynamic)))))
        (:else)))))

(event Custom|PlayActivateFX
  ;[1 1 1] Activate FX
  (Networking|CanExecuteCosmeticEvents
    (:True
      (Utilities|FlowControl|DoOnce false
        (:Completed
          (Development|SetHiddeninGame (Variables|Default|GetLightBeam) false false)
          (Audio|PlaySoundatLocation (Variables|Audio|GetFirst%20Time%20Use%20Audio) (Transformation|GetWorldLocation (Variables|Debug|GetStaticMeshComponent)))
          (Effects|SpawnEmitterAttached :EmitterTemplate "/Game/Athena/Items/Traps/Checkpoint/P_Checkpoint_Activate.P_Checkpoint_Activate" :AttachToComponent (Variables|Debug|GetStaticMeshComponent) :AttachPointName "None" :Location "0,256.000000,35.000000" :Scale "1.000000,1.000000,1.000000" :LocationType "KeepRelativeOffset" :bAutoDestroy true :PoolingMethod "AutoRelease" :bAutoActivate true)
          (if (Variables|Default|GetVisibleDuringMinigameSetting)
            (CallFunction|SpawnFXPlane)
            (Components|Timeline|Stop (Variables|Device_Floor_PlayerCheckpoint|GetTrapVisualStateTransitionTL))
            (bind (_direction _lerptrap _lerpcenterrays _sourcemaskexponent _center _outerraysfade _blocksineadd) (Timeline|AnimatedEmissiveTL PlayFromStart)
              (:Update
                (Rendering|Material|SetVectorParameterValue (Variables|Default|GetTrapMID) "EmissiveMultiplier" (Math|Color|Lerp(LinearColor) (CallFunction|GetTrapIdleEmissiveColor) (Variables|Default|GetActiveEmissiveColor) _lerptrap) :!class /Script/Engine.MaterialInstanceDynamic)
                (Rendering|Material|SetScalarParameterValue (Variables|Default|GetLightBeamMID) "Fade" _lerpcenterrays :!class /Script/Engine.MaterialInstanceDynamic)
                (Rendering|Material|SetScalarParameterValue (Variables|Default|GetFxPlaneMID0) "SourceMaskExponent" _sourcemaskexponent :!class /Script/Engine.MaterialInstanceDynamic)
                (Rendering|Material|SetScalarParameterValue (Variables|Default|GetFxPlaneMID0) "Center" _center :!class /Script/Engine.MaterialInstanceDynamic)
                (Rendering|Material|SetScalarParameterValue (Variables|Default|GetFxPlaneMID1) "Fade" _outerraysfade :!class /Script/Engine.MaterialInstanceDynamic)
                (Rendering|Material|SetScalarParameterValue (Variables|Default|GetFxPlaneMID0) "BlockSineAdd" _blocksineadd :!class /Script/Engine.MaterialInstanceDynamic))
              (:Finished
                (Components|DestroyComponent (Variables|Default|GetFXPlane))))))))))

(event Custom|ResetFXGate
  ;[1 1 1] Activate FX
  (Utilities|FlowControl|DoOnce false
    (:Completed
      (Development|SetHiddeninGame (Variables|Default|GetLightBeam) false false)
      (Audio|PlaySoundatLocation (Variables|Audio|GetFirst%20Time%20Use%20Audio) (Transformation|GetWorldLocation (Variables|Debug|GetStaticMeshComponent)))
      (Effects|SpawnEmitterAttached :EmitterTemplate "/Game/Athena/Items/Traps/Checkpoint/P_Checkpoint_Activate.P_Checkpoint_Activate" :AttachToComponent (Variables|Debug|GetStaticMeshComponent) :AttachPointName "None" :Location "0,256.000000,35.000000" :Scale "1.000000,1.000000,1.000000" :LocationType "KeepRelativeOffset" :bAutoDestroy true :PoolingMethod "AutoRelease" :bAutoActivate true)
      (if (Variables|Default|GetVisibleDuringMinigameSetting)
        (CallFunction|SpawnFXPlane)
        (Components|Timeline|Stop (Variables|Device_Floor_PlayerCheckpoint|GetTrapVisualStateTransitionTL))
        (Timeline|AnimatedEmissiveTL PlayFromStart)))))

(fn GetTrapIdleEmissiveColor
  (return (select (Variables|Default|GetEnabledOnStart) (Variables|Default|GetIdleEmissiveColor) 0)))

(fn GetVisualComponents
  (return (Utilities|Array|MakeArray (Variables|Debug|GetStaticMeshComponent))))

(fn IsVisibleDuringMinigame
  (return (Variables|Default|GetVisibleDuringMinigameSetting)))

(fn SendCheckpointTriggers
  (params (InstigatingActor :type Actor%20Object%20Reference))
  (bind _fortminigameplayerstart (Variables|Default|GetFortMinigamePlayerStart))
  (bind (_outtime _returnvalue) (Minigame|GetPlayerCheckpointLastUsedTime _fortminigameplayerstart (Variables|/Script/Engine.Controller|GetPlayerState (LocalVariables|GetInstigatingController))))
  (Utilities|FlowControl|Sequence
    (:then_0
      (bind _ascontroller (Utilities|Casting|CastToController InstigatingActor)
        (:then
          (LocalVariables|SetInstigatingController _ascontroller))
        (:CastFailed
          (bind _aspawn (Utilities|Casting|CastToPawn InstigatingActor))
          (LocalVariables|SetInstigatingController (Pawn|GetController _aspawn)))))
    (:then_1
      (bind _currentminigame (Variables|/Script/FortniteGame.FortMinigameLogicComponent|GetCurrentMinigame (Variables|Default|GetFortMinigameProgress))
        (:then
          (Minigame|IsInProgress _currentminigame
            (:Yes
              (if (Minigame|HasAnyCheckpointValue _fortminigameplayerstart)
                (if _returnvalue
                  _
                  (else
                    _))
                (else
                  (GameplayEvent|BroadcastEvent (Variables|UserOptions|Events|GetOn%20First%20Activation) 0 (Utilities|Struct|MakeGameplayEventFirstActivate (LocalVariables|GetInstigatingController)))))
              (GameplayEvent|BroadcastEvent (Variables|UserOptions|Events|GetOn%20First%20Activation%20Per%20Player) 0 (Utilities|Struct|MakeGameplayEventFirstActivatePerPlayer (LocalVariables|GetInstigatingController))))))
        (:else)))))

(fn SetPlayerStartsEnabled
  (params (bNewEnabled :type Boolean))
  (Utilities|FlowControl|Sequence
    (:then_0
      (Utilities|SetPlayerStartEnabled (Variables|Default|GetPlayerStartCreative0) bNewEnabled))
    (:then_1
      (Utilities|SetPlayerStartEnabled (Variables|Default|GetPlayerStartCreative1) bNewEnabled))
    (:then_2
      (Utilities|SetPlayerStartEnabled (Variables|Default|GetPlayerStartCreative2) bNewEnabled))
    (:then_3
      (Utilities|SetPlayerStartEnabled (Variables|Default|GetPlayerStartCreative3) bNewEnabled))
    (:then_4
      (Utilities|SetPlayerStartEnabled (Variables|Default|GetPlayerStartCreative4) bNewEnabled))))

(fn SetupTrapMIDs
  (bind _returnvalue (Rendering|Material|CreateDynamicMaterialInstance (Variables|Debug|GetStaticMeshComponent) 0 0 "None" :!class /Script/Engine.PrimitiveComponent))
  (Variables|Default|SetTrapMID _returnvalue)
  (bind _returnvalue_2 (Rendering|Material|CreateDynamicMaterialInstance (Variables|Default|GetLightBeam) 0 0 "None" :!class /Script/Engine.PrimitiveComponent))
  (Variables|Default|SetLightBeamMID _returnvalue_2))

(fn SpawnFXPlane
  (Utilities|IsValid (Variables|Default|GetFXPlane)
    (:Is%20Not%20Valid
      (bind _returnvalue (AddComponent|NODE_AddStaticMeshComponent-0))
      (Variables|Default|SetFXPlane _returnvalue)
      (bind _returnvalue_2 (Rendering|Material|CreateDynamicMaterialInstance (Variables|Default|GetFXPlane) 0 0 "None" :!class /Script/Engine.PrimitiveComponent))
      (Variables|Default|SetFxPlaneMID0 _returnvalue_2)
      (bind _returnvalue_3 (Rendering|Material|CreateDynamicMaterialInstance (Variables|Default|GetFXPlane) 1 0 "None" :!class /Script/Engine.PrimitiveComponent))
      (Variables|Default|SetFxPlaneMID1 _returnvalue_3))))

(fn ConstructionScript)
