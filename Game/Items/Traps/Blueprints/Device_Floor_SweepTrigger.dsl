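; Creative floor trap device that sweeps a trigger volume across a resizable zone, hitting music-sequencer devices it passes and collecting overlapped pawns.
; Pad overlaps, chained sweeps, device events or game phases start it, with restart, toggle or pause modes and instigator tracking.
; A tempo-scaled timeline moves the trigger, supporting looping and bouncing.
; The server replicates signed start time and pause position, letting clients reconstruct direction and progress.
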(event OnComponentBeginOverlap(PadTrigger)
  (params (OverlappedComponent :type Primitive%20Component%20Object%20Reference) (OtherActor :type Actor%20Object%20Reference) (OtherComp :type Primitive%20Component%20Object%20Reference) (OtherBodyIndex :type Integer) (bFromSweep :type Boolean) (SweepResult :type Hit%20Result%20Structure%20%28by%20ref%29))
  ;[1 1 1] Handles checking for overlap from a player hitting the pad.
  (bind _validinstigator (CallFunction|ValidateAndSetInstigator OtherActor))
  (if _validinstigator
    (CallFunction|StartfromTrigger)))

(event OnComponentBeginOverlap(ChainTrigger)
  (params (OverlappedComponent :type Primitive%20Component%20Object%20Reference) (OtherActor :type Actor%20Object%20Reference) (OtherComp :type Primitive%20Component%20Object%20Reference) (OtherBodyIndex :type Integer) (bFromSweep :type Boolean) (SweepResult :type Hit%20Result%20Structure%20%28by%20ref%29))
  ;[1 1 1] Handles checking for overlap from a moving trigger of another Sweep trigger
  (if (and (Components|ComponentHasTag OtherComp "AffectsChainTrigger") (Utilities|NotEqual(Object) OtherComp (Variables|Default|GetMovingTrigger)))
    (if (and (Utilities|DoesObjectImplementInterface OtherActor "/Game/Creative/Devices/Common/Creative_DeviceInterface.Creative_DeviceInterface_C") (Variables|UserKnobs|GetKnob_bSendPlayerInfo))
      (bind _instigator (CreativeDevice|GetInstigatingController(Message) OtherActor))
      (Variables|BuildingMusicSequencer|SetTriggerInstigator _instigator))
    (Variables|Default|SetInitialTrigger false) ; We don't want chained sweeps to have audio
    (CallFunction|StartfromTrigger)))

(event Custom|Fire%20Sequencer%20from%20Start
  ;[0 0.3366 1] Starts the Sequencer from the beginning.
  (CallFunction|SetInitialDirection)
  (Variables|Default|SetNumLoopsRun 0) ; reset loop count
  (CallFunction|ClearPauseLocation)
  (CallFunction|ResumeWave))

(event Custom|StartFromTrigger
  (if (or (== (Variables|UserKnobs|GetRetriggerSetting) 0) (not (Components|Timeline|IsPlaying (Variables|Device_Floor_SweepTrigger|GetTriggerMovement))))
    (CallFunction|PlayAudio)
    ;[1 1 1] Trigger Type Setting handling
    (switch Utilities|FlowControl|Switch|SwitchonECreativeSweepTriggerType (Variables|UserKnobs|GetToggleSetting)
      (:NewEnumerator0
        (CallFunction|FireSequencerFromStart))
      (:NewEnumerator1
        (CallFunction|ToggleLogic))
      (:NewEnumerator2
        (CallFunction|TogglePauseLogic)))))

(event Custom|On%20Wave%20Fired
  (params (Start%20Delta :type Float%20%28double-precision%29))
  (bind _numloopsrun (Variables|Default|GetNumLoopsRun))
  ;[1 1 1] Movement of trigger
  (Collision|ClearMoveIgnoreActors (Variables|Default|GetMovingTrigger))
  (Utilities|Set|Clear (Variables|Default|GetDamaged%20Targets))
  (CallFunction|Calculate&SetWaveTime Start%20Delta)
  (CallFunction|SetTriggerVisualsVisibility true true)
  (if (Variables|Default|GetCurrentDirection)
    (bind (_direction _movement_alpha) (Timeline|TriggerMovement Play :NewTime (stale Custom|Set%20Wave%20Time NewTime))
      (:Update
        (Transformation|SetRelativeLocation (Variables|Default|GetMovingTrigger) (Math|Vector|Lerp(Vector) (CallFunction|GetStartDestinationforTrigger) (CallFunction|GetFinalDestinationforTrigger) _movement_alpha) (not (Variables|Default|GetbIsSettingTime)) false)
        (Variables|BuildingMusicSequencer|SetCachedTriggerInstigator (Variables|BuildingMusicSequencer|GetTriggerInstigator)))
      (:Finished
        ;;[1 1 1] Handles Direction Setting
        (if (and (or (Utilities|Enum|Equal(Enum) (Variables|UserKnobs|GetDirectionSetting) "NewEnumerator2") (Utilities|Enum|Equal(Enum) (Variables|UserKnobs|GetDirectionSetting) "NewEnumerator5")) (Utilities|Enum|Equal(Enum) _direction "Forward"))
          (CallFunction|ClearPauseLocation)
          (CallFunction|ReverseDirection)
          (CallFunction|ResumeWave)
          (elif (== (Variables|UserKnobs|GetNumLoopsSetting) 0)
            ;;[1 1 1] Handles Looping settings
            (Collision|SetCollisionEnabled (Variables|Default|GetMovingTrigger) "NoCollision")
            (else
              (Math|Integer|IncrementInt _numloopsrun
                (:%20%20
                  (if (>= _numloopsrun (Variables|UserKnobs|GetNumLoopsSetting))
                    (CallFunction|ResetDevice)
                    (else
                      _)))))))
        (CallFunction|ClearPauseLocation)
        (CallFunction|ReverseDirection)
        (Transformation|SetRelativeLocation (Variables|Default|GetMovingTrigger) (CallFunction|GetStartDestinationforTrigger) false false)
        (Collision|SetCollisionEnabled (Variables|Default|GetMovingTrigger) "QueryOnly")
        (CallFunction|ResumeWave)))
    (else
      (Timeline|TriggerMovement Reverse))))

(event Custom|StopWave
  (bind _numloopsrun (Variables|Default|GetNumLoopsRun))
  ;[1 1 1] Movement of trigger
  (Timeline|TriggerMovement Stop))

(event Custom|Set%20Wave%20Time
  (params (NewTime :type Float%20%28double-precision%29))
  (bind _numloopsrun (Variables|Default|GetNumLoopsRun))
  ;[1 1 1] Movement of trigger
  (Utilities|FlowControl|Sequence
    (:then_0
      ;;[1 1 1] Prevent sweeping
      (Variables|Default|SetbIsSettingTime true))
    (:then_1
      (Timeline|TriggerMovement SetNewTime))
    (:then_2
      (Variables|Default|SetbIsSettingTime false))))

(event Custom|StartSequenceFromEvent
  (params (Instigator :type Controller%20Object%20Reference))
  ;[0.9808 0 1] Sequencers
  (if (Variables|UserKnobs|GetKnob_bSendPlayerInfo)
    (Variables|BuildingMusicSequencer|SetTriggerInstigator Instigator))
  (CallFunction|FireSequencerFromStart))

(event MinigameStartedDelegate(FortMinigameLogic)
  ;[1 1 1] Minigame start
  (Variables|Default|SetbWasInvulnerable (Variables|BuildingActor|GetbIsInvulnerable))
  (bind _output_get (Variables|BuildingActor|SetbIsInvulnerable false))
  (Fortnite|SetCanbeDamaged self (not _output_get)))

(event Custom|StopSequenceFromEvent
  (params (Instigator :type Controller%20Object%20Reference))
  ;[0.9808 0 1] Sequencers
  (CallFunction|PauseWave))

(event MinigameEndedDelegate(FortMinigameLogic)
  ;[1 1 1] Minigame end
  (bind _output_get (Variables|BuildingActor|SetbIsInvulnerable (Variables|Default|GetbWasInvulnerable)))
  (Fortnite|SetCanbeDamaged self (not _output_get)))

(event Custom|ResumeSequenceFromEvent
  (params (Instigator :type Controller%20Object%20Reference))
  ;[0.9808 0 1] Sequencers
  (if (Variables|UserKnobs|GetKnob_bSendPlayerInfo)
    (Variables|BuildingMusicSequencer|SetTriggerInstigator Instigator))
  (CallFunction|ResumeWave))

(event Custom|ActivateEventOnHitPawns
  (params (Instigator :type Controller%20Object%20Reference))
  ;[0.9808 0 1] Sequencers
  (GameplayEvent|BroadcastEvent (Variables|UserOptions|Events|GetOnPlayerHit) 0 (Utilities|Struct|MakeGameplayEventonPlayerHit Instigator)))

(event MinigameStateChangedDelegate(FortMinigameLogic)
  (params (Minigame :type Fort%20Minigame%20Object%20Reference) (NewMinigameState :type EFortMinigameState%20Enum :default "PreGame"))
  (bind _triggerinstigator (Variables|BuildingMusicSequencer|GetTriggerInstigator))
  (CallFunction|UpdateFrameVisibility)
  (CallFunction|ResetDevice)
  (Networking|SwitchHasAuthority
    (:Authority
      (Utilities|IsValid _triggerinstigator
        (:Is%20Valid
          (if (not (and (Collision|IsOverlappingActor (Variables|Default|GetPadTrigger) (Pawn|GetControlledPawn _triggerinstigator) :!class /Script/Engine.PrimitiveComponent) (Variables|UserKnobs|GetKnob_bSendPlayerInfo)))
            (Variables|BuildingMusicSequencer|SetTriggerInstigator 0 :!label _label))
          (if (Class|CreativeCommonDeviceFunctionLibrary|IsActivatedonPhase NewMinigameState (Variables|UserKnobs|GetActivateOnGamePhase) self)
            (CallFunction|FireSequencerFromStart)))
        (:Is%20Not%20Valid
          _)) ; Check the instigator is valid so we don't spam the logs.
      (goto _label))))

(event OnAnyPropertyChanged(ToyOptionsComponent)
  ;[0.0199 0.4896 0.1098] Initialisation
  (CallFunction|UpdateVolumeSize)
  (CallFunction|ResetDevice)
  (Networking|SwitchHasAuthority
    (:Authority
      (bind _triggerinstigator (Variables|BuildingMusicSequencer|GetTriggerInstigator)
        (:then
          (if (not (and (Collision|IsOverlappingActor (Variables|Default|GetPadTrigger) (Pawn|GetControlledPawn _triggerinstigator) :!class /Script/Engine.PrimitiveComponent) (Variables|UserKnobs|GetKnob_bSendPlayerInfo)))
            (Variables|BuildingMusicSequencer|SetTriggerInstigator 0)))
        (:else
          _)) ; Validate, as on-load issues can cause this to return null.
      (if (Class|CreativeCommonDeviceFunctionLibrary|IsActivatedonPhase "PreGame" (Variables|UserKnobs|GetActivateOnGamePhase) self)
        (CallFunction|FireSequencerFromStart)))))

(event OnPropertiesLoaded(ToyOptionsComponent)
  ;[0.0199 0.4896 0.1098] Initialisation
  (CallFunction|UpdateVolumeSize)
  (CallFunction|ResetDevice)
  (Networking|SwitchHasAuthority
    (:Authority
      (bind _triggerinstigator (Variables|BuildingMusicSequencer|GetTriggerInstigator)
        (:then
          (if (not (and (Collision|IsOverlappingActor (Variables|Default|GetPadTrigger) (Pawn|GetControlledPawn _triggerinstigator) :!class /Script/Engine.PrimitiveComponent) (Variables|UserKnobs|GetKnob_bSendPlayerInfo)))
            (Variables|BuildingMusicSequencer|SetTriggerInstigator 0)))
        (:else
          _)) ; Validate, as on-load issues can cause this to return null.
      (if (Class|CreativeCommonDeviceFunctionLibrary|IsActivatedonPhase "PreGame" (Variables|UserKnobs|GetActivateOnGamePhase) self)
        (CallFunction|FireSequencerFromStart)))))

(event EditorPreview|EventSetupForEditorPreview
  ;[0.0199 0.4896 0.1098] Initialisation
  (CallFunction|UpdateFrameVisibility)
  (CallFunction|UpdateVolumeSize)
  (CallFunction|ResetDevice)
  (Networking|SwitchHasAuthority
    (:Authority
      (bind _triggerinstigator (Variables|BuildingMusicSequencer|GetTriggerInstigator)
        (:then
          (if (not (and (Collision|IsOverlappingActor (Variables|Default|GetPadTrigger) (Pawn|GetControlledPawn _triggerinstigator) :!class /Script/Engine.PrimitiveComponent) (Variables|UserKnobs|GetKnob_bSendPlayerInfo)))
            (Variables|BuildingMusicSequencer|SetTriggerInstigator 0)))
        (:else
          _)) ; Validate, as on-load issues can cause this to return null.
      (if (Class|CreativeCommonDeviceFunctionLibrary|IsActivatedonPhase "PreGame" (Variables|UserKnobs|GetActivateOnGamePhase) self)
        (CallFunction|FireSequencerFromStart)))))

(event GameplayCue|GameplayCue.Abilities.Activation.Traps.ActivateTrap
  (params (EventType :type EGameplayCueEvent%20Enum :default "OnActive") (Parameters :type Gameplay%20Cue%20Parameters%20Structure%20%28by%20ref%29))
  (switch Utilities|FlowControl|Switch|SwitchonEGameplayCueEvent EventType
    (:Executed
      (if (and (CallFunction|HasMinigameStarted) (!= (Variables|UserKnobs|GetDamageSetting) 0))
        (Audio|PlaySoundatLocation (Variables|Default|GetTrapFireSound) (Transformation|GetActorLocation) "0, 0, 0" 0.5)))))

(event EventOnWorldReady
  (CallFunction|OverrideCollisionSettings))

(event EventBeginPlay
  (CallFunction|ResetDevice))

(event EventOnPlaced
  (Utilities|FlowControl|DoOnce false
    (:Completed
      (Trap|AddTriggerComponent (Variables|Default|GetMovingTrigger))
      (Audio|PlaySoundatLocation "/Game/Sounds/Creative/Gadgets/Traps/Trap_Elimination_Placed_Cue.Trap_Elimination_Placed_Cue" (Transformation|GetActorLocation)))))

(event OnAddedtoMinigame(FortMinigameLogic)
  (params (Minigame :type Fort%20Minigame%20Object%20Reference))
  (CallFunction|UpdateFrameVisibility))

(event OnComponentBeginOverlap(MovingTrigger)
  (params (OverlappedComponent :type Primitive%20Component%20Object%20Reference) (OtherActor :type Actor%20Object%20Reference) (OtherComp :type Primitive%20Component%20Object%20Reference) (OtherBodyIndex :type Integer) (bFromSweep :type Boolean) (SweepResult :type Hit%20Result%20Structure%20%28by%20ref%29))
  ;[0.7195 1 0.6408] Moving Trigger's BeginOverlap
  (Collision|IgnoreActorwhenMoving (Variables|Default|GetMovingTrigger) OtherActor true)
  (bind _asfort_music_sequencer_interface (Utilities|Casting|CastToFortMusicSequencerInterface OtherActor)
    (:then
      (Class|FortMusicSequencerInterface|OnHitbyMusicSequencer _asfort_music_sequencer_interface (Variables|BuildingMusicSequencer|GetCachedTriggerInstigator)))
    (:CastFailed
      _))
  ;;[1 1 1] Add overlapped pawns to an array to be processed by the trap when it activates
  (bind _aspawn (Utilities|Casting|CastToPawn OtherActor))
  (Utilities|Set|Add (Variables|Default|GetOverlappedPawns) _aspawn))

(fn ActivateFromOtherDevice
  (params (Instigator :type Controller%20Object%20Reference) (Requester :type ERequesterType%20Enum :default "NewEnumerator0"))
  (return false))

(fn Calculate&SetWaveTime
  (params (Wave%20Start%20Time :type Float%20%28double-precision%29))
  (bind _temposetting (Variables|UserKnobs|GetTempoSetting))
  (bind _returnvalue (Math|Conversions|ToFloat(Integer) _temposetting))
  (bind _returnvalue_2 (/ _returnvalue 6e+01))
  (bind _vertical (CallFunction|SweepDirection))
  (bind _zonezscale (Variables|UserKnobs|GetZoneZScale))
  (bind _zonexscale (Variables|UserKnobs|GetZoneXScale))
  (bind _returnvalue_3 (select _vertical _zonezscale _zonexscale))
  (bind _returnvalue_4 (/ 1.0 _returnvalue_3))
  (bind _returnvalue_5 (* _returnvalue_4 _returnvalue_2))
  (Components|Timeline|SetPlayRate (Variables|Device_Floor_SweepTrigger|GetTriggerMovement) _returnvalue_5) ; Keeps playback rate consistent regardless of length of trigger
  (CallFunction|SetWaveTime (Math|Float|SelectFloat (+ (CallFunction|GetBaseLocationAlpha) (* _returnvalue_5 Wave%20Start%20Time)) (- (CallFunction|GetBaseLocationAlpha) (* _returnvalue_5 Wave%20Start%20Time)) (Variables|Default|GetCurrentDirection))))

(fn CanInteract
  (params (InteractingPawn :type Fort%20Pawn%20Object%20Reference) (InteractionBeingAttempted :type EInteractionBeingAttempted%20Enum :default "FirstInteraction") (InteractionType :type TInteractionType%20Enum :default "IT_NoInteraction"))
  (return false))

(fn ClearPauseLocation
  (Networking|SwitchHasAuthority
    (:Authority
      (if (!= (Variables|Default|GetServerWavePauseLocation) -1.0)
        (Variables|Default|SetServerWavePauseLocation -1.0)))))

(fn ClearOverlappedPawns
  (Utilities|Set|Clear (Variables|Default|GetOverlappedPawns)))

(fn GetBaseLocationAlpha
  (return (select (!= (Variables|Default|GetServerWavePauseLocation) -1.0) (Variables|Default|GetServerWavePauseLocation) (select (Variables|Default|GetCurrentDirection) 0.0 1.0))))

(fn GetDeviceColor
  (return 0))

(fn GetInstigatingController
  (return (Variables|BuildingMusicSequencer|GetTriggerInstigator)))

(fn GetFinalDestinationForTrigger
  ;[1 1 1] Get the distance trigger needs to travel based on scale of volume
  (return (Math|Vector|MakeVector (* (Variables|UserKnobs|GetZoneXScale) (select (CallFunction|SweepDirection) 256.0 512.0)) 0.0 (* (Math|Conversions|ToFloat(Boolean) (CallFunction|SweepDirection)) (* (Variables|UserKnobs|GetZoneZScale) 384.0)))))

(fn GetNewLocation
  (return (Math|Vector|MakeVector (+ (* (- (Variables|UserKnobs|GetZoneXScale) 1.0) 256.0) (.x (Variables|Default|GetStartingLocationOfVolume))) (.y (Variables|Default|GetStartingLocationOfVolume)) (.z (Variables|Default|GetStartingLocationOfVolume)))))

(fn GetSpawnedPickups
  (return 0))

(fn GetStartDestinationForTrigger
  ;[1 1 1]
  (return (Utilities|Select "0, 0, 0" (Math|Vector|MakeVector (* (Variables|UserKnobs|GetZoneXScale) 256.0)) (CallFunction|SweepDirection))))

(fn HasMinigameStarted
  (bind _currentminigame (Variables|/Script/FortniteGame.FortMinigameLogicComponent|GetCurrentMinigame (Variables|Default|GetFortMinigameLogic))
    (:then
      (return (Minigame|HasStarted _currentminigame :!class /Script/FortniteGame.FortMinigame)))
    (:else
      (return false))))

(fn IsEnabled
  (return false))

(fn IsHoldingItemForPawn
  (params (Item :type Item%20Variant%20Handle%20Structure) (Pawn :type Pawn%20Object%20Reference))
  (return false))

(fn IsMinigameOnPregameState
  (bind _currentminigame (Variables|/Script/FortniteGame.FortMinigameLogicComponent|GetCurrentMinigame (Variables|Default|GetFortMinigameLogic))
    (:then
      (return (Utilities|Enum|Equal(Enum) (Variables|/Script/FortniteGame.FortMinigame|GetCurrentState _currentminigame) "PreGame")))
    (:else
      (return (Editor|InEditor)))))

(fn IsTargetableByLinearMover
  (return true))

(fn OnRep_ServerPauseWaveTime
  (if (!= (Variables|Default|GetServerWavePauseLocation) -1.0)
    (CallFunction|SetWaveTime (CallFunction|GetBaseLocationAlpha))
    (CallFunction|StopWave)
    (CallFunction|SetTriggerVisualsVisibility (Variables|UserKnobs|GetKnob_bPausedTriggerIsActive))))

(fn OnRep_ServerWaveStartTime
  (if (== (Variables|Default|GetServerWaveStartTime) 0.0)
    ;[1 0.03568 0] Reset Logic
    (CallFunction|StopWave)
    (CallFunction|SetTriggerVisualsVisibility false)
    (CallFunction|ReturnWavetoStartLocation)
    (Variables|Default|SetInitialTrigger true) ; Reset so player activation will have sound again
    (else
      ;[0.0199 0.4896 0.1098] Start Logic
      (Variables|Default|SetCurrentDirection (>= (Variables|Default|GetServerWaveStartTime) 0.0))
      (CallFunction|OnWaveFired (- (GameState|GetServerWorldTimeSeconds (Game|GetGameState)) (Math|Float|Absolute(Float) (Variables|Default|GetServerWaveStartTime)))))))

(fn OverrideCollisionSettings
  (bind _staticmeshcomponent (Variables|Debug|GetStaticMeshComponent))
  ;[1 0.9605 0.1332] Something is preventing the collision settings from actually sticking. Force them to fit what they should be
  (Collision|SetCollisionResponsetoChannel _staticmeshcomponent "ECC_Visibility" "ECR_Block")
  (Collision|SetCollisionResponsetoChannel _staticmeshcomponent "ECC_GameTraceChannel2" "ECR_Overlap")
  (Collision|SetCollisionResponsetoChannel _staticmeshcomponent "ECC_GameTraceChannel11" "ECR_Block")
  (Collision|SetCollisionResponsetoChannel _staticmeshcomponent "ECC_GameTraceChannel16" "ECR_Overlap"))

(fn PauseWave
  ;[1 1 1] Stuff to help ensure clients are sync'd up as wave travels
  (Networking|SwitchHasAuthority
    (:Authority
      (Networking|FlushNetDormancy)
      (Variables|Default|SetServerWavePauseLocation (Components|Timeline|GetPlaybackPosition (Variables|Device_Floor_SweepTrigger|GetTriggerMovement)))
      (Networking|ForceNetUpdate) ; Replicate ASAP
    )))

(fn PlayAudio
  (Networking|CanExecuteCosmeticEvents
    (:True
      (if (Variables|Default|GetInitialTrigger)
        (Audio|PlaySoundatLocation (Variables|Default|GetStart%20Sequence%20Sound) (Transformation|GetActorLocation))
        (Variables|Default|SetInitialTrigger false)))))

(fn ResetDevice
  (CallFunction|ClearOverlappedPawns)
  (Networking|SwitchHasAuthority
    (:Authority
      (if (!= (Variables|Default|GetServerWaveStartTime) 0.0)
        (Variables|Default|SetServerWaveStartTime 0.0) ; Setting a value of 0 will cause the device to reset.
      ))))

(fn ResumeWave
  ;[1 1 1] Stuff to help ensure clients are sync'd up as wave travels
  (Networking|SwitchHasAuthority
    (:Authority
      (Networking|FlushNetDormancy)
      (Variables|Default|SetServerWaveStartTime (* (GameState|GetServerWorldTimeSeconds (Game|GetGameState)) (Math|Float|SelectFloat 1.0 -1.0 (Variables|Default|GetCurrentDirection)))) ; Notify that a wave has been fired
      (Networking|ForceNetUpdate) ; Replicate ASAP
    )))

(fn ReturnWaveToStartLocation
  (Variables|Default|SetNumLoopsRun 0) ; reset loop count
  (CallFunction|ClearPauseLocation)
  (if (Variables|Default|GetCurrentDirection)
    (Transformation|SetRelativeLocation (Variables|Default|GetMovingTrigger) (CallFunction|GetFinalDestinationforTrigger) false false) ; Get trigger back to start
    (else
      (Transformation|SetRelativeLocation (Variables|Default|GetMovingTrigger) (CallFunction|GetStartDestinationforTrigger) false false) ; Get trigger back to start
    )))

(fn ReverseDirection
  (if (or (Utilities|Enum|Equal(Enum) (Variables|UserKnobs|GetDirectionSetting) "NewEnumerator2") (Utilities|Enum|Equal(Enum) (Variables|UserKnobs|GetDirectionSetting) "NewEnumerator5"))
    (Variables|Default|SetCurrentDirection (not (Variables|Default|GetCurrentDirection))) ; Bounce in the other direction. This isn't replicated directly so it can be client inferred to avoid latency when using the bounce feature.
  ))

(fn SetInitialDirection
  (Variables|Default|SetCurrentDirection (not (or (Utilities|Enum|Equal(Enum) (Variables|UserKnobs|GetDirectionSetting) "NewEnumerator1") (Utilities|Enum|Equal(Enum) (Variables|UserKnobs|GetDirectionSetting) "NewEnumerator4")))))

(fn SetTriggerCollision
  (params (On :type Boolean))
  (return false))

(fn SetTriggerVisualsVisibility
  (params (bNewVisibility :type Boolean) (IlluminateFloorPad :type Boolean))
  (bind _audio_trap_floor_button (Variables|Default|GetAudio_Trap_Floor_Button))
  (bind _mid (Material|GetMID _audio_trap_floor_button))
  (bind _movingtrigger (Variables|Default|GetMovingTrigger))
  (Rendering|SetVisibility _movingtrigger bNewVisibility true)
  (Collision|SetCollisionEnabled _movingtrigger (select bNewVisibility "QueryOnly" "NoCollision"))
  (Components|Activation|SetActive (Variables|Default|GetP_Sequencer_Blocks) bNewVisibility false)
  (Rendering|Material|SetVectorParameterValue _mid "GlowColor" (select IlluminateFloorPad (Variables|Default|GetFloorPadGlowColor) "(R=0.000000,G=0.000000,B=0.000000,A=1.000000)") :!class /Script/Engine.MaterialInstanceDynamic)
  (Rendering|Material|SetTextureParameterValue _mid "Texture" (select IlluminateFloorPad (Variables|Default|GetTextureOn) (Variables|Default|GetTextureOff))))

(fn ShouldInvisiblePropsBeAllowed
  (return false))

(fn SweepDirection
  (return (Utilities|Select :Index (Variables|UserKnobs|GetDirectionSetting) :NewEnumerator0 false :NewEnumerator1 false :NewEnumerator2 false :NewEnumerator3 true :NewEnumerator4 true :NewEnumerator5 true)))

(fn ToggleLogic
  (if (Components|Timeline|IsPlaying (Variables|Device_Floor_SweepTrigger|GetTriggerMovement))
    (CallFunction|ResetDevice)
    (else
      (CallFunction|FireSequencerFromStart))))

(fn TogglePauseLogic
  ;[1 1 1] Stuff to help ensure clients are sync'd up as wave travels
  (if (Components|Timeline|IsPlaying (Variables|Device_Floor_SweepTrigger|GetTriggerMovement))
    (CallFunction|PauseWave)
    (elif (== (Variables|Default|GetServerWavePauseLocation) -1.0)
      (CallFunction|FireSequencerFromStart)
      (else
        (CallFunction|ResumeWave)))))

(fn UpdateFrameVisibility
  (Rendering|SetVisibility (Variables|Default|GetFrame) (or (Math|Boolean|ANDBoolean (not (Replay|GetIsPlayingReplay)) (not (Class|CreativeCommonDeviceFunctionLibrary|IsinPlayMode self)) (CallFunction|IsMinigameonPregameState)) (Variables|Default|GetZoneVisibleInGame)) false)
  (return))

(fn UpdateVolumeSize
  (bind _damagesetting (Variables|UserKnobs|GetDamageSetting))
  (bind _returnvalue (> _damagesetting 0))
  (bind _damagesetting_2 (Variables|UserKnobs|GetDamageSetting))
  (bind _returnvalue_2 (> _damagesetting_2 0))
  (bind _p_sequencer_blocks (Variables|Default|GetP_Sequencer_Blocks))
  (bind _movingtrigger (Variables|Default|GetMovingTrigger))
  (bind _movingtriggermid (Variables|Default|GetMovingTriggerMID))
  (Components|Activation|Deactivate (Variables|Default|GetP_Sequencer_Blocks))
  (Rendering|Material|SetVectorParameterValue (Material|GetMID (Variables|Default|GetFrame)) "SizeBias" (Math|Conversions|ToLinearColor(Vector) (- (* (Math|Vector|MakeVector 256.0 256.0 384.0) (Math|Vector|MakeVector (Variables|UserKnobs|GetZoneXScale) (Variables|UserKnobs|GetZoneYScale) (Variables|UserKnobs|GetZoneZScale))) (Math|Vector|MakeVector 256.0 256.0 384.0))) :!class /Script/Engine.MaterialInstanceDynamic)
  (Rendering|SetBoundsScale (Variables|Default|GetFrame) (* (Math|Float|Max(Float) (Variables|UserKnobs|GetZoneXScale) (Variables|UserKnobs|GetZoneYScale) (Variables|UserKnobs|GetZoneZScale)) 1.5))
  (Transformation|SetWorldScale3D (Variables|Default|GetMovingTrigger) (select (CallFunction|SweepDirection) (Math|Vector|MakeVector (select (Variables|UserOptions|Advanced|GetCustomPulseStyle) (* (Variables|UserKnobs|GetZoneXScale) 512.0) (Variables|UserKnobs|GetZoneXScale)) (select (Variables|UserOptions|Advanced|GetCustomPulseStyle) (* (Variables|UserKnobs|GetZoneYScale) 512.0) (Variables|UserKnobs|GetZoneYScale)) (select (Variables|UserOptions|Advanced|GetCustomPulseStyle) 25.0 0.05)) (Math|Vector|MakeVector (select (Variables|UserOptions|Advanced|GetCustomPulseStyle) 25.0 0.05) (select (Variables|UserOptions|Advanced|GetCustomPulseStyle) (* (Variables|UserKnobs|GetZoneYScale) 512.0) (Variables|UserKnobs|GetZoneYScale)) (select (Variables|UserOptions|Advanced|GetCustomPulseStyle) (* (Variables|UserKnobs|GetZoneZScale) 384.0) (Variables|UserKnobs|GetZoneZScale)))))
  (Transformation|SetRelativeLocation (Variables|Default|GetMovingTrigger) (CallFunction|GetStartDestinationforTrigger) false false)
  (Transformation|SetRelativeLocation (Variables|Default|GetFrame) (CallFunction|GetNewLocation) false false)
  (Variables|/Script/CascadeRuntime.ParticleSystemComponent|SetInstanceParameters (CallFunction|VFXParameters) (Variables|Default|GetP_Sequencer_Blocks))
  (Transformation|SetWorldScale3D _p_sequencer_blocks "1.000000,1.000000,1.000000")
  (if (CallFunction|SweepDirection)
    (Transformation|SetRelativeLocation _p_sequencer_blocks "256.000000,256.000000,0.000000" false false)
    (else
      (Transformation|SetRelativeLocation _p_sequencer_blocks "0,256.000000,384.000000" false false)))
  (Components|Activation|Activate (Variables|Default|GetP_Sequencer_Blocks) true)
  (Components|StaticMesh|SetStaticMesh (Variables|Default|GetMovingTrigger) (select (Variables|UserOptions|Advanced|GetCustomPulseStyle) "/Game/Creative/Devices/Sequencer/Meshes/CP_Device_Sequencer_Volume_01.CP_Device_Sequencer_Volume_01" "/Game/Creative/Devices/Music_Sequencer/S_MusicSequencer_Wave.S_MusicSequencer_Wave"))
  (if (not (Variables|UserOptions|Advanced|GetCustomPulseStyle))
    (Rendering|Material|SetMaterial (Variables|Default|GetMovingTrigger) 0 (select _returnvalue "/Game/Items/Traps/Materials/M_Sequencer_Trigger_02_Inst.M_Sequencer_Trigger_02_Inst" "/Game/Items/Traps/Materials/M_Sequencer_Trigger_01_Inst.M_Sequencer_Trigger_01_Inst"))
    (Rendering|Material|SetMaterial (Variables|Default|GetP_Sequencer_Blocks) 0 (select _returnvalue "/Game/Items/Traps/Materials/M_EliminationZone_Blocks.M_EliminationZone_Blocks" "/Game/Items/Traps/Materials/M_Sequencer_Blocks_Inst.M_Sequencer_Blocks_Inst"))
    (Rendering|Material|SetScalarParameterValueonMaterials (Variables|Default|GetMovingTrigger) "UseFlashingAlert" (select (and (Variables|UserOptions|GetUseFlashingWarningSigns) (> (Variables|UserKnobs|GetDamageSetting) 0)) 1.0 0.0))
    (else
      (Rendering|Material|SetMaterial _movingtrigger 0 "/Game/Creative/Devices/Sequencer/Materials/MI_Device_Sequencer_Volume_01.MI_Device_Sequencer_Volume_01")
      (Material|ReattachOrCreateMID _movingtrigger _movingtriggermid)
      (Rendering|Material|SetVectorParameterValue _movingtriggermid "Base Color" (select _returnvalue_2 "(R=0.901042,G=0.015019,B=0.000000,A=0.000000)" "(R=0.000000,G=0.040255,B=0.901042,A=0.000000)") :!class /Script/Engine.MaterialInstanceDynamic)
      (Rendering|Material|SetVectorParameterValue _movingtriggermid "Swirls Color" (select _returnvalue_2 "(R=1.000000,G=0.108540,B=0.000000,A=0.000000)" "(R=0.027170,G=0.789307,B=1.000000,A=0.000000)") :!class /Script/Engine.MaterialInstanceDynamic)
      (Rendering|Material|SetVectorParameterValue _movingtriggermid "Frame Glow Color" (select _returnvalue_2 "(R=1.000000,G=0.102506,B=0.000000,A=1.000000)" "(R=0.000000,G=0.818932,B=1.000000,A=1.000000)") :!class /Script/Engine.MaterialInstanceDynamic)
      (Rendering|Material|SetScalarParameterValue _movingtriggermid "Width" (* (Variables|UserKnobs|GetZoneXScale) 512.0) :!class /Script/Engine.MaterialInstanceDynamic)
      (Rendering|Material|SetScalarParameterValue _movingtriggermid "Height" (* (Variables|UserKnobs|GetZoneZScale) 512.0) :!class /Script/Engine.MaterialInstanceDynamic)
      (Rendering|Material|SetScalarParameterValue _movingtriggermid "Depth" (* (Variables|UserKnobs|GetZoneYScale) 512.0) :!class /Script/Engine.MaterialInstanceDynamic)))
  (Transformation|SetRelativeTransform (Variables|Default|GetTriggerBase) (Utilities|Array|Get(acopy) (Variables|Default|GetZoneDirections) (Variables|UserKnobs|GetZoneDirection)) false true))

(fn ConstructionScript
  (Networking|CanExecuteCosmeticEvents
    (:True
      (Variables|/Script/Engine.SceneComponent|SetbAbsoluteScale true (Variables|Default|GetP_Sequencer_Blocks)))))

(fn ValidateandSetInstigator
  (params (OtherActor :type Actor%20Object%20Reference))
  (Utilities|FlowControl|Sequence
    (:then_0
      ;[1 1 1] Check for valid actor overlap
      (bind _asplayer_pawn_athena (Utilities|Casting|CastToPlayerPawn_Athena (Variables|GetOtherActor))
        (:then
          (if (not (Vehicle|IsinVehicle _asplayer_pawn_athena))
            (LocalVariables|SetLocalIsActorValid true)
            (LocalVariables|SetLocalInstigator (Pawn|GetController _asplayer_pawn_athena))))
        (:CastFailed
          (bind _asfort_athena_vehicle (Utilities|Casting|CastToFortAthenaVehicle (Variables|GetOtherActor))
            (:then
              (LocalVariables|SetLocalIsActorValid true)
              (bind _returnvalue (Vehicle|GetDriver(Message) _asfort_athena_vehicle))
              (Utilities|IsValid _returnvalue
                (:Is%20Valid
                  (LocalVariables|SetLocalInstigator (Pawn|GetController _returnvalue)))
                (:Is%20Not%20Valid
                  (bind _returnvalue_2 (Vehicle|GetAllPassengers(Message) _asfort_athena_vehicle))
                  (bind _array_element (Utilities|Array|ForEachLoopwithBreak _returnvalue_2 :!label _label)
                    (:LoopBody
                      (Utilities|IsValid _array_element
                        (:Is%20Valid
                          (Utilities|IsValid (Pawn|GetController _array_element)
                            (:Is%20Valid
                              (LocalVariables|SetLocalInstigator (Pawn|GetController _array_element))
                              (goto _label Break))))))))))
            (:CastFailed
              (bind _asphysics_ball_master (Utilities|Casting|CastToPhysicsBall_Master (Variables|GetOtherActor))
                (:then
                  (LocalVariables|SetLocalIsActorValid true)
                  (LocalVariables|SetLocalInstigator (Pawn|GetController (Variables|Actor|GetInstigator _asphysics_ball_master))))
                (:CastFailed
                  (bind _asfort_physics_pawn_object (Utilities|Casting|CastToFortPhysicsPawnObject (Variables|GetOtherActor)))
                  (LocalVariables|SetLocalIsActorValid true)
                  (LocalVariables|SetLocalInstigator (Variables|/Script/Engine.Pawn|GetLastHitBy _asfort_physics_pawn_object)))))))))
    (:then_1
      ;[1 1 1] Assign instigator and bind delegate
      (if (and (LocalVariables|GetLocalIsActorValid) (Variables|UserKnobs|GetKnob_bSendPlayerInfo))
        (Variables|BuildingMusicSequencer|SetTriggerInstigator (LocalVariables|GetLocalInstigator)))
      (return (LocalVariables|GetLocalIsActorValid)))))

(fn VFXParameters
  (return (Utilities|Array|MakeArray :%5B0%5D (Utilities|Struct|MakeParticleSysParam "BlocksVolume" "PSPT_VectorRand" 0.0 0.0 "0.000000,0.000000,0.000000" (* (* (select (CallFunction|SweepDirection) (Variables|Default|GetBlocksVolumeVertical) (Variables|Default|GetBlocksVolumeHorizontal)) (Math|Vector|MakeVector (Variables|UserKnobs|GetZoneXScale) (Variables|UserKnobs|GetZoneYScale) (Variables|UserKnobs|GetZoneZScale))) -1.0)) :%5B1%5D (Utilities|Struct|MakeParticleSysParam "BlocksSpawnRateScale" "PSPT_Scalar" (Utilities|ClampMaxFloat (* (Math|Float|SelectFloat (* (Variables|UserKnobs|GetZoneXScale) (Variables|UserKnobs|GetZoneYScale)) (* (Variables|UserKnobs|GetZoneYScale) (Variables|UserKnobs|GetZoneZScale)) (CallFunction|SweepDirection)) (Variables|Default|GetFX%20Spawn%20Rate%20per%20Unit%20Squared)) (Variables|Default|GetFX%20Max%20Spawn%20Rate))))))
