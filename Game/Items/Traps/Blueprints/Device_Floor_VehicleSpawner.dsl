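; Floor trap device Blueprint that spawns a vehicle at its spawn location.
; When any property on its toy options component changes, the authority spawns an actor of the class that the vehicle type index selects from the vehicle types array.
; Placed and world-ready events play trap sounds at the device location where cosmetic events can execute.
; ShouldTrigger always returns true, and no logic here removes previously spawned vehicles.
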(event EventOnWorldReady
  (Networking|CanExecuteCosmeticEvents
    (:True
      (Audio|PlaySoundatLocation (Variables|Audio|GetTrap_Ready_Sound) (Transformation|GetActorLocation)))))

(event EventOnPlaced
  (Networking|CanExecuteCosmeticEvents
    (:True
      (Audio|PlaySoundatLocation (Variables|Audio|GetTrap_Placed_Sound) (Transformation|GetActorLocation)))))

(event OnAnyPropertyChanged(ToyOptionsComponent)
  (Networking|SwitchHasAuthority
    (:Authority
      (Game|SpawnActorfromClass (Utilities|Array|Get(acopy) (Variables|Default|GetVehicleTypes) (Variables|Default|GetVehicleTypeIndex)) (Transformation|GetWorldTransform (Variables|Default|GetSpawnLocation)) "Undefined" "SelectDefaultAtRuntime" 0 0))))

(fn ShouldTrigger
  (params (TouchingActors :type Array%20of%20Actor%20Object%20References))
  (return true))

(fn ConstructionScript)
