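; Parent Creative floor device that spawns and respawns one vehicle, governed by minigame phase, user options and message receivers.
; Respawn requests pass through the Creative throttle manager and a local randomized spam delay, then the authority async-loads and spawns an Athena or character vehicle.
; The vehicle's EndPlay triggers instant or timed respawn.
; RepNotify enabled and cooldown state drives the pad's emissive colour, flicker and ready sound.
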(event Custom|Completed%20Throttle
  (params (Succeeded :type Boolean))
  ;[1 1 1] Spawning the Vehicle and binding it to various events
  (Variables|SpamProtection|SetWaitingForThrottle false) ; Throttle Complete
  (if Succeeded
    (CallFunction|AttemptSpawnVehicle)
    (else
      (Variables|SpamProtection|SetWaitingToSpawn true)
      (CallFunction|AttemptRespawnAfterSpamProtectionDelay))))

(event Custom|Attempt%20Respawn%20after%20Spam%20Protection%20Delay
  (bind (_spam_protection_min_max_time_x _spam_protection_min_max_time_y) (Math|Vector2D|BreakVector2D (Variables|GetSpam%20Protection%20Min/Max%20Time_X)))
  ;[1 1 1] Spawning the Vehicle and binding it to various events
  ;;[1 1 1] Waits a defined time period after spawning a vehicle, and then checks to see if anything has attempted to spawn a new one.
  ;; This is a buffer on the local device, stopping you mass-respawning vehicles.
  (Utilities|FlowControl|Delay (Math|Random|RandomFloatinRange _spam_protection_min_max_time_x _spam_protection_min_max_time_y))
  (Variables|SpamProtection|SetRecentlySpawned false)
  (if (Variables|SpamProtection|GetWaitingToSpawn)
    (CallFunction|RespawnVehicleFromSettings)))

(event Custom|Attempt%20Spawn%20Vehicle
  ;[1 1 1] Spawning the Vehicle and binding it to various events
  (if (Variables|SpamProtection|GetRecentlySpawned)
    (Variables|SpamProtection|SetWaitingToSpawn true)
    (else
      (CallFunction|SpawnVehicle))))

(event Custom|OnVehicleSpawned
  (params (SpawnedVehicle :type Actor%20Object%20Reference))
  ;[1 1 1] Spawning the Vehicle and binding it to various events
  ;;[1 1 1] Post Vehicle Spawn. What needs to be done after a vehicle is spawned to manage state and hook up delegates
  (Utilities|IsValid SpawnedVehicle
    (:Is%20Valid
      (Variables|SpamProtection|SetWaitingToSpawn false)
      (Utilities|Time|ClearandInvalidateTimerbyHandle (Variables|Default|GetVehicleSpawnTimer))
      (Variables|SpamProtection|SetRecentlySpawned true)
      (Game|AssignOnEndPlay SpawnedVehicle (delegate Custom|VehicleDestroyed))
      (CallFunction|AttemptRespawnAfterSpamProtectionDelay))))

(event Custom|Manually%20Destroy%20Vehicle
  (bind _cachedvehicleinterface (Variables|Default|GetCachedVehicleInterface))
  (bind _returnvalue (Utilities|ToObject(Interface) _cachedvehicleinterface))
  ;[1 1 1] Spawning the Vehicle and binding it to various events
  ;;[1 1 1] Post Vehicle Spawn. What needs to be done after a vehicle is spawned to manage state and hook up delegates
  (Utilities|IsValid _returnvalue
    (:Is%20Valid
      (bind _asactor (Utilities|Casting|CastToActor _returnvalue))
      (CallFunction|BindsVehicleonEndPlay _asactor)
      (Vehicle|DestroyVehicle(Message) _returnvalue))))

(event Custom|BindsVehicleOnEndPlay
  (params (Vehicle :type Actor%20Object%20Reference))
  ;[1 1 1] Spawning the Vehicle and binding it to various events
  ;;[1 1 1] Post Vehicle Spawn. What needs to be done after a vehicle is spawned to manage state and hook up delegates
  (Game|UnbindEventfromOnEndPlay Vehicle (delegate Custom|VehicleDestroyed)))

(event Custom|VehicleDestroyed
  (params (Actor :type Actor%20Object%20Reference) (EndPlayReason :type EEndPlayReason%20Enum :default "Destroyed"))
  (bind _spawndelay (Variables|UserKnobs|GetSpawnDelay))
  ;[1 1 1] Spawning the Vehicle and binding it to various events
  ;;[1 1 1] Post Vehicle Spawn. What needs to be done after a vehicle is spawned to manage state and hook up delegates
  (Networking|SwitchHasAuthority
    (:Authority
      (if (or (== _spawndelay 0.0) (Variables|Default|GetInstantRespawnOverride))
        (CallFunction|RespawnVehicleFromSettings)
        (elif (!= (Variables|UserKnobs|GetSpawnDelay) -1.0)
          (bind _returnvalue (Utilities|Time|SetTimerbyFunctionName 0 "Respawn Vehicle from Settings" _spawndelay))
          (Variables|Default|SetVehicleSpawnTimer _returnvalue)
          (Variables|Default|SetLastCooldownStart (Fortnite|Sync|Time|GetServerWorldTimeSeconds)) ; Start a Cooldown
        )))))

(event Custom|Spawn%20Vehicle
  ;[1 1 1] Spawning the Vehicle and binding it to various events
  ;;[0 0 0] SpawnVehicle Asnycloading the vehicle class if necessary
  (Networking|SwitchHasAuthority
    (:Authority
      (if (and (Fortnite|AreAthenaVehiclesEnabled) (not (CallFunction|IsGloballyDisabled)))
        (if (Utilities|IsValidSoftClassReference (Variables|Default|GetAthenaVehicleTypes))
          (bind _class (Utilities|AsyncLoadClassAsset (Variables|Default|GetAthenaVehicleTypes))
            (:Completed
              (bind _asfort_athena_vehicle (Utilities|Casting|CastToFortAthenaVehicleClass _class))
              (bind _returnvalue (Game|SpawnActorfromClass _asfort_athena_vehicle (CallFunction|GetSpawnLocation) "AlwaysSpawn" "SelectDefaultAtRuntime" 0 false 0))
              (bind _asfort_vehicle_interface (Utilities|Casting|CastToFortVehicleInterface _returnvalue)
                (:then
                  (Variables|Default|SetCachedVehicleInterface _asfort_vehicle_interface))
                (:CastFailed
                  _))
              (InteractInterface|SetAllowInteractionBetweenFortVolumes _returnvalue false :!class /Script/FortniteGame.FortAthenaVehicle)
              (CallFunction|OnVehicleSpawned _returnvalue)))
          (elif (Utilities|IsValidSoftClassReference (Variables|Default|GetCharacterVehicleTypes))
            (bind _class_2 (Utilities|AsyncLoadClassAsset (Variables|Default|GetCharacterVehicleTypes))
              (:Completed
                (bind _asfort_character_vehicle (Utilities|Casting|CastToFortCharacterVehicleClass _class_2))
                (bind _returnvalue_2 (Game|SpawnActorfromClass _asfort_character_vehicle (CallFunction|GetSpawnLocation) "AlwaysSpawn" "SelectDefaultAtRuntime" 0 0))
                (bind _asfort_vehicle_interface_2 (Utilities|Casting|CastToFortVehicleInterface _returnvalue_2)
                  (:then
                    (Variables|Default|SetCachedVehicleInterface _asfort_vehicle_interface_2))
                  (:CastFailed
                    _))
                (InteractInterface|SetAllowInteractionBetweenFortVolumes _returnvalue_2 false :!class /Script/FortniteGame.FortCharacterVehicle)
                (CallFunction|OnVehicleSpawned _returnvalue_2)))))))))

(event EventBeginPlay
  (bind _charactervehicletypes (Variables|Default|GetCharacterVehicleTypes))
  (bind _athenavehicletypes (Variables|Default|GetAthenaVehicleTypes))
  ;[0 1 0.1154] Initialization
  ;;[0 0 0] Async Load the vehicle we intend to use
  (if (not (CallFunction|IsGloballyDisabled))
    (if (Utilities|IsValidSoftClassReference _athenavehicletypes)
      (bind _class (Utilities|AsyncLoadClassAsset _athenavehicletypes)
        (:Completed
          (Variables|Default|SetLoadedAthenaVehicle _class)))
      (elif (Utilities|IsValidSoftClassReference _charactervehicletypes)
        (bind _class_2 (Utilities|AsyncLoadClassAsset _charactervehicletypes)
          (:Completed
            (Variables|Default|SetLoadedAthenaVehicle _class_2)))))
    (Networking|SwitchHasAuthority
      (:Authority
        (CallFunction|UpdateEnabledFromPhase "PreGame")))))

(event EventEndPlay
  (params (EndPlayReason :type EEndPlayReason%20Enum :default "Destroyed"))
  (bind _cachedvehicleinterface (Variables|Default|GetCachedVehicleInterface))
  ;[1 0 0] Destruction
  (Networking|SwitchHasAuthority
    (:Authority
      (Utilities|IsValid (Utilities|ToObject(Interface) _cachedvehicleinterface)
        (:Is%20Valid
          (if (<= (Utilities|Array|Length (Vehicle|GetAllPassengers _cachedvehicleinterface)) 0)
            (CallFunction|ManuallyDestroyVehicle)
            (else
              (Vehicle|SetDestroyonLastExit (Variables|Default|GetCachedVehicleInterface) true))))))))

(event EventOnDestroy
  (bind _cachedvehicleinterface (Variables|Default|GetCachedVehicleInterface))
  ;[1 0 0] Destruction
  (Networking|SwitchHasAuthority
    (:Authority
      (Utilities|IsValid (Utilities|ToObject(Interface) _cachedvehicleinterface)
        (:Is%20Valid
          (if (<= (Utilities|Array|Length (Vehicle|GetAllPassengers _cachedvehicleinterface)) 0)
            (CallFunction|ManuallyDestroyVehicle)
            (else
              (Vehicle|SetDestroyonLastExit (Variables|Default|GetCachedVehicleInterface) true))))))))

(event OnAnyPropertyChanged(ToyOptionsComponent)
  ;[0 1 0.1154] Initialization
  (CallFunction|UpdateEnabledFromPhase "PreGame" true))

(event OnPropertiesLoaded(ToyOptionsComponent)
  ;[0 1 0.1154] Initialization
  (CallFunction|UpdateEnabledFromPhase "PreGame" true))

(event EventOnPlaced
  ;[0 1 0.1154] Initialization
  (Networking|CanExecuteCosmeticEvents
    (:True
      (Audio|PlaySoundatLocation (Variables|Audio|GetTrap_Placed_Sound) (Transformation|GetActorLocation)))))

(event MinigameStateChangedDelegate(FortMinigameProgress)
  (params (Minigame :type Fort%20Minigame%20Object%20Reference) (NewMinigameState :type EFortMinigameState%20Enum :default "PreGame"))
  ;[0 0.432 1] Minigame Progression
  (CallFunction|UpdateEnabledFromPhase NewMinigameState)
  (switch Utilities|FlowControl|Switch|SwitchonEFortMinigameState NewMinigameState
    (:InProgress
      _)
    (:PostGameEnd
      _))
  (CallFunction|RefreshVehicleSpawn))

(event Attachment|EventOnAttachToBuilding
  (params (AttachedToBuilding :type Building%20Actor%20Object%20Reference))
  ;[1 1 1] Temporary Workaround for the Biplane Specifically
  (Utilities|IsValid AttachedToBuilding
    (:Is%20Valid
      (Utilities|Array|AddUnique (Variables|/Script/Engine.Actor|GetTags AttachedToBuilding) "NoVehicleDamage"))))

(event OnGameplayMessageReceived(EnableVehicleRespawnReceiverComponent)
  (params (TriggerInstigator :type Controller%20Object%20Reference))
  ;[0.6648 0 1] Receivers
  (if (not (CallFunction|IsEnabled))
    (Utilities|StopOtherforDelay
      (:Out%20A
        (Variables|Default|SetbEnabled true))
      (:Out%20B
        (Variables|Default|SetbEnabled false))) ; When one input is triggered, pass through\nand stop the other input passing through\nfor VehicleMinToggleTime seconds
  ))

(event OnGameplayMessageReceived(DisableVehicleRespawnReceiverComponent)
  (params (TriggerInstigator :type Controller%20Object%20Reference))
  ;[0.6648 0 1] Receivers
  (if (CallFunction|IsEnabled)
    (Utilities|StopOtherforDelay
      (:Out%20A
        (Variables|Default|SetbEnabled true))
      (:Out%20B
        (Variables|Default|SetbEnabled false))) ; When one input is triggered, pass through\nand stop the other input passing through\nfor VehicleMinToggleTime seconds
  ))

(event OnGameplayMessageReceived(RespawnReceiverComponent)
  (params (TriggerInstigator :type Controller%20Object%20Reference))
  ;[0.6648 0 1] Receivers
  (CallFunction|RespawnVehicleFromSettings)
  (Development|PrintString "Call Respawn -2-" true true "(R=0.000000,G=0.660000,B=1.000000,A=1.000000)" 1e+01))

(event Custom|Blend%20Emissive%20Color
  (params (Color :type Linear%20Color%20Structure))
  ;[1 0.7982 0] Blend Emissive quickly to new colour.
  (if (Math|Color|NotEqual(LinearColor) Color (Variables|ColourSwitches|GetCurrent%20Color))
    (Variables|ColourSwitches|SetOld%20Color (Variables|ColourSwitches|GetCurrent%20Color))
    (Variables|ColourSwitches|SetCurrent%20Color Color)
    (bind (_direction _alpha) (Timeline|Colour%20Switch%20Timeline PlayFromStart)
      (:Update
        (Rendering|Material|SetVectorParameterValue (Variables|Default|GetMID%20SpawnPlate) "Emissive" (Math|Color|Lerp(LinearColor) (Variables|ColourSwitches|GetOld%20Color) (Variables|ColourSwitches|GetCurrent%20Color) _alpha) :!class /Script/Engine.MaterialInstanceDynamic)))))

(event OnGameplayMessageReceived(DestroyReceiverComponent)
  (params (TriggerInstigator :type Controller%20Object%20Reference))
  ;[0.6648 0 1] Receivers
  (CallFunction|ManuallyDestroyVehicle)
  (CallFunction|VehicleDestroyed 0))

(event Custom|Play%20On%20Cooldown%20Effect
  ;[1 0.7982 0] Blend Emissive quickly to new colour.
  ;;[0.599 0.009532 0] This section plays a looping pulse on the emissive that scales based on the time remaining on the timer. Don't think it's worth it, so disabling and just using the material pulsing emissive.
  (CallFunction|SetBlinkSpeed)
  (bind (_direction _blink_cooldown_colours) (Timeline|Blink%20Colour%20Lerp PlayFromStart)
    (:Update
      (bind _output_get (Variables|ColourSwitches|SetCurrent%20Color (Math|Color|Lerp(LinearColor) (Variables|ColourSwitches|GetColorCooldown) (Variables|ColourSwitches|GetColorCooldownBlink) _blink_cooldown_colours)))
      (Rendering|Material|SetVectorParameterValue (Variables|Default|GetMID%20SpawnPlate) "Emissive" _output_get :!class /Script/Engine.MaterialInstanceDynamic))
    (:Finished
      (if (CallFunction|OnCooldown)
        (CallFunction|PlayonCooldownEffect)))))

(event Custom|Stop%20On%20Cooldown%20Effect
  ;[1 0.7982 0] Blend Emissive quickly to new colour.
  ;;[0.599 0.009532 0] This section plays a looping pulse on the emissive that scales based on the time remaining on the timer. Don't think it's worth it, so disabling and just using the material pulsing emissive.
  (Timeline|Blink%20Colour%20Lerp Stop))

(fn ShouldTrigger
  (params (TouchingActors :type Array%20of%20Actor%20Object%20References))
  (return true))

(fn DestroyVehicleifNeverSpawn
  (Networking|SwitchHasAuthority
    (:Authority
      (if (== (Variables|UserKnobs|GetSpawnDelay) -1.0)
        (CallFunction|ManuallyDestroyVehicle)))))

(fn GetSpawnLocation
  (return (Transformation|GetWorldTransform (Variables|Default|GetSpawnLocation))))

(fn isEnabled
  (return (and (Variables|Default|GetbEnabled) (not (CallFunction|IsGloballyDisabled)))))

(fn isGloballyDisabled
  (return (and (not (Variables|Default|GetAsync%20Converted)) (ScalableFloat|GetValueasBool (Variables|Default|GetDisableAsync)))))

(fn OnCooldown
  (return (and (CallFunction|IsEnabled) (or (Math|Boolean|ANDBoolean (> (+ (Variables|Default|GetLastCooldownStart) (Variables|UserKnobs|GetSpawnDelay)) (Fortnite|Sync|Time|GetServerWorldTimeSeconds)) (> (Variables|UserKnobs|GetSpawnDelay) 0.0) (> (Variables|Default|GetLastCooldownStart) 0.0)) (== (Variables|UserKnobs|GetSpawnDelay) -1.0)))))

(fn OnRep_bEnabled
  (CallFunction|UpdateVehicleSpawnSettings))

(fn OnRep_LastCooldownStart
  (CallFunction|UpdateEffects))

(fn RefreshVehicleSpawn
  (Networking|SwitchHasAuthority
    (:Authority
      (Utilities|Time|ClearandInvalidateTimerbyHandle (Variables|Default|GetVehicleSpawnTimer))
      (Variables|Default|SetInstantRespawnOverride true)
      (Utilities|IsValid (Utilities|ToObject(Interface) (Variables|Default|GetCachedVehicleInterface))
        (:Is%20Valid
          (Vehicle|DestroyVehicle (Variables|Default|GetCachedVehicleInterface)))
        (:Is%20Not%20Valid
          (CallFunction|RespawnVehicleFromSettings))))))

(fn RespawnVehiclefromSettings
  (bind _respawnonenable (Variables|UserKnobs|GetRespawnOnEnable))
  (bind _returnvalue (!= _respawnonenable 0))
  (Networking|SwitchHasAuthority
    (:Authority
      (if (and (Variables|Default|GetInstantRespawnOverride) _returnvalue)
        (Variables|Default|SetInstantRespawnOverride false)
        (else
          _))
      (if _returnvalue
        (if (CallFunction|IsEnabled)
          (Utilities|FlowControl|Sequence
            (:then_0
              (CallFunction|ManuallyDestroyVehicle))
            (:then_1
              (if (not (Variables|SpamProtection|GetWaitingForThrottle))
                (bind _returnvalue_2 (Class|CreativeBlueprintLibrary|GetCreativeGeneralThrottleManager))
                (Variables|SpamProtection|SetWaitingForThrottle true)
                (Class|FortCreativeGeneralThrottleManager|RequestThrottleObject _returnvalue_2 self (Math|Conversions|ToInteger(Byte) (Utilities|Enum|LiteralenumEGeneralThrottleBuckets "NewEnumerator0")) (EventDispatchers|CreateEvent :!event Completed%20Throttle))
                (Variables|Default|SetLastCooldownStart 0.0)))))))))

(fn SetBlinkSpeed
  (bind (_minmax_cooldown_blink_time_x _minmax_cooldown_blink_time_y) (Math|Vector2D|BreakVector2D (Variables|GetMinMax%20Cooldown%20Blink%20Time_X)))
  (bind (_minmax_cooldown_blink_speed_x _minmax_cooldown_blink_speed_y) (Math|Vector2D|BreakVector2D (Variables|GetMinMax%20Cooldown%20Blink%20Speed_X)))
  (Components|Timeline|SetPlayRate (Variables|Device_Floor_VehicleSpawnerParent|GetColour%20Switch%20Timeline) (Math|Float|MapRangeClamped (- (Variables|UserKnobs|GetSpawnDelay) (- (Fortnite|Sync|Time|GetServerWorldTimeSeconds) (Variables|Default|GetLastCooldownStart))) _minmax_cooldown_blink_time_x _minmax_cooldown_blink_time_y _minmax_cooldown_blink_speed_x _minmax_cooldown_blink_speed_y)))

(fn UpdateEffects
  (LocalVariables|Setlocal%20Cooldown (CallFunction|OnCooldown))
  (Networking|CanExecuteCosmeticEvents
    (:True
      (Utilities|FlowControl|Sequence
        (:then_0
          ;[1 0.6163 0] Visual Effects
          (bind _mid_spawnplate (Variables|Default|GetMID%20SpawnPlate)
            (:then
              (Rendering|Material|SetMaterial (Variables|Debug|GetStaticMeshComponent) 0 _mid_spawnplate))
            (:else
              (bind _returnvalue (Rendering|Material|CreateDynamicMaterialInstance (Variables|Debug|GetStaticMeshComponent) 0 "/Game/Creative/Devices/VehicleSpawner/MI_Floor_Spawnpad_Creative_Vehicle_Inst.MI_Floor_Spawnpad_Creative_Vehicle_Inst" "None" :!class /Script/Engine.PrimitiveComponent))
              (Variables|Default|SetMID%20SpawnPlate _returnvalue)
              (Rendering|Material|SetTextureParameterValue (Variables|Default|GetMID%20SpawnPlate) "EmissiveTexture" (Variables|Default|GetEmissive%20Texture))))
          (bind _output_get (LocalVariables|Setlocal%20Colour (Math|Color|SelectColor (Variables|ColourSwitches|GetDisabledColor) (Math|Color|SelectColor (Variables|ColourSwitches|GetColorCooldown) (Variables|ColourSwitches|GetEnabledColor) (LocalVariables|Getlocal%20Cooldown)) (not (CallFunction|IsEnabled)))))
          (CallFunction|BlendEmissiveColor _output_get)
          (Rendering|Material|SetScalarParameterValue (Variables|Default|GetMID%20SpawnPlate) "FlickerScale" (select (LocalVariables|Getlocal%20Cooldown) 0.5 0.0) :!class /Script/Engine.MaterialInstanceDynamic))
        (:then_1
          (if (CallFunction|IsEnabled)
            (Audio|PlaySoundatLocation (Variables|Audio|GetTrap_Ready_Sound) (Transformation|GetActorLocation))))))))

(fn UpdateEnabledfromPhase
  (params (Current%20Phase :type EFortMinigameState%20Enum :default "PreGame") (Reset%20if%20no%20Change :type Boolean))
  (LocalVariables|Setlocal%20Phase Current%20Phase)
  (LocalVariables|Setlocal%20Enabled (Math|Boolean|ORBoolean (== (Variables|UserKnobs|GetEnabledIndex) 0) (and (== (Variables|UserKnobs|GetEnabledIndex) 1) (Utilities|Enum|Equal(Enum) (LocalVariables|Getlocal%20Phase) "PreGame")) (and (Utilities|Array|ContainsItem (Variables|Default|GetGameplay%20Phases) (LocalVariables|Getlocal%20Phase)) (== (Variables|UserKnobs|GetEnabledIndex) 2))))
  (if (Math|Boolean|NotEqual(Boolean) (Variables|Default|GetbEnabled) (LocalVariables|Getlocal%20Enabled))
    (Networking|SwitchHasAuthority
      (:Authority
        (Variables|Default|SetbEnabled (LocalVariables|Getlocal%20Enabled))))
    (elif Reset%20if%20no%20Change
      (CallFunction|UpdateVehicleSpawnSettings))))

(fn UpdateVehicleSpawnSettings
  (Networking|SwitchHasAuthority
    (:Authority
      (Utilities|Time|ClearandInvalidateTimerbyHandle (Variables|Default|GetVehicleSpawnTimer))
      (if (CallFunction|IsEnabled)
        (if (and (or (== (Variables|UserKnobs|GetRespawnOnEnable) 2) (and (== (Variables|UserKnobs|GetRespawnOnEnable) 1) (not (Utilities|IsValid (Utilities|ToObject(Interface) (Variables|Default|GetCachedVehicleInterface)) :!class /Script/Engine.KismetSystemLibrary)))) (not (CallFunction|OnCooldown)))
          (CallFunction|RespawnVehicleFromSettings))
        (else
          (if (Variables|UserKnobs|GetbDestroyOnDisable)
            (CallFunction|ManuallyDestroyVehicle)))))
    (:Remote
      (CallFunction|UpdateEffects) ; UpdateEffects starts with a cosmetic check, so we don't need one here
    )))

(fn ConstructionScript)
