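; Blueprint for a half-size damage rail floor trap, a child class that leaves its core behavior to its parent.
; BeginPlay calls the parent, then caps the max draw distance of its electric particle component.
; Overrides make ShouldTrigger always return true regardless of touching actors, mark the trap as targetable by linear movers, and disallow invisible props.
; The construction script only calls the parent.
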(event EventBeginPlay
  (|CallParentFunction)
  (LOD|SetMaxDrawDistance (Variables|Default|GetP_Elemental_Electric_011) 4.5e+03))

(fn ShouldTrigger
  (params (TouchingActors :type Array%20of%20Actor%20Object%20References))
  (return true))

(fn IsTargetableByLinearMover
  (return true))

(fn ShouldInvisiblePropsBeAllowed
  (return false))

(fn ConstructionScript
  (|CallParentFunction))
