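; Parent Blueprint for the Creative floor damage rail trap toy, handling its trigger hookup and zap feedback.
; On placement it adds the trigger component and starts an attached audio loop, later faded on death or destruction; ShouldTrigger always returns true.
; A generic ability GameplayCue plays the zap sound and spawns an emitter at the hit.
; A custom event broadcasts an ability impact using the vehicle's up vector and reversed impact normal.
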(event EventOnWorldReady
  (Networking|CanExecuteCosmeticEvents
    (:True
      (Audio|PlaySoundatLocation (Variables|Audio|GetTrap_Ready_Sound) (Transformation|GetActorLocation)))))

(event Custom|AbilityImpactLocation
  (params (VehicleActor :type Actor%20Object%20Reference) (HitResult :type Hit%20Result%20Structure))
  (bind (_bblockinghit _binitialoverlap _time _distance _location _impactpoint _normal _impactnormal) (Collision|BreakHitResult HitResult))
  (Audio|BroadcastAbilityImpactatLocation 0 VehicleActor (+ (* (Transformation|GetActorUpVector VehicleActor) 5e+02) (* _impactnormal -1e+03)) 777.0))

(event EventOnDeathPlayEffects
  (params (Damage :type Float%20%28single-precision%29) (DamageTags :type Gameplay%20Tag%20Container%20Structure%20%28by%20ref%29) (Momentum :type Vector :default "0, 0, 0") (HitInfo :type Hit%20Result%20Structure%20%28by%20ref%29) (InstigatedBy :type Fort%20Pawn%20Object%20Reference) (DamageCauser :type Actor%20Object%20Reference) (EffectContext :type Gameplay%20Effect%20Context%20Handle%20Structure))
  (Audio|Components|Audio|FadeOut (stale EventOnPlaced _returnvalue) 0.5 0.0 "Linear" :!class /Script/Engine.AudioComponent))

(event EventOnPlaced
  (Trap|AddTriggerComponent (Variables|Trap_Floor_Player_Jump_Pad|GetTrigger))
  (Networking|CanExecuteCosmeticEvents
    (:True
      (Audio|PlaySoundatLocation (Variables|Audio|GetTrap_Placed_Sound) (Transformation|GetActorLocation))
      (bind _returnvalue (Audio|SpawnSoundAttached (Variables|Audio|GetTrap_Looping_Sound) (Variables|Debug|GetStaticMeshComponent))))))

(event Custom|Stop%20Audio%20Loop
  (Audio|Components|Audio|FadeOut (stale EventOnPlaced _returnvalue) 0.5 0.0 "Linear" :!class /Script/Engine.AudioComponent))

(event EventDestroyed
  ;[1 1 1] Stop audio when removed with delete tool
  (CallFunction|StopAudioLoop))

(event GameplayCue|GameplayCue.Abilities.Generic
  (params (EventType :type EGameplayCueEvent%20Enum :default "OnActive") (Parameters :type Gameplay%20Cue%20Parameters%20Structure%20%28by%20ref%29))
  (bind _returnvalue (Transformation|GetActorLocation))
  (bind (_bblockinghit _binitialoverlap _time _distance _location _impactpoint) (Collision|BreakHitResult (Ability|GameplayCue|GetHitResult Parameters)))
  (Audio|PlaySoundatLocation "/Game/Sounds/Creative/Toys/DamageRails/DamageRails_Zap_Athena_Cue.DamageRails_Zap_Athena_Cue" _returnvalue)
  (Effects|SpawnEmitteratLocation "/Game/Abilities/Toys/DamageRail/P_Damage_Rail_Zap.P_Damage_Rail_Zap" _location (Math|Conversions|RotationFromXVector (- _impactpoint _returnvalue)) "1.500000,1.500000,1.500000"))

(fn ShouldTrigger
  (params (TouchingActors :type Array%20of%20Actor%20Object%20References))
  (return true))

(fn ConstructionScript)
