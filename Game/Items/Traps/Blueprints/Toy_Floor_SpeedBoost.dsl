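; Creative floor speed boost trap that exposes preset vehicle and pawn impulse magnitudes, with ready, placed and firing feedback.
; Option changes refresh visuals: a preset index of zero or below dims the pad and stops idle particles, while higher ones speed them up.
; The server toggles trigger collision to match, and cosmetic changes run only on clients that can show them.
; GetImpulseMagnitudes honors an override index but validates only the normal index.
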(event EventOnWorldReady
  (OptionData|BindEventtoOnAnyPropertyChanged (Variables|Default|GetToyOptionsComponent) (EventDispatchers|CreateEvent :!event Custom|SetTrapVisuals))
  (CallFunction|SetTrapVisuals)
  (Networking|CanExecuteCosmeticEvents
    (:True
      (Audio|PlaySoundatLocation (Variables|Audio|GetTrap_Ready_Sound) (Transformation|GetWorldLocation (Variables|Debug|GetStaticMeshComponent))))))

(event EventOnPlaced
  (Trap|AddTriggerComponent (Variables|Trap_Floor_Player_Jump_Pad|GetTrigger))
  (Networking|CanExecuteCosmeticEvents
    (:True
      (Audio|PlaySoundatLocation (Variables|Audio|GetTrap_Placed_Sound) (Transformation|GetWorldLocation (Variables|Debug|GetStaticMeshComponent))))))

(event GameplayCue|GameplayCue.Abilities.Activation
  (params (EventType :type EGameplayCueEvent%20Enum :default "OnActive") (Parameters :type Gameplay%20Cue%20Parameters%20Structure%20%28by%20ref%29))
  (Audio|PlaySoundatLocation (Variables|Audio|GetTrap_Fire_Sound) (Transformation|GetWorldLocation (Variables|Debug|GetStaticMeshComponent)))
  (Effects|SpawnEmitterAttached :EmitterTemplate "/Game/Weapons/FORT_TRAPS/Trap_SpeedBoosts/FX/P_SpeedBoost_Creative_Activate.P_SpeedBoost_Creative_Activate" :AttachToComponent (Variables|Debug|GetStaticMeshComponent) :AttachPointName "None" :Location "0,256.000000,30.000000" :Scale "1.000000,1.000000,1.000000" :LocationType "KeepRelativeOffset" :bAutoDestroy true :PoolingMethod "AutoRelease" :bAutoActivate true))

(event Custom|SetTrapVisuals
  (bind _p_speedboost_idle (Variables|Default|GetP_SpeedBoost_Idle))
  ;[1 1 1] Update Visuals / Collision
  (Utilities|FlowControl|Sequence
    (:then_0
      ;;[1 1 1] Client
      (Networking|CanExecuteCosmeticEvents
        (:True
          (CallFunction|GetTrapMID)
          (if (<= (Variables|Default|GetImpulsePresetIndex) 0)
            (Components|Activation|Deactivate _p_speedboost_idle)
            (Rendering|Material|SetVectorParameterValue (Variables|Default|GetTrapMID) "EmissiveMultiplier" "(R=0.000000,G=0.000000,B=0.000000,A=1.000000)" :!class /Script/Engine.MaterialInstanceDynamic)
            (else
              (Components|Activation|Activate _p_speedboost_idle false)
              (Rendering|Material|SetVectorParameterValue (Variables|Default|GetTrapMID) "EmissiveMultiplier" (Variables|Default|GetTrapActiveEmissive) :!class /Script/Engine.MaterialInstanceDynamic)
              (Variables|/Script/CascadeRuntime.ParticleSystemComponent|SetCustomTimeDilation (/ (Math|Conversions|ToFloat(Integer) (Variables|Default|GetImpulsePresetIndex)) 2.0) (Variables|Default|GetP_SpeedBoost_Idle)))))))
    (:then_1
      ;;[1 1 1] Server
      (Networking|SwitchHasAuthority
        (:Authority
          (Collision|SetCollisionEnabled (Variables|Trap_Floor_Player_Jump_Pad|GetTrigger) (select (<= (Variables|Default|GetImpulsePresetIndex) 0) "NoCollision" "QueryOnly")))))))

(event Custom|BoostVehicle
  (params (VehicleActor :type Actor%20Object%20Reference) (ImpulseVector :type Vector))
  (Audio|BroadcastAbilityImpactatLocation (Variables|Audio|GetTrap_Fire_Sound) VehicleActor ImpulseVector 777.0))

(event EventBeginPlay
  (|CallParentFunction)
  (LOD|SetMaxDrawDistance (Variables|Default|GetP_SpeedBoost_Idle) 5e+03))

(fn ShouldTrigger
  (params (TouchingActors :type Array%20of%20Actor%20Object%20References))
  (return true))

(fn GetImpulseMagnitudes
  (bind (_vehicleimpulsemagniture_4_3f79b7c3469bc952924faa8bc00e2eb7 _pawnimpulsemagnitude_6_89dcc04642398d0b9d2e1ba2823ac12b) (Utilities|Struct|BreakSpeedBoostImpulsePreset (Utilities|Array|Get(aref) (Variables|Default|GetImpulsePresets) (Math|Integer|SelectInt (Variables|Default|GetImpulseOverride) (Variables|Default|GetImpulsePresetIndex) (>= (Variables|Default|GetImpulseOverride) 0)))))
  (bind _impulsepresets (Variables|Default|GetImpulsePresets))
  (bind _impulsepresetindex (Variables|Default|GetImpulsePresetIndex))
  (if (Utilities|Array|IsValidIndex _impulsepresets _impulsepresetindex)
    (return _vehicleimpulsemagniture_4_3f79b7c3469bc952924faa8bc00e2eb7 _pawnimpulsemagnitude_6_89dcc04642398d0b9d2e1ba2823ac12b)
    (else
      (return 0.0 0.0))))

(fn GetTrapMID
  (Variables|Default|GetTrapMID
    (:then)
    (:else
      (bind _returnvalue (Rendering|Material|CreateDynamicMaterialInstance (Variables|Debug|GetStaticMeshComponent) 0 0 "None" :!class /Script/Engine.PrimitiveComponent))
      (Variables|Default|SetTrapMID _returnvalue))))

(fn OnRep_ImpulseVector)

(fn ConstructionScript)
