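; Creative floor trap device that applies a configurable speed modifier through a gameplay ability, triggered by touch or an Activate event.
; Minigame phase and Enable/Disable events drive the enabled component, and UpdateTrap syncs trigger collision, speed state, emissive color and chevron particles.
; Activate sets the instigator as RemoteTarget so the ability skips the trigger search.
; Device Triggered clears that target and, for valid instigators, starts the reset delay and broadcasts OnActivation.
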(event MinigameStateChangedDelegate(FortMinigameProgress)
  (params (Minigame :type Fort%20Minigame%20Object%20Reference) (NewMinigameState :type EFortMinigameState%20Enum :default "PreGame"))
  ;[0 0.5075 1 0.5] State Changes
  (Class|CreativeEnabledComponent|SetEnabledOnPhase (Variables|Default|GetCreative_Enabled_Component) NewMinigameState (Variables|UserOptions|GetEnabledDuringPhase)))

(event Custom|Enable%20Event
  (params (Instigator :type Controller%20Object%20Reference))
  ;[0.8046 0 1 0.5] Gameplay Functions
  (Class|CreativeEnabledComponent|SetEnabled (Variables|Default|GetCreative_Enabled_Component) true))

(event EventBeginPlay
  ;[0 1 0.3143 0.5] Initialization
  (|CallParentFunction)
  (LOD|SetMaxDrawDistance (Variables|Default|GetP_SpeedBoost_Idle) 5e+03))

(event Custom|Disable%20Event
  (params (Instigator :type Controller%20Object%20Reference))
  ;[0.8046 0 1 0.5] Gameplay Functions
  (Class|CreativeEnabledComponent|SetEnabled (Variables|Default|GetCreative_Enabled_Component) false))

(event OnEnabledStateChanged(Creative_Enabled_Component)
  (params (Enabled :type Boolean))
  ;[0 0.5075 1 0.5] State Changes
  (CallFunction|UpdateTrap))

(event EventPropertiesUpdated
  ;[0 1 0.3143 0.5] Initialization
  (|CallParentFunction)
  (Class|CreativeEnabledComponent|SetEnabledOnPhase (Variables|Default|GetCreative_Enabled_Component) (CallFunction|GetCurrentState) (Variables|UserOptions|GetEnabledDuringPhase))
  (CallFunction|UpdateTrap))

(event Custom|Activate%20Event
  (params (Instigator :type Controller%20Object%20Reference))
  ;[0.8046 0 1 0.5] Gameplay Functions
  (if (Class|CreativeEnabledComponent|IsEnabled (Variables|Default|GetCreative_Enabled_Component))
    (bind _output_get (Variables|Default|SetRemoteTarget Instigator)) ; If a remote target is set\nthe Ability will apply\ndirectly to that actor rather\nthan search on the trigger.
    (Utilities|IsValid _output_get
      (:Is%20Valid
        (Ability|SendGameplayEventtoActor self "(TagName=\"Gameplay.Effect.Trap.Activate\")")))))

(event EventOnPlaced
  ;[0 1 0.3143 0.5] Initialization
  (Trap|AddTriggerComponent (Variables|Trap_Floor_Player_Jump_Pad|GetTrigger))
  (Networking|SwitchHasAuthority
    (:Authority
      (CallFunction|PropertiesUpdated))
    (:Remote
      (Audio|PlaySoundatLocation (Variables|Audio|GetTrap_Placed_Sound) (Transformation|GetWorldLocation (Variables|Debug|GetStaticMeshComponent))))))

(event Custom|Update%20Trap
  ;[0 0.5217 1 0.5] Update Trap Logic
  (Collision|SetCollisionEnabled (Variables|Trap_Floor_Player_Jump_Pad|GetTrigger) (select (Class|CreativeEnabledComponent|IsEnabled (Variables|Default|GetCreative_Enabled_Component)) "QueryOnly" "NoCollision"))
  (CallFunction|SetSpeedState)
  (CallFunction|SetTrapVisuals))

(event Custom|SetTrapVisuals
  (bind _p_speedboost_idle (Variables|Default|GetP_SpeedBoost_Idle))
  ;[1 0.7825 0 0.5] Client - Update Trap Visuals with the current state
  (Networking|CanExecuteCosmeticEvents
    (:True
      (if (Visibility|ShouldbeVisible (Variables|Default|GetVisibleInGame))
        (Rendering|Material|SetVectorParameterValue (Material|GetMID (Variables|Debug|GetStaticMeshComponent)) "EmissiveMultiplier" (Utilities|Map|Find (Variables|Default|GetSpeedEmissives) (Variables|Default|GetSpeedState)) :!class /Script/Engine.MaterialInstanceDynamic)
        (switch Utilities|FlowControl|Switch|SwitchonECreativeSpeedModType (Variables|Default|GetSpeedState)
          (:NewEnumerator0
            _)
          (:NewEnumerator1
            (Components|Activation|Activate _p_speedboost_idle false)
            (Rendering|Material|CreateDynamicMaterialInstance _p_speedboost_idle 0 (select (Utilities|Enum|Equal(Enum) (Variables|Default|GetSpeedState) "NewEnumerator3") "/Game/Weapons/FORT_TRAPS/Trap_SpeedBoosts/FX/M_SpeedBoost_Chevron_Creative_Fast_Inst.M_SpeedBoost_Chevron_Creative_Fast_Inst" "/Game/Weapons/FORT_TRAPS/Trap_SpeedBoosts/FX/M_SpeedBoost_Chevron_Creative_Slow_Inst.M_SpeedBoost_Chevron_Creative_Slow_Inst") "None" :!class /Script/Engine.PrimitiveComponent)
            (Variables|/Script/CascadeRuntime.ParticleSystemComponent|SetCustomTimeDilation (Utilities|Array|Get(acopy) (Variables|Default|GetSpeedBoostFXTimeDilations) (Variables|UserOptions|Basic|GetSpeed)) (Variables|Default|GetP_SpeedBoost_Idle)))
          (:NewEnumerator2
            _)
          (:NewEnumerator3
            _))
        (Components|Activation|Deactivate (Variables|Default|GetP_SpeedBoost_Idle))))))

(event Custom|Device%20Triggered
  (params (TriggerInstigator :type Controller%20Object%20Reference))
  ;[0 0.4302 1 0.5] Device Triggered Logic
  (Variables|Default|SetRemoteTarget 0) ; Ability Activation is complete\nWipe the remote target
  (Utilities|IsValid TriggerInstigator
    (:Is%20Valid
      (Reset|ResetDelayStart (Variables|Default|GetCreative_Enabled_Component) (Variables|UserOptions|GetResetDelay))
      (GameplayEvent|BroadcastEvent (Variables|UserOptions|Events|GetOnActivation) 0 (Utilities|Struct|MakeGameplayEventActivated TriggerInstigator)))) ; If instigator is invalid, it's not actually successfully \napplied to anything, so we don't want to count\nthis as 'activated'
)

(fn ShouldTrigger
  (params (TouchingActors :type Array%20of%20Actor%20Object%20References))
  (return true))

(fn GetCurrentState
  (bind _fortminigameprogress (Variables|Default|GetFortMinigameProgress)
    (:then
      (bind _currentminigame (Variables|/Script/FortniteGame.FortMinigameLogicComponent|GetCurrentMinigame _fortminigameprogress)
        (:then
          (return (Variables|/Script/FortniteGame.FortMinigame|GetCurrentState _currentminigame)))
        (:else
          _)))
    (:else
      _))
  (return "PreGame"))

(fn GetVisualComponents
  (return (Utilities|Array|MakeArray (Variables|Debug|GetStaticMeshComponent) (Variables|Default|GetP_SpeedBoost_Idle))))

(fn IsVisibleDuringMinigame
  (return (Variables|UserOptions|Basic|GetVisibleDuringGame)))

(fn IsValidTeam
  (params (Actor :type Actor%20Object%20Reference))
  (return (or (Knobs|Team|IsTeamAny (Creative|TeamOption|GetLegacyTeamIndex (Variables|UserOptions|GetActivatingTeam))) (== (Knobs|Team|GetTeamValueasIndex (Creative|TeamOption|GetLegacyTeamIndex (Variables|UserOptions|GetActivatingTeam))) (Math|Conversions|ToInteger(Byte) (Team|GetActorPvPTeamIndex Actor))))))

(fn OnRep_ImpulseVector)

(fn SetSpeedState
  (Variables|Default|SetSpeedState (select (Class|CreativeEnabledComponent|IsEnabled (Variables|Default|GetCreative_Enabled_Component)) (Utilities|Select :Index (Variables|UserOptions|Basic|GetSpeed) :Option%200 "NewEnumerator1" :Option%201 "NewEnumerator1" :Option%202 "NewEnumerator1" :Option%203 "NewEnumerator2" :Option%204 "NewEnumerator3" :Option%205 "NewEnumerator3" :Option%206 "NewEnumerator3") "NewEnumerator0")))

(fn ConstructionScript)
