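; Creative sentry floor-trap device that spawns an AI turret pawn on its configured team, defaulting to Monster.
; Minigame start optionally spawns it via the CallBackup gameplay event; death broadcasts On Eliminated and schedules preset-delayed respawns; minigame end cleans up.
; Outside play mode, local pawn overlap fades a timeline-driven sphere showing the preset range.
; Turret-tagged damage to the attached building is zeroed.
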(event EventEndPlay
  (params (EndPlayReason :type EEndPlayReason%20Enum :default "Destroyed"))
  (bind _spawnedpawn (Variables|Default|GetSpawnedPawn))
  (Networking|SwitchHasAuthority
    (:Authority
      (Utilities|IsValid _spawnedpawn
        (:Is%20Valid
          (Actor|SetLifeSpan _spawnedpawn 0.1))))))

(event MinigameStartedDelegate(FortMinigameProgress)
  (bind _spawnedpawn (Variables|Default|GetSpawnedPawn))
  ;[1 1 1] Initialization if placed by a player as a trap, also ensuring the turret uses the owner's team
  (Networking|SwitchHasAuthority
    (:Authority
      (bind _currentminigame (Variables|/Script/FortniteGame.FortMinigameLogicComponent|GetCurrentMinigame (Variables|Default|GetFortMinigameProgress))
        (:then
          (Minigame|BindEventtoPlayerChangedTeamDelegate _currentminigame (delegate Custom|OnPlayerChangedTeams)))
        (:else
          _))
      (CallFunction|SetTeam (CallFunction|GetTeamIndex))
      (if (Variables|UserOptions|GetbSpawnOnMinigameStart)
        (CallFunction|BeginSpawn true)
        (elif (Utilities|IsValid _spawnedpawn :!class /Script/Engine.KismetSystemLibrary)
          (Actor|DestroyActor _spawnedpawn))))))

(event Custom|OnPlayerChangedTeams
  (params (PlayerState :type Fort%20Player%20State%20Object%20Reference) (TeamId :type Byte) (OldTeamId :type Byte) (bTeamChangedWithoutRespawn :type Boolean))
  ;[1 1 1] Initialization if placed by a player as a trap, also ensuring the turret uses the owner's team
  (if bTeamChangedWithoutRespawn
    (bind _spawnedpawn (Variables|Default|GetSpawnedPawn)
      (:then
        (Class|AITurretPawn|ReEvaluateTargets _spawnedpawn))
      (:else))))

(event MinigameEndedDelegate(FortMinigameProgress)
  (bind _spawnedpawn (Variables|Default|GetSpawnedPawn))
  (Networking|SwitchHasAuthority
    (:Authority
      (bind _currentminigame (Variables|/Script/FortniteGame.FortMinigameLogicComponent|GetCurrentMinigame (Variables|Default|GetFortMinigameProgress))
        (:then
          (Minigame|UnbindEventfromPlayerChangedTeamDelegate _currentminigame (delegate Custom|OnPlayerChangedTeams)))
        (:else
          _)))
    (:Remote
      _))
  (Utilities|Time|ClearandInvalidateTimerbyHandle (Variables|Default|GetSpawn%20Timer%20Handle))
  (Utilities|IsValid _spawnedpawn
    (:Is%20Valid
      (Actor|SetLifeSpan _spawnedpawn 0.1))))

(event Custom|SpawnEvent
  (params (Instigator :type Controller%20Object%20Reference))
  (bind _currentminigame (Variables|/Script/FortniteGame.FortMinigameLogicComponent|GetCurrentMinigame (Variables|Default|GetFortMinigameProgress))
    (:then
      (Minigame|IsInProgress _currentminigame
        (:Yes
          _)))
    (:else
      _))
  (CallFunction|BeginSpawn false))

(event Custom|BindDeathDelegate
  ;[1 1 1] Binding Associated Pawn Death Delegate
  (EventDispatchers|BindEventtoOnDied (Variables|Default|GetSpawnedPawn) (delegate Custom|OnDied_Bind) :!class /Script/FortniteGame.FortPawn))

(event GameplayCue|GameplayCue.Abilities.Generic.BotTurret
  (params (EventType :type EGameplayCueEvent%20Enum :default "OnActive") (Parameters :type Gameplay%20Cue%20Parameters%20Structure%20%28by%20ref%29))
  (bind _spawnpoint (Variables|Default|GetSpawnPoint))
  (bind _returnvalue (Transformation|GetWorldLocation _spawnpoint))
  ;[1 1 1] GameplayCue for spawning
  (Effects|SpawnEmitteratLocation (Variables|Default|GetSpawnParticle) _returnvalue)
  (Audio|PlaySoundatLocation (Variables|Default|GetSpawnSound) _returnvalue))

(event Custom|OnDied_Bind
  (params (DamagedActor :type Actor%20Object%20Reference) (Damage :type Float%20%28single-precision%29) (InstigatedBy :type Controller%20Object%20Reference) (DamageCauser :type Actor%20Object%20Reference) (HitLocation :type Vector) (FHitComponent :type Primitive%20Component%20Object%20Reference) (BoneName :type Name) (Momentum :type Vector))
  (bind _respawnpresetindex (Variables|UserOptions|GetRespawnPresetIndex))
  (bind _respawnpresets (Variables|Configurable|GetRespawnPresets))
  (bind _output (Utilities|Array|Get(aref) _respawnpresets _respawnpresetindex))
  ;[1 1 1] Binding Associated Pawn Death Delegate
  (Utilities|IsValid (Variables|Default|GetSpawnedPawn)
    (:Is%20Valid
      (Variables|Default|SetSpawnedPawn 0)
      (GameplayEvent|BroadcastEvent (Variables|UserOptions|Events|GetOn%20Eliminated) 0 (Utilities|Struct|MakeGameplayEventonEliminationTransmit InstigatedBy))
      (if (!= _output -1.0)
        (bind _currentminigame (Variables|/Script/FortniteGame.FortMinigameLogicComponent|GetCurrentMinigame (Variables|Default|GetFortMinigameProgress))
          (:then
            (if (Utilities|Enum|Equal(Enum) (Variables|/Script/FortniteGame.FortMinigame|GetCurrentState _currentminigame) "InProgress")
              (CallFunction|BeginSpawn false _output)))
          (:else))))))

(event Custom|ShowRangeVisualization
  ;[1 0.639 0 0.5] Expand and Show or Shrink and Hide (or just immediately hide) the Sphere to visualize sentry aggro range.
  (Development|SetHiddeninGame (Variables|Default|GetVisualizationSphere) false false)
  (bind _material_instance_dynamic (CallFunction|GetVisualizationSphereMID))
  (bind (_direction _lerp) (Timeline|FadeVisualizationCubeTL Play)
    (:Update
      (Rendering|Material|SetScalarParameterValue (stale Custom|HideRangeVisualization _material_instance_dynamic) "Opacity" _lerp :!class /Script/Engine.MaterialInstanceDynamic)
      (Transformation|SetWorldScale3D (Variables|Default|GetVisualizationSphere) (Math|Conversions|ToVector(Float) (Math|Float|Lerp (/ (Utilities|Array|Get(acopy) (Variables|Configurable|GetRangePresets) (Variables|UserOptions|GetRangePresetIndex)) 82.0) (/ (Utilities|Array|Get(acopy) (Variables|Configurable|GetRangePresets) (Variables|UserOptions|GetRangePresetIndex)) 5e+01) _lerp))))
    (:Finished
      (switch Utilities|FlowControl|Switch|SwitchonETimelineDirection _direction
        (:Backward
          (Development|SetHiddeninGame (Variables|Default|GetVisualizationSphere) true false))))))

(event Custom|HideRangeVisualization
  (params (Immediate :type Boolean))
  ;[1 0.639 0 0.5] Expand and Show or Shrink and Hide (or just immediately hide) the Sphere to visualize sentry aggro range.
  (bind _material_instance_dynamic (CallFunction|GetVisualizationSphereMID))
  (if (not Immediate)
    (Timeline|FadeVisualizationCubeTL Reverse)
    (else
      (Timeline|FadeVisualizationCubeTL SetNewTime))))

(event OnComponentBeginOverlap(VisualizationRange)
  (params (OverlappedComponent :type Primitive%20Component%20Object%20Reference) (OtherActor :type Actor%20Object%20Reference) (OtherComp :type Primitive%20Component%20Object%20Reference) (OtherBodyIndex :type Integer) (bFromSweep :type Boolean) (SweepResult :type Hit%20Result%20Structure%20%28by%20ref%29))
  (if (HelperFunctions|IsLocalPawn OtherActor)
    (CallFunction|ShowRangeVisualization)))

(event OnComponentEndOverlap(VisualizationRange)
  (params (OverlappedComponent :type Primitive%20Component%20Object%20Reference) (OtherActor :type Actor%20Object%20Reference) (OtherComp :type Primitive%20Component%20Object%20Reference) (OtherBodyIndex :type Integer))
  (if (HelperFunctions|IsLocalPawn OtherActor)
    (CallFunction|HideRangeVisualization false)))

(event OnPropertiesLoaded(ToyOptionsComponent)
  ;[0 0.4792 0.1629] Initialization
  (Transformation|SetWorldScale3D (Variables|Default|GetVisualizationSphere) (Math|Conversions|ToVector(Float) (/ (Utilities|Array|Get(acopy) (Variables|Configurable|GetRangePresets) (Variables|UserOptions|GetRangePresetIndex)) 5e+01)))
  (CallFunction|UpdateVisualizationState))

(event OnAnyPropertyChanged(ToyOptionsComponent)
  ;[0 0.4792 0.1629] Initialization
  (Transformation|SetWorldScale3D (Variables|Default|GetVisualizationSphere) (Math|Conversions|ToVector(Float) (/ (Utilities|Array|Get(acopy) (Variables|Configurable|GetRangePresets) (Variables|UserOptions|GetRangePresetIndex)) 5e+01)))
  (CallFunction|UpdateVisualizationState))

(event Custom|Begin%20Spawn
  (params (bKillSpawnedPawnIfValid :type Boolean) (DelayTime :type Float%20%28double-precision%29))
  (bind _spawnedpawn (Variables|Default|GetSpawnedPawn))
  (if (not (Utilities|Time|IsTimerActivebyHandle (Variables|Default|GetSpawn%20Timer%20Handle)))
    (if (Utilities|IsValid _spawnedpawn :!class /Script/Engine.KismetSystemLibrary)
      (if bKillSpawnedPawnIfValid
        (Actor|DestroyActor _spawnedpawn)))
    (if (> DelayTime 0.0)
      (bind _returnvalue (Utilities|Time|SetTimerbyEvent (delegate Custom|Spawn%20Backup) DelayTime))
      (Variables|Default|SetSpawn%20Timer%20Handle _returnvalue)
      (else
        (CallFunction|SpawnBackup)))))

(event Custom|Spawn%20Backup
  (if (Fortnite|IsCreativeorPlaygroundModePlaylist)
    (bind _currentminigame (Variables|/Script/FortniteGame.FortMinigameLogicComponent|GetCurrentMinigame (Variables|Default|GetFortMinigameProgress))
      (:then
        (if (Minigame|IsinProgress _currentminigame)
          _))
      (:else)))
  (Ability|SendGameplayEventtoActor self "(TagName=\"Abilities.Generic.CallBackup\")"))

(event MinigameStateChangedDelegate(FortMinigameProgress)
  (params (Minigame :type Fort%20Minigame%20Object%20Reference) (NewMinigameState :type EFortMinigameState%20Enum :default "PreGame"))
  (if (or (Utilities|Enum|Equal(Enum) NewMinigameState "WaitingForCameras") (Utilities|Enum|Equal(Enum) NewMinigameState "PreGame"))
    (CallFunction|UpdateVisualizationState)))

(fn GetTeamIndex
  (return (select (Creative|TeamOption|IsTeamType (Variables|UserOptions|GetTeamPresetIndex) "None") (Utilities|Enum|LiteralenumEFortTeam "Monster") (Creative|TeamOption|ConverttoTeam (Variables|UserOptions|GetTeamPresetIndex)))))

(fn GetVisualizationSphereMID
  (bind _asmaterial_instance_dynamic (Utilities|Casting|CastToMaterialInstanceDynamic (Rendering|Material|GetMaterial (Variables|Default|GetVisualizationSphere) 0))
    (:then
      (return _asmaterial_instance_dynamic))
    (:CastFailed
      (bind _returnvalue (Rendering|Material|CreateDynamicMaterialInstance (Variables|Default|GetVisualizationSphere) 0 0 "None" :!class /Script/Engine.PrimitiveComponent))
      (return _returnvalue))))

(fn ModifyIncomingDamageForBuildingActorAttachedToMe
  (params (AttachedBuildingSMActor :type Building%20Static%20Mesh%20Object%20Reference) (Damage :type Float%20%28single-precision%29) (InTags :type Gameplay%20Tag%20Container%20Structure%20%28by%20ref%29) (EffectContext :type Gameplay%20Effect%20Context%20Handle%20Structure%20%28by%20ref%29) (EventInstigator :type Controller%20Object%20Reference) (DamageCauser :type Actor%20Object%20Reference))
  (return (Math|Float|SelectFloat 0.0 Damage (GameplayTags|DoesContainerMatchTagQuery InTags (GameplayTags|MakeGameplayTagQuery "(TokenStreamVersion=0,TagDictionary=((TagName=\"NPC.EnemyType.Turret\")),QueryTokenStream=(0,1,1,1,0),UserDescription=\"\",AutoDescription=\" ANY( NPC.EnemyType.Turret )\")")))))

(fn SendEliminatedTrigger
  (params (Instigator :type Actor%20Object%20Reference)))

(fn UpdateVisualizationState
  (Networking|CanExecuteCosmeticEvents
    (:True
      (if (not (Class|CreativeCommonDeviceFunctionLibrary|IsinPlayMode self))
        (Collision|SetCollisionEnabled (Variables|Default|GetVisualizationRange) "QueryOnly")
        (else
          (Collision|SetCollisionEnabled (Variables|Default|GetVisualizationRange) "NoCollision")
          (CallFunction|HideRangeVisualization true))))))

(fn ConstructionScript
  (|CallParentFunction)
  (Rendering|SetVisibility (Variables|Default|GetVisualizationSphere) (Variables|UserOptions|GetShow%20Range%20Visualization) false)
  (if (Variables|Stat|GetCanTrapBeDamaged)
    (Fortnite|SetCanbeDamaged 0 true)
    (else
      (Fortnite|SetCanbeDamaged 0))))
