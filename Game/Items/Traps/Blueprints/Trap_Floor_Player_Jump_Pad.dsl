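; Floor jump pad trap Blueprint that provides a launch impulse from a small share of an actor's horizontal velocity plus a set upward strength.
; Placement creates a dynamic material, registers the trigger and plays dust.
; Emissive glow turns on when building or reloading finishes and off on the fire gameplay cue, which also activates jump particles.
; Trap state changes play sounds; depleted durability triggers a delayed explosion sound and destruction emitter.
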(event EventOnPlaced
  ;[1 1 1] Initial Setup
  (bind _returnvalue (Rendering|Material|CreateDynamicMaterialInstance (Building|GetBuildingMeshComponent) 0 (Rendering|Material|GetMaterial (Building|GetBuildingMeshComponent) 0) "None" :!class /Script/Engine.PrimitiveComponent))
  (Variables|Default|SetJumpMaterial _returnvalue)
  (Rendering|Material|SetScalarParameterValue (Variables|Default|GetJumpMaterial) "Emissive_Mult" 0.0 :!class /Script/Engine.MaterialInstanceDynamic)
  (Trap|AddTriggerComponent (Variables|Trap_Floor_Player_Jump_Pad|GetTrigger))
  (Audio|PlaySoundatLocation (Variables|Audio|GetTrap_Placed_Sound) (Transformation|GetActorLocation))
  (Utilities|TrapPlacementDustFX :Attach%20to%20Component (Variables|Debug|GetStaticMeshComponent) :Socket "PlacedFX" :Rotation "0, 0, 0" :Burst%20Rate 8 :Size%20%28Volume%29 "100.000000,100.000000,15.000000" :Velocity%20Scale "1,1,1" :Size%20Scale%20%28Particles%29 1))

(event EventOnFinishedBuilding
  ;[1 1 1] Finished building
  (Audio|PlaySoundatLocation (Variables|Audio|GetTrap_Active_Sound) (Transformation|GetActorLocation))
  (Rendering|Material|SetScalarParameterValue (Variables|Default|GetJumpMaterial) "Emissive_Mult" 3.5 :!class /Script/Engine.MaterialInstanceDynamic))

(event GameplayCue|GameplayCue.Abilities.Activation.Traps.ActivateTrap
  (params (EventType :type EGameplayCueEvent%20Enum :default "OnActive") (Parameters :type Gameplay%20Cue%20Parameters%20Structure%20%28by%20ref%29))
  ;[1 1 1] On Fire
  (switch Utilities|FlowControl|Switch|SwitchonEGameplayCueEvent EventType
    (:Executed
      (Audio|PlaySoundatLocation (Variables|Audio|GetTrap_Fire_Sound) (Transformation|GetActorLocation))
      (Components|Activation|Activate (Variables|Trap_Floor_Player_Jump_Pad|GetJumpParticleEffect) true)
      (Rendering|Material|SetScalarParameterValue (Variables|Default|GetJumpMaterial) "Emissive_Mult" 0.0 :!class /Script/Engine.MaterialInstanceDynamic))))

(event GameplayCue|GameplayCue.Abilities.Activation.Traps.ReloadBegin
  (params (EventType :type EGameplayCueEvent%20Enum :default "OnActive") (Parameters :type Gameplay%20Cue%20Parameters%20Structure%20%28by%20ref%29))
  ;[1 1 1] Reload Begin
  (switch Utilities|FlowControl|Switch|SwitchonEGameplayCueEvent EventType
    (:Executed
      (Audio|PlaySoundatLocation (Variables|Audio|GetTrap_Reload_Sound) (Transformation|GetActorLocation)))))

(event EventOnReloadEnd
  ;[1 1 1] Reload End
  (Audio|PlaySoundatLocation (Variables|Audio|GetTrap_Active_Sound) (Transformation|GetActorLocation))
  (Rendering|Material|SetScalarParameterValue (Variables|Default|GetJumpMaterial) "Emissive_Mult" 3.5 :!class /Script/Engine.MaterialInstanceDynamic))

(event EventOnOutOfDurability
  ;[1 1 1] Durability Change
  (Utilities|FlowControl|Delay 1)
  (Audio|PlaySoundatLocation (Variables|Audio|GetTrap_Explode_Sound) (Transformation|GetActorLocation))
  (Effects|SpawnEmitteratLocation "/Game/Weapons/FORT_TRAPS/Spike_Trap/P_Spike_Trap_Destruction.P_Spike_Trap_Destruction" (Transformation|GetWorldLocation (Variables|Transformation|GetRootComponent))))

(fn GetImpulseDirection
  (params (LaunchingActor :type Actor%20Object%20Reference))
  (return (Math|Vector|MakeVector (* (.x (Transformation|GetVelocity LaunchingActor)) 0.1) (* (.y (Transformation|GetVelocity LaunchingActor)) 0.1) (Variables|Default|GetLaunchStrength))))

(fn ConstructionScript)
