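; Creative floor-trap scoreboard that shows Team A and Team B scores on a sign, with overlap boxes for scoring and resetting.
; The boxes send gameplay-tag events to the actor itself, while separate functions increment the scores and minigame start and end zero them.
; Score RepNotifies on remote clients play sounds and set the "Score" parameter on per-team dynamic materials.
; Construction creates and colors those materials, which are recreated after bounce animations via a dirty flag.
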(event Collision|EventActorBeginOverlap
  (params (OtherActor :type Actor%20Object%20Reference)))

(event EventTick
  (params (DeltaSeconds :type Float%20%28single-precision%29)))

(event EventOnPlaced
  (Utilities|FlowControl|DoOnce false
    (:Completed
      (Audio|PlaySoundatLocation "/Game/Sounds/Creative/Gadgets/Traps/Trap_Equip_Cue.Trap_Equip_Cue" (Transformation|GetActorLocation)))))

(event OnComponentBeginOverlap(TeamBBox)
  (params (OverlappedComponent :type Primitive%20Component%20Object%20Reference) (OtherActor :type Actor%20Object%20Reference) (OtherComp :type Primitive%20Component%20Object%20Reference) (OtherBodyIndex :type Integer) (bFromSweep :type Boolean) (SweepResult :type Hit%20Result%20Structure%20%28by%20ref%29))
  (Utilities|Casting|CastToFortPlayerPawn OtherActor)
  (Ability|SendGameplayEventtoActor self "(TagName=\"Gameplay.Effect.Toy.Activate.ScoreTeamB\")"))

(event OnComponentBeginOverlap(ResetBox)
  (params (OverlappedComponent :type Primitive%20Component%20Object%20Reference) (OtherActor :type Actor%20Object%20Reference) (OtherComp :type Primitive%20Component%20Object%20Reference) (OtherBodyIndex :type Integer) (bFromSweep :type Boolean) (SweepResult :type Hit%20Result%20Structure%20%28by%20ref%29))
  (Utilities|Casting|CastToFortPlayerPawn OtherActor)
  (Ability|SendGameplayEventtoActor self "(TagName=\"Gameplay.Effect.Toy.Activate.ResetScore\")")
  (Variables|Default|SetIsResetOverlapped true)
  (Audio|PlaySoundatLocation (Variables|Default|GetResetScoreSound) (Transformation|GetWorldLocation (Variables|Default|GetScoreSign)))
  (Variables|Default|SetIsResetOverlapped false))

(event OnComponentBeginOverlap(TeamABox)
  (params (OverlappedComponent :type Primitive%20Component%20Object%20Reference) (OtherActor :type Actor%20Object%20Reference) (OtherComp :type Primitive%20Component%20Object%20Reference) (OtherBodyIndex :type Integer) (bFromSweep :type Boolean) (SweepResult :type Hit%20Result%20Structure%20%28by%20ref%29))
  (Utilities|Casting|CastToFortPlayerPawn OtherActor)
  (Ability|SendGameplayEventtoActor self "(TagName=\"Gameplay.Effect.Toy.Activate.ScoreTeamA\")"))

(event EventOnBounceAnimationUpdate
  (params (Data :type Fort%20Bounce%20Data%20Structure%20%28by%20ref%29))
  (|CallParentFunction Data)
  (Variables|Default|SetbRegetDynamicMaterialInstances true))

(event MinigameStartedDelegate(FortMinigameProgress)
  (Variables|Default|SetTeamAScore 0)
  (Variables|Default|SetTeamBScore 0))

(event MinigameEndedDelegate(FortMinigameProgress)
  (Variables|Default|SetTeamAScore 0)
  (Variables|Default|SetTeamBScore 0))

(fn ApplyImpulse
  (params (ActorToApplyImpulse :type Actor%20Object%20Reference) (ComponentToGetDirectionFrom :type Scene%20Component%20Object%20Reference) (NewParam1 :type Gameplay%20Tag%20Structure))
  (Ability|SendGameplayEventtoActor self NewParam1))

(fn CheckForRegetDynamicMaterialInstances
  (if (Variables|Default|GetbRegetDynamicMaterialInstances)
    (CallFunction|RegetDynamicMaterialInstances))
  (return))

(fn IncrementTeamAScore
  (Variables|Default|SetTeamAScore (+ (Variables|Default|GetTeamAScore) 1)))

(fn IncrementTeamBScore
  (Variables|Default|SetTeamBScore (+ (Variables|Default|GetTeamBScore) 1)))

(fn IsTargetableByLinearMover
  (return true))

(fn OnRep_TeamAScore
  (Networking|SwitchHasAuthority
    (:Remote
      (Audio|PlaySoundatLocation (Variables|Default|GetScoreUpSound_A) (Transformation|GetWorldLocation (Variables|Default|GetScorePadA))) ; Score Up Sound
      (CallFunction|CheckforRegetDynamicMaterialInstances)
      (Rendering|Material|SetScalarParameterValue (Variables|Default|GetMID_TeamAScore) "Score" (Math|Conversions|ToFloat(Integer) (Variables|Default|GetTeamAScore)) :!class /Script/Engine.MaterialInstanceDynamic))))

(fn OnRep_TeamBScore
  (Networking|SwitchHasAuthority
    (:Remote
      (Audio|PlaySoundatLocation (Variables|Default|GetScoreUpSound_B) (Transformation|GetWorldLocation (Variables|Default|GetScorePadB))) ; Score Up Sound
      (CallFunction|CheckforRegetDynamicMaterialInstances)
      (Rendering|Material|SetScalarParameterValue (Variables|Default|GetMID_TeamBScore) "Score" (Math|Conversions|ToFloat(Integer) (Variables|Default|GetTeamBScore)) :!class /Script/Engine.MaterialInstanceDynamic))))

(fn RegetDynamicMaterialInstances
  (Utilities|FlowControl|Sequence
    (:then_0
      (bind _returnvalue (Rendering|Material|CreateDynamicMaterialInstance (Variables|Default|GetScoreSign) (Variables|Default|GetTeamAScoreMatIndex) 0 "None" :!class /Script/Engine.PrimitiveComponent))
      (Variables|Default|SetMID_TeamAScore _returnvalue))
    (:then_1
      (bind _returnvalue_2 (Rendering|Material|CreateDynamicMaterialInstance (Variables|Default|GetScoreSign) (Variables|Default|GetTeamBScoreMatIndex) 0 "None" :!class /Script/Engine.PrimitiveComponent))
      (Variables|Default|SetMID_TeamBScore _returnvalue_2))
    (:then_2
      (Variables|Default|SetbRegetDynamicMaterialInstances false))))

(fn ShouldInvisiblePropsBeAllowed
  (return false))

(fn ConstructionScript
  (bind _scoresign (Variables|Default|GetScoreSign))
  (bind _teamascorematindex (Variables|Default|GetTeamAScoreMatIndex))
  (bind _teamascoremat (Variables|Default|GetTeamAScoreMat))
  (bind _teambscorematindex (Variables|Default|GetTeamBScoreMatIndex))
  (bind _teambscoremat (Variables|Default|GetTeamBScoreMat))
  (bind _scoresign_2 (Variables|Default|GetScoreSign))
  (Utilities|FlowControl|Sequence
    (:then_0
      ;[1 1 1] Set up TeamA Scoreboard
      (if (Utilities|NotEqual(Object) (Variables|Default|GetTeamAScoreMat) 0)
        (Rendering|Material|SetMaterial _scoresign _teamascorematindex _teamascoremat)
        (else
          _))
      (bind _returnvalue (Rendering|Material|CreateDynamicMaterialInstance _scoresign _teamascorematindex _teamascoremat "None" :!class /Script/Engine.PrimitiveComponent))
      (bind _output_get (Variables|Default|SetMID_TeamAScore _returnvalue))
      (Rendering|Material|SetScalarParameterValue _output_get "Score" 0.0 :!class /Script/Engine.MaterialInstanceDynamic)
      (Rendering|Material|SetVectorParameterValue _output_get "Color" (Variables|Default|GetTeamAScoreColor) :!class /Script/Engine.MaterialInstanceDynamic))
    (:then_1
      ;[1 1 1] Set up TeamB Scoreboard
      (if (Utilities|NotEqual(Object) (Variables|Default|GetTeamBScoreMat) 0)
        (Rendering|Material|SetMaterial _scoresign_2 _teambscorematindex _teambscoremat)
        (else
          _))
      (bind _returnvalue_2 (Rendering|Material|CreateDynamicMaterialInstance _scoresign_2 _teambscorematindex _teambscoremat "None" :!class /Script/Engine.PrimitiveComponent))
      (bind _output_get_2 (Variables|Default|SetMID_TeamBScore _returnvalue_2))
      (Rendering|Material|SetScalarParameterValue _output_get_2 "Score" 0.0 :!class /Script/Engine.MaterialInstanceDynamic)
      (Rendering|Material|SetVectorParameterValue _output_get_2 "Color" (Variables|Default|GetTeamBScoreColor) :!class /Script/Engine.MaterialInstanceDynamic))))
