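# Unlit material for a trap barrier that draws a procedural blue sky from the camera vector.
# A curve-atlas gradient is layered with sun disk and glow, projected panning clouds and a rotating spherical-mapped cloud ring, then blended into horizon fog.
# Clouds mix near and far tilings of one texture, with a swirl flowmap twisting the near layer.
# A time-of-day multiplier comes from dotting per-period weights with the Fortnite parameter collection.
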Material /Game/Items/Traps/Materials/M_Barrier_BlueSky
graph {
 properties {pixelDepthOffsetMode=PDOM_Legacy, shadingModel=MSM_Unlit, usedShadingModels=Unlit}
 nodes [
  n1 = NamedRerouteUsage()
  n2 = LinearInterpolate(
   A=LinearInterpolate(
    A=CurveAtlasRowParameter(CurveTime=@n1) {atlas=/Game/Items/Traps/Textures/CA_BarrierCurves, curve=/Game/Items/Traps/Textures/C_BarrierBlueSky, parameterName=ColorCurve}
    B=NamedRerouteUsage()
    Alpha=NamedRerouteUsage()
   )
   B=VectorParameter() {parameterName=Cloud Ring Color, defaultValue=<rgba 0.9167 1 1 0>}.RGB
   Alpha=NamedRerouteUsage()
  )
  n3 = NamedRerouteUsage()
  n4 = ComponentMask(None=CameraVectorWS()) {b=true}
  n5 = Multiply(A=@n4) {constB=-1}
  n6 = StaticSwitchParameter(True=OneMinus(None=Abs(None=@n4)), False=OneMinus(None=Saturate(None=@n5))) {defaultValue=true, group=Fog, parameterName=Fog only the horizon}
  n7 = ScalarParameter() {defaultValue=15, group=Fog, parameterName=Fog Exp}
  n8 = Multiply(A=Add(A=Power(Base=@n6, Exp=@n7)) {constB=1.5}) {constB=0.5}
  n9 = Saturate(None=ConstantBiasScale(None=Sine(None=@n8)))
  n10 = ScalarParameter() {defaultValue=1, group=Fog, parameterName=Fog Intensity}
  Multiply(
   A=LinearInterpolate(A=Add(A=@n2, B=@n3), B=VectorParameter() {parameterName=Fog Color, defaultValue=<rgba 0.9532 1 0.9263 1>}.RGB, Alpha=Multiply(A=@n9, B=@n10))
   B=NamedRerouteUsage()
  )
  n11 = Saturate(None=NamedRerouteUsage())
  n12 = Multiply(
   A=Power(Base=@n11, Exp=ScalarParameter() {defaultValue=150, parameterName=Sun Disk Exp}) {constExponent=1000}
   B=VectorParameter() {parameterName=Sun Disk Color, defaultValue=<rgba 1 0.6135 0.2825 0>}.RGB
  )
  n13 = ScalarParameter() {defaultValue=0.5, parameterName=Disk Luminance}
  n14 = OneMinus(None=Divide(A=ScalarParameter() {defaultValue=0.4, parameterName=Sun Size}) {constB=1000})
  n15 = Divide(A=Subtract(A=@n11, B=@n14), B=OneMinus(None=@n14))
  n16 = ScalarParameter() {defaultValue=2, parameterName=Sun Exp}
  n17 = Multiply(
   A=Saturate(None=Power(Base=@n15, Exp=@n16) {constExponent=4})
   B=VectorParameter() {parameterName=Sun Color, defaultValue=<rgba 1 0.6084 0.2014 1>}.RGB
  )
  n18 = ScalarParameter() {defaultValue=50, parameterName=Sun Intensity}
  NamedRerouteDeclaration(None=Add(A=Multiply(A=@n12, B=@n13), B=Multiply(A=@n17, B=@n18))) {name=SUN, nodeColor=<rgba 1 0.9217 0 1>}
  n19 = Normalize(VectorInput=VectorParameter() {parameterName=Sun Position, defaultValue=<rgba 1 1 1 0>}.RGB)
  n20 = Multiply(A=DotProduct(A=NamedRerouteDeclaration(None=@n19) {name=Sun Direction, nodeColor=<rgba 1 0.3344 0 1>}, B=CameraVectorWS())) {constB=-1}
  NamedRerouteDeclaration(None=@n20) {name="Sun Gradient -1 ► 1", nodeColor=<rgba 1 0.0282 0.5297 1>}
  NamedRerouteDeclaration(None=ConstantBiasScale(None=@n20)) {name="Sun Gradient 0 ► 1", nodeColor=<rgba 1 0.1013 0.7612 1>}
  n21 = CameraVectorWS()
  n22 = Saturate(None=Multiply(A=ComponentMask(None=@n21) {b=true}) {constB=-1})
  n23 = MaterialFunctionCall(
   In=Power(Base=@n22, Exp=ScalarParameter() {defaultValue=0.9458, group=Clouds, parameterName=Cloud Interface Exp, sortPriority=102}) {constExponent=3}
   Contrast=ScalarParameter() {defaultValue=1, group=Clouds, parameterName=Cloud Interface Contrast, sortPriority=103}
  ) {materialFunction=&CheapContrast}
  n24 = ScalarParameter() {defaultValue=0.89, group=Clouds, parameterName=Cloud Tiling Near, sortPriority=101}
  n25 = CameraVectorWS()
  n26 = ComponentMask(None=@n25) {b=true}
  n27 = Divide(
   A=ComponentMask(None=@n25) {g=true, r=true}
   B=Power(Base=Abs(None=@n26), Exp=ScalarParameter() {defaultValue=0.8, group=Clouds, parameterName=Cloud UV Exponent, sortPriority=104})
  )
  n28 = Multiply(A=@n27, B=@n24)
  n29 = ScalarParameter() {group=Clouds, parameterName=Cloud Speed Direction, sliderMax=1, sortPriority=0}
  n30 = ScalarParameter() {defaultValue=1, group=Clouds, parameterName=Cloud Speed}
  n31 = Multiply(A=AppendVector(A=Cosine(None=@n29), B=Sine(None=@n29)), B=Multiply(A=Divide(A=@n30) {constB=100}, B=Time()))
  n32 = Multiply(A=@n31, B=@n24)
  n33 = Multiply(A=@n27, B=ScalarParameter() {defaultValue=0.5, group=Clouds, parameterName=Cloud Twist Tiling, sortPriority=201})
  n34 = Time()
  n35 = ScalarParameter() {defaultValue=0.01, group=Clouds, parameterName=Cloud Twist speed, sortPriority=202}
  n36 = TextureSampleParameter2D(UVs=Add(A=Add(A=@n33, B=@n31), B=Multiply(A=@n34, B=@n35))) {
   group=Clouds
   parameterName=Swirl Texture
   samplerType=SAMPLERTYPE_LinearColor
   texture=/Game/VisualThreatManager/StormVisuals/LargeClouds/T_Swirl_Flowmap
  }
  n37 = Subtract(A=AppendVector(A=@n36.R, B=@n36.R)) {constB=0.5}
  n38 = ScalarParameter() {defaultValue=0.15, group=Clouds, parameterName=Cloud Twist Strength, sortPriority=200}
  n39 = TextureObject() {samplerType=SAMPLERTYPE_LinearColor, texture=/Game/Items/Traps/Textures/T_Barrier_Clouds_M}
  n40 = TextureSample(UVs=Add(A=Add(A=@n28, B=@n32), B=Multiply(A=@n37, B=@n38)), Tex=@n39)
  n41 = ScalarParameter() {defaultValue=0.452, group=Clouds, parameterName=Cloud Tiling Far, sortPriority=100}
  n42 = Multiply(A=@n41, B=@n31)
  n43 = TextureSample(UVs=Add(A=Multiply(A=@n41, B=@n27), B=Multiply(A=@n42) {constB=0.6}), Tex=@n39)
  n44 = Saturate(None=LinearInterpolate(
    A=ScalarParameter() {group=Clouds, parameterName=Clouds Low, sortPriority=104}
    B=ScalarParameter() {defaultValue=1.1, group=Clouds, parameterName=Clouds High, sortPriority=105}
    Alpha=LinearInterpolate(A=@n43.R, B=@n40.R, Alpha=@n23.Result)
   ))
  n45 = LinearInterpolate(Alpha=NamedRerouteUsage()) {constA=-1}
  n46 = Multiply(
   A=Multiply(A=Saturate(None=@n44), B=Ceil(None=@n22))
   B=Power(Base=Saturate(None=@n45), Exp=ScalarParameter() {defaultValue=0.256, group=Clouds, parameterName=Cloud Horizon Fade, sortPriority=300})
  )
  NamedRerouteDeclaration(None=Multiply(A=VectorParameter() {parameterName=Cloud Color, defaultValue=<rgba 0.5849 0.8788 1 1>}.RGB, B=@n46)) {name=Clouds, nodeColor=<rgba 0.45 0.7743 1 1>}
  n47 = DotProduct(A=CameraVectorWS(), B=Constant3Vector() {constant=<rgba 0 0 1 0>})
  NamedRerouteDeclaration(None=Divide(A=ArccosineFast!(None=@n47)) {constB=3.1416}) {name="Gradient Vertical 0 ► 1", nodeColor=<rgba 1 0 0.1592 1>}
  NamedRerouteDeclaration(None=Power(Base=@n46, Exp=ScalarParameter() {defaultValue=3, group=Clouds, parameterName=Cloud Opacity Mask Exp, sortPriority=106})) {name=Cloud Mask, nodeColor=<rgba 0.4757 0.8169 1 1>}
  n48 = Constant2Vector() {r=-1}
  n49 = CameraVectorWS()
  n50 = Normalize(VectorInput=ComponentMask(None=@n49) {g=true, r=true})
  n51 = Divide(A=ArccosineFast(None=DotProduct(A=@n48, B=@n50))) {constB=6.2832}
  n52 = OneMinus(None=@n51)
  n53 = Constant3Vector() {constant=<rgba 0 0 -1 0>}
  n54 = Divide(A=ArccosineFast(None=DotProduct(A=@n49, B=@n53))) {constB=3.1416}
  n55 = ScalarParameter() {defaultValue=3.2957, parameterName=Cloud Ring Scale}
  n56 = Saturate(None=Add(A=Multiply(A=@n54, B=@n55), B=ScalarParameter() {defaultValue=-0.852, parameterName=Cloud Ring Y Offset}))
  n57 = Time() {bOverride_Period=true, period=3600}
  n58 = ScalarParameter() {defaultValue=0.001, parameterName=Cloud Ring Rotation Speed}
  n59 = AppendVector(A=Add(A=ScalarParameter() {defaultValue=0.12, parameterName=Cloud Ring X Offset}, B=Multiply(A=@n57, B=@n58)), B=Constant())
  n60 = TextureSample(UVs=Add(A=AppendVector(A=@n52, B=@n56), B=@n59)) {texture=/Game/Items/Traps/Textures/T_Barrier_Clouds_02_M}
  n61 = TextureSample(UVs=Add(A=AppendVector(A=@n51, B=@n56), B=@n59)) {texture=&T_Barrier_Clouds_02_M}
  n62 = Saturate(None=Ceil(None=ComponentMask(None=@n49) {g=true}))
  NamedRerouteDeclaration(None=Multiply(A=LinearInterpolate(A=@n60.R, B=@n61.R, Alpha=@n62), B=ScalarParameter() {defaultValue=0.5, parameterName=Cloud Ring Strength})) {name=Cloud Ring, nodeColor=<rgba 0 1 0.2031 1>}
  n63 = CollectionParameter() {
   collection=&FortniteMaterialParameters
   parameterName=FortniteActiveTimeOfDay
  }
  n64 = DotProduct(
   A=MaterialFunctionCall(
    R=ScalarParameter() {defaultValue=0.35, group=Time Of Day, parameterName=Morning Mult, sortPriority=20}
    G=ScalarParameter() {defaultValue=1, group=Time Of Day, parameterName=Day Mult, sortPriority=20}
    B=ScalarParameter() {defaultValue=0.35, group=Time Of Day, parameterName=Evening Mult, sortPriority=20}
    A=ScalarParameter() {defaultValue=0.5, group=Time Of Day, parameterName=Night Mult, sortPriority=21}
   ) {materialFunction=&AppendMany}.RGBA
   B=@n63
  )
  NamedRerouteDeclaration(None=@n64) {name=TOD Multiplier, nodeColor=<rgba 1 0.3047 0 1>}
  Reroute(None=@n64)
 ]
}
