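# Masked, unlit, dithered material for a trap barrier's bounds portal, showing a tiled space texture.
# The texture is triplanar-projected along the camera vector, so it follows view direction rather than mesh UVs.
# Optional hue shift and desaturation switches feed scale, contrast and bias color grading.
# A separate switchable noise-driven sine flicker branches from the raw texture and does not feed that grading chain.
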Material /Game/Items/Traps/Materials/M_Barrier_BoundsPortal
graph {
 properties {
  bUsedWithInstancedStaticMeshes=true
  blendMode=BLEND_Masked
  ditherOpacityMask=true
  opacityMaskClipValue=0.5
  pixelDepthOffsetMode=PDOM_Legacy
  shadingModel=MSM_Unlit
  usedShadingModels=Unlit
 }
 nodes [
  n1 = MaterialFunctionCall!(
   CameraVector
   Texture=TextureObjectParameter() {parameterName=TriPlanarTexture, texture=/Game/Effects/Fort_Effects/Textures/Tile/T_Portal_Space_D}
   Tiling=ScalarParameter() {defaultValue=1, parameterName=TriPlanarTiling}
   Offset
   AxisFadeContrast=ScalarParameter() {defaultValue=4, parameterName=TriPlanarAxisFadeContrast}
  ) {materialFunction=/Engine/Functions/Engine_MaterialFunctions01/Texturing/TriplanarCameraVector}
  n2 = StaticSwitchParameter(
   True=MaterialFunctionCall(Hue Shift Percentage=ScalarParameter() {parameterName=HueShift}, Texture=@n1.Result) {materialFunction=&HueShift}.Result
   False=@n1.Result
  ) {parameterName=UseHueShift}
  n3 = StaticSwitchParameter(True=Desaturation(None=@n2, Fraction=ScalarParameter() {parameterName=DesaturationFraction}), False=@n2) {parameterName=UseDesaturation}
  n4 = VectorParameter() {parameterName=ColorScale, defaultValue=<rgba 1 1 1 0>}
  Add(
   A=VectorParameter() {parameterName=ColorBias, defaultValue=<rgba 0 0.005 0.01 0>}.RGB
   B=Power(Base=Multiply(A=@n3, B=@n4.RGB), Exp=ScalarParameter() {defaultValue=1, parameterName=ContrastExponent})
  )
  ScalarParameter() {defaultValue=1, parameterName=Opacity}
  n5 = MaterialFunctionCall!(
   PrimaryTextureCoordinates=Multiply(A=TextureCoordinate(), B=ScalarParameter() {defaultValue=100, parameterName=FlickerRandomSampleFrequency})
   2ndOrderTextureCoordinates
   RandomSeed=Constant() {r=43}
   NoiseDensityBias
   Bypass Noise Influence Crosstalk
  ) {materialFunction=/Game/Effects/Fort_Effects/Materials/_Functions/MF_RandomUVSample}
  n6 = Multiply(A=@n1.Result, B=@n5.NoiseWithRange)
  n7 = ScalarParameter() {defaultValue=0.1, parameterName=FlickerSpeedLow}
  n8 = ScalarParameter() {defaultValue=1, parameterName=FlickerSpeedHigh}
  n9 = Sine(None=Multiply(A=LinearInterpolate(A=@n7, B=@n8, Alpha=@n5.NoiseWithRange), B=Time()))
  StaticSwitchParameter(True=Add(A=@n1.Result, B=Multiply(A=@n6, B=@n9)), False=@n1.Result) {parameterName=UseRandomFlicker}
 ]
}
