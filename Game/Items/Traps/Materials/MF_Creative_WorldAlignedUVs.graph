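# Material function for Creative trap materials that generates world-aligned UVs from pixel position instead of mesh UVs.
# Position, optionally made object-relative in local space by a static switch, is divided by a WorldScale input and multiplied by object scale.
# Local-space normals drive Ceil/Saturate masks intended to pick a projection per dominant axis.
# As wired, repeated channels make both the first mask zero and the output always the scaled X component duplicated.
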MaterialFunction /Game/Items/Traps/Materials/MF_Creative_WorldAlignedUVs
graph {nodes [
  n1 = WorldPosition()
  n2 = ObjectPositionWS()
  n3 = StaticSwitch(
   True=Transform(None=Subtract(A=@n1.XYZ, B=@n2)) {transformSourceType=TRANSFORMSOURCE_World, transformType=TRANSFORM_Local}
   False=WorldPosition().XYZ
   Value=FunctionInput(Preview=StaticBool() {value=true}) {bUsePreviewValueAsDefault=true, inputName=TransformToLocalSpace, inputType=FunctionInput_StaticBool, sortPriority=1}
  )
  n4 = FunctionInput(Preview=VectorParameter() {parameterName=WorldScale, defaultValue=<rgba 256 256 256 0>}.RGB) {bUsePreviewValueAsDefault=true, inputName=WorldScale}
  n5 = MaterialFunctionCall(Float3=Multiply(
    A=MaterialFunctionCall() {materialFunction=/Engine/Functions/Engine_MaterialFunctions02/WorldPositionOffset/ObjectScale}.Scale XYZ
    B=Divide(A=@n3, B=@n4)
   )) {materialFunction=&BreakOutFloat3Components}
  n6 = AppendVector(A=@n5.R, B=@n5.R)
  n7 = AppendVector(A=@n5.G, B=@n5.G)
  n8 = Transform(None=VertexNormalWS()) {transformSourceType=TRANSFORMSOURCE_World, transformType=TRANSFORM_Local}
  n9 = MaterialFunctionCall(Float3=Abs(None=Normalize(VectorInput=@n8))) {materialFunction=&BreakOutFloat3Components}
  n10 = Saturate(None=Ceil(None=Subtract(A=@n9.R, B=@n9.R)))
  n11 = Ceil(None=Subtract(A=@n9.B, B=Max(A=@n9.R, B=@n9.R)))
  FunctionOutput(None=LinearInterpolate(A=LinearInterpolate(A=@n6, B=@n7, Alpha=@n10), B=AppendVector(A=@n5.R, B=@n5.R), Alpha=Saturate(None=@n11)))
 ]}
