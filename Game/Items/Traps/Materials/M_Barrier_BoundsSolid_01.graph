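# Masked, lit material for trap barrier bounds, by its path and name, giving a solid surface with faint rock detail.
# A world-aligned UV function, transformed to local space and driven by a WorldScale parameter, maps rock base color and normal textures.
# The grayscale rock texture blends with a tint color, and a vector parameter scales normal strength.
# Scalar parameters provide metallic, specular, roughness and opacity, with dithered masking enabled.
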Material /Game/Items/Traps/Materials/M_Barrier_BoundsSolid_01
graph {
 properties {blendMode=BLEND_Masked, ditherOpacityMask=true, opacityMaskClipValue=0.5, pixelDepthOffsetMode=PDOM_Legacy, usedShadingModels=Default Lit}
 nodes [
  n1 = MaterialFunctionCall!(WorldScale=VectorParameter() {parameterName=WorldScale, defaultValue=<rgba 400 400 400 0>}.RGB, TransformToLocalSpace) {materialFunction=&/MF_Creative_WorldAlignedUVs}
  LinearInterpolate(
   A=VectorParameter() {parameterName=ColorLerp, defaultValue=<rgba 0.3438 0.3438 0.3438 0>}.RGB
   B=TextureSample(UVs=@n1.Result) {samplerType=SAMPLERTYPE_LinearGrayscale, texture=/Game/Environments/Sets/Arid/Textures/RockDetail_basecolor}.RGB
   Alpha=ScalarParameter() {defaultValue=0.1, parameterName=Lerp}
  )
  Multiply(
   A=TextureSample(UVs=@n1.Result) {samplerType=SAMPLERTYPE_Normal, texture=/Game/Environments/Sets/Arid/Textures/RockDetail_normal}.RGB
   B=VectorParameter() {parameterName=NormalStrength, defaultValue=<rgba 0.3 0.3 1 0>}.RGB
  )
  ScalarParameter() {defaultValue=0.1, parameterName=Metallic}
  ScalarParameter() {defaultValue=0.8, parameterName=Specular}
  ScalarParameter() {defaultValue=0.8, parameterName=Roughness}
  ScalarParameter() {defaultValue=1, parameterName=Opacity}
 ]
}
