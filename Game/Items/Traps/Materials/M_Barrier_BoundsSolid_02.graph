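# Masked, dithered Default Lit material for the solid bounds of a trap barrier, textured as metal panels.
# A world-aligned UV function, driven by a WorldScale parameter, supplies shared coordinates to the diffuse, normal and mask texture samples.
# A scalar Opacity parameter sits alongside, with dithered opacity masking enabled on the material.
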Material /Game/Items/Traps/Materials/M_Barrier_BoundsSolid_02
graph {
 properties {blendMode=BLEND_Masked, ditherOpacityMask=true, opacityMaskClipValue=0.5, pixelDepthOffsetMode=PDOM_Legacy, usedShadingModels=Default Lit}
 nodes [
  ScalarParameter() {defaultValue=1, parameterName=Opacity}
  n1 = MaterialFunctionCall!(WorldScale=VectorParameter() {parameterName=WorldScale, defaultValue=<rgba 512 512 512 0>}.RGB, TransformToLocalSpace) {materialFunction=&/MF_Creative_WorldAlignedUVs}
  TextureSample(UVs=@n1.Result) {samplerType=SAMPLERTYPE_Normal, texture=/Game/Packages/Fortress_Floors_Generic/Textures/T_FORT_MetalPanels02_N}
  TextureSample(UVs=@n1.Result) {texture=/Game/Packages/Fortress_Floors_Generic/Textures/T_FORT_MetalPanels02_D}
  TextureSample(UVs=@n1.Result) {samplerType=SAMPLERTYPE_Masks, texture=/Game/Packages/Fortress_Floors_Generic/Textures/T_FORT_MetalPanels02_S}
 ]
}
