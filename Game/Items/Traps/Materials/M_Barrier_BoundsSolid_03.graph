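# M_Barrier_BoundsSolid_03 is a masked, lit brick wall material, apparently for the solid bounds of a trap barrier item.
# The MF_Creative_WorldAlignedUVs function supplies shared UVs to brick color, normal and mask texture samples from Fortress_Walls_Generic.
# An Opacity scalar parameter defaults to fully opaque, and the material enables dithered opacity masking.
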Material /Game/Items/Traps/Materials/M_Barrier_BoundsSolid_03
graph {
 properties {blendMode=BLEND_Masked, ditherOpacityMask=true, opacityMaskClipValue=0.5, pixelDepthOffsetMode=PDOM_Legacy, usedShadingModels=Default Lit}
 nodes [
  ScalarParameter() {defaultValue=1, parameterName=Opacity}
  n1 = MaterialFunctionCall!(WorldScale, TransformToLocalSpace) {materialFunction=&/MF_Creative_WorldAlignedUVs}
  TextureSample(UVs=@n1.Result) {samplerType=SAMPLERTYPE_Normal, texture=/Game/Packages/Fortress_Walls_Generic/Textures/T_BrickWall_005a_N}
  TextureSample(UVs=@n1.Result) {texture=/Game/Packages/Fortress_Walls_Generic/Textures/T_BrickWall_005}
  TextureSample(UVs=@n1.Result) {samplerType=SAMPLERTYPE_Masks, texture=/Game/Packages/Fortress_Walls_Generic/Textures/T_BrickWall_006_New_S}
 ]
}
