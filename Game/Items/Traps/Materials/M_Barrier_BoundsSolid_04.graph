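# M_Barrier_BoundsSolid_04 is an unlit masked material for trap items, named as a solid bounds visual for a barrier.
# Its graph holds only a constant and a scalar Opacity parameter that defaults to fully opaque, with no wiring between them or to material outputs shown.
# Dithered opacity masking is enabled, which lets a masked material approximate partial transparency.
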Material /Game/Items/Traps/Materials/M_Barrier_BoundsSolid_04
graph {
 properties {
  blendMode=BLEND_Masked
  ditherOpacityMask=true
  opacityMaskClipValue=0.5
  pixelDepthOffsetMode=PDOM_Legacy
  shadingModel=MSM_Unlit
  usedShadingModels=Unlit
 }
 nodes [
  Constant()
  ScalarParameter() {defaultValue=1, parameterName=Opacity}
 ]
}
