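# Additive, unlit, two-sided material for the translucent bounds of the barrier trap, drawing animated stripes, dots and noise.
# World-aligned UVs drive rotated stripes, a row-staggered dot grid that fades with camera distance, and two panning noise samples.
# These plus fresnel blend background and circle colors, masked by a V-coordinate band that also scales dot size.
# Backfaces are dimmed; only High quality adds near-camera and depth fades.
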Material /Game/Items/Traps/Materials/M_Barrier_BoundsTranslucent_01
graph {
 properties {
  bEnableResponsiveAA=true
  blendMode=BLEND_Additive
  pixelDepthOffsetMode=PDOM_Legacy
  shadingModel=MSM_Unlit
  twoSided=true
  usedShadingModels=Unlit
 }
 nodes [
  n1 = MaterialFunctionCall!(WorldScale=VectorParameter() {parameterName=WorldScale, defaultValue=<rgba 20 20 20 0>}.RGB, TransformToLocalSpace) {materialFunction=&/MF_Creative_WorldAlignedUVs}
  n2 = Frac(None=Rotator(Coordinate=@n1.Result, Time=ScalarParameter() {defaultValue=9.3, parameterName=Time}))
  n3 = Divide(A=@n1.Result, B=ScalarParameter() {defaultValue=0.2, parameterName=DotsRepeat})
  n4 = Add(A=Constant2Vector() {r=0.5}, B=@n3)
  n5 = ComponentMask(None=Divide(A=WorldPosition().XYZ, B=ScalarParameter() {defaultValue=19.48, parameterName=WorldScaleDots})) {b=true}
  n6 = Multiply(A=ScalarParameter() {defaultValue=0.2, parameterName=DotsRepeat}) {constB=2}
  n7 = Round(None=Frac(None=Divide(A=@n5, B=@n6) {constB=16}))
  n8 = MaterialFunctionCall(Float2=TextureCoordinate()) {materialFunction=&BreakOutFloat2Components}
  n9 = SphereMask(A=@n8.G, B=Constant() {r=0.5}, Radius=ScalarParameter() {defaultValue=0.5, parameterName=BGRadius}, Hardness=Constant())
  n10 = SphereMask(
   A=Frac(None=LinearInterpolate(A=@n4, B=@n3, Alpha=@n7))
   B=Constant() {r=0.5}
   Radius=LinearInterpolate(A=ScalarParameter() {defaultValue=0.1, parameterName=Radius}, B=ScalarParameter() {defaultValue=0.25, parameterName=RadiusLarge}, Alpha=@n9)
   Hardness=ScalarParameter() {defaultValue=0.5, parameterName=DotsHardness}
  )
  n11 = Add(A=Distance(A=CameraPositionWS(), B=WorldPosition().XYZ), B=ScalarParameter() {defaultValue=-100, parameterName=DistanceBias})
  n12 = ScalarParameter() {defaultValue=800, parameterName=FadeScale}
  n13 = OneMinus(None=Saturate(None=Divide(A=@n11, B=@n12)))
  n14 = VectorParameter() {parameterName=PanSpeedNoiseA, defaultValue=<rgba 0.3 0.6 11 0>}
  n15 = TextureSampleParameter2D(UVs=Divide(A=Panner(Coordinate=@n1.Result, Speed=@n14.RGB) {speedX=5}, B=ScalarParameter() {defaultValue=30, parameterName=RepeatNoiseA})) {parameterName=Noise, texture=/Game/Global/Textures/Default/Noise/T_Dudebro_Noise_01}
  n16 = VectorParameter() {parameterName=PanSpeedNoiseB, defaultValue=<rgba -0.44 -1.2808 0 0>}
  n17 = TextureSampleParameter2D(UVs=Divide(A=Panner(Coordinate=@n1.Result, Speed=@n16.RGB) {speedX=5}, B=ScalarParameter() {defaultValue=-20, parameterName=RepeatNoiseB})) {parameterName=Noise, texture=&T_Dudebro_Noise_01}
  n18 = SphereMask(
   A=Multiply(A=Multiply(A=@n15.R, B=@n17.R)) {constB=2.5}
   B=@n8.G
   Radius=ScalarParameter() {defaultValue=0.5, parameterName=NoiseRadius}
   Hardness=ScalarParameter() {parameterName=NoiseHardness}
  )
  n19 = ScalarParameter() {defaultValue=0.7, parameterName=NoiseScale}
  n20 = Add(
   A=SphereMask(
    A=MaterialFunctionCall(Float2=@n2) {materialFunction=&BreakOutFloat2Components}.R
    B=Constant() {r=0.5}
    Radius=ScalarParameter() {defaultValue=0.2, parameterName=StripeRadius}
    Hardness=ScalarParameter() {defaultValue=0.9, parameterName=StripeHarndess}
   )
   B=Add(A=Multiply(A=@n10, B=@n13), B=Multiply(A=@n18, B=@n19))
  )
  n21 = MaterialFunctionCall(
   Brightness=ScalarParameter() {defaultValue=0.5, parameterName=Fresnel Brightness}
   Exponent=ScalarParameter() {defaultValue=4, parameterName=Fresnel Exponent}
  ) {materialFunction=&MF_Fresnel_FX}
  n22 = Multiply(
   A=LinearInterpolate(
    A=VectorParameter() {parameterName=BGColor, defaultValue=<rgba 0.1198 0.0609 0 0>}.RGB
    B=VectorParameter() {parameterName=Circles, defaultValue=<rgba 1 0.2688 0.099 0>}.RGB
    Alpha=Add(A=@n20, B=@n21.Fresnel Inverted)
   )
   B=@n9
  )
  n23 = LinearInterpolate(A=Multiply(A=ScalarParameter() {defaultValue=0.2, parameterName=BacksideOpacity}, B=@n22), B=@n22, Alpha=Saturate(None=TwoSidedSign()))
  n24 = Multiply(
   A=@n23
   B=MaterialFunctionCall(Distance=ScalarParameter() {defaultValue=200, parameterName="Fade_Start_Distance (S)"}) {materialFunction=&Near_Camera_Fade}.Result
  )
  n25 = ScalarParameter() {defaultValue=15, parameterName=FadeDistance}
  Multiply(
   A=QualitySwitch(Default=@n23, Low=@n23, High=DepthFade(Opacity=@n24, FadeDistance=@n25), Medium=@n23)
   B=ScalarParameter() {defaultValue=1, parameterName=Color Boost}
  )
  ScalarParameter() {defaultValue=1, parameterName=Opacity}
 ]
}
