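# Additive, unlit, two-sided material for a trap barrier's bounds, showing colored stripes that fade with camera distance.
# Stripes come from a sphere mask on rotated, repeating world-aligned UVs plus an inverted Fresnel term, with back faces dimmed.
# Fade uses camera-to-pivot distance minus a view-direction-blended extent from object scale, assuming a base pivot.
# Fade runs through a vertex interpolator, and a world-offset branch collapses fully faded vertices onto the pivot.
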Material /Game/Items/Traps/Materials/M_Barrier_BoundsTranslucent_02
graph {
 properties {
  bEnableResponsiveAA=true
  blendMode=BLEND_Additive
  pixelDepthOffsetMode=PDOM_Legacy
  shadingModel=MSM_Unlit
  twoSided=true
  usedShadingModels=Unlit
 }
 nodes [
  n1 = Subtract(
   A=CameraPositionWS()
   B=MaterialFunctionCall() {materialFunction=&ObjectPivotPoint}.Object Pivot Location
  )
  n2 = MaterialFunctionCall(Float3=Divide(
    A=MaterialFunctionCall() {materialFunction=/Engine/Functions/Engine_MaterialFunctions02/WorldPositionOffset/ObjectScale}.Scale XYZ
    B=Constant3Vector() {desc=Since pivot is at the base on Z, constant=<rgba 2 2 1 0>}
   )) {materialFunction=&BreakOutFloat3Components}
  n3 = Normalize(VectorInput=@n1) {desc=vector to camera}
  n4 = Multiply(A=@n3, B=Constant3Vector() {constant=<rgba 1 1 0 0>}) {desc=XY Only}
  n5 = Normalize(VectorInput=Transform(None=Constant3Vector() {constant=<rgba 1 0 0 0>}) {transformSourceType=TRANSFORMSOURCE_Local})
  n6 = LinearInterpolate(A=@n2.G, B=@n2.G, Alpha=Abs(None=DotProduct(A=@n4, B=@n5) {desc=dot against forward vector})) {desc=" 1=X, 0=Y"}
  n7 = Constant3Vector() {constant=<rgba 0 0 1 0>}
  n8 = Transform(None=Constant3Vector() {constant=<rgba 0 0 1 0>}) {transformSourceType=TRANSFORMSOURCE_Local}
  n9 = Abs(None=DotProduct(A=Multiply(A=@n3, B=@n7) {desc=Z Only}, B=Normalize(VectorInput=@n8)) {desc=dot against up vector})
  n10 = MaterialFunctionCall!(
   min=ScalarParameter() {defaultValue=1280, parameterName=FadeEndDistance}
   max=ScalarParameter() {defaultValue=768, parameterName=FadeStartDistance}
   value=Subtract(
    A=MaterialFunctionCall(Vector 3=@n1) {materialFunction=/Engine/Functions/Engine_MaterialFunctions02/Utility/VectorLength}.V3 Length
    B=LinearInterpolate(A=@n6, B=@n2.B, Alpha=@n9) {desc="1 = Z, 0 = XY"}
   ) {desc=distance from pivot to camera - bounds size}
  ) {materialFunction=/Game/Effects/Fort_Effects/Materials/_Functions/InverseLerp_MF}
  n11 = Subtract(
   A=MaterialFunctionCall() {materialFunction=&ObjectPivotPoint}.Object Pivot Location
   B=WorldPosition().XYZ
  )
  If(A=@n10.Clamped, B=Constant() {r=0.001}, "A > B"=Constant(), "A == B"=@n11, "A < B"=@n11)
  n12 = ScalarParameter() {defaultValue=0.2, parameterName=BacksideOpacity}
  n13 = MaterialFunctionCall!(WorldScale=VectorParameter() {parameterName=WorldScale, defaultValue=<rgba 20 20 20 0>}.RGB, TransformToLocalSpace) {materialFunction=&/MF_Creative_WorldAlignedUVs}
  n14 = Constant() {r=9.3}
  n15 = MaterialFunctionCall(Float2=Frac(None=Rotator(Coordinate=@n13.Result, Time=@n14))) {materialFunction=&BreakOutFloat2Components}
  n16 = Constant() {r=0.5}
  n17 = ScalarParameter() {defaultValue=0.2, parameterName=StripeRadius}
  n18 = ScalarParameter() {defaultValue=0.9, parameterName=StripeHarndess}
  n19 = ScalarParameter() {defaultValue=0.4, parameterName=Fresnel Brightness}
  n20 = ScalarParameter() {defaultValue=4, parameterName=Fresnel Exponent}
  n21 = LinearInterpolate(
   A=VectorParameter() {parameterName=BGColor, defaultValue=<rgba 0.1198 0.01 0 0>}.RGB
   B=VectorParameter() {parameterName=Circles, defaultValue=<rgba 1 0.3242 0.099 0>}.RGB
   Alpha=Add(
    A=SphereMask(A=@n15.R, B=@n16, Radius=@n17, Hardness=@n18)
    B=MaterialFunctionCall(Brightness=@n19, Exponent=@n20) {materialFunction=&MF_Fresnel_FX}.Fresnel Inverted
   )
  )
  n22 = TwoSidedSign()
  Multiply(
   A=LinearInterpolate(A=Multiply(A=@n12, B=@n21), B=@n21, Alpha=Saturate(None=@n22))
   B=ComponentMask(None=VertexInterpolator(VS=@n10.Clamped).PS) {r=true}
  )
  ScalarParameter() {defaultValue=1, parameterName=Opacity}
 ]
}
