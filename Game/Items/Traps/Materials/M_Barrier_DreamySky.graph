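# Unlit material for a trap barrier, named as a dreamy sky, that colors the surface with a tinted sky HDRI.
# The camera vector, with its Z axis flipped, serves as the lookup coordinate into the HDRI texture, so the image follows view direction rather than mesh UVs.
# The sampled color is multiplied by a Tint vector parameter, pale pink by default.
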Material /Game/Items/Traps/Materials/M_Barrier_DreamySky
graph {
 properties {pixelDepthOffsetMode=PDOM_Legacy, shadingModel=MSM_Unlit, usedShadingModels=Unlit}
 nodes [
  n1 = CameraVectorWS()
  n2 = Constant3Vector() {constant=<rgba 1 1 -1 0>}
  Multiply(
   A=TextureSample(UVs=Multiply(A=@n1, B=@n2)) {samplerType=SAMPLERTYPE_LinearColor, texture=/Game/Items/Traps/Textures/T_DreamySky_Hdri}.RGB
   B=VectorParameter() {parameterName=Tint, defaultValue=<rgba 1 0.7394 0.7493 0>}.RGB
  )
 ]
}
