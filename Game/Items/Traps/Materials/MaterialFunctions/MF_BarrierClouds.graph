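# Material function that renders animated sky-like clouds for barrier materials under Traps, projected from the camera vector rather than mesh UVs.
# Far and near samples of the cloud texture scroll and blend by view elevation, with a swirl flowmap twisting the near layer.
# A distant-cloud boost fills toward the horizon before a fade, optionally mirrored below it.
# Outputs tinted color with optional normal-based sun lighting, a cloud mask and a debug value.
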MaterialFunction /Game/Items/Traps/Materials/MaterialFunctions/MF_BarrierClouds
graph {nodes [
  n1 = FunctionInput(Preview=Constant3Vector() {constant=<rgba 1 1 1 0>}) {bUsePreviewValueAsDefault=true, inputName=Light Source Color}
  n2 = Multiply(A=FunctionInput() {inputName=Light Source Direction}) {constB=-1}
  n3 = FunctionInput(Preview=TextureObject() {samplerType=SAMPLERTYPE_Normal, texture=&FlatNormalBlue}) {bUsePreviewValueAsDefault=true, inputName=Cloud Normals, inputType=FunctionInput_Texture2D}
  n4 = ScalarParameter() {defaultValue=0.2, group=Clouds, parameterName=Cloud Tiling Far, sortPriority=100}
  n5 = CameraVectorWS()
  n6 = ComponentMask(None=@n5) {b=true}
  n7 = Divide(
   A=ComponentMask(None=@n5) {g=true, r=true}
   B=Power(Base=Abs(None=@n6), Exp=ScalarParameter() {defaultValue=0.8, group=Clouds, parameterName=Cloud UV Exponent, sortPriority=104})
  )
  n8 = ScalarParameter() {group=Clouds, parameterName=Cloud Speed Direction, sliderMax=1, sortPriority=0}
  n9 = ScalarParameter() {defaultValue=1, group=Clouds, parameterName=Cloud Speed}
  n10 = Multiply(A=AppendVector(A=Cosine(None=@n8), B=Sine(None=@n8)), B=Multiply(A=Divide(A=@n9) {constB=100}, B=Time()))
  n11 = Multiply(A=@n4, B=@n10)
  n12 = FunctionInput(Preview=Constant()) {bUsePreviewValueAsDefault=true, inputName=UV Distortion, inputType=FunctionInput_Vector2, sortPriority=90}
  n13 = Add(A=Add(A=Multiply(A=@n4, B=@n7), B=Multiply(A=@n11) {constB=0.6}), B=@n12)
  n14 = Multiply(A=@n6) {constB=-1}
  n15 = StaticSwitch(
   True=Abs(None=@n14)
   False=Saturate(None=@n14)
   Value=FunctionInput(Preview=StaticBool()) {bUsePreviewValueAsDefault=true, inputName=Mirrored on Horizon, inputType=FunctionInput_StaticBool, sortPriority=91}
  )
  n16 = MaterialFunctionCall(
   In=Power(Base=@n15, Exp=ScalarParameter() {defaultValue=1, group=Clouds, parameterName=Cloud Interface Exp, sortPriority=102}) {constExponent=3}
   Contrast=ScalarParameter() {defaultValue=1, group=Clouds, parameterName=Cloud Interface Contrast, sortPriority=103}
  ) {materialFunction=&CheapContrast}
  n17 = ScalarParameter() {defaultValue=0.7, group=Clouds, parameterName=Cloud Tiling Near, sortPriority=101}
  n18 = Multiply(A=@n7, B=@n17)
  n19 = Multiply(A=@n10, B=@n17)
  n20 = Multiply(A=@n7, B=ScalarParameter() {defaultValue=0.66, group=Clouds, parameterName=Cloud Twist Tiling, sortPriority=201})
  n21 = Time()
  n22 = ScalarParameter() {defaultValue=0.01, group=Clouds, parameterName=Cloud Twist speed, sortPriority=202}
  n23 = TextureSampleParameter2D(UVs=Add(A=Add(A=@n20, B=@n10), B=Multiply(A=@n21, B=@n22))) {
   group=Clouds
   parameterName=Swirl Texture
   samplerType=SAMPLERTYPE_LinearColor
   texture=/Game/VisualThreatManager/StormVisuals/LargeClouds/T_Swirl_Flowmap
  }
  n24 = AppendVector(A=@n23.R, B=@n23.R)
  n25 = Add(
   A=Add(A=Add(A=@n18, B=@n19), B=@n12)
   B=Multiply(A=Subtract(A=@n24) {constB=0.5}, B=ScalarParameter() {defaultValue=0.3, group=Clouds, parameterName=Cloud Twist Strength, sortPriority=200})
  )
  n26 = MaterialFunctionCall(Float3=LinearInterpolate(
    A=TextureSample(UVs=@n13, Tex=@n3) {samplerType=SAMPLERTYPE_Normal}.RGB
    B=TextureSample(UVs=@n25, Tex=@n3) {samplerType=SAMPLERTYPE_Normal}.RGB
    Alpha=@n16.Result
   )) {materialFunction=&BreakOutFloat3Components}
  n27 = ComponentMask(None=CameraVectorWS()) {b=true}
  n28 = OneMinus(None=Saturate(None=Multiply(A=@n27) {constB=-1}))
  n29 = Multiply(A=@n28, B=@n28)
  n30 = MaterialFunctionCall(R=@n26.R, G=@n26.R, B=LinearInterpolate(A=Multiply(A=@n29) {constA=-2, constB=-2}, Alpha=@n26.B)) {materialFunction=&AppendMany}
  n31 = Saturate(None=DotProduct(A=Normalize(VectorInput=@n2), B=Normalize(VectorInput=@n30.RGB)))
  n32 = ScalarParameter() {defaultValue=2, group=Clouds, parameterName=Sun Effect, sortPriority=303}
  n33 = CameraVectorWS()
  n34 = OneMinus(None=Abs(None=ComponentMask(None=@n33) {b=true}))
  n35 = ScalarParameter() {defaultValue=10, group=Clouds, parameterName=Distant Clouds Exp, sortPriority=301}
  n36 = Multiply(
   A=MaterialFunctionCall!(Input=Power(Base=@n34, Exp=@n35)) {materialFunction=&/MF_BarrierSmoothStep}.Output
   B=ScalarParameter() {defaultValue=5.6, group=Clouds, parameterName=Distant Clouds Intensity, sortPriority=302}
  )
  n37 = OneMinus(None=Saturate(None=@n36))
  n38 = StaticSwitch(
   True=Multiply(A=Multiply(A=@n1, B=@n31), B=Multiply(A=@n32, B=@n37))
   False=Constant()
   Value=FunctionInput(Preview=StaticBool()) {bUsePreviewValueAsDefault=true, inputName=Normals on clouds, inputType=FunctionInput_StaticBool}
  )
  n39 = FunctionInput(Preview=TextureObject() {texture=/Game/Creative/BuildingActors/Props/Textures/CP_White_Texutre}) {bUsePreviewValueAsDefault=true, inputName=Cloud Texture, inputType=FunctionInput_Texture2D}
  n40 = LinearInterpolate(
   A=ScalarParameter() {defaultValue=-0.5, group=Clouds, parameterName=Clouds Low, sortPriority=104}
   B=ScalarParameter() {defaultValue=1.5, group=Clouds, parameterName=Clouds High, sortPriority=105}
   Alpha=LinearInterpolate(A=TextureSample(UVs=@n13, Tex=@n39).R, B=TextureSample(UVs=@n25, Tex=@n39).R, Alpha=@n16.Result)
  )
  n41 = Saturate(None=Add(A=@n36, B=Saturate(None=@n40)))
  n42 = OneMinus(None=@n15)
  n43 = ScalarParameter() {defaultValue=6, group=Clouds, parameterName=Cloud Horizon Fade, sortPriority=300}
  n44 = Multiply(A=Multiply(A=@n41, B=Ceil(None=@n15)), B=OneMinus(None=Power(Base=@n42, Exp=@n43)))
  n45 = FunctionInput(Preview=Constant3Vector() {constant=<rgba 0.2918 0.2747 0.2423 1>}) {bUsePreviewValueAsDefault=true, inputName=Cloud Color}
  FunctionOutput(None=Add(A=Multiply(A=@n38, B=@n44), B=Multiply(A=@n45, B=@n44)))
  FunctionOutput(None=Power(Base=@n44, Exp=ScalarParameter() {defaultValue=2, group=Clouds, parameterName=Cloud Opacity Mask Exp, sortPriority=106})) {outputName=Cloud Mask, sortPriority=1}
  FunctionOutput(None=@n41) {outputName=DEBUG, sortPriority=1}
 ]}
