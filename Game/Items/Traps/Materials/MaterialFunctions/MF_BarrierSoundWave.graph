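# Material function for trap barriers that outputs sound-wave signals: Fake Signal, Music Amplitude, Music Signal and a thresholded debug view.
# Fake Signal layers panned Perlin noise offset by vertex color, using smooth-min at higher quality and a simpler blend on lower tiers.
# Music Signal shapes four bands into sine lobes across U, weighted and blended between sum and max.
# Static switches select Jukebox or audio-reactive parameter collections, or synthetic debug signals.
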MaterialFunction /Game/Items/Traps/Materials/MaterialFunctions/MF_BarrierSoundWave
graph {nodes [
  n1 = AppendVector(
   A=FunctionInput(Preview=Constant() {r=1}) {bUsePreviewValueAsDefault=true, inputName=Fake Signal Tiling, inputType=FunctionInput_Scalar, sortPriority=1}
   B=Constant()
  )
  n2 = NamedRerouteUsage()
  n3 = Multiply(A=@n2, B=@n1)
  n4 = Multiply(
   A=Time()
   B=FunctionInput(Preview=Constant() {r=2.5}) {bUsePreviewValueAsDefault=true, inputName=Fake Signal Speed, inputType=FunctionInput_Scalar, sortPriority=5}
  )
  n5 = VertexColor()
  n6 = TextureCoordinate()
  n7 = DDX(Value=@n6)
  n8 = DDY(Value=@n6)
  n9 = TextureSample!(UVs=Add(A=Panner(Coordinate=@n3, Time=@n4) {speedX=0.05, speedY=0.12}, B=@n5.G), Tex, "DDX(UVs)"=@n7, "DDY(UVs)"=@n8, Apply View MipBias) {mipValueMode=TMVM_Derivative, samplerSource=SSM_Wrap_WorldGroupSettings, texture=/Game/Items/Traps/Textures/T_PerlinRGB_01_M}
  n10 = Constant() {r=3}
  n11 = Multiply(
   A=@n2
   B=Multiply(
    A=@n1
    B=FunctionInput(Preview=@n10) {bUsePreviewValueAsDefault=true, inputName=Fake Signal Tiling Multiplier 1, inputType=FunctionInput_Scalar, sortPriority=2}
   ) {constB=3}
  )
  n12 = TextureSample(UVs=Add(A=Panner(Coordinate=@n11, Time=@n4) {speedX=0.05, speedY=0.12}, B=@n5.G), "DDX(UVs)"=@n7, "DDY(UVs)"=@n8) {mipValueMode=TMVM_Derivative, samplerSource=SSM_Wrap_WorldGroupSettings, texture=&T_PerlinRGB_01_M}
  n13 = Multiply(A=@n12.G) {constB=0.6}
  n14 = Constant() {r=0.5}
  n15 = Multiply(
   A=@n1
   B=FunctionInput(Preview=Constant() {r=6}) {bUsePreviewValueAsDefault=true, inputName=Fake Signal Tiling Multiplier 2, inputType=FunctionInput_Scalar, sortPriority=3}
  ) {constB=6}
  n16 = Add(A=Panner(Coordinate=Multiply(A=@n2, B=@n15), Time=@n4) {speedX=0.1, speedY=0.2}, B=@n5.G)
  n17 = Multiply(
   A=Add(A=MaterialFunctionCall!(A=@n9.G, B=@n13, K=@n14) {materialFunction=&/MF_BarrierSmoothMin}.Result) {constB=0.5}
   B=LinearInterpolate(B=TextureSample(UVs=@n16, "DDX(UVs)"=@n7, "DDY(UVs)"=@n8) {mipValueMode=TMVM_Derivative, samplerSource=SSM_Wrap_WorldGroupSettings, texture=&T_PerlinRGB_01_M}.B) {constA=1, constAlpha=0.3}
  )
  n18 = LinearInterpolate(Alpha=@n12.R) {constA=0.3}
  n19 = ShadingPathSwitch(Default=@n17, Deferred=@n17, Mobile=@n18) {desc=Migrated from FeatureLevelSwitch}
  FunctionOutput(None=QualitySwitch(Default=@n19, Low=@n18, High=@n19, Medium=@n18, Epic=@n19)) {outputName=Fake Signal, sortPriority=13}
  n20 = Time()
  n21 = Power(Base=OneMinus(None=Frac(None=@n20))) {constExponent=5}
  n22 = Multiply(A=ConstantBiasScale(None=Sine(None=@n21))) {constB=0.8}
  n23 = ComponentMask(None=CollectionParameter() {collection=/Game/Creative/Devices/Jukebox/Creative_Jukebox_MPC, parameterName=Percussion}) {r=true}
  n24 = CollectionParameter() {collection=/Game/Athena/Events/AudioAnalysis/MPC_Audio_Reactive_Parameters, parameterName=Amplitude_Music}
  FunctionOutput(None=StaticSwitchParameter(True=Add(A=@n22) {constB=0.2}, False=StaticSwitchParameter(True=@n23, False=@n24) {defaultValue=true, parameterName=Use Jukebox MPC}) {parameterName=DEBUG SOUND SIGNAL}) {outputName=Music Amplitude, sortPriority=12}
  n25 = ComponentMask(None=NamedRerouteUsage()) {r=true}
  n26 = OneMinus(None=Power(Base=OneMinus(None=@n25)) {constExponent=7})
  n27 = ScalarParameter() {defaultValue=0.03, group=Barrier Sound Wave MF, parameterName=Offset Ends}
  n28 = Sine(None=Add(A=Multiply(A=@n26) {constB=0.5}, B=@n27) {constB=0})
  n29 = Constant()
  n30 = Panner(Coordinate=@n29) {speedX=-1, speedY=-0.8}
  n31 = Add(A=@n29) {constB=0.17}
  n32 = Panner(Coordinate=@n31) {speedX=-0.6, speedY=0.2}
  n33 = Add(A=@n31) {constB=0.35}
  n34 = Panner(Coordinate=@n33) {speedX=0.2, speedY=0.3}
  n35 = Panner(Coordinate=Add(A=@n33) {constB=0.58}) {speedX=0.7, speedY=-0.5}
  n36 = VectorParameter() {group=Barrier Sound Wave MF, parameterName=Debug Sound Spectrum Strengths, defaultValue=<rgba 2 0.5 0.13 0.05>}
  n37 = Multiply(
   A=MaterialFunctionCall(
    X=TextureSample!(UVs=@n30, Tex, Apply View MipBias) {samplerSource=SSM_Wrap_WorldGroupSettings, texture=&T_PerlinRGB_01_M}.R
    Y=TextureSample(UVs=@n32) {samplerSource=SSM_Wrap_WorldGroupSettings, texture=&T_PerlinRGB_01_M}.R
    Z=TextureSample(UVs=@n34) {samplerSource=SSM_Wrap_WorldGroupSettings, texture=&T_PerlinRGB_01_M}.R
    A=TextureSample(UVs=@n35) {samplerSource=SSM_Wrap_WorldGroupSettings, texture=&T_PerlinRGB_01_M}.R
   ) {materialFunction=&MakeFloat4}.Result
   B=AppendVector(A=@n36.RGB, B=@n36.RGB)
  )
  n38 = Power(Base=StaticSwitchParameter(True=@n37, False=CollectionParameter() {collection=&Creative_Jukebox_MPC, parameterName=Percussion}) {parameterName=DEBUG SOUND SIGNAL}) {constExponent=0.5}
  n39 = MaterialFunctionCall(Float4=FunctionInput(Preview=Constant4Vector() {constant=<rgba 1 2 5 10>}.RGBA) {bUsePreviewValueAsDefault=true, inputName="Spectrum Strength (V4)", inputType=FunctionInput_Vector4}) {materialFunction=&BreakOutFloat4Components}
  n40 = Multiply(A=Multiply(A=Saturate(None=@n28), B=ComponentMask(None=@n38) {r=true}), B=@n39.R)
  n41 = OneMinus(None=Power(Base=OneMinus(None=@n25)) {constExponent=1.5})
  n42 = Power(Base=Sine(None=Multiply(A=@n41) {constB=0.5})) {constExponent=6}
  n43 = Multiply(A=Multiply(A=Saturate(None=@n42), B=ComponentMask(None=@n38) {g=true}), B=@n39.G)
  n44 = Power(Base=@n25) {constExponent=1.5}
  n45 = Power(Base=Sine(None=Multiply(A=@n44) {constB=0.5})) {constExponent=6}
  n46 = Multiply(A=Multiply(A=Saturate(None=@n45), B=ComponentMask(None=@n38) {b=true}), B=@n39.B)
  n47 = Add(A=Add(A=@n40, B=@n43), B=@n46)
  n48 = Power(Base=@n25) {constExponent=7}
  n49 = Sine(None=Add(A=Multiply(A=@n48) {constB=0.5}, B=OneMinus(None=@n27)) {constB=0})
  n50 = Multiply(A=Multiply(A=Saturate(None=@n49), B=ComponentMask(None=@n38) {a=true}), B=@n39.A)
  n51 = Max(A=@n40, B=@n43)
  n52 = LinearInterpolate(A=Multiply(A=Add(A=@n47, B=@n50)) {constB=0.75}, B=Max(A=Max(A=@n51, B=@n46), B=@n50)) {constAlpha=0.6}
  n53 = ComponentMask(None=NamedRerouteUsage()) {r=true}
  n54 = Power(Base=OneMinus(None=@n53)) {constExponent=7}
  n55 = Sine(None=Multiply(A=OneMinus(None=@n54)) {constB=0.5})
  n56 = Power(Base=StaticSwitchParameter(
    True=@n37
    False=CollectionParameter() {collection=&MPC_Audio_Reactive_Parameters, parameterName=Spectrum_Music}
   ) {group=Barrier Sound Wave MF, parameterName=DEBUG SOUND SIGNAL}) {constExponent=0.5}
  n57 = Multiply(A=Multiply(A=Saturate(None=@n55), B=ComponentMask(None=@n56) {r=true}), B=@n39.R)
  n58 = OneMinus(None=Power(Base=OneMinus(None=@n53)) {constExponent=1.5})
  n59 = Power(Base=Sine(None=Multiply(A=@n58) {constB=0.5})) {constExponent=6}
  n60 = Multiply(A=Multiply(A=Saturate(None=@n59), B=ComponentMask(None=@n56) {g=true}), B=@n39.G)
  n61 = Power(Base=@n53) {constExponent=1.5}
  n62 = Power(Base=Sine(None=Multiply(A=@n61) {constB=0.5})) {constExponent=6}
  n63 = Multiply(A=Multiply(A=Saturate(None=@n62), B=ComponentMask(None=@n56) {b=true}), B=@n39.B)
  n64 = Add(A=Add(A=@n57, B=@n60), B=@n63)
  n65 = Sine(None=Multiply(A=Power(Base=@n53) {constExponent=7}) {constB=0.5})
  n66 = Multiply(A=Multiply(A=Saturate(None=@n65), B=ComponentMask(None=@n56) {a=true}), B=@n39.A)
  n67 = Max(A=@n57, B=@n60)
  n68 = LinearInterpolate(A=Multiply(A=Add(A=@n64, B=@n66)) {constB=0.75}, B=Max(A=Max(A=@n67, B=@n63), B=@n66)) {constAlpha=0.6}
  n69 = NamedRerouteUsage()
  n70 = Subtract(
   A=FunctionOutput(None=StaticSwitchParameter(True=@n52, False=@n68) {defaultValue=true, group=Barrier Sound Wave MF, parameterName=Use Jukebox MPC}) {outputName=Music Signal, sortPriority=10}
   B=ComponentMask(None=OneMinus(None=@n69)) {g=true}
  )
  FunctionOutput(None=Ceil(None=Saturate(None=@n70))) {outputName=DEBUG, sortPriority=99}
  NamedRerouteDeclaration(None=FunctionInput(Preview=TextureCoordinate()) {bUsePreviewValueAsDefault=true, inputName=UVs, inputType=FunctionInput_Vector2}) {name=UV INPUT, nodeColor=<rgba 1 0.1464 0 1>}
 ]}
