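# Camera triplanar material function, filed under trap materials for barriers, that maps a texture by view direction rather than surface position.
# It projects the world-space camera vector onto planes perpendicular to Z, X and Y to form three UV sets, each adjusted by UV Scale and UV Distortion.
# Two nested lerps blend the RGB samples, weighted by the camera vector's absolute X and Y components through CheapContrast with a shared contrast input.
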MaterialFunction /Game/Items/Traps/Materials/MaterialFunctions/MF_BarrierTriplanar
graph {
 properties {bExposeToLibrary=true, description=Camera Triplanar, libraryCategoriesText []}
 nodes [
  n1 = Constant3Vector() {constant=<rgba 0 0 1 0>}
  n2 = Add(A=CameraVectorWS(), B=@n1)
  n3 = CrossProduct(A=@n1, B=Constant3Vector() {constant=<rgba 1 0 0 0>})
  n4 = CrossProduct(A=@n3, B=@n1)
  n5 = AppendVector(A=DotProduct(A=@n2, B=Normalize(VectorInput=@n3)), B=DotProduct(A=@n2, B=Normalize(VectorInput=@n4)))
  n6 = FunctionInput(Preview=Constant() {r=1}) {bUsePreviewValueAsDefault=true, inputName=UV Scale, inputType=FunctionInput_Vector2, previewValue=<1 1 0 1>}
  n7 = FunctionInput(Preview=Constant()) {bUsePreviewValueAsDefault=true, inputName=UV Distortion, inputType=FunctionInput_Vector2}
  n8 = FunctionInput(Preview=TextureObject() {texture=null}) {inputType=FunctionInput_Texture2D}
  n9 = TextureSample(UVs=Add(A=Multiply(A=@n5, B=@n6), B=@n7), Tex=@n8)
  n10 = CameraVectorWS()
  n11 = Constant3Vector() {constant=<rgba 1 0 0 0>}
  n12 = Add(A=@n10, B=@n11)
  n13 = CrossProduct(A=@n11, B=Constant3Vector() {constant=<rgba 0 0 1 0>})
  n14 = CrossProduct(A=@n13, B=@n11)
  n15 = AppendVector(A=DotProduct(A=@n12, B=Normalize(VectorInput=@n13)), B=DotProduct(A=@n12, B=Normalize(VectorInput=@n14)))
  n16 = TextureSample(UVs=Add(A=Multiply(A=@n15, B=@n6), B=@n7), Tex=@n8)
  n17 = FunctionInput() {bUsePreviewValueAsDefault=true, inputName=contrast, inputType=FunctionInput_Scalar, previewValue=<1 0 0 1>}
  n18 = MaterialFunctionCall(In=Abs(None=DotProduct(A=@n10, B=@n11)), Contrast=@n17) {materialFunction=&CheapContrast}
  n19 = CameraVectorWS()
  n20 = Constant3Vector() {constant=<rgba 0 1 0 0>}
  n21 = Add(A=@n19, B=@n20)
  n22 = CrossProduct(A=@n20, B=Constant3Vector() {constant=<rgba 0 0 1 0>})
  n23 = DotProduct(A=@n21, B=Normalize(VectorInput=@n22))
  n24 = DotProduct(A=@n21, B=Normalize(VectorInput=CrossProduct(A=@n22, B=@n20)))
  n25 = Add(A=Multiply(A=AppendVector(A=@n23, B=@n24), B=@n6), B=@n7)
  n26 = Abs(None=DotProduct(A=@n19, B=@n20))
  FunctionOutput(None=LinearInterpolate(
    A=LinearInterpolate(A=@n9.RGB, B=@n16.RGB, Alpha=@n18.Result)
    B=TextureSample(UVs=@n25, Tex=@n8).RGB
    Alpha=MaterialFunctionCall(In=@n26, Contrast=@n17) {materialFunction=&CheapContrast}.Result
   ))
 ]
}
