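# Material function that derives several UV projections from the camera vector, with optional local-space and rotation switches on that vector.
# It outputs spherical (plain, remapped, mirrored), cylindrical, planar and radial UVs, with planar supporting exponent and dome blending.
# Quad UVs project the view onto a plane placed by horizontal and vertical angles, with a mask clipping bounds and the opposite hemisphere.
# Extra outputs give the adjusted camera vector, a hemisphere mask and a generic UV bounds mask.
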MaterialFunction /Game/Packages/Fortress_SharedMaterials/Base_Material_Functions/MF_CameraBasedUVs
graph {nodes [
  n1 = FunctionInput(Preview=Constant()) {bUsePreviewValueAsDefault=true, inputName=Quad UV H pos, inputType=FunctionInput_Scalar, sortPriority=50}
  n2 = Cosine(None=@n1)
  n3 = Multiply(A=FunctionInput(Preview=Constant() {r=0.25}) {bUsePreviewValueAsDefault=true, inputName=Quad UV V pos, inputType=FunctionInput_Scalar, sortPriority=51}) {constB=0.25}
  n4 = Cosine(None=@n3)
  n5 = Sine(None=@n1)
  n6 = Normalize(VectorInput=MaterialFunctionCall(R=Multiply(A=@n2, B=@n4), G=Multiply(A=@n5, B=@n4), B=Sine(None=@n3)) {materialFunction=&AppendMany}.RGB)
  n7 = CrossProduct(A=@n6, B=Constant3Vector() {constant=<rgba 0 0 1 0>})
  n8 = NamedRerouteUsage()
  n9 = Add(A=@n8, B=@n6)
  n10 = DotProduct(A=@n9, B=Normalize(VectorInput=@n7))
  n11 = DotProduct(A=@n9, B=Normalize(VectorInput=CrossProduct(A=@n7, B=@n6)))
  n12 = DotProduct(A=@n8, B=@n6)
  n13 = Constant2Vector() {g=1, r=1}
  n14 = Multiply(
   A=Divide(A=AppendVector(A=@n10, B=@n11), B=@n12)
   B=Divide(B=FunctionInput(Preview=@n13) {bUsePreviewValueAsDefault=true, inputName=Quad UV Scale, inputType=FunctionInput_Vector2, sortPriority=52})
  )
  n15 = Constant() {r=0.5}
  FunctionOutput(None=OneMinus(None=Add(A=@n14, B=@n15))) {outputName=Quad UV, sortPriority=50}
  n16 = FunctionInput(Preview=CameraVectorWS()) {bUsePreviewValueAsDefault=true, inputName=Camera Vector, sortPriority=-10}
  n17 = StaticSwitch(
   True=Multiply(
    A=MaterialFunctionCall() {materialFunction=&ObjectScale}.Scale XYZ
    B=Transform(None=@n16) {transformSourceType=TRANSFORMSOURCE_World, transformType=TRANSFORM_Local}
   )
   False=@n16
   Value=FunctionInput(Preview=StaticBool()) {bUsePreviewValueAsDefault=true, inputName=Local Space, inputType=FunctionInput_StaticBool}
  )
  n18 = Constant3Vector() {constant=<rgba 0 0 1 0>}
  n19 = FunctionInput(Preview=@n18) {bUsePreviewValueAsDefault=true, inputName=Rotation Vector, sortPriority=61, previewValue=<0 0 1 1>}
  n20 = MaterialFunctionCall!(VectorToRotate=@n17, LookAtVector=@n18, RestingVector=@n19) {materialFunction=/Engine/Functions/Engine_MaterialFunctions01/Vectors/RotateVector}
  n21 = FunctionInput(Preview=StaticBool()) {bUsePreviewValueAsDefault=true, inputName=Enable Rotation, inputType=FunctionInput_StaticBool, sortPriority=60, previewValue=<0 0 1 1>}
  FunctionOutput(None=NamedRerouteDeclaration(None=StaticSwitch(True=@n20.Result, False=@n17, Value=@n21)) {name=Cam Vector, nodeColor=<rgba 1 0 0.0437 1>}) {outputName=Camera Vector, sortPriority=90}
  n22 = FunctionInput(Preview=Constant() {r=1}) {bUsePreviewValueAsDefault=true, inputName=Planar Radial UV Scale, inputType=FunctionInput_Scalar, sortPriority=40}
  n23 = Multiply(A=NamedRerouteUsage(), B=@n22)
  n24 = Multiply(A=@n22) {constB=-0.5}
  FunctionOutput(None=MaterialFunctionCall(Vector or UVs=Add(A=@n23, B=@n24)) {materialFunction=&VectorToRadialValue}.Radial Coordinates) {outputName=Planar Radial UVs, sortPriority=40}
  n25 = Ceil(None=Multiply(A=@n12) {constB=-1})
  n26 = Multiply(A=Abs(None=@n14)) {constB=2}
  n27 = OneMinus(None=Saturate(None=Floor(None=@n26)))
  n28 = ComponentMask(None=@n27) {r=true}
  n29 = ComponentMask(None=@n27) {g=true}
  FunctionOutput(None=Multiply(A=Saturate(None=@n25), B=Multiply(A=@n28, B=@n29))) {outputName=Quad UV Mask, sortPriority=51}
  n30 = Constant2Vector() {r=1}
  n31 = NamedRerouteUsage()
  n32 = Normalize(VectorInput=ComponentMask(None=@n31) {g=true, r=true})
  n33 = Divide(A=ArccosineFast!(None=DotProduct(A=@n30, B=@n32))) {constB=6.2832}
  n34 = OneMinus(None=@n33)
  n35 = DotProduct(A=@n31, B=Constant3Vector() {constant=<rgba 0 0 -1 0>})
  n36 = Saturate(None=Divide(A=ArccosineFast(None=@n35)) {constB=3.1416})
  n37 = ComponentMask(None=@n31) {g=true}
  n38 = LinearInterpolate(
   A=Add(A=AppendVector(A=@n34, B=@n36), B=Constant())
   B=Add(A=AppendVector(A=@n33, B=@n36), B=Constant())
   Alpha=Saturate(None=Ceil(None=@n37))
  )
  n39 = MaterialFunctionCall(Float2=FunctionOutput(None=NamedRerouteDeclaration(None=@n38) {name=Sphere UV, nodeColor=<rgba 0.7109 0 1 1>}) {outputName=Spherical UVs, sortPriority=10}) {materialFunction=&BreakOutFloat2Components}
  n40 = FunctionInput(Preview=Constant() {r=0.2}) {bUsePreviewValueAsDefault=true, inputName=Spherical UV Remap Top, inputType=FunctionInput_Scalar, sortPriority=10}
  n41 = OneMinus(None=@n40)
  n42 = Divide(B=@n41)
  n43 = Multiply(A=@n39.G, B=@n42)
  n44 = Subtract(A=@n42)
  n45 = FunctionInput(Preview=Constant() {r=0.2}) {bUsePreviewValueAsDefault=true, inputName=Spherical UV Remap Bot, inputType=FunctionInput_Scalar, sortPriority=11}
  n46 = Divide(A=Subtract(A=OneMinus(None=@n45), B=@n40), B=@n41)
  n47 = Saturate(None=Multiply(A=Subtract(A=@n43, B=@n44), B=Divide(B=@n46)))
  FunctionOutput(None=NamedRerouteDeclaration(None=AppendVector(A=@n39.R, B=@n47)) {name=Spherical Clamped UV, nodeColor=<rgba 1 0.396 0.849 1>}) {outputName=Spherical Remapped UVs, sortPriority=11}
  n48 = Subtract(A=@n47) {constB=0.5}
  n49 = OneMinus(None=Abs(None=Multiply(A=@n48) {constB=2}))
  FunctionOutput(None=NamedRerouteDeclaration(None=AppendVector(A=@n39.R, B=@n49)) {name=Spherical Clamped Mirrored UV, nodeColor=<rgba 1 0.396 0.849 1>}) {outputName=Spherical Remapped Mirrored UVs, sortPriority=12}
  n50 = NamedRerouteUsage()
  n51 = MaterialFunctionCall(Vector or UVs=ComponentMask(None=@n50) {g=true, r=true}) {materialFunction=&VectorToRadialValue}
  n52 = ComponentMask(None=@n50) {b=true}
  n53 = Multiply(A=@n52, B=@n52)
  n54 = Divide(A=Constant() {r=10000}, B=Subtract(B=Divide(B=@n53)))
  n55 = Multiply(A=Power(Base=Abs(None=@n54)) {constExponent=0.5}, B=Sign(None=@n52))
  n56 = Constant() {r=1}
  n57 = Add(A=Multiply(
    A=Divide(A=@n55) {constB=500}
    B=FunctionInput(Preview=@n56) {bUsePreviewValueAsDefault=true, inputName=Cylinder UV V Scale, sortPriority=20}
   )) {constB=0.5}
  FunctionOutput(None=NamedRerouteDeclaration(None=AppendVector(A=@n51.Vector Converted to Angle, B=@n57)) {name=Cylinder UV, nodeColor=<rgba 1 0.4604 0 1>}) {outputName=Cylinder UVs, sortPriority=20}
  n58 = NamedRerouteUsage()
  n59 = Power(
   Base=Abs(None=ComponentMask(None=@n58) {b=true})
   Exp=FunctionInput(Preview=Constant() {r=1}) {bUsePreviewValueAsDefault=true, inputName=Planar UV Exponent, inputType=FunctionInput_Scalar, sortPriority=30}
  )
  n60 = FunctionInput(Preview=Constant()) {bUsePreviewValueAsDefault=true, inputName=Planar UV Dome, inputType=FunctionInput_Scalar, sortPriority=31}
  n61 = Multiply(
   A=Divide(A=ComponentMask(None=@n58) {g=true, r=true}, B=LinearInterpolate(A=@n59, Alpha=@n60) {constA=1})
   B=FunctionInput(Preview=Constant() {r=1}) {bUsePreviewValueAsDefault=true, inputName=Planar UV Scale, inputType=FunctionInput_Scalar, sortPriority=32}
  )
  FunctionOutput(None=NamedRerouteDeclaration(None=Add(A=@n61) {constB=0.5}) {name=Planar UVs, nodeColor=<rgba 0 0.7656 1 1>}) {outputName=Planar UVs, sortPriority=30}
  n62 = Abs(None=Subtract(A=FunctionInput() {inputName=Mask UV, inputType=FunctionInput_Vector2, sortPriority=100}, B=Constant() {r=0.5}))
  n63 = Subtract(A=FunctionInput(Preview=Constant() {r=1}) {bUsePreviewValueAsDefault=true, inputName=Mask Size, inputType=FunctionInput_Scalar, sortPriority=101}) {constB=0.5}
  n64 = Subtract(A=ComponentMask(None=@n62) {r=true}, B=@n63) {constB=0.5}
  n65 = Subtract(A=ComponentMask(None=@n62) {g=true}, B=@n63) {constB=0.5}
  n66 = Add(A=Ceil(None=Saturate(None=@n64)), B=Ceil(None=Saturate(None=@n65)))
  FunctionOutput(None=OneMinus(None=Saturate(None=@n66))) {outputName=UV Mask, sortPriority=100}
  n67 = ComponentMask(None=NamedRerouteUsage()) {b=true}
  FunctionOutput(None=Ceil(None=Multiply(A=@n67) {constB=-1})) {outputName=Top Mask, sortPriority=91}
  FunctionOutput(None=FunctionInput() {inputName=-------------, sortPriority=99}) {outputName=-------------, sortPriority=99}
  NamedRerouteUsage()
  NamedRerouteUsage()
  NamedRerouteUsage()
  NamedRerouteUsage()
  n68 = TextureCoordinate()
  TextureSample!(UVs=NamedRerouteUsage(), Tex, "DDX(UVs)"=DDX(Value=@n68), "DDY(UVs)"=DDY(Value=@n68), Apply View MipBias) {mipValueMode=TMVM_Derivative, texture=/Game/FN_PublicTurntable/Turntable_Lowpoly/Content/BlueprintAssets/Checker}
 ]}
