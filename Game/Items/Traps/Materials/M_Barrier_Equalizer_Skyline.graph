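# Unlit material for a trap barrier, drawing a music-reactive night skyline as an emissive fake environment.
# Camera-based spherical and planar UVs place layered city textures, distant lights, bottom smoke, clouds, fog and triplanar stars.
# MF_BarrierSoundWave blends music and fake signals for equalizer and pulse effects, with colors from a scrolling curve-atlas hue ramp.
# Skyline samples take explicit, Fresnel-scaled UV derivatives to counter blurring; static switches toggle distant city and desaturation.
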Material /Game/Items/Traps/Materials/M_Barrier_Equalizer_Skyline
graph {
 properties {bUseMaterialAttributes=true, pixelDepthOffsetMode=PDOM_Legacy, shadingModel=MSM_Unlit, usedShadingModels=Unlit}
 nodes [
  n1 = Multiply(
   A=NamedRerouteDeclaration(None=MaterialFunctionCall!(
     Camera Vector
     Local Space
     Spherical UV Remap Top
     Spherical UV Remap Bot
     Cylinder UV V Scale
     Planar UV Exponent
     Planar UV Dome
     Planar UV Scale
     Planar Radial UV Scale
     Quad UV H pos
     Quad UV V pos
     Quad UV Scale
     Enable Rotation
     Rotation Vector
     -------------
     Mask UV
     Mask Size
    ) {materialFunction=/Game/Packages/Fortress_SharedMaterials/Base_Material_Functions/MF_CameraBasedUVs}.Planar UVs) {name=Planar UV, nodeColor=<rgba 0.4531 0 1 1>}
   B=ScalarParameter() {defaultValue=0.75, group=Smoke - Bottom, parameterName=Bottom Smoke Tiling}
  )
  n2 = Multiply(A=Time(), B=ScalarParameter() {defaultValue=0.3, group=Smoke - Bottom, parameterName=Bottom Smoke Speed})
  n3 = Panner(Coordinate=Multiply(A=@n1) {constB=0.5}, Time=@n2) {speedX=0.02, speedY=0.05}
  n4 = TextureObjectParameter() {group=Environment - Clouds, parameterName=Clouds Texture, sortPriority=10, texture=/Game/Items/Traps/Textures/T_Barrier_Clouds_03_M}
  n5 = Panner(Coordinate=Multiply(A=@n1) {constB=0.8}, Time=@n2) {speedX=0.01, speedY=0.04}
  n6 = ScalarParameter() {defaultValue=-7, group=Environment - PlanarCity, parameterName=Bottom Smoke Switch A}
  n7 = ScalarParameter() {defaultValue=4, group=Environment - PlanarCity, parameterName=Bottom Smoke Switch B}
  n8 = ComponentMask(None=NamedRerouteUsage()) {g=true}
  n9 = LinearInterpolate(
   A=NamedRerouteUsage()
   B=Add(
    A=TextureSample(UVs=@n3, Tex=@n4) {samplerSource=SSM_Wrap_WorldGroupSettings}.R
    B=TextureSample(UVs=@n5, Tex=@n4) {samplerSource=SSM_Wrap_WorldGroupSettings}.R
   )
   Alpha=Saturate(None=LinearInterpolate(A=@n6, B=@n7, Alpha=@n8))
  )
  NamedRerouteDeclaration(None=Saturate(None=@n9)) {name=Bottom Cloud Mask, nodeColor=<rgba 0 1 1 1>}
  n10 = MaterialFunctionCall(
   Spherical UV Remap Top=ScalarParameter() {defaultValue=0.4, group=Environment - City, parameterName=City UV Remap Top}
   Spherical UV Remap Bot=ScalarParameter() {defaultValue=0.3, group=Environment - City, parameterName=City UV Remap Bottom}
  ) {materialFunction=&MF_CameraBasedUVs}
  n11 = MaterialFunctionCall(Float2=NamedRerouteDeclaration(None=@n10.Spherical UVs) {name=Spherical UVs, nodeColor=<rgba 1 0 0.2109 1>}) {materialFunction=&BreakOutFloat2Components}
  n12 = Multiply(A=@n11.R, B=ScalarParameter() {defaultValue=1, group=Equalizer, parameterName=Equalizer Remap UV X})
  n13 = Add(
   A=Multiply(A=@n11.G, B=ScalarParameter() {defaultValue=3, group=Equalizer, parameterName=Equalizer Remap UV Y}) {constB=3}
   B=ScalarParameter() {defaultValue=-1.1, group=Equalizer, parameterName=Equalizer Remap UV Y Offset}
  ) {constB=-1.1}
  NamedRerouteDeclaration(None=AppendVector(A=Frac(None=@n12), B=Saturate(None=@n13))) {name=Equalizer UV, nodeColor=<rgba 0.4297 0 1 1>}
  n14 = Multiply(A=Time(), B=ScalarParameter() {defaultValue=0.5, group=Colors, parameterName=Hue Shift Speed})
  n15 = ComponentMask(None=@n10.Spherical UVs) {r=true}
  n16 = CurveAtlasRowParameter(CurveTime=Frac(None=Add(A=@n14, B=@n15))) {
   atlas=/Game/Items/Traps/Textures/CA_BarrierCurves
   curve=/Game/Items/Traps/Textures/C_BarrierEqualizer
   defaultValue=3
   group=Colors
   parameterName=Color Curve
  }
  n17 = Multiply(A=NamedRerouteUsage()) {constB=2}
  NamedRerouteDeclaration(None=LinearInterpolate(
    A=VectorParameter() {group=Colors, parameterName=Top color, defaultValue=<rgba 0.7865 0.7865 0.7865 0>}.RGB
    B=@n16
    Alpha=Saturate(None=@n17)
   )) {name=RGB Color, nodeColor=<rgba 0.7138 1 0 1>}
  n18 = TextureObjectParameter() {
   group=Environment - City
   parameterName=City Tex A 01
   samplerType=SAMPLERTYPE_LinearColor
   sortPriority=14
   texture=/Game/Items/Traps/Textures/T_Barrier_CitySkyline_B_01
  }
  n19 = NamedRerouteUsage()
  n20 = NamedRerouteUsage()
  n21 = Multiply(A=NamedRerouteUsage(), B=Constant2Vector() {g=1, r=2})
  n22 = TextureSample!(UVs=@n21, Tex=@n18, "DDX(UVs)"=@n19, "DDY(UVs)"=@n20, Apply View MipBias) {mipValueMode=TMVM_Derivative, samplerType=SAMPLERTYPE_LinearColor, texture=/Game/Items/Traps/Textures/T_Barrier_CitySkyline_A_01}
  n23 = TextureObjectParameter() {
   group=Environment - City
   parameterName=City Tex B 01
   samplerType=SAMPLERTYPE_LinearColor
   sortPriority=15
   texture=&T_Barrier_CitySkyline_B_01
  }
  n24 = TextureSample(UVs=@n21, Tex=@n23, "DDX(UVs)"=@n19, "DDY(UVs)"=@n20) {mipValueMode=TMVM_Derivative, samplerType=SAMPLERTYPE_LinearColor, texture=&T_Barrier_CitySkyline_B_01}
  n25 = Ceil(None=Subtract(A=ComponentMask(None=@n21) {r=true}))
  n26 = LinearInterpolate(A=@n22.RGB, B=@n24.RGB, Alpha=@n25)
  n27 = ComponentMask(None=@n26) {b=true}
  NamedRerouteDeclaration(None=@n27) {name=Skyline Mask, nodeColor=<rgba 0.1514 0.179 0.1875 1>}
  n28 = MaterialFunctionCall!(
   "Spectrum Strength (V4)"
   UVs=NamedRerouteUsage()
   Fake Signal Tiling=ScalarParameter() {defaultValue=1, group=Music Reactivity, parameterName=Fake Signal Tiling}
   Fake Signal Tiling Multiplier 1=ScalarParameter() {defaultValue=2, group=Music Reactivity, parameterName=Fake Signal Tiling 2}
   Fake Signal Tiling Multiplier 2=ScalarParameter() {defaultValue=4, group=Music Reactivity, parameterName=Fake Signal Tiling 3}
   Fake Signal Speed
  ) {materialFunction=/Game/Items/Traps/Materials/MaterialFunctions/MF_BarrierSoundWave}
  n29 = NamedRerouteDeclaration(None=LinearInterpolate(
    A=ScalarParameter() {defaultValue=0.5, parameterName=Amplitude Min}
    B=ScalarParameter() {defaultValue=2, parameterName=Amplitude Max}
    Alpha=@n28.Music Amplitude
   )) {name=Music Reactivity - Amplitude, nodeColor=<rgba 1 0 0.9744 1>}
  n30 = VectorParameter() {
   group=Music Reactivity
   parameterName=EQ Control
   channelNames {a=Global Intensity, b=Fake Signal Proportion, g=Fake Signal Strength, r=Music Signal Strength}
   defaultValue=<rgba 1 1 0.8 1>
  }
  n31 = LinearInterpolate(
   A=Multiply(A=@n28.Music Signal, B=@n30.Music Signal Strength)
   B=Multiply(A=@n28.Fake Signal, B=@n30.Fake Signal Strength)
   Alpha=@n30.Fake Signal Proportion
  ) {constA=1}
  NamedRerouteDeclaration(None=Multiply(A=Multiply(A=@n29, B=@n31), B=@n30.Global Intensity)) {name=Equalizer Signal, nodeColor=<rgba 1 0 0.599 1>}
  n32 = NamedRerouteUsage()
  n33 = Constant2Vector() {g=3, r=6}
  n34 = TextureObjectParameter() {group=Environment - Clouds, parameterName=Clouds Texture, sortPriority=10, texture=&T_Barrier_Clouds_03_M}
  n35 = TextureCoordinate()
  n36 = TextureSample(
   UVs=Panner(Coordinate=Multiply(A=@n32, B=@n33)) {speedX=0.01, speedY=0.005}
   Tex=@n34
   "DDX(UVs)"=DDX(Value=@n35)
   "DDY(UVs)"=DDY(Value=@n35)
  ) {mipValueMode=TMVM_Derivative, samplerSource=SSM_Wrap_WorldGroupSettings}
  NamedRerouteDeclaration(None=@n36.RGB) {name=Spherical Cloud, nodeColor=<rgba 1 0.2344 0 1>}
  n37 = Constant()
  n38 = MaterialFunctionCall(Spherical UV Remap Top=@n37, Spherical UV Remap Bot=@n37) {materialFunction=&MF_CameraBasedUVs}
  NamedRerouteDeclaration(None=@n38.Top Mask) {name=Top Mask, nodeColor=<rgba 1 0.6145 0 1>}
  NamedRerouteDeclaration(None=ComponentMask(None=@n38.Spherical Remapped Mirrored UVs) {g=true}) {name=Mask Vertical Mirroed, nodeColor=<rgba 1 0.6145 0 1>}
  NamedRerouteDeclaration(None=@n10.Spherical Remapped UVs) {name=City UVs, nodeColor=<rgba 1 0 0.1057 1>}
  NamedRerouteDeclaration(None=ComponentMask(None=@n38.Spherical UVs) {g=true}) {name=Mask Vertical, nodeColor=<rgba 1 0.6145 0 1>}
  n39 = TextureCoordinate()
  n40 = DDX(Value=@n39)
  n41 = DotProduct(A=VertexNormalWS(), B=CameraVectorWS())
  n42 = Saturate(None=Power(Base=Abs(None=@n41), Exp=ScalarParameter() {defaultValue=1, group=UV DD, parameterName=DD Fresnel Exp}))
  n43 = ScalarParameter() {defaultValue=0.05, group=UV DD, parameterName=DD Fix for blurry textures}
  NamedRerouteDeclaration(None=Multiply(A=Multiply(A=@n40, B=@n42), B=@n43)) {name=DDX, nodeColor=<rgba 0 1 0.6719 1>}
  n44 = DDY(Value=@n39)
  NamedRerouteDeclaration(None=Multiply(A=Multiply(A=@n44, B=@n42), B=@n43)) {name=DDY, nodeColor=<rgba 0 1 0.2266 1>}
  Reroute(None=@n36.RGB)
  n45 = Power(Base=NamedRerouteUsage())
  n46 = ScalarParameter() {defaultValue=5, group=Music Reactivity, parameterName=Music Pulse Intensity}
  n47 = NamedRerouteUsage()
  n48 = ComponentMask(None=MaterialFunctionCall!(
    contrast=ScalarParameter() {defaultValue=3, group=Environment - Stars, parameterName=Stars Contrast}
    UV Scale=ScalarParameter() {defaultValue=2, group=Environment - Stars, parameterName=Stars UV}
    UV Distortion
    In=TextureObjectParameter() {group=Environment - Stars, parameterName=Stars Texture, texture=/Game/Items/Traps/Textures/T_Barrier_Stars_01_M}
   ) {materialFunction=/Game/Items/Traps/Materials/MaterialFunctions/MF_BarrierTriplanar}.Result) {r=true}
  n49 = TextureObjectParameter() {group=Environment - Clouds, parameterName=Clouds Texture, sortPriority=10, texture=&T_Barrier_Clouds_03_M}
  n50 = Constant() {r=1}
  n51 = StaticBool()
  n52 = TextureObjectParameter() {
   group=Environment - Clouds
   parameterName=Clouds Normal
   samplerType=SAMPLERTYPE_Normal
   sortPriority=11
   texture=&FlatNormalBlue
  }
  n53 = Multiply(
   A=NamedRerouteUsage()
   B=Power(
    Base=MaterialFunctionCall!(
     Cloud Texture=@n49
     Cloud Color=@n50
     Light Source Color=@n50
     Light Source Direction
     Normals on clouds=@n51
     Cloud Normals=@n52
     UV Distortion
     Mirrored on Horizon=@n51
    ) {materialFunction=/Game/Items/Traps/Materials/MaterialFunctions/MF_BarrierClouds}.Result
    Exp=ScalarParameter() {defaultValue=2, group=Environment - Clouds, parameterName=Cloud Exp}
   )
  )
  n54 = Multiply(A=Subtract(A=Multiply(A=@n47, B=@n48), B=@n53), B=OneMinus(None=NamedRerouteUsage()))
  n55 = ComponentMask(None=NamedRerouteUsage()) {g=true}
  n56 = Saturate(None=LinearInterpolate(
    A=ScalarParameter() {defaultValue=-1.63, group=Environment - Stars, parameterName=Star Mask Low}
    B=ScalarParameter() {defaultValue=1.42, group=Environment - Stars, parameterName=Star Mask High}
    Alpha=OneMinus(None=@n55)
   ))
  n57 = Multiply(A=Saturate(None=Multiply(A=@n54, B=@n56)), B=ScalarParameter() {defaultValue=20, group=Environment - Stars, parameterName=Stars Intensity})
  n58 = VectorParameter() {group=Environment - Stars, parameterName=Stars Color, defaultValue=<rgba 0.9688 0.9277 0.8874 0>}
  n59 = Multiply(A=@n53, B=OneMinus(None=NamedRerouteUsage()))
  n60 = Multiply(
   A=VectorParameter() {group=Environment - Clouds, parameterName=Cloud Color, defaultValue=<rgba 0.013 0.0661 0.1198 0>}.RGB
   B=LinearInterpolate(B=NamedRerouteUsage(), Alpha=ScalarParameter() {defaultValue=1, parameterName=Cloud Hue}) {constA=1}
  )
  n61 = ComponentMask(None=NamedRerouteUsage()) {g=true}
  n62 = OneMinus(None=ScalarParameter() {defaultValue=0.3, group=Music Reactivity, parameterName=Rectivity Center Position})
  n63 = LinearInterpolate(
   A=ScalarParameter() {group=Music Reactivity, parameterName=Rectivity Min Size}
   B=ScalarParameter() {defaultValue=1.5, group=Music Reactivity, parameterName=Rectivity Max Size}
   Alpha=Saturate(None=NamedRerouteUsage())
  )
  n64 = Add(A=Multiply(A=Subtract(A=@n61, B=@n62), B=Divide(B=@n63)), B=@n62)
  n65 = OneMinus(None=@n64)
  n66 = OneMinus(None=@n62)
  n67 = Power(Base=MaterialFunctionCall!(A=Divide(A=@n64, B=@n62), B=Divide(A=@n65, B=@n66), K=Constant() {r=0.5}) {materialFunction=/Game/Items/Traps/Materials/MaterialFunctions/MF_BarrierSmoothMin}.Result) {constExponent=4}
  n68 = OneMinus(None=NamedRerouteUsage())
  n69 = Multiply(
   A=Multiply(A=NamedRerouteDeclaration(None=@n67) {name=Equalizer, nodeColor=<rgba 0 0.7187 1 1>}, B=LinearInterpolate(Alpha=@n68))
   B=NamedRerouteUsage()
  )
  n70 = TextureObjectParameter() {group=Environment - City, parameterName=City Tex A 02, sortPriority=10, texture=/Game/Items/Traps/Textures/T_Barrier_CitySkyline_B_02}
  n71 = TextureSample(UVs=@n21, Tex=@n70, "DDX(UVs)"=@n19, "DDY(UVs)"=@n20) {mipValueMode=TMVM_Derivative, texture=/Game/Items/Traps/Textures/T_Barrier_CitySkyline_A_02}
  n72 = TextureObjectParameter() {group=Environment - City, parameterName=City Tex B 02, sortPriority=11, texture=&T_Barrier_CitySkyline_B_02}
  n73 = TextureSample(UVs=@n21, Tex=@n72, "DDX(UVs)"=@n19, "DDY(UVs)"=@n20) {mipValueMode=TMVM_Derivative, texture=&T_Barrier_CitySkyline_B_02}
  n74 = TextureObjectParameter() {
   group=Environment - City
   parameterName=City Tex A 03
   samplerType=SAMPLERTYPE_LinearColor
   sortPriority=12
   texture=/Game/Items/Traps/Textures/T_Barrier_CitySkyline_B_03
  }
  n75 = TextureSample(UVs=@n21, Tex=@n74, "DDX(UVs)"=@n19, "DDY(UVs)"=@n20) {mipValueMode=TMVM_Derivative, samplerType=SAMPLERTYPE_LinearColor, texture=/Game/Items/Traps/Textures/T_Barrier_CitySkyline_A_03}
  n76 = TextureObjectParameter() {
   group=Environment - City
   parameterName=City Tex B 03
   samplerType=SAMPLERTYPE_LinearColor
   sortPriority=13
   texture=&T_Barrier_CitySkyline_B_03
  }
  n77 = TextureSample(UVs=@n21, Tex=@n76, "DDX(UVs)"=@n19, "DDY(UVs)"=@n20) {mipValueMode=TMVM_Derivative, samplerType=SAMPLERTYPE_LinearColor, texture=&T_Barrier_CitySkyline_B_03}
  n78 = ComponentMask(None=LinearInterpolate(A=@n75.RGB, B=@n77.RGB, Alpha=@n25)) {g=true}
  n79 = Multiply(A=Multiply(A=LinearInterpolate(A=@n71.RGB, B=@n73.RGB, Alpha=@n25), B=@n27), B=@n78) {constA=1}
  n80 = VectorParameter() {group=Environment - City, parameterName=City Color, defaultValue=<rgba 0.0625 0.0625 0.0625 0>}
  n81 = NamedRerouteDeclaration(None=@n78) {name=Skyline Depth Mask, nodeColor=<rgba 0.1667 0.1397 0.1181 1>}
  n82 = VectorParameter() {group=Environment - City, parameterName=City Window Colors, defaultValue=<rgba 0.3281 0.301 0.2546 0>}
  n83 = ComponentMask(None=@n26) {r=true}
  n84 = NamedRerouteUsage()
  n85 = Multiply(A=Multiply(A=@n81, B=@n81), B=Multiply(A=Multiply(A=@n82.RGB, B=@n83), B=Add(A=@n84)))
  n86 = Multiply(
   A=Multiply(A=NamedRerouteUsage(), B=@n9)
   B=VectorParameter() {group=Smoke - Bottom, parameterName=Bottom Smoke Color, defaultValue=<rgba 0.0469 0.0469 0.0469 0>}.RGB
  )
  n87 = ScalarParameter() {defaultValue=-12, group=Smoke - Bottom, parameterName=Bottom Smoke Mask A}
  n88 = ScalarParameter() {defaultValue=9, group=Smoke - Bottom, parameterName=Bottom Smoke Mask B}
  n89 = ComponentMask(None=NamedRerouteUsage()) {g=true}
  n90 = Power(Base=Saturate(None=LinearInterpolate(A=@n87, B=@n88, Alpha=@n89) {constA=-10, constB=10}))
  n91 = LinearInterpolate(A=Add(A=Multiply(A=@n79, B=@n80.RGB), B=@n85), B=@n86, Alpha=@n90)
  n92 = Add(A=Multiply(A=Add(A=@n69, B=@n91), B=NamedRerouteUsage()), B=Multiply(A=OneMinus(None=@n90), B=@n85))
  n93 = ComponentMask(None=NamedRerouteUsage()) {g=true}
  n94 = OneMinus(None=@n93)
  n95 = OneMinus(None=Saturate(None=Multiply(A=@n94) {constB=2}))
  n96 = ScalarParameter() {defaultValue=5, group=Environment - PlanarCity, parameterName=Distance City Falloff}
  n97 = NamedRerouteUsage()
  n98 = Saturate(None=LinearInterpolate(
    A=ScalarParameter() {defaultValue=-5, group=Environment - PlanarCity, parameterName=Distance City Near Mask A}
    B=ScalarParameter() {defaultValue=3.1, group=Environment - PlanarCity, parameterName=Distance City Near Mask B}
    Alpha=@n93
   ) {constA=-1, constB=2})
  n99 = Multiply(A=Saturate(None=Multiply(A=@n95, B=@n96) {constB=10}), B=Max(A=OneMinus(None=@n97), B=@n98))
  n100 = NamedRerouteUsage()
  n101 = TextureObjectParameter() {
   group=Environment - PlanarCity
   parameterName=City Texture
   samplerType=SAMPLERTYPE_LinearColor
   texture=/Game/Items/Traps/Textures/T_Barrier_Stars_02_M
  }
  n102 = TextureSample!(
   UVs=Multiply(A=@n100, B=ScalarParameter() {defaultValue=0.46, group=Environment - PlanarCity, parameterName=Distant Lights Tiling Far}) {constB=0.25}
   Tex=@n101
   Apply View MipBias
  ) {samplerSource=SSM_Wrap_WorldGroupSettings, samplerType=SAMPLERTYPE_LinearColor, texture=&T_Barrier_Stars_02_M}
  n103 = TextureSample(
   UVs=Multiply(A=@n100, B=ScalarParameter() {defaultValue=1.57, group=Environment - PlanarCity, parameterName=Distant Lights Tiling Near}) {constB=0.25}
   Tex=@n101
  ) {samplerSource=SSM_Wrap_WorldGroupSettings, samplerType=SAMPLERTYPE_LinearColor, texture=&T_Barrier_Stars_02_M}
  n104 = TextureSample(
   UVs=Multiply(A=@n100, B=ScalarParameter() {defaultValue=1.57, group=Environment - PlanarCity, parameterName=Distant Lights Tiling Far Tilebreak}) {constB=0.25}
   Tex=@n101
  ) {samplerSource=SSM_Wrap_WorldGroupSettings, samplerType=SAMPLERTYPE_LinearColor, texture=&T_Barrier_Stars_02_M}
  n105 = LinearInterpolate(A=Multiply(A=Multiply(A=@n102.R, B=@n103.G), B=@n104.G), B=@n103.R, Alpha=@n98)
  n106 = VectorParameter() {group=Environment - PlanarCity, parameterName=Planar City Intensity Far, defaultValue=<rgba 0.4219 0.4219 0.4219 1>}
  n107 = VectorParameter() {group=Environment - PlanarCity, parameterName=Planar City Intensity Near, defaultValue=<rgba 0.1042 0.1042 0.1042 1>}
  n108 = Ceil(None=@n98)
  n109 = StaticSwitchParameter(True=Multiply(A=Multiply(A=@n99, B=@n105), B=LinearInterpolate(A=@n106.RGB, B=@n107.RGB, Alpha=@n108)), False=Constant()) {defaultValue=true, group=Environment - PlanarCity, parameterName=Distant City}
  n110 = Add(
   A=Add(A=Multiply(A=@n59, B=@n60) {constB=0.1}, B=Add(A=@n92, B=@n109))
   B=VectorParameter() {group=Environment - Background, parameterName=Background, defaultValue=<rgba 0.0048 0.0005 0.0052 0>}.RGB
  )
  n111 = NamedRerouteUsage()
  n112 = VectorParameter() {parameterName=Distant Fog Color, defaultValue=<rgba 0.0338 0.0319 0.0365 0>}
  n113 = Abs(None=Subtract(A=NamedRerouteUsage()) {constB=0.5})
  n114 = OneMinus(None=Multiply(A=ComponentMask(None=@n113) {g=true}) {constB=2})
  n115 = ScalarParameter() {defaultValue=30, group=Environment - Fog, parameterName=Fog Exp}
  n116 = Multiply(A=Multiply(A=Multiply(A=@n111, B=@n112.RGB), B=Power(Base=@n114, Exp=@n115) {constExponent=30}), B=OneMinus(None=NamedRerouteUsage()))
  n117 = Multiply(A=Add(A=Multiply(A=@n45, B=@n46) {constB=9}), B=Add(A=Multiply(A=@n57, B=@n58.RGB), B=Add(A=@n110, B=@n116)))
  n118 = Desaturation(None=@n117, Fraction=ScalarParameter() {group=Colors, parameterName=Desaturation Factor})
  SetMaterialAttributes(Emissive Color=Multiply(
    A=StaticSwitchParameter(True=@n118, False=@n117) {group=_Options, parameterName=Desaturation}
    B=ScalarParameter() {defaultValue=0.5, group=Colors, parameterName=General Intensity Multiplier}
   )) {attributeSetTypes [#1]}
  SetMaterialAttributes(Emissive Color=@n86) {attributeSetTypes [#1]}
 ]
}
