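# Unlit fire material for a trap barrier, usable on instanced static meshes, combining animated flames, embers and smoke.
# A camera-based UV function supplies spherical coordinates, with a sine-warped variant, shared through named reroutes.
# Three noise-distorted flame layers drive a color-graded power curve, alongside flickering sparks and lit clouds, with many samples taking explicit mesh-UV derivatives for mip selection.
# The sum of three rerouted layers is tinted by time of day.
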Material /Game/Items/Traps/Materials/M_Barrier_Fire
graph {
 properties {bUsedWithInstancedStaticMeshes=true, pixelDepthOffsetMode=PDOM_Legacy, shadingModel=MSM_Unlit, usedShadingModels=Unlit}
 nodes [
  n1 = NamedRerouteUsage()
  n2 = NamedRerouteUsage()
  n3 = VectorParameter() {group=Time of Day, parameterName=Color Mult Morning, sortPriority=1, defaultValue=<rgba 1 0.9384 0.9219 1>}
  n4 = VectorParameter() {group=Time of Day, parameterName=Color Mult Day, sortPriority=2, defaultValue=<rgba 1 0.7246 0.6847 1>}
  n5 = VectorParameter() {group=Time of Day, parameterName=Color Mult Evening, sortPriority=3, defaultValue=<rgba 1 0.7438 0.5801 1>}
  n6 = VectorParameter() {group=Time of Day, parameterName=Color Mult Night, sortPriority=4, defaultValue=<rgba 1 0.8502 0.8202 1>}
  Multiply(
   A=Add(A=Add(A=@n1, B=@n2), B=NamedRerouteUsage())
   B=MaterialFunctionCall(Color_Morning=@n3.RGB, Color_Day=@n4.RGB, Color_Evening=@n5.RGB, Color_Night=@n6.RGB) {materialFunction=&Color_Time_of_Day}.Result
  )
  n7 = MaterialFunctionCall!(
   Camera Vector
   Local Space
   Spherical UV Remap Top
   Spherical UV Remap Bot
   Cylinder UV V Scale
   Planar UV Exponent
   Planar UV Dome
   Planar UV Scale
   Planar Radial UV Scale
   Quad UV H pos
   Quad UV V pos
   Quad UV Scale
   Enable Rotation
   Rotation Vector
   -------------
   Mask UV
   Mask Size
  ) {materialFunction=/Game/Packages/Fortress_SharedMaterials/Base_Material_Functions/MF_CameraBasedUVs}
  NamedRerouteDeclaration(None=@n7.Spherical Remapped Mirrored UVs) {name=SRM UVs, nodeColor=<rgba 1 0 0.0866 1>}
  NamedRerouteDeclaration(None=@n7.Spherical UVs) {name=Spherical UVs, nodeColor=<rgba 1 0.0819 0.1316 1>}
  n8 = NamedRerouteUsage()
  n9 = NamedRerouteUsage()
  n10 = Constant2Vector() {g=1, r=4}
  n11 = NamedRerouteUsage()
  n12 = ScalarParameter() {defaultValue=0.3, group=Ember, parameterName=Ember Distortion Speed}
  n13 = TextureSample(UVs=Panner(Coordinate=Multiply(A=@n9, B=@n10), Time=Multiply(A=@n11, B=@n12)) {speedX=0.1, speedY=0.1}) {samplerType=SAMPLERTYPE_Normal, texture=/Game/Effects/Fort_Effects/Textures/Tile/T_GaussianNoise_02_N}
  n14 = ComponentMask(None=@n13.RGB) {r=true}
  n15 = ScalarParameter() {defaultValue=0.05, group=Ember, parameterName=Ember Distortion}
  n16 = ScalarParameter() {defaultValue=4, group=Ember, parameterName=Embers Tiling}
  n17 = ScalarParameter() {defaultValue=0.3, group=Ember, parameterName=Embers Aspect Ratio}
  n18 = Multiply(A=NamedRerouteUsage(), B=AppendVector(A=@n16, B=Multiply(A=@n16, B=@n17)))
  n19 = Add(A=@n18, B=AppendVector(A=Multiply(A=@n14, B=@n15) {constB=0.05}, B=Constant()))
  n20 = Multiply(A=NamedRerouteUsage(), B=ScalarParameter() {defaultValue=0.5, group=Ember, parameterName=Embers Speed})
  n21 = VectorParameter() {
   group=Ember
   parameterName=Embers Speeds
   channelNames {a=Spark 2 Y, b=Spark 1 X, g=Spark 1 Y, r=Spark 1 X}
   defaultValue=<rgba -0.01 0.1 0.01 0.1>
  }
  n22 = TextureCoordinate()
  n23 = TextureSample!(
   UVs=Panner(Coordinate=Add(A=@n19), Time=@n20, Speed=AppendVector(A=@n21.Spark 1 X, B=@n21.Spark 1 X)) {speedX=0.035, speedY=0.2}
   Tex
   "DDX(UVs)"=DDX(Value=@n22)
   "DDY(UVs)"=DDY(Value=@n22)
   Apply View MipBias
  ) {
   mipValueMode=TMVM_Derivative
   samplerSource=SSM_Wrap_WorldGroupSettings
   samplerType=SAMPLERTYPE_LinearColor
   texture=/Game/Items/Traps/Textures/T_Barrier_Stars_02_M
  }
  n24 = Multiply(A=ComponentMask(None=@n13.RGB) {g=true}, B=@n15) {constB=0.05}
  n25 = Constant()
  n26 = Multiply(A=Add(A=@n18, B=AppendVector(A=@n24, B=@n25))) {constB=0.8}
  n27 = TextureCoordinate()
  n28 = TextureSample(
   UVs=Panner(Coordinate=Add(A=@n26), Time=@n20, Speed=AppendVector(A=@n21.Spark 1 X, B=@n21.Spark 1 X)) {speedX=0.02, speedY=0.1}
   "DDX(UVs)"=DDX(Value=@n27)
   "DDY(UVs)"=DDY(Value=@n27)
  ) {
   mipValueMode=TMVM_Derivative
   samplerSource=SSM_Wrap_WorldGroupSettings
   samplerType=SAMPLERTYPE_LinearColor
   texture=&T_Barrier_Stars_02_M
  }
  n29 = Multiply(A=Power(Base=ComponentMask(None=@n8) {g=true}), B=Add(A=Multiply(A=@n23.R, B=@n23.R), B=Multiply(A=@n28.R, B=@n28.R)))
  n30 = Add(A=@n23.G, B=@n28.G)
  n31 = Multiply(A=NamedRerouteUsage(), B=ScalarParameter() {defaultValue=0.2, group=Ember, parameterName=Ember Flicker Speed}) {constB=0.5}
  n32 = ConstantBiasScale(None=Sine(None=Add(A=@n30, B=@n31)))
  NamedRerouteDeclaration(None=Multiply(A=Multiply(A=@n29, B=@n32), B=VectorParameter() {group=Ember, parameterName=Embers Color, defaultValue=<rgba 20 2.7657 0 0>}.RGB)) {name=Embers, nodeColor=<rgba 1 0.3764 0 1>}
  NamedRerouteDeclaration(None=Multiply(A=Time(), B=ScalarParameter() {defaultValue=3, group=Global, parameterName=Global Speed Scale})) {name=Global Time, nodeColor=<rgba 0.875 1 0 1>}
  n33 = ScalarParameter() {defaultValue=0.1, group=Fire, parameterName=Fire Remap UVs B, sortPriority=-6}
  n34 = NamedRerouteUsage()
  n35 = Saturate(None=LinearInterpolate(
    A=ScalarParameter() {defaultValue=0.32, group=Fire, parameterName=Fire Remap UVs A, sortPriority=-7}
    B=Add(A=@n33)
    Alpha=ComponentMask(None=@n34) {g=true}
   ))
  n36 = Sine(None=Add(A=Multiply(A=@n35) {constB=0.5}) {constB=-0.25})
  n37 = Saturate(None=LinearInterpolate(A=ScalarParameter() {defaultValue=-0.6, group=Fire, parameterName=UV Cutoff, sortPriority=-5}, Alpha=ConstantBiasScale(None=@n36)))
  n38 = NamedRerouteUsage()
  n39 = Multiply(
   A=NamedRerouteUsage()
   B=AppendVector(
    A=ScalarParameter() {defaultValue=2, group=Fire, parameterName=Fire Distortion UV X, sortPriority=0}
    B=ScalarParameter() {defaultValue=2, group=Fire, parameterName=Fire Distortion UV Y, sortPriority=1}
   )
  )
  n40 = TextureSample!(UVs=Panner(Coordinate=@n39, Time=@n38) {speedX=0.1, speedY=0.1}, Tex, Apply View MipBias) {
   samplerSource=SSM_Wrap_WorldGroupSettings
   samplerType=SAMPLERTYPE_Normal
   texture=&T_GaussianNoise_02_N
  }
  n41 = Divide(A=Multiply(A=AppendVector(A=@n40.R, B=@n40.R), B=ScalarParameter() {defaultValue=0.5, group=Fire, parameterName=Fire 1 Distortion, sortPriority=13}) {constB=0.02}) {constB=100}
  n42 = ScalarParameter() {defaultValue=1, group=Fire, parameterName=Fire Global tiling, sortPriority=-9}
  n43 = ScalarParameter() {defaultValue=1, group=Fire, parameterName=Fire Global Aspect Ratio, sortPriority=-8}
  n44 = Multiply(A=NamedRerouteUsage(), B=AppendVector(A=@n42, B=Multiply(A=@n42, B=@n43)))
  n45 = ScalarParameter() {defaultValue=7, group=Fire, parameterName=Fire 1 Tiling, sortPriority=10}
  n46 = Multiply(A=@n45, B=ScalarParameter() {defaultValue=0.5, group=Fire, parameterName=Fire 1 Aspect Ratio, sortPriority=11})
  n47 = Panner(
   Coordinate=Multiply(A=Add(A=@n44, B=@n41), B=AppendVector(A=@n45, B=@n46))
   Time=Multiply(A=NamedRerouteUsage(), B=ScalarParameter() {defaultValue=0.1, group=Fire, parameterName=Fire 1 Speed, sortPriority=12})
  ) {speedX=0.1, speedY=1}
  n48 = TextureObjectParameter() {
   group=Fire
   parameterName=Fire Texture
   samplerType=SAMPLERTYPE_LinearColor
   sortPriority=-10
   texture=/Game/Items/Traps/Textures/T_Plasma_Flames_02_M
  }
  n49 = TextureCoordinate()
  n50 = DDX(Value=@n49)
  n51 = DDY(Value=@n49)
  n52 = VectorParameter() {group=Fire, parameterName=Fire Strengths, sortPriority=8, channelNames {a="", b=3, g=2, r=1}, defaultValue=<rgba 0.5 0.35 0.15 0>}
  n53 = TextureSample(UVs=Panner(Coordinate=Multiply(A=@n39) {constB=2}, Time=@n38) {speedX=-0.05, speedY=0.05}) {
   samplerSource=SSM_Wrap_WorldGroupSettings
   samplerType=SAMPLERTYPE_Normal
   texture=&T_GaussianNoise_02_N
  }
  n54 = Divide(A=Multiply(A=AppendVector(A=@n53.R, B=@n53.R), B=ScalarParameter() {defaultValue=1, group=Fire, parameterName=Fire 2 Distortion, sortPriority=23}) {constB=0.02}) {constB=100}
  n55 = ScalarParameter() {defaultValue=10, group=Fire, parameterName=Fire 2 Tiling, sortPriority=20}
  n56 = Multiply(A=@n55, B=ScalarParameter() {defaultValue=0.4, group=Fire, parameterName=Fire 2 Aspect Ratio, sortPriority=21})
  n57 = Panner(
   Coordinate=Multiply(A=Add(A=@n44, B=@n54), B=AppendVector(A=@n55, B=@n56))
   Time=Multiply(A=NamedRerouteUsage(), B=ScalarParameter() {defaultValue=0.1, group=Fire, parameterName=Fire 2 Speed, sortPriority=22})
  ) {speedX=-0.1, speedY=1}
  n58 = TextureCoordinate()
  n59 = DDX(Value=@n58)
  n60 = DDY(Value=@n58)
  n61 = Add(
   A=Multiply(
    A=TextureSample!(UVs=@n47, Tex=@n48, "DDX(UVs)"=@n50, "DDY(UVs)"=@n51, Apply View MipBias) {mipValueMode=TMVM_Derivative, samplerSource=SSM_Wrap_WorldGroupSettings}.R
    B=@n52.1
   )
   B=Multiply(
    A=TextureSample(UVs=@n57, Tex=@n48, "DDX(UVs)"=@n59, "DDY(UVs)"=@n60) {mipValueMode=TMVM_Derivative, samplerSource=SSM_Wrap_WorldGroupSettings}.R
    B=@n52.2
   )
  )
  n62 = TextureSample!(UVs=Panner(Coordinate=Multiply(A=@n39) {constB=3}, Time=@n38) {speedX=-0.05, speedY=0.05}, Tex, Apply View MipBias) {
   samplerSource=SSM_Wrap_WorldGroupSettings
   samplerType=SAMPLERTYPE_Normal
   texture=&T_GaussianNoise_02_N
  }
  n63 = Multiply(A=AppendVector(A=@n62.R, B=@n62.R), B=ScalarParameter() {defaultValue=0.5, group=Fire, parameterName=Fire 3 Distortion, sortPriority=33}) {constB=0.02}
  n64 = ScalarParameter() {defaultValue=16, group=Fire, parameterName=Fire 3 Tiling, sortPriority=30}
  n65 = ScalarParameter() {defaultValue=0.37, group=Fire, parameterName=Fire 3 Aspect Ratio, sortPriority=31}
  n66 = Multiply(A=Add(A=@n44, B=Divide(A=@n63) {constB=100}), B=AppendVector(A=@n64, B=Multiply(A=@n64, B=@n65)))
  n67 = Multiply(A=NamedRerouteUsage(), B=ScalarParameter() {defaultValue=0.22, group=Fire, parameterName=Fire 3 Speed})
  n68 = TextureCoordinate()
  n69 = Multiply(
   A=TextureSample!(
    UVs=Panner(Coordinate=@n66, Time=@n67) {speedX=-0.1, speedY=1}
    Tex=@n48
    "DDX(UVs)"=DDX(Value=@n68)
    "DDY(UVs)"=DDY(Value=@n68)
    Apply View MipBias
   ) {mipValueMode=TMVM_Derivative, samplerSource=SSM_Wrap_WorldGroupSettings}.R
   B=@n52.3
  )
  n70 = Max(A=LinearInterpolate(
    A=ScalarParameter() {defaultValue=30, group=Fire, parameterName=Fire Exponent on Low, sortPriority=6}
    B=ScalarParameter() {defaultValue=2, group=Fire, parameterName=Fire Exponent on High, sortPriority=7}
    Alpha=Add(A=@n61, B=@n69)
   )) {constB=0}
  n71 = VectorParameter() {group=Fire, parameterName=Flame Color 2, sortPriority=9, defaultValue=<rgba 10 1.8747 0.8778 1>}
  n72 = VectorParameter() {group=Fire, parameterName=Flame Color, sortPriority=9, defaultValue=<rgba 5 1.0158 0.3669 1>}
  n73 = Power(Base=@n37)
  NamedRerouteDeclaration(None=Multiply(A=Power(Base=@n37, Exp=@n70), B=LinearInterpolate(A=@n71.RGB, B=@n72.RGB, Alpha=@n73))) {name=Fire, nodeColor=<rgba 1 0.3994 0 1>}
  n74 = NamedRerouteUsage()
  n75 = Constant2Vector() {g=1, r=4}
  n76 = TextureSample!(UVs=Panner(Coordinate=Multiply(A=@n74, B=@n75), Time=NamedRerouteUsage()) {speedX=0.1, speedY=0.1}, Tex, Apply View MipBias) {samplerType=SAMPLERTYPE_Normal, texture=&T_GaussianNoise_02_N}
  n77 = ScalarParameter() {defaultValue=0.02, group=Smoke, parameterName=Cloud Distortion}
  n78 = Multiply(A=ComponentMask(None=@n76.RGB) {r=true}, B=@n77) {constB=0.05}
  n79 = Constant()
  n80 = MaterialFunctionCall(Float2=NamedRerouteUsage()) {materialFunction=&BreakOutFloat2Components}
  n81 = ScalarParameter() {defaultValue=2, group=Smoke, parameterName=Smoke Tiling}
  n82 = ScalarParameter() {defaultValue=1, group=Smoke, parameterName=Smoke Aspect Ratio}
  n83 = Multiply(A=AppendVector(A=@n80.R, B=Power(Base=@n80.G) {constExponent=0.25}), B=AppendVector(A=@n81, B=Multiply(A=@n81, B=@n82)))
  n84 = Multiply(A=NamedRerouteUsage(), B=ScalarParameter() {defaultValue=0.25, group=Smoke, parameterName=Cloud Speed})
  n85 = VectorParameter() {
   group=Smoke
   parameterName=Cloud Speeds
   channelNames {a=Spark 2 Y, b=Spark 1 X, g=Spark 1 Y, r=Spark 1 X}
   defaultValue=<rgba 0.03 0.12 0.02 0.1>
  }
  n86 = Panner(Coordinate=Add(A=@n83, B=AppendVector(A=@n78, B=@n79)), Time=@n84, Speed=AppendVector(A=@n85.Spark 1 X, B=@n85.Spark 1 X)) {speedX=0.035, speedY=0.2}
  n87 = TextureCoordinate()
  n88 = TextureSample!(UVs=@n86, Tex, "DDX(UVs)"=DDX(Value=@n87), "DDY(UVs)"=DDY(Value=@n87), Apply View MipBias) {mipValueMode=TMVM_Derivative, samplerSource=SSM_Wrap_WorldGroupSettings, texture=/Game/Effects/Fort_Effects/Textures/Smoke/T_Clouds_01}
  n89 = AppendVector(A=Multiply(A=ComponentMask(None=@n76.RGB) {g=true}, B=@n77) {constB=0.05}, B=Constant())
  n90 = Panner(Coordinate=Multiply(A=Add(A=@n83, B=@n89)) {constB=2}, Time=@n84, Speed=AppendVector(A=@n85.Spark 1 X, B=@n85.Spark 1 X)) {speedX=0.02, speedY=0.1}
  n91 = TextureCoordinate()
  n92 = TextureSample(UVs=@n90, "DDX(UVs)"=DDX(Value=@n91), "DDY(UVs)"=DDY(Value=@n91)) {mipValueMode=TMVM_Derivative, samplerSource=SSM_Wrap_WorldGroupSettings, texture=&T_Clouds_01}
  n93 = Add(A=@n88.R, B=@n92.R)
  n94 = TextureCoordinate()
  n95 = TextureSample(UVs=@n86, "DDX(UVs)"=DDX(Value=@n94), "DDY(UVs)"=DDY(Value=@n94)) {
   mipValueMode=TMVM_Derivative
   samplerSource=SSM_Wrap_WorldGroupSettings
   samplerType=SAMPLERTYPE_Normal
   texture=/Game/Effects/Fort_Effects/Textures/Smoke/T_Clouds_01_M
  }
  n96 = TextureCoordinate()
  n97 = TextureSample(UVs=@n90, "DDX(UVs)"=DDX(Value=@n96), "DDY(UVs)"=DDY(Value=@n96)) {
   mipValueMode=TMVM_Derivative
   samplerSource=SSM_Wrap_WorldGroupSettings
   samplerType=SAMPLERTYPE_Normal
   texture=&T_Clouds_01_M
  }
  n98 = Saturate(None=Add(A=@n95.G, B=@n97.G))
  n99 = ScalarParameter() {defaultValue=2, group=Smoke, parameterName=Lighting Intensity}
  n100 = NamedRerouteUsage()
  n101 = Saturate(None=Multiply(A=ComponentMask(None=@n100) {g=true}) {constB=2})
  n102 = Add(A=Multiply(A=@n93, B=@n93), B=Multiply(A=Multiply(A=@n98, B=@n99) {constB=2}, B=Power(Base=@n101) {constExponent=5}))
  n103 = ComponentMask(None=NamedRerouteUsage()) {g=true}
  n104 = Power(Base=Saturate(None=Multiply(A=@n103) {constB=2})) {constExponent=3}
  NamedRerouteDeclaration(None=Multiply(A=Multiply(A=@n102, B=@n104), B=VectorParameter() {group=Smoke, parameterName=Smoke Color, defaultValue=<rgba 0.0156 0.0004 0 1>}.RGB)) {name=Smoke, nodeColor=<rgba 1 0.3764 0 1>}
  NamedRerouteDeclaration(None=@n7.Top Mask) {name=Top Mask, nodeColor=<rgba 1 0.375 0 1>}
  NamedRerouteDeclaration(None=@n7.Spherical Remapped UVs) {name=Spherical Remapped UVs, nodeColor=<rgba 1 0 0.0273 1>}
  n105 = NamedRerouteUsage()
  n106 = ComponentMask(None=@n105) {g=true}
  n107 = Multiply(
   A=Sine(None=ComponentMask(None=@n106) {r=true})
   B=Divide(A=ScalarParameter() {defaultValue=8, group=Global, parameterName=Spherical UV Compensaton}) {constB=100}
  )
  n108 = ComponentMask(None=@n106) {r=true}
  NamedRerouteDeclaration(None=AppendVector(A=ComponentMask(None=@n105) {r=true}, B=Add(A=@n107, B=@n108))) {name=Warped Spherical UVs, nodeColor=<rgba 1 0.091 0 1>}
  NamedRerouteUsage()
  NamedRerouteUsage()
 ]
}
