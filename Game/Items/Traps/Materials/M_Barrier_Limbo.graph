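# Unlit material for a "Limbo" trap barrier that draws an animated otherworldly backdrop with a floating video screen.
# Camera-based UVs map the backdrop and screen by view direction, so they read as distant scenery rather than surface texture.
# The backdrop layers noise-distorted ink ribbons, radial ink, flickering stars, gradient clouds and a horizon glow.
# A rounded-rectangle mask eroded by plasma noise blends in the streaming media texture, with a static switch for placeholder key art.
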Material /Game/Items/Traps/Materials/M_Barrier_Limbo
graph {
 properties {pixelDepthOffsetMode=PDOM_Legacy, shadingModel=MSM_Unlit, usedShadingModels=Unlit}
 nodes [
  n1 = NamedRerouteUsage()
  n2 = Constant() {r=0.25}
  n3 = NamedRerouteUsage()
  n4 = OneMinus(None=ComponentMask(None=@n3) {g=true})
  n5 = Power(Base=Add(A=Multiply(A=@n4)) {constB=-0.5})
  n6 = Multiply(A=ComponentMask(None=@n3) {g=true})
  n7 = Subtract(A=Constant() {r=0.25}, B=Power(Base=Add(A=@n6) {constB=-0.5}))
  n8 = ComponentMask(None=NamedRerouteUsage()) {g=true}
  n9 = LinearInterpolate(
   A=Power(Base=Subtract(A=@n2, B=@n5)) {constExponent=0.5}
   B=OneMinus(None=Power(Base=@n7) {constExponent=0.5})
   Alpha=Ceil(None=Subtract(A=@n8) {constB=0.5})
  )
  n10 = Multiply(A=AppendVector(A=ComponentMask(None=@n1) {r=true}, B=Multiply(A=@n9) {constB=2})) {constB=3}
  Add(
   A=TextureSample(UVs=@n10) {texture=/Game/FN_PublicTurntable/Turntable_Lowpoly/Content/BlueprintAssets/Checker}.RGB
   B=TextureSample(UVs=Multiply(A=@n10) {constB=10}) {texture=/Game/Effects/Fort_Effects/Textures/Mask/T_HalfSphere_D}.R
  )
  n11 = NamedRerouteUsage()
  n12 = MaterialFunctionCall!(
   UVs=@n11
   Radius=ScalarParameter() {defaultValue=0.2, parameterName=Inner Frame Radius, sliderMax=1}
   Aspect Ratio=NamedRerouteUsage()
   UV Scale from center
   Hardness
  ) {materialFunction=/Game/Packages/Fortress_SharedMaterials/Base_Material_Functions/MF_RoundedRectangle}
  n13 = Multiply(A=NamedRerouteUsage()) {constB=2}
  n14 = Panner(Coordinate=Multiply(A=@n13, B=Constant2Vector() {g=2, r=5})) {speedX=0.1, speedY=0.2}
  n15 = TextureCoordinate()
  n16 = DDX(Value=@n15)
  n17 = DDY(Value=@n15)
  n18 = Panner(Coordinate=Multiply(A=@n13, B=Constant2Vector() {g=3, r=4})) {speedX=-0.05, speedY=0.15}
  n19 = Subtract(
   A=Multiply(A=ScalarParameter() {defaultValue=2.4, parameterName=Video Frame Hardness}, B=SmoothStep(Value=@n12.Result)) {constB=10}
   B=Add(
    A=TextureSample!(UVs=@n14, Tex, "DDX(UVs)"=@n16, "DDY(UVs)"=@n17, Apply View MipBias) {mipValueMode=TMVM_Derivative, texture=/Game/Weapons/WeaponSkins/Wraps/Textures/PlasmaClouds/T_Wrap_PlasmaClouds__M}.G
    B=TextureSample(UVs=@n18, "DDX(UVs)"=@n16, "DDY(UVs)"=@n17) {mipValueMode=TMVM_Derivative, texture=&T_Wrap_PlasmaClouds__M}.B
   )
  )
  n20 = Multiply(A=NamedRerouteUsage(), B=Power(Base=Saturate(None=@n19)))
  n21 = TextureSample!(UVs=@n11, Tex=TextureObject() {texture=/Game/UI/Foundation/Textures/LoadingScreens/Fort_Load_SeasonKeyArt}, Apply View MipBias) {texture=&Fort_Load_SeasonKeyArt}
  n22 = MaterialFunctionCall!(UVs=@n11, Texture=TextureObject() {texture=/Game/UI/Foundation/Movie/StreamingTelevisionMediaTexture}) {materialFunction=/Game/Movies/InversePostProcess/MF_invertFNtonemap}
  LinearInterpolate(
   A=NamedRerouteUsage()
   B=Multiply(A=@n20, B=StaticSwitchParameter(True=@n21.RGB, False=@n22.Result) {parameterName=DEBUG Placeholder video texture})
   Alpha=@n20
  )
  Multiply(A=NamedRerouteUsage(), B=Constant2Vector() {g=1, r=2})
  n23 = Constant2Vector() {g=1, r=1.777}
  n24 = MaterialFunctionCall!(
   Camera Vector
   Local Space
   Spherical UV Remap Top
   Spherical UV Remap Bot
   Cylinder UV V Scale
   Planar UV Exponent
   Planar UV Dome
   Planar UV Scale
   Planar Radial UV Scale
   Quad UV H pos
   Quad UV V pos=ScalarParameter() {defaultValue=0.17, parameterName=Screen V Position}
   Quad UV Scale=Multiply(
    A=NamedRerouteDeclaration(None=@n23) {name=Aspect Ratio, nodeColor=<rgba 0.0547 0 1 1>}
    B=ScalarParameter() {defaultValue=0.6, parameterName=Scale}
   )
   Enable Rotation
   Rotation Vector
   -------------
   Mask UV
   Mask Size
  ) {materialFunction=/Game/Packages/Fortress_SharedMaterials/Base_Material_Functions/MF_CameraBasedUVs}
  NamedRerouteDeclaration(None=@n24.Quad UV) {name=Screen UV, nodeColor=<rgba 1 0 0.0159 1>}
  NamedRerouteDeclaration(None=@n24.Quad UV Mask) {name=Screen UV Mask, nodeColor=<rgba 1 0 0.3185 1>}
  NamedRerouteDeclaration(None=VectorParameter() {parameterName=Background Color 2, defaultValue=<rgba 0.0716 0.0502 0.1562 0>}.RGB) {name=Background Color 2, nodeColor=<rgba 1 0.0469 0 1>}
  n25 = Power(Base=ComponentMask(None=NamedRerouteUsage()) {g=true})
  n26 = Saturate(None=LinearInterpolate(Alpha=OneMinus(None=@n25)) {constA=-0.1})
  n27 = NamedRerouteUsage()
  n28 = Multiply(A=@n27, B=Constant2Vector() {g=3, r=6}) {constB=6}
  n29 = AppendVector(A=Constant(), B=Sine(None=ComponentMask(None=@n27) {g=true}))
  n30 = ScalarParameter() {parameterName=aaa}
  n31 = Ceil(None=ComponentMask(None=@n28) {r=true})
  n32 = ScalarParameter() {parameterName=strand offset}
  n33 = Add(A=Add(A=@n28, B=Multiply(A=@n29, B=@n30)), B=AppendVector(A=Constant(), B=Multiply(A=@n31, B=@n32)))
  n34 = Multiply(A=NamedRerouteUsage(), B=ScalarParameter() {defaultValue=1, parameterName=Distortion Tiling 1})
  n35 = Multiply(A=Time(), B=ScalarParameter() {defaultValue=0.1, parameterName=Distorton Speed 1})
  n36 = TextureCoordinate()
  n37 = TextureSample!(
   UVs=Multiply(A=Panner(Coordinate=@n34, Time=@n35) {speedX=0.01, speedY=-0.1}, B=Constant2Vector() {g=1, r=1})
   Tex
   "DDX(UVs)"=DDX(Value=@n36)
   "DDY(UVs)"=DDY(Value=@n36)
   Apply View MipBias
  ) {
   mipValueMode=TMVM_Derivative
   samplerSource=SSM_Wrap_WorldGroupSettings
   samplerType=SAMPLERTYPE_LinearColor
   texture=/Game/Effects/Fort_Effects/Textures/Tile/T_PerlinRGB_Displacement
  }
  n38 = Multiply(A=Subtract(A=Multiply(A=@n37.RGB) {constB=2}), B=ScalarParameter() {defaultValue=0.3, parameterName=Ribbon Distortion 1})
  n39 = Multiply(A=NamedRerouteUsage(), B=ScalarParameter() {defaultValue=6, parameterName=Distortion Tiling 2})
  n40 = Multiply(A=Time(), B=ScalarParameter() {defaultValue=3, parameterName=Distorton Speed 2})
  n41 = TextureCoordinate()
  n42 = TextureSample(
   UVs=Multiply(A=Panner(Coordinate=@n39, Time=@n40) {speedX=0.01, speedY=-0.1}, B=Constant2Vector() {g=1, r=3})
   "DDX(UVs)"=DDX(Value=@n41)
   "DDY(UVs)"=DDY(Value=@n41)
  ) {
   mipValueMode=TMVM_Derivative
   samplerSource=SSM_Wrap_WorldGroupSettings
   samplerType=SAMPLERTYPE_LinearColor
   texture=&T_PerlinRGB_Displacement
  }
  n43 = Multiply(A=Subtract(A=Multiply(A=@n42.RGB) {constB=2}), B=ScalarParameter() {defaultValue=0.01, parameterName=Ribbon Distortion 2})
  n44 = MaterialFunctionCall(Float3=NamedRerouteDeclaration(None=Add(A=@n38, B=@n43)) {name=Distort Texture, nodeColor=<rgba 1 0 0.9844 1>}) {materialFunction=&BreakOutFloat3Components}
  n45 = AppendVector(A=@n44.R, B=Constant())
  n46 = Multiply(A=Time(), B=ScalarParameter() {defaultValue=0.2, parameterName=Ink Scroll Speed})
  n47 = TextureObject() {texture=/Game/Items/Traps/Textures/T_Limbo_Ink_01_M}
  n48 = TextureCoordinate()
  n49 = TextureSample(UVs=Panner(Coordinate=Add(A=@n33, B=@n45), Time=@n46) {speedY=0.5}, Tex=@n47, "DDX(UVs)"=DDX(Value=@n48), "DDY(UVs)"=DDY(Value=@n48)) {mipValueMode=TMVM_Derivative, samplerSource=SSM_Wrap_WorldGroupSettings, texture=/Game/Effects/Fort_Effects/Textures/T_InkySmokeTrail}
  n50 = NamedRerouteUsage()
  n51 = TextureCoordinate()
  n52 = TextureSample(
   UVs=Multiply(A=Panner(Coordinate=@n50) {speedX=0.1, speedY=0.15}, B=Constant2Vector() {g=0.5, r=1})
   "DDX(UVs)"=DDX(Value=@n51)
   "DDY(UVs)"=DDY(Value=@n51)
  ) {mipValueMode=TMVM_Derivative, samplerType=SAMPLERTYPE_LinearColor, texture=&T_PerlinRGB_Displacement}
  n53 = AppendVector(A=@n44.G, B=Constant())
  n54 = TextureCoordinate()
  n55 = TextureSample(UVs=Panner(Coordinate=Add(A=@n33, B=@n53), Time=@n46) {speedY=0.75}, Tex=@n47, "DDX(UVs)"=DDX(Value=@n54), "DDY(UVs)"=DDY(Value=@n54)) {mipValueMode=TMVM_Derivative, samplerSource=SSM_Wrap_WorldGroupSettings, texture=&T_InkySmokeTrail}
  n56 = Add(A=Multiply(A=@n49.R, B=Power(Base=@n52.R) {constExponent=3}), B=Multiply(A=@n55.G, B=Power(Base=@n52.G) {constExponent=3}))
  n57 = AppendVector(A=@n44.B, B=Constant())
  n58 = TextureCoordinate()
  n59 = TextureSample(UVs=Panner(Coordinate=Add(A=@n33, B=@n57), Time=@n46) {speedY=1}, Tex=@n47, "DDX(UVs)"=DDX(Value=@n58), "DDY(UVs)"=DDY(Value=@n58)) {mipValueMode=TMVM_Derivative, samplerSource=SSM_Wrap_WorldGroupSettings, texture=&T_InkySmokeTrail}
  n60 = Multiply(A=@n59.B, B=Power(Base=@n52.B) {constExponent=3})
  n61 = NamedRerouteUsage()
  n62 = Constant() {r=0.5}
  n63 = Constant() {r=1}
  n64 = Constant() {r=0.5}
  n65 = NamedRerouteUsage()
  n66 = Constant2Vector() {g=0.3, r=3}
  n67 = TextureSample!(UVs=Panner(Coordinate=Multiply(A=@n65, B=@n66)) {speedX=0.1, speedY=0.2}, Tex, Apply View MipBias) {texture=/Game/Items/Traps/Textures/T_PerlinRGB_01_M}
  n68 = Add(A=NamedRerouteUsage()) {constB=-0.5}
  n69 = Multiply(
   A=NamedRerouteDeclaration(None=MaterialFunctionCall(Vector or UVs=@n68) {materialFunction=&VectorToRadialValue}.Radial Coordinates) {name=Screen Radial UV, nodeColor=<rgba 0.6406 1 0 1>}
   B=Constant2Vector() {g=0.1, r=1}
  ) {constB=4}
  n70 = TextureCoordinate()
  n71 = DDX(Value=@n70)
  n72 = DDY(Value=@n70)
  n73 = TextureSample!(UVs=Panner(Coordinate=@n69) {speedX=0.02, speedY=0.05}, Tex, "DDX(UVs)"=@n71, "DDY(UVs)"=@n72, Apply View MipBias) {mipValueMode=TMVM_Derivative, texture=/Game/Items/Traps/Textures/T_Limbo_Ink_02_M}
  n74 = TextureSample(UVs=Panner(Coordinate=Multiply(A=@n69) {constB=2}) {speedX=-0.01, speedY=0.03}, "DDX(UVs)"=@n71, "DDY(UVs)"=@n72) {mipValueMode=TMVM_Derivative, texture=&T_Limbo_Ink_02_M}
  n75 = Multiply(
   A=Power(Base=Add(A=@n73.R, B=@n74.R))
   B=Power(Base=SphereMask(A=@n61, B=@n62, Radius=@n63, Hardness=@n64) {hardnessPercent=1}, Exp=LinearInterpolate(Alpha=@n67.R) {constA=3, constB=10})
  )
  n76 = Add(A=Multiply(A=SmoothStep(Value=@n26), B=Add(A=@n56, B=@n60)), B=Multiply(A=Multiply(A=@n75) {constB=3}, B=NamedRerouteUsage()))
  n77 = LinearInterpolate(A=NamedRerouteUsage(), B=NamedRerouteUsage(), Alpha=Multiply(A=@n52.B, B=@n52.B))
  n78 = TextureObject() {texture=/Game/Items/Traps/Textures/T_Limbo_Stars_M}
  n79 = ComponentMask(None=MaterialFunctionCall!(
    contrast=ScalarParameter() {defaultValue=3, group=Environment - Stars, parameterName=Stars Contrast}
    UV Scale=ScalarParameter() {defaultValue=3, group=Environment - Stars, parameterName=Stars UV}
    UV Distortion
    In=@n78
   ) {materialFunction=/Game/Items/Traps/Materials/MaterialFunctions/MF_BarrierTriplanar}.Result) {b=true}
  n80 = Multiply(A=NamedRerouteUsage(), B=Multiply(A=ScalarParameter() {defaultValue=3, parameterName=Big Star UV}, B=Constant2Vector() {g=1, r=2}))
  n81 = TextureSample!(UVs=@n80, Tex=@n78, Apply View MipBias) {texture=&T_Limbo_Stars_M}
  n82 = Time()
  n83 = Add(A=@n81.G, B=Multiply(A=@n82, B=ScalarParameter() {defaultValue=0.1, parameterName=Big Star Speed}))
  n84 = AppendVector(A=Sine(None=@n83), B=Cosine(None=@n83))
  n85 = ScalarParameter() {defaultValue=0.03, parameterName=Big Star Movement}
  n86 = TextureSample(UVs=Add(A=@n80, B=Multiply(A=@n84, B=@n85) {constB=0})) {texture=&T_Limbo_Stars_M}
  n87 = ScalarParameter() {defaultValue=2, parameterName=Big Speckles}
  n88 = Multiply(A=@n82, B=ScalarParameter() {defaultValue=0.17, parameterName=Big Flicker Star Speed})
  n89 = Saturate(None=Sine(None=Add(A=@n81.G, B=@n88)))
  n90 = ComponentMask(None=NamedRerouteUsage()) {g=true}
  n91 = Constant() {r=0.5}
  n92 = Constant()
  n93 = Multiply(A=Multiply(A=Multiply(A=@n86.R, B=@n87), B=SmoothStep(Value=@n89)), B=Power(Base=SphereMask(A=@n90, B=@n91, Radius=@n91, Hardness=@n92)))
  n94 = Add(
   A=Multiply(A=Multiply(A=@n76, B=@n77), B=ScalarParameter() {defaultValue=2, parameterName=Ribbon Intensity})
   B=Multiply(A=Add(A=@n79, B=@n93), B=VectorParameter() {parameterName=Speckles color, defaultValue=<rgba 1 0.3344 0.1028 1>}.RGB)
  )
  n95 = Multiply(A=NamedRerouteUsage(), B=Multiply(A=Constant2Vector() {g=1, r=2}, B=ScalarParameter() {defaultValue=2, parameterName=Background Tiling 1}))
  n96 = Multiply(A=@n95, B=ScalarParameter() {defaultValue=1.5, parameterName=Background Tiling 2})
  n97 = Add(
   A=TextureSample(UVs=Panner(Coordinate=@n95) {speedX=0.005, speedY=0.02}) {texture=&T_Limbo_Ink_02_M}.R
   B=TextureSample(UVs=Panner(Coordinate=@n96) {speedX=0.005, speedY=0.02}) {texture=&T_Limbo_Ink_02_M}.R
  )
  n98 = Multiply(A=ComponentMask(None=NamedRerouteUsage()) {g=true}) {constB=2}
  n99 = Multiply(A=@n97, B=Saturate(None=OneMinus(None=@n98)))
  n100 = LinearInterpolate(
   A=LinearInterpolate(A=NamedRerouteUsage(), B=NamedRerouteUsage(), Alpha=@n99)
   B=NamedRerouteUsage()
   Alpha=Power(Base=@n99) {constExponent=3}
  )
  n101 = NamedRerouteUsage()
  n102 = NamedRerouteUsage()
  n103 = Abs(None=LinearInterpolate(Alpha=ComponentMask(None=@n102) {g=true}) {constA=-1})
  n104 = ScalarParameter() {defaultValue=45, parameterName=Horizon Exponent Min}
  n105 = ScalarParameter() {defaultValue=70, parameterName=Horizon Exponent Max}
  n106 = ComponentMask(None=@n97) {r=true}
  n107 = OneMinus(None=ComponentMask(None=NamedRerouteUsage()) {g=true})
  n108 = Multiply(A=Power(Base=OneMinus(None=@n103), Exp=LinearInterpolate(A=@n104, B=@n105, Alpha=@n106)) {constExponent=6}, B=Multiply(A=@n107, B=@n107))
  NamedRerouteDeclaration(None=Add(A=Add(A=@n94, B=@n100), B=Multiply(A=@n101, B=@n108)) {constB=0}) {name=Background, nodeColor=<rgba 0.5234 0 1 1>}
  NamedRerouteDeclaration(None=@n24.Spherical UVs) {name=Sphere UVs, nodeColor=<rgba 1 0 0.0866 1>}
  NamedRerouteDeclaration(None=@n24.Top Mask) {name=Top Mask, nodeColor=<rgba 0 1 0.4141 1>}
  NamedRerouteDeclaration(None=VectorParameter() {parameterName=Horizon Color, defaultValue=<rgba 1 0.3181 0.5308 0>}.RGB) {name=Horizon Color, nodeColor=<rgba 1 0 0.7734 1>}
  NamedRerouteDeclaration(None=VectorParameter() {parameterName=Background Color 3, defaultValue=<rgba 0.2063 0.604 0.8333 0>}.RGB) {name=Background Color 3, nodeColor=<rgba 1 0.0469 0 1>}
  n109 = VectorParameter() {parameterName="Rotation Vector (V3)", defaultValue=<rgba 1 0 0.15 0>}
  n110 = MaterialFunctionCall(Enable Rotation=StaticBool() {value=true}, Rotation Vector=@n109.RGB) {materialFunction=&MF_CameraBasedUVs}
  n111 = Constant2Vector() {r=0.5}
  NamedRerouteDeclaration(None=Frac(None=Add(A=@n110.Spherical UVs, B=@n111))) {name=Spherical Directional UV, nodeColor=<rgba 0 1 0.3672 1>}
  n112 = Multiply(A=DotProduct(A=CameraVectorWS(), B=@n109.RGB)) {constB=-1}
  NamedRerouteDeclaration(None=Saturate(None=Ceil(None=@n112))) {name=Front Side Mask, nodeColor=<rgba 0 0.9062 1 1>}
  NamedRerouteDeclaration(None=VectorParameter() {parameterName=Ribbon Color 1, defaultValue=<rgba 2 0.2878 0.2281 0>}.RGB) {name=Ribbon Color 1, nodeColor=<rgba 1 0.4922 0 1>}
  NamedRerouteDeclaration(None=VectorParameter() {parameterName=Ribbon Color 2, defaultValue=<rgba 1.1584 0.4363 2 0>}.RGB) {name=Ribbon Color 2, nodeColor=<rgba 1 0.4922 0 1>}
  NamedRerouteDeclaration(None=VectorParameter() {parameterName=Background Color 1, defaultValue=<rgba 0.0076 0.0066 0.0156 0>}.RGB) {name=Background Color 1, nodeColor=<rgba 1 0.0469 0 1>}
  TextureSample() {texture=&T_Wrap_PlasmaClouds__M}
  TextureSample() {texture=/Game/Effects/Fort_Effects/Textures/T_FX_SpringerPlasma}
  TextureSample() {texture=/Game/Effects/Fort_Effects/Textures/Tile/T_Inky_Smoke_Tile2}
  TextureSample() {texture=/Game/Effects/Fort_Effects/Textures/Tile/T_PerlinRGB_01_M}
  TextureSample() {texture=/Game/Effects/Fort_Effects/Textures/Plasma/T_Plasma_03_M}
  TextureSample() {texture=/Game/Effects/Fort_Effects/Textures/Smoke/T_Clouds_02_M}
 ]
}
