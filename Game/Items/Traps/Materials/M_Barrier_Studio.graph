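# M_Barrier_Studio is an unlit material for a trap barrier that paints an animated space scene with planets, black hole, beam, stars, clouds and streaks.
# MF_CameraBasedUVs supplies spherical, cylindrical and quad UVs from the camera vector, optionally in local space.
# Layers use panned noise, triplanar blinking stars and a UV grid scattering small planets.
# Named reroutes carry layers and masks into a final sum, multiplied by a global Tint.
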Material /Game/Items/Traps/Materials/M_Barrier_Studio
graph {
 properties {pixelDepthOffsetMode=PDOM_Legacy, shadingModel=MSM_Unlit, usedShadingModels=Unlit}
 nodes [
  n1 = Multiply(
   A=NamedRerouteUsage()
   B=LinearInterpolate(A=ScalarParameter() {defaultValue=0.7, group=PLANETS - GENERAL, parameterName=Planet Background Cloud}, Alpha=NamedRerouteUsage())
  )
  n2 = OneMinus(None=NamedRerouteUsage())
  n3 = Multiply(A=Multiply(A=NamedRerouteUsage(), B=NamedRerouteUsage()), B=NamedRerouteUsage())
  n4 = NamedRerouteUsage()
  n5 = Add(A=Add(A=Multiply(A=@n1, B=@n2), B=Multiply(A=@n3, B=@n4)), B=NamedRerouteUsage())
  n6 = NamedRerouteUsage()
  n7 = Add(A=Add(A=Add(A=@n5, B=@n6), B=NamedRerouteUsage()), B=NamedRerouteUsage())
  n8 = NamedRerouteUsage()
  n9 = VectorParameter() {group=GLOBAL, parameterName=Tint, channelNames {a=Intensity, b="", g="", r=""}, defaultValue=<rgba 0.145 0.6838 1 5>}
  Multiply(A=Add(A=Add(A=@n7, B=@n8), B=NamedRerouteUsage()), B=Multiply(A=@n9, B=@n9))
  n10 = ScalarParameter() {defaultValue=0.27, group=UVs, parameterName=Remap Sphere UV}
  n11 = ScalarParameter() {group=UVs, parameterName=Quad UV Scale Position H}
  n12 = MaterialFunctionCall!(
   Camera Vector
   Local Space=StaticBoolParameter() {group=GLOBAL, parameterName=Local Space}
   Spherical UV Remap Top=@n10
   Spherical UV Remap Bot=@n10
   Cylinder UV V Scale=ScalarParameter() {defaultValue=1, group=UVs, parameterName=Cylinder UV Scale}
   Planar UV Exponent
   Planar UV Dome
   Planar UV Scale
   Planar Radial UV Scale
   Quad UV H pos=NamedRerouteDeclaration(None=@n11) {name=Center Planet H Position, nodeColor=<rgba 1 0 0.9844 1>}
   Quad UV V pos=ScalarParameter() {group=UVs, parameterName=Quad UV Scale Position V}
   Quad UV Scale=ScalarParameter() {defaultValue=1, group=UVs, parameterName=Quad UV Scale}
   Enable Rotation
   Rotation Vector
   -------------
   Mask UV
   Mask Size
  ) {materialFunction=/Game/Packages/Fortress_SharedMaterials/Base_Material_Functions/MF_CameraBasedUVs}
  NamedRerouteDeclaration(None=@n12.Camera Vector) {name=Cam Vector, nodeColor=<rgba 0.6641 0 1 1>}
  NamedRerouteDeclaration(None=@n12.Cylinder UVs) {name=Cylinder UVs, nodeColor=<rgba 1 0 0.0596 1>}
  n13 = MaterialFunctionCall(UVs=NamedRerouteUsage(), Texture Scale=ScalarParameter() {defaultValue=0.756, group=PLANETS - CENTER, parameterName=Planet Scale}) {materialFunction=&ScaleUVsByCenter}
  n14 = Constant() {r=0.5}
  n15 = Constant() {r=0.44}
  n16 = Constant() {r=0.99}
  n17 = NamedRerouteUsage()
  n18 = Saturate(None=Multiply(A=SphereMask(A=@n13.UVs, B=@n14, Radius=@n15, Hardness=@n16), B=@n17))
  n19 = Add(
   A=NamedRerouteUsage()
   B=AppendVector(A=ScalarParameter() {defaultValue=0.056, group=PLANETS - SMALL, parameterName=Planets UV Rotation}, B=Constant())
  )
  n20 = Floor(None=ScalarParameter() {defaultValue=16, group=PLANETS - SMALL, parameterName=Count X})
  n21 = ScalarParameter() {defaultValue=4, group=PLANETS - SMALL, parameterName=Count Y}
  n22 = Floor(None=@n21)
  n23 = Multiply(A=Subtract(A=@n21)) {constB=-1}
  n24 = TextureSample(UVs=@n19) {samplerType=SAMPLERTYPE_LinearColor, texture=/Game/Items/Traps/Textures/T_Barrier_Mask_01_M}
  n25 = Add(A=Multiply(A=@n19, B=AppendVector(A=@n20, B=@n22)), B=AppendVector(A=Constant(), B=Multiply(A=@n23, B=@n24.G)))
  n26 = MaterialFunctionCall(
   UVs=Frac(None=@n25)
   Texture Scale=LinearInterpolate(A=ScalarParameter() {defaultValue=0.5, group=PLANETS - SMALL, parameterName=Min Planet Scale}, Alpha=@n24.B)
  ) {materialFunction=&ScaleUVsByCenter}
  n27 = TextureCoordinate()
  n28 = DDX(Value=@n27)
  n29 = DDY(Value=@n27)
  n30 = Abs(None=Subtract(A=ComponentMask(None=@n25) {g=true}) {constB=0.5})
  n31 = Saturate(None=Floor(None=Multiply(A=@n30) {constB=2}))
  n32 = Multiply(
   A=Multiply(
    A=TextureSample!(UVs=@n26.UVs, Tex, "DDX(UVs)"=@n28, "DDY(UVs)"=@n29, Apply View MipBias) {mipValueMode=TMVM_Derivative, texture=/Game/Items/Traps/Textures/T_Barrier_Planet_02_M}.RGB
    B=@n26.0-1 mask
   )
   B=Multiply(A=Ceil(None=@n24.G), B=OneMinus(None=@n31))
  )
  n33 = Ceil(None=ComponentMask(None=@n32) {r=true})
  NamedRerouteDeclaration(None=OneMinus(None=Max(A=@n18, B=@n33))) {name=Planet Mask, nodeColor=<rgba 1 0.2152 0.0955 1>}
  n34 = NamedRerouteDeclaration(None=OneMinus(None=@n18)) {name=Center Planet Mask, nodeColor=<rgba 1 0 0.5625 1>}
  n35 = Multiply(A=@n34, B=ComponentMask(None=@n32) {g=true})
  n36 = Multiply(A=@n17, B=@n13.0-1 mask)
  n37 = TextureSample!(UVs=@n13.UVs, Tex, Apply View MipBias) {texture=/Game/Items/Traps/Textures/T_Barrier_Planet_01_M}
  n38 = NamedRerouteDeclaration(None=Multiply(A=@n36, B=@n37.G)) {name=Center Planet Glow, nodeColor=<rgba 0.1484 0 1 1>}
  n39 = Multiply(A=@n34, B=ComponentMask(None=@n32) {r=true})
  n40 = Multiply(A=Multiply(A=@n36, B=@n37.R), B=ScalarParameter() {defaultValue=1, group=PLANETS - CENTER, parameterName=Center Planet Brightness})
  n41 = Add(
   A=Multiply(
    A=Add(A=@n38, B=@n35)
    B=VectorParameter() {group=PLANETS - GENERAL, parameterName=Planet Glow Color, defaultValue=<rgba 0.9352 0.9635 0.9423 0>}.RGB
   )
   B=Multiply(
    A=Add(A=@n39, B=@n40)
    B=VectorParameter() {group=PLANETS - GENERAL, parameterName=Planet Color, defaultValue=<rgba 0.2193 0.2865 0.1121 0>}.RGB
   )
  )
  n42 = ScalarParameter() {defaultValue=0.1, group=PLANETS - GENERAL, parameterName=Planet Intesity}
  n43 = NamedRerouteUsage()
  n44 = Time()
  n45 = ScalarParameter() {defaultValue=0.5, group=PLANETS - CENTER, parameterName=Dot Brightness Speed}
  n46 = TextureSample(UVs=Panner(Coordinate=@n43, Time=Multiply(A=@n44, B=@n45)) {speedX=0.1, speedY=0.2}) {texture=/Game/Global/Textures/Default/Noise/3xPerlin_M}
  n47 = LinearInterpolate(Alpha=OneMinus(None=ComponentMask(None=@n43) {g=true})) {constA=-0.2, constB=1.2}
  n48 = NamedRerouteUsage()
  n49 = Constant() {r=0.5}
  n50 = Multiply(A=Time(), B=ScalarParameter() {defaultValue=0.005, group=PLANETS - CENTER, parameterName=Dot rotation speed})
  n51 = Constant() {r=0.5}
  n52 = Multiply(
   A=TextureSample(UVs=MaterialFunctionCall(UVs=@n48, Rotation Center=@n49, "Rotation Angle (0-1)"=@n50) {materialFunction=&CustomRotator}.Rotated Values) {texture=&T_Barrier_Planet_01_M}.B
   B=SphereMask(A=NamedRerouteUsage(), B=@n51, Radius=@n51)
  )
  n53 = NamedRerouteUsage()
  n54 = Multiply(A=Multiply(A=Multiply(A=@n46.R, B=@n46.R), B=Saturate(None=@n47)), B=Power(Base=Multiply(A=@n52, B=@n53)))
  n55 = VectorParameter() {group=PLANETS - CENTER, parameterName=Dots Color, defaultValue=<rgba 1000 539.1445 357.451 0>}
  NamedRerouteDeclaration(None=Add(A=Multiply(A=@n54, B=@n55.RGB), B=Multiply(A=@n41, B=@n42))) {name=Planets, nodeColor=<rgba 1 0.4453 0 1>}
  NamedRerouteDeclaration(None=@n12.Quad UV) {name=Quad UV, nodeColor=<rgba 1 0 0.1406 1>}
  NamedRerouteDeclaration(None=@n12.Quad UV Mask) {name=Quad UV Mask, nodeColor=<rgba 1 0 0.5156 1>}
  n56 = Divide(B=ScalarParameter() {defaultValue=0.3, group=BEAM, parameterName=Beam UV Width})
  n57 = Subtract(
   A=MaterialFunctionCall(Local Space=StaticBoolParameter() {group=GLOBAL, parameterName=Local Space}) {materialFunction=&MF_CameraBasedUVs}.Cylinder UVs
   B=AppendVector(A=NamedRerouteUsage(), B=Constant())
  )
  n58 = LinearInterpolate(A=Multiply(A=@n56) {constB=-1}, B=Add(A=@n56), Alpha=@n57) {constA=-10, constB=11}
  n59 = Subtract(A=@n58) {constB=0.5}
  n60 = Floor(None=Multiply(A=Abs(None=@n59)) {constB=2})
  n61 = Subtract(A=Saturate(None=@n58)) {constB=0.25}
  n62 = Multiply(A=Saturate(None=OneMinus(None=@n60)), B=ConstantBiasScale(None=Sine(None=@n61)))
  n63 = ComponentMask(None=NamedRerouteUsage()) {g=true}
  n64 = Power(Base=Saturate(None=Multiply(A=@n63) {constB=2})) {constExponent=20}
  n65 = Multiply(A=@n62, B=Multiply(A=NamedRerouteUsage(), B=OneMinus(None=@n64)))
  n66 = Multiply(A=Power(Base=@n65), B=ScalarParameter() {defaultValue=0.7, group=BEAM, parameterName=Beam Glow Intensity})
  n67 = Multiply(A=NamedRerouteUsage(), B=ScalarParameter() {defaultValue=1, group=BEAM, parameterName=Center Planet Glow})
  n68 = VectorParameter() {group=BEAM, parameterName=Beam Streaks UVs, channelNames {a=UV2 Y, b=UV1 X, g=UV1 Y, r=UV1 X}, defaultValue=<rgba 4 1 8 3>}
  n69 = Multiply(A=AppendVector(A=@n68.UV1 X, B=@n68.UV1 X), B=@n57) {constB=10}
  n70 = Multiply(A=ScalarParameter() {defaultValue=0.3, group=BEAM, parameterName=Beam Streaks}, B=Time())
  n71 = AppendVector(A=Constant(), B=Multiply(A=ScalarParameter() {defaultValue=0.1, group=BEAM, parameterName=Vertical Distortion}, B=@n62) {constA=10})
  n72 = TextureSample(UVs=Add(A=Panner(Coordinate=@n69, Time=@n70) {speedY=1}, B=@n71)) {texture=/Game/Items/Traps/Textures/T_Barrier_Streaks_01_M}
  n73 = Multiply(A=AppendVector(A=@n68.UV1 X, B=@n68.UV1 X), B=@n57) {constB=8}
  n74 = TextureSample(UVs=Add(A=Panner(Coordinate=@n73, Time=@n70) {speedY=1}, B=@n71)) {texture=&T_Barrier_Streaks_01_M}
  n75 = Add(A=Add(A=@n72.R, B=@n72.R), B=Add(A=@n74.G, B=@n74.G))
  n76 = Subtract(A=@n75, B=NamedRerouteUsage())
  n77 = Multiply(A=Add(A=@n75, B=Max(A=@n76) {constB=0}), B=@n65)
  n78 = ScalarParameter() {defaultValue=2, group=BEAM, parameterName=Beam Streaks Intensity}
  n79 = NamedRerouteUsage()
  n80 = Saturate(None=Multiply(A=ComponentMask(None=@n79) {g=true}) {constB=2})
  n81 = Multiply(A=Add(A=Multiply(A=@n77, B=@n78), B=Add(A=@n66, B=@n67)), B=Saturate(None=LinearInterpolate(Alpha=@n80) {constA=-0.2}))
  NamedRerouteDeclaration(None=Multiply(A=Power(Base=@n81), B=ScalarParameter() {defaultValue=1, group=BEAM, parameterName=Beam General Intensity})) {name=Beam, nodeColor=<rgba 0 0.8125 1 1>}
  n82 = NamedRerouteUsage()
  n83 = LinearInterpolate(
   A=VectorParameter() {group=BACKGROUND, parameterName=Color Top, defaultValue=<rgba 0.3385 0.2078 0.1283 0>}.RGB
   B=VectorParameter() {group=BACKGROUND, parameterName=Color Bot, defaultValue=<rgba 0.3594 0.2684 0.2134 0>}.RGB
   Alpha=Power(Base=ComponentMask(None=@n82) {g=true})
  )
  n84 = ScalarParameter() {defaultValue=0.01, group=BACKGROUND, parameterName=Background Multiplier}
  n85 = ScalarParameter() {defaultValue=0.1, group=BACKGROUND, parameterName=BG Color on Blackhole}
  n86 = Power(
   Base=OneMinus(None=NamedRerouteUsage())
   Exp=ScalarParameter() {defaultValue=15, group=BACKGROUND, parameterName=BG Color Blackhole Exponent}
  ) {constExponent=5}
  NamedRerouteDeclaration(None=Multiply(A=Multiply(A=@n83, B=@n84) {constB=0.001}, B=LinearInterpolate(A=@n85, Alpha=@n86))) {name=Background, nodeColor=<rgba 0 1 0.3203 1>}
  n87 = MaterialFunctionCall(
   Local Space=StaticBoolParameter() {group=GLOBAL, parameterName=Local Space}
   Quad UV H pos=Constant() {r=0.2}
   Quad UV V pos=Constant() {r=0.999}
   Quad UV Scale=Constant() {r=1}
  ) {materialFunction=&MF_CameraBasedUVs}
  n88 = AppendVector(A=ComponentMask(None=@n87.Quad UV) {r=true}, B=OneMinus(None=ComponentMask(None=@n87.Quad UV) {g=true}))
  n89 = TextureSample(UVs=@n88) {texture=&T_Barrier_Planet_01_M}
  n90 = SphereMask(A=@n88, B=Constant() {r=0.5}, Radius=Constant() {r=0.44}, Hardness=Constant())
  n91 = Multiply(A=Saturate(None=Multiply(A=@n90) {constB=1000}), B=OneMinus(None=NamedRerouteUsage()))
  NamedRerouteDeclaration(None=Multiply(
    A=Multiply(A=@n89.R, B=@n91)
    B=VectorParameter() {group=PLANETS - BOTTOM, parameterName=Bottom Planet Colot, defaultValue=<rgba 0.2812 0.16 0.1319 0>}.RGB
   )) {name=Bottom Planet, nodeColor=<rgba 0.8281 0 1 1>}
  n92 = NamedRerouteUsage()
  n93 = ScalarParameter() {defaultValue=1, group=STREAKS, parameterName=Streak Tiling V, sortPriority=0}
  n94 = ScalarParameter() {defaultValue=4, group=STREAKS, parameterName=Streak Tiling U, sortPriority=1}
  n95 = MaterialFunctionCall(Float2=Multiply(
    A=MaterialFunctionCall(XYZ=@n92, XY=@n92) {materialFunction=&Swizzle}.YX
    B=AppendVector(A=@n93, B=@n94)
   )) {materialFunction=&BreakOutFloat2Components}
  n96 = AppendVector(A=@n95.G, B=@n95.G)
  n97 = VectorParameter() {group=STREAKS, parameterName=Streak Tilings, sortPriority=10, channelNames {a="", b=Far, g=Mid, r=Near}, defaultValue=<rgba 1 3 6 0>}
  n98 = Multiply(
   A=ScalarParameter() {defaultValue=0.2, group=STREAKS, parameterName=Streaks Global Speed, sortPriority=3}
   B=Time() {bOverride_Period=true, period=7200}
  )
  n99 = VectorParameter() {group=STREAKS, parameterName=Streak Speeds, sortPriority=11, channelNames {a="", b=Far, g=Mid, r=Near}, defaultValue=<rgba 1 0.65 0.35 0>}
  n100 = TextureCoordinate()
  n101 = DDX(Value=@n100)
  n102 = DDY(Value=@n100)
  n103 = TextureSample!(
   UVs=Panner(Coordinate=Multiply(A=@n96, B=@n97.Near), Time=Multiply(A=@n98, B=@n99.Near)) {speedY=0.5}
   Tex
   "DDX(UVs)"=@n101
   "DDY(UVs)"=@n102
   Apply View MipBias
  ) {mipValueMode=TMVM_Derivative, samplerSource=SSM_Wrap_WorldGroupSettings, texture=&T_Barrier_Streaks_01_M}
  n104 = Abs(None=ComponentMask(None=NamedRerouteUsage()) {b=true})
  n105 = Saturate(None=LinearInterpolate(Alpha=OneMinus(None=@n104)) {constA=-0.05, constB=1.5})
  n106 = VectorParameter() {group=STREAKS, parameterName=Streak Exponents, sortPriority=12, channelNames {a="", b=Far, g=Mid, r=Near}, defaultValue=<rgba 0.35 0.75 1 0>}
  n107 = VectorParameter() {group=STREAKS, parameterName=Streak Intensity, sortPriority=13, channelNames {a="", b=Far, g=Mid, r=Near}, defaultValue=<rgba 1 0.7 0.35 0>}
  n108 = Multiply(A=Multiply(A=@n103.R, B=Power(Base=@n105, Exp=@n106.Near) {constExponent=0.25}), B=@n107.Near)
  n109 = TextureSample(UVs=Panner(Coordinate=Multiply(A=@n96, B=@n97.Mid), Time=Multiply(A=@n98, B=@n99.Mid)) {speedY=0.5}, "DDX(UVs)"=@n101, "DDY(UVs)"=@n102) {mipValueMode=TMVM_Derivative, samplerSource=SSM_Wrap_WorldGroupSettings, texture=&T_Barrier_Streaks_01_M}
  n110 = Multiply(A=Multiply(A=@n109.G, B=Power(Base=@n105, Exp=@n106.Mid) {constExponent=0.5}), B=@n107.Mid) {constA=0.5, constB=0.66}
  n111 = TextureSample(UVs=Panner(Coordinate=Multiply(A=@n96, B=@n97.Far), Time=Multiply(A=@n98, B=@n99.Far)) {speedY=0.5}, "DDX(UVs)"=@n101, "DDY(UVs)"=@n102) {mipValueMode=TMVM_Derivative, samplerSource=SSM_Wrap_WorldGroupSettings, texture=&T_Barrier_Streaks_01_M}
  n112 = Multiply(A=@n111.B, B=Power(Base=@n105, Exp=@n106.Far) {constExponent=1})
  n113 = Saturate(None=Add(A=Add(A=@n108, B=@n110), B=Multiply(A=@n112, B=@n107.Far) {constA=0.25, constB=0.33}))
  n114 = Multiply(A=@n113, B=ScalarParameter() {defaultValue=3, group=STREAKS, parameterName=Streak Mult, sortPriority=30}) {constB=2}
  n115 = ScalarParameter() {defaultValue=35, group=STREAKS, parameterName=Streak Substract Tiling, sortPriority=20}
  n116 = Multiply(A=@n115, B=ScalarParameter() {defaultValue=1, group=STREAKS, parameterName=Streak Subtract Aspect, sortPriority=21})
  n117 = Panner(Coordinate=Multiply(A=NamedRerouteUsage(), B=AppendVector(A=@n115, B=@n116))) {speedY=1}
  n118 = Multiply(
   A=NamedRerouteDeclaration(None=TextureSample!(UVs=@n117, Tex, Apply View MipBias) {samplerType=SAMPLERTYPE_LinearColor, texture=/Game/Effects/Fort_Effects/Textures/Tile/LinerNoise64x64}.R) {name=Panning Noise, nodeColor=<rgba 0.5 1 0 1>}
   B=ScalarParameter() {defaultValue=3, group=STREAKS, parameterName=Subtract Intensity, sortPriority=22}
  )
  n119 = Power(Base=Saturate(None=Subtract(A=@n114, B=@n118)))
  n120 = ScalarParameter() {defaultValue=5, group=STREAKS, parameterName=Streaks Center Intensity, sortPriority=31}
  n121 = Multiply(A=Power(Base=@n113), B=ScalarParameter() {defaultValue=3, group=STREAKS, parameterName=Streak Soft Center})
  n122 = Multiply(A=@n113, B=ScalarParameter() {defaultValue=1, group=STREAKS, parameterName=Streak Soft, sortPriority=33})
  n123 = VectorParameter() {group=STREAKS, parameterName=Streaks Color, sortPriority=35, defaultValue=<rgba 0.1667 0.0709 0.0503 0>}
  n124 = Multiply(A=Add(A=Multiply(A=@n119, B=@n120), B=Add(A=@n121, B=@n122)), B=@n123.RGB)
  n125 = NamedRerouteUsage()
  n126 = Sine(None=Multiply(A=ComponentMask(None=@n125) {g=true}) {constB=0.5})
  NamedRerouteDeclaration(None=Multiply(A=Multiply(A=@n124, B=@n126), B=ScalarParameter() {defaultValue=0.3, group=STREAKS, parameterName=Streaks Global Intensity, sortPriority=34})) {name=Streaks, nodeColor=<rgba 0.7302 1 0.3408 1>}
  NamedRerouteDeclaration(None=@n12.Top Mask) {name=Top Mask, nodeColor=<rgba 0.4186 0 0.8542 1>}
  NamedRerouteDeclaration(None=OneMinus(None=@n91)) {name=Bottom Planet Mask, nodeColor=<rgba 0 0.8125 1 1>}
  NamedRerouteDeclaration(None=@n12.Spherical UVs) {name=Sphere UVs, nodeColor=<rgba 1 0.1268 0.0371 1>}
  n127 = TextureObjectParameter() {
   group=STARS
   parameterName=Stars Texture
   samplerType=SAMPLERTYPE_LinearColor
   sortPriority=10
   texture=/Game/Items/Traps/Textures/T_Barrier_Stars_02_M
  }
  n128 = MaterialFunctionCall(Float3=MaterialFunctionCall!(
    contrast=ScalarParameter() {defaultValue=10, group=STARS, parameterName=Stars Triplanar Contrast, sortPriority=21}
    UV Scale=ScalarParameter() {defaultValue=2, group=STARS, parameterName=Stars UV, sortPriority=22}
    UV Distortion
    In=@n127
   ) {materialFunction=/Game/Items/Traps/Materials/MaterialFunctions/MF_BarrierTriplanar_LinearColor}.Result) {materialFunction=&BreakOutFloat3Components}
  n129 = Add(A=@n128.G, B=Multiply(A=Time(), B=ScalarParameter() {defaultValue=0.1, group=STARS, parameterName=Stars Blink Speed, sortPriority=30}))
  n130 = Saturate(None=Multiply(A=@n128.R, B=Sine(None=@n129)))
  n131 = VectorParameter() {group=STARS, parameterName=Stars Color, sortPriority=31, defaultValue=<rgba 1 0.6821 0.4775 0>}
  NamedRerouteDeclaration(None=Multiply(A=Multiply(A=@n130, B=@n131.RGB), B=ScalarParameter() {defaultValue=0.2, group=STARS, parameterName=Stars Intensity})) {name=Stars, nodeColor=<rgba 1 0.5868 0.0263 1>}
  n132 = ScalarParameter() {defaultValue=2, group=CLOUDS, parameterName=Cloud U}
  n133 = Multiply(A=@n132, B=ScalarParameter() {defaultValue=0.6, group=CLOUDS, parameterName=Cloud Ratio})
  n134 = NamedRerouteUsage()
  n135 = AppendVector(
   A=ComponentMask(None=@n134) {r=true}
   B=Power(Base=ComponentMask(None=@n134) {g=true}, Exp=ScalarParameter() {defaultValue=1.5, parameterName=Cloud UV Stretching})
  )
  n136 = ScalarParameter() {defaultValue=4.4, group=CLOUDS, parameterName=Distortion Cloud U}
  n137 = AppendVector(A=@n136, B=Multiply(A=@n136, B=ScalarParameter() {defaultValue=0.6, group=CLOUDS, parameterName=Cloud Distortion Ratio}))
  n138 = Time()
  n139 = ScalarParameter() {defaultValue=0.001, group=CLOUDS, parameterName=Distortion Speed}
  n140 = TextureCoordinate()
  n141 = TextureSample!(
   UVs=Panner(Coordinate=Multiply(A=@n137, B=@n135), Time=Multiply(A=@n138, B=@n139)) {speedX=1, speedY=2}
   Tex=TextureObjectParameter() {parameterName=Cloud Distortion Tex, samplerType=SAMPLERTYPE_Normal, texture=/Game/Items/Traps/Textures/T_Barrier_Clouds_N}
   "DDX(UVs)"=DDX(Value=@n140)
   "DDY(UVs)"=DDY(Value=@n140)
   Apply View MipBias
  ) {mipValueMode=TMVM_Derivative, samplerType=SAMPLERTYPE_Normal, texture=/Game/Effects/Fort_Effects/Textures/Tile/T_GaussianNoise_02_N}
  n142 = Add(
   A=Multiply(A=AppendVector(A=@n132, B=@n133), B=@n135)
   B=Multiply(A=AppendVector(A=@n141.R, B=@n141.R), B=ScalarParameter() {defaultValue=0.01, group=CLOUDS, parameterName=Distortion})
  )
  n143 = VectorParameter() {
   group=CLOUDS
   parameterName=Clouds Tex 1
   sortPriority=10
   channelNames {a="", b=Tiling Mult, g=Speed Y, r=Speed X}
   defaultValue=<rgba 0 2 1 0>
  }
  n144 = Multiply(A=Time(), B=ScalarParameter() {defaultValue=0.005, group=CLOUDS, parameterName=Cloud Panning Speed})
  n145 = Panner(Coordinate=Multiply(A=@n142, B=@n143.Tiling Mult), Time=@n144, Speed=AppendVector(A=@n143.Speed X, B=@n143.Speed X)) {speedY=1.2}
  n146 = TextureObjectParameter() {group=CLOUDS, parameterName=Background Texture, samplerType=SAMPLERTYPE_LinearColor, texture=/Game/Items/Traps/Textures/T_Barrier_Clouds_M}
  n147 = TextureCoordinate()
  n148 = DDX(Value=@n147)
  n149 = DDY(Value=@n147)
  n150 = VectorParameter() {
   group=CLOUDS
   parameterName=Clouds Tex 2
   sortPriority=11
   channelNames {a="", b=Tiling Mult, g=Speed Y, r=Speed X}
   defaultValue=<rgba 1.2 1.5 2 0>
  }
  n151 = Panner(Coordinate=Multiply(A=@n142, B=@n150.Tiling Mult) {constB=2}, Time=@n144, Speed=AppendVector(A=@n150.Speed X, B=@n150.Speed X)) {speedX=0.2, speedY=1.1}
  n152 = VectorParameter() {
   group=CLOUDS
   parameterName=Clouds Tex 3
   sortPriority=12
   channelNames {a="", b=Tiling Mult, g=Speed Y, r=Speed X}
   defaultValue=<rgba -1.2 1 2 0>
  }
  n153 = Multiply(A=@n142, B=@n152.Tiling Mult) {constB=3}
  n154 = AppendVector(A=@n152.Speed X, B=@n152.Speed X)
  n155 = Multiply(
   A=Multiply(
    A=TextureSample(UVs=@n145, Tex=@n146, "DDX(UVs)"=@n148, "DDY(UVs)"=@n149) {mipValueMode=TMVM_Derivative, samplerSource=SSM_Wrap_WorldGroupSettings, samplerType=SAMPLERTYPE_LinearColor}.B
    B=TextureSample(UVs=@n151, Tex=@n146, "DDX(UVs)"=@n148, "DDY(UVs)"=@n149) {mipValueMode=TMVM_Derivative, samplerSource=SSM_Wrap_WorldGroupSettings, samplerType=SAMPLERTYPE_LinearColor}.B
   )
   B=TextureSample(UVs=Panner(Coordinate=@n153, Time=@n144, Speed=@n154) {speedX=-0.2, speedY=1}, Tex=@n146, "DDX(UVs)"=@n148, "DDY(UVs)"=@n149) {mipValueMode=TMVM_Derivative, samplerSource=SSM_Wrap_WorldGroupSettings, samplerType=SAMPLERTYPE_LinearColor}.B
  )
  n156 = ScalarParameter() {defaultValue=2, group=CLOUDS, parameterName=Cloud Strength}
  n157 = Abs(None=ComponentMask(None=NamedRerouteUsage()) {b=true})
  n158 = ScalarParameter() {defaultValue=20, group=CLOUDS, parameterName=Cloud Fadeout Exp}
  n159 = LinearInterpolate(Alpha=OneMinus(None=Power(Base=@n157, Exp=@n158) {constExponent=4})) {constA=-0.2}
  n160 = Multiply(
   A=Power(Base=Multiply(A=@n155, B=@n156), Exp=ScalarParameter() {defaultValue=1, group=CLOUDS, parameterName=Cloud Exponent}) {constExponent=1}
   B=Power(Base=Saturate(None=@n159))
  )
  n161 = ComponentMask(None=NamedRerouteUsage()) {g=true}
  n162 = ScalarParameter() {defaultValue=0.5, group=CLOUDS, parameterName=Cloud Fadeout Position, sliderMax=2, sliderMin=0.1}
  n163 = Add(A=Multiply(A=Power(Base=@n161, Exp=@n162)) {constB=0.5}) {constB=0.25}
  n164 = LinearInterpolate(
   A=VectorParameter() {group=CLOUDS, parameterName=Cloud Color Bot, defaultValue=<rgba 0.2448 0.0831 0.0314 0>}.RGB
   B=VectorParameter() {group=CLOUDS, parameterName=Cloud Color Top, defaultValue=<rgba 1.5 1.385 1.2094 0>}.RGB
   Alpha=ConstantBiasScale(None=Sine(None=@n163))
  )
  n165 = ScalarParameter() {defaultValue=1.81, group=CLOUDS, parameterName=Cloud Intensity}
  NamedRerouteDeclaration(None=Multiply(A=Multiply(A=@n160, B=@n164), B=Divide(A=@n165) {constB=100})) {name=Clouds, nodeColor=<rgba 0 0.3437 1 1>}
  n166 = ComponentMask(None=NamedRerouteUsage()) {g=true}
  n167 = Divide(B=ScalarParameter() {defaultValue=0.18, group=BLACK HOLE, parameterName=Black Hole size})
  n168 = Subtract(A=Multiply(A=OneMinus(None=@n166), B=Add(A=@n167)), B=@n167)
  n169 = Saturate(None=@n168)
  n170 = Ceil(None=@n169)
  NamedRerouteDeclaration(None=OneMinus(None=@n170)) {name=Black Hole Mask, nodeColor=<rgba 0 0.625 1 1>}
  n171 = ScalarParameter() {defaultValue=1, group=BLACK HOLE, parameterName=Black Hole Texture Tiling}
  n172 = Multiply(
   A=NamedRerouteUsage()
   B=AppendVector(
    A=ScalarParameter() {defaultValue=2, group=BLACK HOLE, parameterName=Black Hole Global tiling X}
    B=ScalarParameter() {defaultValue=0.9, group=BLACK HOLE, parameterName=Black Hole Global tiling Y}
   )
  )
  n173 = ScalarParameter() {defaultValue=0.2, group=BLACK HOLE, parameterName=Black Hole Texture Speed}
  n174 = Multiply(A=Time(), B=ScalarParameter() {defaultValue=0.1, group=BLACK HOLE, parameterName=Black Hole Speed})
  n175 = Add(
   A=Multiply(A=Multiply(A=@n172, B=@n171), B=Constant2Vector() {g=0.35, r=1})
   B=Multiply(A=Multiply(A=@n174, B=@n173), B=Constant2Vector() {g=1, r=0.05})
  )
  n176 = TextureObject() {samplerType=SAMPLERTYPE_LinearColor, texture=&T_Barrier_Clouds_M}
  n177 = TextureCoordinate()
  n178 = DDX(Value=@n177)
  n179 = DDY(Value=@n177)
  n180 = ScalarParameter() {defaultValue=2, group=BLACK HOLE, parameterName=Black Hole Texture Tiling 2}
  n181 = ScalarParameter() {defaultValue=0.1, group=BLACK HOLE, parameterName=Black Hole Texture Speed 2}
  n182 = Add(
   A=Multiply(A=Multiply(A=@n172, B=@n180), B=Constant2Vector() {g=0.7, r=1})
   B=Multiply(A=Multiply(A=@n174, B=@n181), B=Constant2Vector() {g=1, r=0.05})
  )
  n183 = TextureCoordinate()
  n184 = DDX(Value=@n183)
  n185 = DDY(Value=@n183)
  n186 = Multiply(
   A=Multiply(A=@n172, B=ScalarParameter() {defaultValue=4, group=BLACK HOLE, parameterName=Black Hole Texture Tiling 3})
   B=Constant2Vector() {g=0.5, r=1}
  )
  n187 = Multiply(
   A=Multiply(A=@n174, B=ScalarParameter() {defaultValue=0.3, group=BLACK HOLE, parameterName=Black Hole Texture Speed 3})
   B=Constant2Vector() {g=1, r=0.05}
  )
  n188 = TextureCoordinate()
  n189 = Add(
   A=Add(
    A=TextureSample(UVs=@n175, Tex=@n176, "DDX(UVs)"=@n178, "DDY(UVs)"=@n179) {mipValueMode=TMVM_Derivative}.G
    B=TextureSample(UVs=@n182, Tex=@n176, "DDX(UVs)"=@n184, "DDY(UVs)"=@n185) {mipValueMode=TMVM_Derivative}.B
   )
   B=TextureSample(UVs=Add(A=@n186, B=@n187), Tex=@n176, "DDX(UVs)"=DDX(Value=@n188), "DDY(UVs)"=DDY(Value=@n188)) {mipValueMode=TMVM_Derivative}.R
  )
  n190 = LinearInterpolate(B=ScalarParameter() {defaultValue=1.2, group=BLACK HOLE, parameterName=Black Hole Aura Outer Scale}, Alpha=@n189) {constA=0.3}
  n191 = OneMinus(None=Multiply(A=Divide(B=@n190), B=Abs(None=@n168)))
  n192 = Subtract(A=Saturate(None=Power(Base=@n191) {constExponent=5}), B=@n170)
  n193 = LinearInterpolate(A=ScalarParameter() {defaultValue=0.704, group=BLACK HOLE, parameterName=Black Hole Aura inner Scale}, Alpha=Multiply(A=@n189) {constB=0.2}) {constB=0}
  n194 = OneMinus(None=Multiply(A=Divide(B=@n193) {constB=0.01}, B=Abs(None=@n168)))
  n195 = Subtract(A=Saturate(None=Power(Base=@n194) {constExponent=5}), B=OneMinus(None=@n170))
  n196 = Power(Base=Add(A=Saturate(None=@n192), B=Saturate(None=@n195)))
  n197 = VectorParameter() {group=BLACK HOLE, parameterName=Black Hole Aura Color, defaultValue=<rgba 0.1667 0.0897 0.073 0>}
  NamedRerouteDeclaration(None=Multiply(A=Multiply(A=@n196, B=@n197.RGB), B=ScalarParameter() {defaultValue=1, group=BLACK HOLE, parameterName=Black Hole Intensity})) {name=Black Hole, nodeColor=<rgba 0.0296 0 0.0417 1>}
  NamedRerouteDeclaration(None=@n169) {name=Black Hole Mask Soft, nodeColor=<rgba 0 0.5547 1 1>}
  NamedRerouteDeclaration(None=@n12.Spherical Remapped UVs) {name=Spherical Remapped, nodeColor=<rgba 1 0.0351 0 1>}
  NamedRerouteUsage()
  Reroute(None=@n11)
 ]
}
