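# Unlit material for the barrier trap that paints an animated underwater scene using camera-based UVs from MF_CameraBasedUVs.
# Turbulence-distorted panning caustics and a top layer of scrolling normals dotted with vectors drive water color, alongside a similarly lit sand bottom.
# View-vector gradients form fog and masks; the water blends toward fog color.
# Sine-modulated light rays on rotated cylinder UVs and drifting particles provide further named layers for the additive composite.
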Material /Game/Items/Traps/Materials/M_Barrier_Underwater
graph {
 properties {pixelDepthOffsetMode=PDOM_Legacy, shadingModel=MSM_Unlit, usedShadingModels=Unlit}
 nodes [
  n1 = NamedRerouteUsage()
  n2 = NamedRerouteUsage()
  n3 = NamedRerouteUsage()
  n4 = ScalarParameter() {defaultValue=1, parameterName=Distortion tiling}
  n5 = TextureSample(UVs=Panner(Coordinate=Multiply(A=@n3, B=@n4)) {speedX=-0.025, speedY=-0.02}) {texture=/Game/Global/Textures/Default/Caustics/T_Turbulence_Seamless}
  n6 = AppendVector(A=@n5.R, B=@n5.R)
  n7 = ScalarParameter() {defaultValue=0.1, parameterName=Water caustics distortion}
  n8 = TextureSample(UVs=Add(A=Panner(Coordinate=@n2) {speedX=0.02, speedY=0.02}, B=Multiply(A=@n6, B=@n7) {constB=0})) {samplerType=SAMPLERTYPE_LinearColor, texture=/FortWater/Textures/T_Caustic_M}
  n9 = NamedRerouteUsage()
  n10 = NamedRerouteUsage()
  n11 = ScalarParameter() {defaultValue=1, parameterName=Distortion tiling}
  n12 = TextureSample(UVs=Panner(Coordinate=Multiply(A=@n10, B=@n11)) {speedX=0.025, speedY=0.02}) {texture=&T_Turbulence_Seamless}
  n13 = AppendVector(A=@n12.R, B=@n12.R)
  n14 = TextureSample(UVs=Add(A=Panner(Coordinate=@n9) {speedX=-0.02, speedY=-0.02}, B=Multiply(A=@n13, B=@n7) {constB=0.1})) {samplerType=SAMPLERTYPE_LinearColor, texture=&T_Caustic_M}
  n15 = Multiply(
   A=LinearInterpolate(
    A=ScalarParameter() {defaultValue=0.5, parameterName=Bottom water color intensity}
    B=ScalarParameter() {defaultValue=30, parameterName=Bottom Caustics intensity}
    Alpha=Multiply(A=@n8.G, B=@n14.G)
   ) {constA=0.1, constB=10}
   B=NamedRerouteUsage()
  )
  n16 = NamedRerouteUsage()
  n17 = LinearInterpolate(
   A=LinearInterpolate(
    A=VectorParameter() {parameterName=Water colour low, defaultValue=<rgba 0 0.0358 0.0625 1>}.RGB
    B=VectorParameter() {parameterName=Water colour hight, defaultValue=<rgba 0 0.2582 0.4508 1>}.RGB
    Alpha=LinearInterpolate(A=@n1, B=@n15, Alpha=@n16) {constAlpha=1, constB=0}
   )
   Alpha=NamedRerouteUsage()
  )
  n18 = VectorParameter() {parameterName=Fog color, defaultValue=<rgba 0 0.0116 0.0296 1>}
  n19 = NamedRerouteUsage()
  Add(A=Add(A=NamedRerouteUsage(), B=LinearInterpolate(A=@n17, B=@n18.RGB, Alpha=@n19)), B=NamedRerouteUsage())
  n20 = MaterialFunctionCall!(
   Camera Vector
   Local Space
   Spherical UV Remap Top
   Spherical UV Remap Bot
   Cylinder UV V Scale
   Planar UV Exponent=Constant() {r=1.3}
   Planar UV Dome
   Planar UV Scale=Constant() {r=0.5}
   Planar Radial UV Scale
   Quad UV H pos
   Quad UV V pos
   Quad UV Scale
   Enable Rotation
   Rotation Vector
   -------------
   Mask UV
   Mask Size
  ) {materialFunction=/Game/Packages/Fortress_SharedMaterials/Base_Material_Functions/MF_CameraBasedUVs}
  NamedRerouteDeclaration(None=@n20.Spherical UVs) {name=Spherical UV, nodeColor=<rgba 1 0 0.8906 1>}
  NamedRerouteDeclaration(None=@n20.Planar UVs) {name=Planar UV, nodeColor=<rgba 0.9453 0 1 1>}
  NamedRerouteDeclaration(None=@n20.Quad UV) {name=Quad uv, nodeColor=<rgba 0.9219 1 0 1>}
  NamedRerouteDeclaration(None=@n20.Planar Radial UVs) {name=Planar radial uv, nodeColor=<rgba 0.8516 1 0 1>}
  n21 = Multiply(A=NamedRerouteUsage(), B=ScalarParameter() {defaultValue=1, parameterName=Sand tiling})
  n22 = Constant3Vector() {constant=<rgba 0 -0.4 0.5 0>}
  n23 = NamedRerouteUsage()
  n24 = Multiply(
   A=DotProduct(
    A=TextureSample(UVs=@n21) {samplerType=SAMPLERTYPE_Normal, texture=/Game/Environments/Sets/Arid/Textures/DesertSand_normal}.RGB
    B=Normalize(VectorInput=@n22)
   )
   B=Multiply(A=@n23) {constB=0.3}
  )
  n25 = VectorParameter() {parameterName=Sand color dark, defaultValue=<rgba 0.0729 0.0407 0.0152 1>}
  n26 = VectorParameter() {parameterName=Sand color hightlight, defaultValue=<rgba 1 0.713 0.4211 1>}
  NamedRerouteDeclaration(None=LinearInterpolate(A=@n24, B=LinearInterpolate(A=@n25.RGB, B=@n26.RGB, Alpha=@n24), Alpha=@n23)) {name=Sand bottom, nodeColor=<rgba 0 1 0.4609 1>}
  n27 = MaterialFunctionCall(
   Enable Rotation=StaticBool() {value=true}
   Rotation Vector=VectorParameter() {parameterName=Light direction vector, defaultValue=<rgba 0 -0.3 0.5 1>}.RGB
  ) {materialFunction=&MF_CameraBasedUVs}
  n28 = ComponentMask(None=Multiply(A=@n27.Cylinder UVs, B=ScalarParameter() {defaultValue=9, group=Light rays, parameterName=Light rays tiling}) {constB=7}) {r=true}
  n29 = Constant()
  n30 = TextureSample(UVs=Panner(Coordinate=AppendVector(A=@n28, B=@n29)) {speedY=0.03}) {samplerType=SAMPLERTYPE_Masks, texture=/Game/Characters/Player/Male/Large/Bodies/M_LRG_Clash_V/Textures/Simple_Perlin_M}
  n31 = Multiply(A=Time()) {constB=0.05}
  n32 = Multiply(A=@n30.RGB, B=ConstantBiasScale(None=Sine(None=@n31)))
  n33 = Multiply(A=Time()) {constB=0.013}
  n34 = Multiply(A=@n30.RGB, B=ConstantBiasScale(None=Sine(None=@n33)))
  n35 = AppendVector(A=ComponentMask(None=Multiply(A=@n27.Cylinder UVs)) {r=true}, B=Constant())
  n36 = Multiply(
   A=Multiply(
    A=LinearInterpolate(A=@n32, B=@n34)
    B=TextureSample(UVs=@n35) {samplerType=SAMPLERTYPE_LinearGrayscale, texture=/Game/Global/Textures/Default/Masks_and_Gradients/AnimRamp}.RGB
   )
   B=Power(Base=NamedRerouteUsage(), Exp=ScalarParameter() {defaultValue=1.6, parameterName=Horizontal Light mask})
  )
  NamedRerouteDeclaration(None=Multiply(A=Saturate(None=@n36), B=VectorParameter() {parameterName=Light color, defaultValue=<rgba 0 0.2099 0.383 1>}.RGB)) {name=Volumetric light, nodeColor=<rgba 1 0 0.2578 1>}
  n37 = ComponentMask(None=CameraVectorWS()) {b=true}
  n38 = Sine(None=Multiply(A=Add(A=@n37) {constB=1.5}) {constB=0.5})
  NamedRerouteDeclaration(None=OneMinus(None=ConstantBiasScale(None=@n38))) {name=Vertical gradient, nodeColor=<rgba 1 0 0.1406 1>}
  NamedRerouteDeclaration(None=Normalize(VectorInput=VectorParameter() {parameterName=Light direction vector, defaultValue=<rgba 0 -0.3 0.5 1>}.RGB)) {name=Light direction vector normalized, nodeColor=<rgba 0.2188 0 1 1>}
  NamedRerouteDeclaration(None=Multiply(A=Multiply(A=@n6) {constB=0.3}, B=ScalarParameter() {defaultValue=0.1, parameterName=Top distortion intensity}) {constB=0.01}) {name=distortion, nodeColor=<rgba 1 0.1875 0 1>}
  n39 = ScalarParameter() {defaultValue=7, parameterName=Top water color intensity}
  n40 = ScalarParameter() {defaultValue=20, parameterName="Top water  highlights intensity"}
  n41 = Panner(Coordinate=NamedRerouteUsage()) {speedX=0.05, speedY=0.05}
  n42 = TextureSample(UVs=Add(A=Multiply(A=@n41) {constB=0.6}, B=NamedRerouteUsage())) {samplerType=SAMPLERTYPE_Normal, texture=/Game/Global/Textures/Default/Tileable_Patterns/1Normal}
  n43 = NamedRerouteUsage()
  n44 = NamedRerouteUsage()
  n45 = Add(A=Multiply(A=Panner(Coordinate=@n44) {speedX=-0.05, speedY=-0.05}) {constB=0.6}, B=NamedRerouteUsage())
  n46 = Multiply(A=Panner(Coordinate=NamedRerouteUsage()) {speedX=-0.05, speedY=0.05}, B=ScalarParameter() {defaultValue=0.1, parameterName=Top layer tiling}) {constB=0.5}
  n47 = NamedRerouteUsage()
  n48 = DotProduct(
   A=TextureSample(UVs=Add(A=@n46, B=@n47)) {samplerType=SAMPLERTYPE_Normal, texture=/Game/Packages/Fortnite_UniqueMaterials_1/Textures/FoamNormals}.RGB
   B=NamedRerouteUsage()
  )
  n49 = Multiply(
   A=DotProduct(
    A=TextureSample(UVs=@n45) {samplerType=SAMPLERTYPE_Normal, texture=&1Normal}.RGB
    B=NamedRerouteUsage()
   )
   B=@n48
  )
  n50 = Multiply(A=Panner(Coordinate=NamedRerouteUsage()) {speedX=0.05, speedY=-0.05}, B=ScalarParameter() {defaultValue=0.1, parameterName=Top layer tiling}) {constB=0.5}
  n51 = NamedRerouteUsage()
  n52 = DotProduct(
   A=TextureSample(UVs=Add(A=@n50, B=@n51)) {samplerType=SAMPLERTYPE_Normal, texture=&FoamNormals}.RGB
   B=NamedRerouteUsage()
  )
  n53 = Multiply(A=Multiply(A=DotProduct(A=@n42.RGB, B=@n43), B=@n49), B=Multiply(A=@n49, B=Multiply(A=@n48, B=@n52)))
  NamedRerouteDeclaration(None=Multiply(A=LinearInterpolate(A=@n39, B=@n40, Alpha=@n53) {constA=1, constB=5}, B=NamedRerouteUsage())) {name=Top layer water, nodeColor=<rgba 1 0.3984 0 1>}
  NamedRerouteDeclaration(None=@n20.Top Mask) {name=Top mask, nodeColor=<rgba 0.8047 0 1 1>}
  NamedRerouteDeclaration(None=@n20.Spherical Remapped UVs) {name=Spherical RE UV, nodeColor=<rgba 0 0.8359 1 1>}
  n54 = Power(Base=ComponentMask(None=CameraVectorWS()) {b=true}) {constExponent=1.1}
  n55 = Sine(None=Multiply(A=Add(A=@n54) {constB=1.5}) {constB=0.5})
  NamedRerouteDeclaration(None=Saturate(None=ConstantBiasScale(None=@n55))) {name=Circular mask, nodeColor=<rgba 0.4297 0 1 1>}
  NamedRerouteDeclaration(None=@n20.Cylinder UVs) {name=Cylinder uv, nodeColor=<rgba 1 0.359 0 1>}
  n56 = NamedRerouteUsage()
  n57 = ScalarParameter() {defaultValue=5, parameterName=Particles tiling}
  n58 = TextureSample(UVs=Panner(Coordinate=Multiply(A=@n56, B=@n57) {constB=5}) {speedX=0.001, speedY=0.003}) {samplerType=SAMPLERTYPE_LinearColor, texture=/Game/Items/Traps/Textures/T_Barrier_Stars_02_M}
  n59 = Add(A=Multiply(A=Time(), B=Constant() {r=0.1}) {constB=0.1}, B=@n58.G)
  n60 = Multiply(
   A=Multiply(A=@n58.R, B=Sine(None=@n59))
   B=Power(Base=NamedRerouteUsage(), Exp=ScalarParameter() {defaultValue=10, parameterName=Particle mask})
  )
  NamedRerouteDeclaration(None=Multiply(A=Clamp(None=@n60), B=VectorParameter() {parameterName=Particle color, defaultValue=<rgba 0.0468 0.1146 0.1141 1>}.RGB) {constB=0.4}) {name=particles, nodeColor=<rgba 1 0 0.4922 1>}
  n61 = NamedRerouteUsage()
  n62 = ScalarParameter() {defaultValue=1.5, group=Fog, parameterName=Fog Exp}
  NamedRerouteDeclaration(None=Multiply(A=Power(Base=@n61, Exp=@n62), B=ScalarParameter() {defaultValue=1, group=Fog, parameterName=Fog Intensity})) {name=FOG, nodeColor=<rgba 0 1 0.3438 1>}
  NamedRerouteDeclaration(None=@n20.Spherical Remapped Mirrored UVs) {name=Spherical RM uvs, nodeColor=<rgba 0.7344 0 1 1>}
  n63 = ComponentMask(None=CameraVectorWS()) {b=true}
  n64 = Power(Base=OneMinus(None=Abs(None=@n63))) {constExponent=3}
  n65 = Sine(None=Multiply(A=Add(A=@n64) {constB=1.5}) {constB=0.5})
  n66 = Multiply(A=Saturate(None=ConstantBiasScale(None=@n65)))
  n67 = NamedRerouteUsage()
  NamedRerouteDeclaration(None=Saturate(None=LinearInterpolate(B=@n66, Alpha=@n67) {constA=1})) {name=Fog as mask, nodeColor=<rgba 0 1 0.7188 1>}
 ]
}
