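# Unlit translucent particle sprite material in the trap items folder, named for control blocks, that draws a soft rounded-square shape without any textures.
# Particle color RGB passes through MF_FX_EmissiveAdjustments for the glow color.
# Centered UVs are made absolute and cubed before a SphereMask, bending the circular falloff into a rounded square.
# That mask multiplies particle alpha to shape the sprite's transparency.
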Material /Game/Items/Traps/Materials/M_Control_Blocks
graph {
 properties {
  bUsedWithParticleSprites=true
  blendMode=BLEND_Translucent
  pixelDepthOffsetMode=PDOM_Legacy
  shadingModel=MSM_Unlit
  usedShadingModels=Unlit
 }
 nodes [
  n1 = ParticleColor()
  MaterialFunctionCall(Emissive=@n1.RGB) {materialFunction=&MF_FX_EmissiveAdjustments}
  n2 = Abs(None=Subtract(A=TextureCoordinate()) {constB=0.5})
  n3 = Constant() {r=0.1}
  Multiply(A=@n1.A, B=SphereMask(A=Power(Base=@n2) {constExponent=3}, B=Constant(), Radius=@n3, Hardness=@n3))
 ]
}
