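# Additive, unlit, two-sided material for the elimination zone trap's barrier.
# Projects world-space UVs by rounded normal, optionally standard UVs, for tiled warning icons with optional flashing, camera-distance-faded dots and panning noise.
# These layers plus Fresnel and depth-behind-translucency blend background to circle color, masked by a UV gradient and tinted by particle color.
# High quality adds near-camera fade; MF_CreativeMoveTool processes the final attributes.
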Material /Game/Items/Traps/Materials/M_EliminationZone_Barrier_01
graph {
 properties {
  allowTranslucentCustomDepthWrites=true
  bEnableResponsiveAA=true
  bUseMaterialAttributes=true
  bUsedWithMeshParticles=true
  blendMode=BLEND_Additive
  pixelDepthOffsetMode=PDOM_Legacy
  shadingModel=MSM_Unlit
  twoSided=true
  usedShadingModels=Unlit
 }
 nodes [
  n1 = VectorParameter() {parameterName=BGColor, defaultValue=<rgba 0.7448 0.0124 0 0>}
  n2 = VectorParameter() {parameterName=Circles, defaultValue=<rgba 10 1.6189 0.3215 0>}
  n3 = WorldPosition()
  n4 = MaterialFunctionCall(Float3=Multiply(A=@n3.XYZ) {constB=-1}) {materialFunction=&BreakOutFloat3Components}
  n5 = AppendVector(A=@n4.G, B=@n4.G)
  n6 = AppendVector(A=@n4.R, B=@n4.R)
  n7 = VertexNormalWS()
  n8 = MaterialFunctionCall(Float3=Abs(None=Round(None=@n7))) {materialFunction=&BreakOutFloat3Components}
  n9 = LinearInterpolate(
   A=LinearInterpolate(A=LinearInterpolate(A=@n5, B=@n6, Alpha=@n8.G), B=AppendVector(A=@n4.R, B=@n4.R), Alpha=@n8.B)
   B=Multiply(A=TextureCoordinate(), B=ScalarParameter() {defaultValue=6, parameterName=DotsRepeat})
   Alpha=ScalarParameter() {parameterName=UseStandardUVs}
  )
  n10 = TextureSampleParameter2D(UVs=Divide(A=@n9, B=ScalarParameter() {defaultValue=150, parameterName=IconTilingScale})) {
   parameterName=IconTexture
   samplerSource=SSM_Wrap_WorldGroupSettings
   texture=/Game/UI/Foundation/Textures/Banner/Standard/T-Banners-Icons-Warning-L
  }
  n11 = Time() {period=100000}
  n12 = ScalarParameter() {defaultValue=1, parameterName=AlertFlashTime}
  n13 = Multiply(
   A=Clamp(None=MaterialFunctionCall(Value=@n11, Period=@n12) {materialFunction=&LinearSine}."Rounded Linear Sine ") {maxDefault=0.8, minDefault=0.05}
   B=ScalarParameter() {parameterName=UseFlashingAlert, sliderMax=1}
  )
  n14 = ScalarParameter() {defaultValue=0.05, parameterName=IconOpacityScale}
  n15 = ScalarParameter() {defaultValue=1, parameterName=DotsScale}
  n16 = Divide(A=@n9, B=ScalarParameter() {defaultValue=6, parameterName=DotsRepeat})
  n17 = Add(A=Constant2Vector() {r=0.5}, B=@n16)
  n18 = ComponentMask(None=@n3.XYZ) {b=true}
  n19 = Multiply(A=ScalarParameter() {defaultValue=6, parameterName=DotsRepeat}) {constB=2}
  n20 = Round(None=Frac(None=Divide(A=@n18, B=@n19) {constB=16}))
  n21 = MaterialFunctionCall(Float2=TextureCoordinate()) {materialFunction=&BreakOutFloat2Components}
  n22 = SphereMask(
   A=@n21.G
   B=Constant() {r=0.5}
   Radius=ScalarParameter() {defaultValue=0.5, parameterName=BGGradientRadius}
   Hardness=ScalarParameter() {parameterName=BGGradientHardness}
  )
  n23 = SphereMask(
   A=Frac(None=LinearInterpolate(A=@n17, B=@n16, Alpha=@n20))
   B=Constant() {r=0.5}
   Radius=LinearInterpolate(A=ScalarParameter() {defaultValue=0.1, parameterName=Radius}, B=ScalarParameter() {defaultValue=0.35, parameterName=RadiusLarge}, Alpha=@n22)
   Hardness=ScalarParameter() {parameterName=DotsHardness}
  )
  n24 = Distance(A=CameraPositionWS(), B=WorldPosition().XYZ)
  n25 = ScalarParameter() {defaultValue=-100, parameterName=DistanceBias}
  n26 = Saturate(None=Divide(A=Add(A=@n24, B=@n25), B=ScalarParameter() {defaultValue=800, parameterName=FadeScale}))
  n27 = Panner(Coordinate=@n9, Speed=VectorParameter() {parameterName=PanSpeedNoiseA, defaultValue=<rgba 7.5 15 11 0>}.RGB) {speedX=5}
  n28 = ScalarParameter() {defaultValue=1000, parameterName=RepeatNoiseA}
  n29 = Panner(Coordinate=@n9, Speed=VectorParameter() {parameterName=PanSpeedNoiseB, defaultValue=<rgba -11 -32.02 0 0>}.RGB) {speedX=5}
  n30 = ScalarParameter() {defaultValue=-1300, parameterName=RepeatNoiseB}
  n31 = Multiply(
   A=TextureSampleParameter2D(UVs=Divide(A=@n27, B=@n28)) {parameterName=Noise, texture=/Game/Global/Textures/Default/Noise/T_Dudebro_Noise_01}.R
   B=TextureSampleParameter2D(UVs=Divide(A=@n29, B=@n30)) {parameterName=Noise, texture=&T_Dudebro_Noise_01}.R
  )
  n32 = Multiply(A=Time(), B=ScalarParameter() {defaultValue=0.35, parameterName=NoiseSineSpeed})
  n33 = ConstantBiasScale(None=Sine(None=Add(A=@n32, B=@n21.G)) {period=3})
  n34 = ScalarParameter() {defaultValue=0.5, parameterName=NoiseRadius}
  n35 = ScalarParameter() {parameterName=NoiseHardness}
  n36 = Add(
   A=Multiply(A=Multiply(A=@n15, B=@n23), B=OneMinus(None=@n26))
   B=Multiply(A=SphereMask(A=@n31, B=@n33, Radius=@n34, Hardness=@n35), B=ScalarParameter() {defaultValue=0.05, parameterName=NoiseScale})
  )
  n37 = MaterialFunctionCall(
   Brightness=ScalarParameter() {defaultValue=0.5, parameterName=Fresnel Brightness}
   Exponent=ScalarParameter() {defaultValue=4, parameterName=Fresnel Exponent}
  ) {materialFunction=&MF_Fresnel_FX}
  n38 = MaterialFunctionCall!(Distance Bias, Distance Divisor=ScalarParameter() {defaultValue=20, parameterName="Distance Divisor (S)"}, Scene Depth, Pixel Depth) {materialFunction=/Game/Effects/Fort_Effects/Materials/_Functions/MF_DistanceBehindTranslucency_01}
  n39 = ScalarParameter() {defaultValue=0.2, parameterName=DistanceBehindTranslucencyScale}
  n40 = Add(A=Multiply(A=@n10.RGB, B=Add(A=@n13, B=@n14)), B=Add(A=Add(A=@n36, B=@n37.Fresnel Inverted), B=Multiply(A=@n38.Normalized Inverted, B=@n39)))
  n41 = Multiply(A=ParticleColor().RGB, B=Multiply(A=LinearInterpolate(A=@n1.RGB, B=@n2.RGB, Alpha=@n40), B=@n22))
  n42 = Multiply(
   A=@n41
   B=MaterialFunctionCall(Distance=ScalarParameter() {defaultValue=200, parameterName="Fade_Start_Distance (S)"}) {materialFunction=&Near_Camera_Fade}.Result
  )
  MaterialFunctionCall(MA In=SetMaterialAttributes!(
    MaterialAttributes
    Emissive Color=QualitySwitch(Default=@n41, Low=@n41, High=@n42, Medium=@n41)
    Opacity=ScalarParameter() {defaultValue=1, parameterName=Opacity}
   ) {attributeSetTypes [#1, #2]}) {materialFunction=&MF_CreativeMoveTool}
 ]
}
