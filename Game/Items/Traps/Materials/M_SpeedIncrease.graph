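# Additive unlit particle-sprite material under Traps, named for a speed increase, that draws a soft circular, panning energy-streak glow.
# Streak texture, tiled and panned by dynamic parameters (PanU drives both axes), blends low and high colors.
# On deferred, scene world normals distort UVs and drive a Fresnel mask multiplied by depth-behind-sprite falloff; other paths use constants.
# Sphere mask, depth fade and particle color shape the result before emissive adjustment.
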Material /Game/Items/Traps/Materials/M_SpeedIncrease
graph {
 properties {bUsedWithParticleSprites=true, blendMode=BLEND_Additive, pixelDepthOffsetMode=PDOM_Legacy, shadingModel=MSM_Unlit, usedShadingModels=Unlit}
 nodes [
  n1 = VectorParameter() {parameterName=FresnelColorLow, defaultValue=<rgba 0.017 0.0962 0.3 0>}
  n2 = VectorParameter() {parameterName=FresnelColorHigh, defaultValue=<rgba 2.2024 20 12.387 0>}
  n3 = DynamicParameter() {paramNames [PanU,PanV,UVTile,Param4]}
  n4 = Panner(Coordinate=Multiply(A=TextureCoordinate(), B=@n3.UVTile), Speed=AppendVector(A=@n3.PanU, B=@n3.PanU)) {bFractionalPart=true}
  n5 = Constant()
  n6 = SceneTexture() {sceneTextureId=PPI_WorldNormal}
  n7 = Transform(None=@n6.Color) {transformSourceType=TRANSFORMSOURCE_World, transformType=TRANSFORM_View}
  n8 = ShadingPathSwitch(
   Default=@n5
   Deferred=Multiply(A=ScalarParameter() {defaultValue=0.2, parameterName=NormalStrength}, B=ComponentMask(None=@n7) {g=true, r=true})
   Forward=@n5
   Mobile=@n5
  )
  n9 = Power(
   Base=TextureSample(UVs=Add(A=@n4, B=@n8)) {texture=/Game/Effects/Fort_Effects/Textures/Energy/T_OutpostShockwaveStreaks_02}.G
   Exp=ScalarParameter() {defaultValue=2, parameterName=TextureExponent}
  )
  n10 = Constant() {r=0.2}
  n11 = ScalarParameter() {defaultValue=35, parameterName="Distance Divisor (S)"}
  n12 = Multiply(
   A=MaterialFunctionCall(
    Exponent=ScalarParameter() {defaultValue=1, parameterName=FresnelExponent}
    Camera Vector=ComponentMask(None=@n6.Color) {b=true, g=true, r=true}
   ) {materialFunction=&MF_Fresnel_FX}.Fresnel Inverted
   B=Power(
    Base=MaterialFunctionCall!(Distance Bias, Distance Divisor=@n11, Scene Depth, Pixel Depth) {materialFunction=/Game/Effects/Fort_Effects/Materials/_Functions/MF_DistanceBehindTranslucency_01}.Normalized Inverted
    Exp=ScalarParameter() {defaultValue=1, parameterName=DistanceExponent}
   )
  )
  n13 = Constant() {r=0.5}
  n14 = Multiply(
   A=Multiply(A=LinearInterpolate(A=@n1.RGB, B=@n2.RGB, Alpha=@n9), B=ShadingPathSwitch(Default=@n10, Deferred=@n12, Forward=@n10, Mobile=@n10))
   B=SphereMask(A=TextureCoordinate(), B=@n13, Radius=@n13, Hardness=Constant())
  )
  n15 = ScalarParameter() {defaultValue=3, parameterName=FadeDistance}
  MaterialFunctionCall(Emissive=Multiply(A=DepthFade(Opacity=@n14, FadeDistance=@n15), B=ParticleColor().RGB)) {materialFunction=&MF_FX_EmissiveAdjustments}
 ]
}
