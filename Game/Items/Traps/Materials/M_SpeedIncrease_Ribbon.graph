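# Unlit alpha-composite ribbon material for speed increase trap effects, usable on Niagara ribbons, sprites and beam trails.
# A tiled, panning smoke-trail texture, raised to a power, blends two "Fresnel" colors without any Fresnel node, with optional particle color override.
# Depth-faded particle alpha shapes color and opacity; color passes twice through MF_FX_EmissiveAdjustments.
# An optional UV1 gradient with swap and flip controls fades both along one axis.
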Material /Game/Items/Traps/Materials/M_SpeedIncrease_Ribbon
graph {
 properties {
  bUsedWithBeamTrails=true
  bUsedWithNiagaraRibbons=true
  bUsedWithNiagaraSprites=true
  blendMode=BLEND_AlphaComposite
  pixelDepthOffsetMode=PDOM_Legacy
  shadingModel=MSM_Unlit
  usedShadingModels=Unlit
 }
 nodes [
  n1 = AppendVector(A=ScalarParameter() {defaultValue=1, parameterName=Tiling X}, B=ScalarParameter() {defaultValue=1, parameterName=Tiling Y})
  n2 = TextureCoordinate()
  n3 = If(
   A=ScalarParameter() {parameterName=UseTextCoord0}
   B=Constant() {r=0.99}
   "A > B"=@n2
   "A == B"=@n2
   "A < B"=TextureCoordinate() {coordinateIndex=1}
  )
  n4 = ScalarParameter() {parameterName=PanU}
  n5 = ScalarParameter() {parameterName=PanV}
  n6 = TextureSample(UVs=Panner(Coordinate=Multiply(A=@n1, B=@n3), Speed=AppendVector(A=@n4, B=@n5))) {samplerType=SAMPLERTYPE_Grayscale, texture=&T_Smoke_Trail_Soft}
  n7 = Power(Base=@n6.R, Exp=ScalarParameter() {defaultValue=4, parameterName=Exponent})
  n8 = LinearInterpolate(
   A=VectorParameter() {parameterName=FresnelColorLow, defaultValue=<rgba 0.017 0.0962 0.3 0>}.RGB
   B=VectorParameter() {parameterName=FresnelColorHigh, defaultValue=<rgba 0.2604 10 5.8339 0>}.RGB
   Alpha=@n7
  )
  n9 = ParticleColor()
  n10 = ScalarParameter() {parameterName=UseParticleColor}
  n11 = DepthFade(Opacity=ParticleColor().A, FadeDistance=ScalarParameter() {defaultValue=3, parameterName=FadeDistance})
  n12 = Multiply(A=Multiply(A=LinearInterpolate(A=@n8, B=@n9.RGB, Alpha=@n10), B=@n6.R), B=@n11)
  n13 = ScalarParameter() {defaultValue=1, parameterName=Opacity Multiplier}
  n14 = Multiply(
   A=MaterialFunctionCall(Emissive=MaterialFunctionCall(Emissive=@n12) {materialFunction=&MF_FX_EmissiveAdjustments}.Result) {materialFunction=&MF_FX_EmissiveAdjustments}.Result
   B=@n13
  )
  n15 = TextureCoordinate() {coordinateIndex=1}
  n16 = ComponentMask(None=LinearInterpolate(
    A=@n15
    B=MaterialFunctionCall(XY=@n15) {materialFunction=&Swizzle}.YX
    Alpha=ScalarParameter() {defaultValue=1, parameterName=UV Swap}
   )) {g=true}
  n17 = Power(
   Base=LinearInterpolate(A=@n16, B=OneMinus(None=@n16), Alpha=ScalarParameter() {parameterName=UV Flip})
   Exp=ScalarParameter() {defaultValue=1.5, parameterName=Opacity Gradient Pow}
  )
  n18 = ScalarParameter() {parameterName=Use Opacity Gradient}
  LinearInterpolate(A=@n14, B=Multiply(A=@n14, B=@n17), Alpha=@n18)
  n19 = Multiply(A=@n7, B=@n11)
  Multiply(A=LinearInterpolate(A=@n19, B=Multiply(A=@n19, B=@n17), Alpha=@n18), B=@n13)
 ]
}
